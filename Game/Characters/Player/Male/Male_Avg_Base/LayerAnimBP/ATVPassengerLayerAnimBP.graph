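# Animation layer Blueprint for player characters riding as ATV passengers, implementing only the VehicleSplitBodyOverride layer.
# The cached upper/lower-body pose gets its root reset and a slope pitch/roll aim offset, then feeds a lean-and-bounce state machine, bypassed by a crouch-turning flag.
# Quad input drives lean start, nested lean/bounce and end states; recent firing returns to the slope pose.
# Leans and bounces blend by gender, while start/end use CMM clips only.
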AnimBlueprint /Game/Characters/Player/Male/Male_Avg_Base/LayerAnimBP/ATVPassengerLayerAnimBP
target_skeleton=&Fortnite_M_Avg_Player_Skeleton
graphs [
 AnimGraph {nodes [Output Pose() {layerGroup=None}]}
 EventGraph {links [], nodes []}
 VehicleFinalPoseOverride {nodes [
   Input Pose() {name=InFinalPose}
   Output Pose() {layerGroup=VehicleOverride}
  ]}
 VehicleFullBodyOverride {nodes [
   Input Pose() {name=InPoseFullBody}
   Output Pose() {layerGroup=VehicleOverride}
  ]}
 VehicleLowerBodyOverride {nodes [
   Input Pose() {name=InPoseLowerBodyDefault}
   Input Pose() {name=InPoseUpperBody}
   Output Pose() {layerGroup=VehicleOverride}
  ]}
 VehicleSplitBodyOverride {nodes [
   Input Pose() {name=InPoseSplitBody}
   Input Pose() {name=InPoseUpperAndLowerBodyRemoveRoot}
   AnimGraphNode_StateMachine_0 = ATV_LeansAndBounce() {is=title AnimGraphNode_StateMachine}
   Output Pose(Result=Blend Poses by bool(
     BlendPose_0=Use cached pose 'PostSlopeAOPose'()
     BlendPose_1=@AnimGraphNode_StateMachine_0
     BlendTime_0=0.3
     BlendTime_1=0.35
     bActiveValue=Get bTransitionLocomotionAdditiveCrouchTurning() -> (bTransitionLocomotionAdditiveCrouchTurning:bool)
    ) {blendType=Linear}) {layerGroup=Vehicle Override}
   Save cached pose 'UpperAndLowerBodyCachePose'(Pose=Input Pose() {name=InPoseUpperAndLowerBody}) {cacheName=UpperAndLowerBodyCachePose}
   n1 = Reset Root Transform!(
    ComponentPose:Component Space Pose Link Structure=Local To Component(LocalPose=Use cached pose 'UpperAndLowerBodyCachePose'())
    Alpha:float=1
   ) -> (Pose:Component Space Pose Link Structure) {is=title AnimGraphNode_ResetRoot, alphaBoolBlend {blendOption=Linear}}
   n2 = Break Fort Anim Input Quad!(FortAnimInput_Quad:"Fort Anim Input Quad Structure (by ref)"=Get QuadVehicleInput() -> (QuadVehicleInput:Fort Anim Input Quad Structure)) -> (
    bIsUsingQuad:bool=~
    bIsDriver:bool=~
    bIsFrontPassenger:bool=~
    bIsBackPassenger:bool=~
    bIsBackPassengerAndLeaning:bool=~
    bIsDrifting:bool=~
    bIsBoosting:bool=~
    bIsReversing:bool=~
    bIsBraking:bool=~
    bIsMoving:bool=~
    bIsMovingForward:bool=~
    bIsLeaning:bool=~
    bIsLeaningOrBouncing:bool=~
    bIsBounceCrouching:bool=~
    bIsBounceCrouched:bool=~
    bIsBounceJumping:bool=~
    bIsBounceRecoiling:bool=~
    bIsSteeringRight:bool=~
    bIsSteeringLeft:bool=~
    RunForwardAlpha:float=~
    BounceCompression:float=~
    LeanPosition:Vector=~
    LeanPositionX:float=~
    LeanPositionY:float=~
    LeanPositionZ:float=~
    VerticalVelocity:float=~
    VerticalAcceleration:float=~
    bAimFWD:bool=~
    bAimBWD:bool=~
    bAimLFT:bool=~
    bAimRGT:bool=~
    PawnToVehicleDeltaYawAngleDegrees:float=~
    AimCardDirDeadZoneAngleDegrees:float=~
    AimCardDirAngleOffsetDegrees:float=~
    AimFWDDeltaAngleDegrees:float=~
    AimBWDDeltaAngleDegrees:float=~
    AimLFTDeltaAngleDegrees:float=~
    AimRGTDeltaAngleDegrees:float=~
    SlopePitchDegreeAngle:float=~
    SlopeRollDegreeAngle:float=~
    SteerAngle:float=~
    SteerAlpha:float=~
    SteerAngleDeadZoneDegrees:float=~
    SteeringRotation:float=~
    VehiclePitch:float=~
    VehicleRoll:float=~
   ) {is=title K2Node_BreakStruct}
   Save cached pose 'PostSlopeAOPose'(Pose=AimOffset Player 'ATV_PassengerSlope_AO'!(
     X:float=@n2.SlopeRollDegreeAngle
     Y:float=@n2.SlopePitchDegreeAngle
     BasePose:Pose Link Structure=Component To Local(ComponentPose=@n1)
     Alpha:float=1
    ) -> (Pose:Pose Link Structure) {blendSpace=/Game/Animation/Game/MainPlayer/Vehicles/ATV/AO/ATV_PassengerSlope_AO, alphaBoolBlend {blendOption=Linear}}) {cacheName=PostSlopeAOPose}
  ]}
 ATV_LeansAndBounce {
  entry=SplitBodyCacheState
  path=VehicleSplitBodyOverride.AnimGraphNode_StateMachine_0.ATV_LeansAndBounce
  ids=id
  nodes [
   AnimStateNode_0 = SplitBodyCacheState() {is=title AnimStateNode}
   AnimStateNode_17 = ATV_BouncingLeans() {is=title AnimStateNode}
   AnimStateNode_6 = ATV_BouncingLeans_Start() {is=title AnimStateNode}
   AnimStateNode_7 = ATV_BouncingLeans_End() {is=title AnimStateNode}
   AnimStateTransitionNode_0 = AnimStateTransitionNode() {crossfadeDuration=0}
   AnimStateTransitionNode_1 = AnimStateTransitionNode() {crossfadeDuration=0}
   AnimStateTransitionNode_3 = AnimStateTransitionNode() {crossfadeDuration=0}
   AnimStateTransitionNode_5 = AnimStateTransitionNode() {bAutomaticRuleBasedOnSequencePlayerInState=true, priorityOrder=2}
   AnimStateTransitionNode_7 = AnimStateTransitionNode() {bAutomaticRuleBasedOnSequencePlayerInState=true}
   AnimStateTransitionNode_8 = AnimStateTransitionNode() {crossfadeDuration=0}
   AnimStateTransitionNode_9 = AnimStateTransitionNode() {crossfadeDuration=0.0667, priorityOrder=2}
  ]
  links [
   {
    dst=ATV_BouncingLeans_End
    rule=VehicleSplitBodyOverride.AnimGraphNode_StateMachine_0.ATV_LeansAndBounce.AnimStateTransitionNode_9.AnimationTransitionGraph_0
    src=ATV_BouncingLeans
    via=AnimStateTransitionNode_9
   }
   {
    dst=SplitBodyCacheState
    rule=VehicleSplitBodyOverride.AnimGraphNode_StateMachine_0.ATV_LeansAndBounce.AnimStateTransitionNode_3.Transition
    src=ATV_BouncingLeans
    via=AnimStateTransitionNode_3
   }
   {
    dst=SplitBodyCacheState
    rule=VehicleSplitBodyOverride.AnimGraphNode_StateMachine_0.ATV_LeansAndBounce.AnimStateTransitionNode_5.Transition
    src=ATV_BouncingLeans_End
    via=AnimStateTransitionNode_5
   }
   {
    dst=SplitBodyCacheState
    rule=VehicleSplitBodyOverride.AnimGraphNode_StateMachine_0.ATV_LeansAndBounce.AnimStateTransitionNode_8.Transition
    src=ATV_BouncingLeans_End
    via=AnimStateTransitionNode_8
   }
   {
    dst=ATV_BouncingLeans
    rule=VehicleSplitBodyOverride.AnimGraphNode_StateMachine_0.ATV_LeansAndBounce.AnimStateTransitionNode_7.AnimationTransitionGraph_0
    src=ATV_BouncingLeans_Start
    via=AnimStateTransitionNode_7
   }
   {
    dst=SplitBodyCacheState
    rule=VehicleSplitBodyOverride.AnimGraphNode_StateMachine_0.ATV_LeansAndBounce.AnimStateTransitionNode_1.Transition
    src=ATV_BouncingLeans_Start
    via=AnimStateTransitionNode_1
   }
   {
    dst=ATV_BouncingLeans_Start
    rule=VehicleSplitBodyOverride.AnimGraphNode_StateMachine_0.ATV_LeansAndBounce.AnimStateTransitionNode_0.Transition
    src=SplitBodyCacheState
    via=AnimStateTransitionNode_0
   }
  ]
 }
 SplitBodyCacheState {
  path=VehicleSplitBodyOverride.AnimGraphNode_StateMachine_0.ATV_LeansAndBounce.AnimStateNode_0.SplitBodyCacheState
  nodes [Output Animation Pose(Result=Use cached pose 'PostSlopeAOPose'())]
 }
 ATV_BouncingLeans {
  path=VehicleSplitBodyOverride.AnimGraphNode_StateMachine_0.ATV_LeansAndBounce.AnimStateNode_17.ATV_BouncingLeans
  nodes [
   AnimGraphNode_StateMachine_0 = BouncingLeansATV() {is=title AnimGraphNode_StateMachine}
   Output Animation Pose(Result=@AnimGraphNode_StateMachine_0)
  ]
 }
 BouncingLeansATV {
  entry=ATV_Leans
  path=VehicleSplitBodyOverride.AnimGraphNode_StateMachine_0.ATV_LeansAndBounce.AnimStateNode_17.ATV_BouncingLeans.AnimGraphNode_StateMachine_0.BouncingLeansATV
  ids=id
  nodes [
   AnimStateNode_0 = ATV_Bounce() {is=title AnimStateNode}
   AnimStateNode_1 = ATV_Leans() {is=title AnimStateNode}
   AnimStateTransitionNode_0 = AnimStateTransitionNode() {bAutomaticRuleBasedOnSequencePlayerInState=true, crossfadeDuration=0.25}
   AnimStateTransitionNode_1 = AnimStateTransitionNode() {crossfadeDuration=0.25, priorityOrder=2}
   AnimStateTransitionNode_2 = AnimStateTransitionNode() {crossfadeDuration=0.1}
  ]
  links [
   {
    dst=ATV_Leans
    rule=VehicleSplitBodyOverride.AnimGraphNode_StateMachine_0.ATV_LeansAndBounce.AnimStateNode_17.ATV_BouncingLeans.AnimGraphNode_StateMachine_0.BouncingLeansATV.AnimStateTransitionNode_0.Transition
    src=ATV_Bounce
    via=AnimStateTransitionNode_0
   }
   {
    dst=ATV_Leans
    rule=VehicleSplitBodyOverride.AnimGraphNode_StateMachine_0.ATV_LeansAndBounce.AnimStateNode_17.ATV_BouncingLeans.AnimGraphNode_StateMachine_0.BouncingLeansATV.AnimStateTransitionNode_1.Transition
    src=ATV_Bounce
    via=AnimStateTransitionNode_1
   }
   {
    dst=ATV_Bounce
    rule=VehicleSplitBodyOverride.AnimGraphNode_StateMachine_0.ATV_LeansAndBounce.AnimStateNode_17.ATV_BouncingLeans.AnimGraphNode_StateMachine_0.BouncingLeansATV.AnimStateTransitionNode_2.Transition
    src=ATV_Leans
    via=AnimStateTransitionNode_2
   }
  ]
 }
 ATV_Bounce {
  path=VehicleSplitBodyOverride.AnimGraphNode_StateMachine_0.ATV_LeansAndBounce.AnimStateNode_17.ATV_BouncingLeans.AnimGraphNode_StateMachine_0.BouncingLeansATV.AnimStateNode_0.ATV_Bounce
  nodes [
   n1 = "Blend Poses (EFortCustomGender)"!(
    BlendPose_0:Pose Link Structure=Sequence Player 'ATV_Passenger_Bounce_CMM'!(PlayRate:float=1.2) -> (Pose) {bLoopAnimation=false, sequence=/Game/Animation/Game/MainPlayer/Vehicles/ATV/ATV_Passenger_Bounce_CMM}
    BlendPose_1:Pose Link Structure=Sequence Player 'ATV_Passenger_Bounce_CMF'!(PlayRate:float=1.2) -> (Pose) {sequence=/Game/Animation/Game/MainPlayer/Vehicles/ATV/ATV_Passenger_Bounce_CMF}
    BlendTime_0:float=0.1
    BlendTime_1:float=0.1
    ActiveEnumValue:Byte=Get Gender() -> (Gender:EFortCustomGender Enum=Invalid)
   ) -> (Pose:Pose Link Structure) {
    is=title AnimGraphNode_BlendListByEnum
    blendType=Linear
    blendPose [{linkId=-1, sourceLinkId=-1}, {linkId=-1, sourceLinkId=-1}]
    blendTime [0.1,0.1]
   }
   Output Animation Pose(Result=@n1)
  ]
 }
 ATV_Leans {
  path=VehicleSplitBodyOverride.AnimGraphNode_StateMachine_0.ATV_LeansAndBounce.AnimStateNode_17.ATV_BouncingLeans.AnimGraphNode_StateMachine_0.BouncingLeansATV.AnimStateNode_1.ATV_Leans
  nodes [
   n1 = Break Fort Anim Input Quad(FortAnimInput_Quad=Get QuadVehicleInput()) -> (
    bIsUsingQuad=~
    bIsDriver=~
    bIsFrontPassenger=~
    bIsBackPassenger=~
    bIsBackPassengerAndLeaning=~
    bIsDrifting=~
    bIsBoosting=~
    bIsReversing=~
    bIsBraking=~
    bIsMoving=~
    bIsMovingForward=~
    bIsLeaning=~
    bIsLeaningOrBouncing=~
    bIsBounceCrouching=~
    bIsBounceCrouched=~
    bIsBounceJumping=~
    bIsBounceRecoiling=~
    bIsSteeringRight=~
    bIsSteeringLeft=~
    RunForwardAlpha=~
    BounceCompression=~
    LeanPosition=~
    LeanPositionX=~
    LeanPositionY=~
    LeanPositionZ=~
    VerticalVelocity=~
    VerticalAcceleration=~
    bAimFWD=~
    bAimBWD=~
    bAimLFT=~
    bAimRGT=~
    PawnToVehicleDeltaYawAngleDegrees=~
    AimCardDirDeadZoneAngleDegrees=~
    AimCardDirAngleOffsetDegrees=~
    AimFWDDeltaAngleDegrees=~
    AimBWDDeltaAngleDegrees=~
    AimLFTDeltaAngleDegrees=~
    AimRGTDeltaAngleDegrees=~
    SlopePitchDegreeAngle=~
    SlopeRollDegreeAngle=~
    SteerAngle=~
    SteerAlpha=~
    SteerAngleDeadZoneDegrees=~
    SteeringRotation=~
    VehiclePitch=~
    VehicleRoll=~
   )
   n2 = "Blend Poses (EFortCustomGender)"(
    BlendPose_0=Blendspace Player 'ATV_Passenger_Leans_BS_CMM'(X=@n1.LeanPositionY, Y=@n1.LeanPositionZ) {blendSpace=/Game/Animation/Game/MainPlayer/Vehicles/ATV/ATV_Passenger_Leans_BS_CMM}
    BlendPose_1=Blendspace Player 'ATV_Passenger_Leans_BS_CMF'(X=@n1.LeanPositionY, Y=@n1.LeanPositionZ) {blendSpace=/Game/Animation/Game/MainPlayer/Vehicles/ATV/ATV_Passenger_Leans_BS_CMF}
    BlendTime_0=0.1
    BlendTime_1=0.1
    ActiveEnumValue=Get Gender() -> (Gender=Invalid)
   ) {blendType=Linear, blendPose [{linkId=-1, sourceLinkId=-1}, {linkId=-1, sourceLinkId=-1}], blendTime [0.1,0.1]}
   Output Animation Pose(Result=Apply Additive(
     Base=@n2
     Additive="Sequence Player (additive) 'ATV_Passenger_Idle'"() {sequence=/Game/Animation/Game/MainPlayer/Vehicles/ATV/ATV_Passenger_Idle}
    ) {alphaBoolBlend {blendOption=Linear}})
  ]
 }
 Transition {
  path=VehicleSplitBodyOverride.AnimGraphNode_StateMachine_0.ATV_LeansAndBounce.AnimStateNode_17.ATV_BouncingLeans.AnimGraphNode_StateMachine_0.BouncingLeansATV.AnimStateTransitionNode_0.Transition
  nodes [Result()]
 }
 Transition {
  path=VehicleSplitBodyOverride.AnimGraphNode_StateMachine_0.ATV_LeansAndBounce.AnimStateNode_17.ATV_BouncingLeans.AnimGraphNode_StateMachine_0.BouncingLeansATV.AnimStateTransitionNode_1.Transition
  nodes [Result(bCanEnterTransition=Break Fort Anim Input Quad(FortAnimInput_Quad=Get QuadVehicleInput()) -> (
     bIsUsingQuad=~
     bIsDriver=~
     bIsFrontPassenger=~
     bIsBackPassenger=~
     bIsBackPassengerAndLeaning=~
     bIsDrifting=~
     bIsBoosting=~
     bIsReversing=~
     bIsBraking=~
     bIsMoving=~
     bIsMovingForward=~
     bIsLeaning=~
     bIsLeaningOrBouncing=~
     bIsBounceCrouching=~
     bIsBounceCrouched=~
     bIsBounceJumping=~
     bIsBounceRecoiling=~
     bIsSteeringRight=~
     bIsSteeringLeft=~
     RunForwardAlpha=~
     BounceCompression=~
     LeanPosition=~
     LeanPositionX=~
     LeanPositionY=~
     LeanPositionZ=~
     VerticalVelocity=~
     VerticalAcceleration=~
     bAimFWD=~
     bAimBWD=~
     bAimLFT=~
     bAimRGT=~
     PawnToVehicleDeltaYawAngleDegrees=~
     AimCardDirDeadZoneAngleDegrees=~
     AimCardDirAngleOffsetDegrees=~
     AimFWDDeltaAngleDegrees=~
     AimBWDDeltaAngleDegrees=~
     AimLFTDeltaAngleDegrees=~
     AimRGTDeltaAngleDegrees=~
     SlopePitchDegreeAngle=~
     SlopeRollDegreeAngle=~
     SteerAngle=~
     SteerAlpha=~
     SteerAngleDeadZoneDegrees=~
     SteeringRotation=~
     VehiclePitch=~
     VehicleRoll=~
    ).bIsBounceCrouched)]
 }
 Transition {
  path=VehicleSplitBodyOverride.AnimGraphNode_StateMachine_0.ATV_LeansAndBounce.AnimStateNode_17.ATV_BouncingLeans.AnimGraphNode_StateMachine_0.BouncingLeansATV.AnimStateTransitionNode_2.Transition
  nodes [Result(bCanEnterTransition=Break Fort Anim Input Quad(FortAnimInput_Quad=Get QuadVehicleInput()) -> (
     bIsUsingQuad=~
     bIsDriver=~
     bIsFrontPassenger=~
     bIsBackPassenger=~
     bIsBackPassengerAndLeaning=~
     bIsDrifting=~
     bIsBoosting=~
     bIsReversing=~
     bIsBraking=~
     bIsMoving=~
     bIsMovingForward=~
     bIsLeaning=~
     bIsLeaningOrBouncing=~
     bIsBounceCrouching=~
     bIsBounceCrouched=~
     bIsBounceJumping=~
     bIsBounceRecoiling=~
     bIsSteeringRight=~
     bIsSteeringLeft=~
     RunForwardAlpha=~
     BounceCompression=~
     LeanPosition=~
     LeanPositionX=~
     LeanPositionY=~
     LeanPositionZ=~
     VerticalVelocity=~
     VerticalAcceleration=~
     bAimFWD=~
     bAimBWD=~
     bAimLFT=~
     bAimRGT=~
     PawnToVehicleDeltaYawAngleDegrees=~
     AimCardDirDeadZoneAngleDegrees=~
     AimCardDirAngleOffsetDegrees=~
     AimFWDDeltaAngleDegrees=~
     AimBWDDeltaAngleDegrees=~
     AimLFTDeltaAngleDegrees=~
     AimRGTDeltaAngleDegrees=~
     SlopePitchDegreeAngle=~
     SlopeRollDegreeAngle=~
     SteerAngle=~
     SteerAlpha=~
     SteerAngleDeadZoneDegrees=~
     SteeringRotation=~
     VehiclePitch=~
     VehicleRoll=~
    ).bIsBounceJumping)]
 }
 ATV_BouncingLeans_Start {
  path=VehicleSplitBodyOverride.AnimGraphNode_StateMachine_0.ATV_LeansAndBounce.AnimStateNode_6.ATV_BouncingLeans_Start
  nodes [Output Animation Pose(Result=Sequence Player 'ATV_Passenger_Lean_Start_CMM'() {bLoopAnimation=false, sequence=/Game/Animation/Game/MainPlayer/Vehicles/ATV/ATV_Passenger_Lean_Start_CMM})]
 }
 ATV_BouncingLeans_End {
  path=VehicleSplitBodyOverride.AnimGraphNode_StateMachine_0.ATV_LeansAndBounce.AnimStateNode_7.ATV_BouncingLeans_End
  nodes [Output Animation Pose(Result=Sequence Player 'ATV_Passenger_Lean_End_CMM'() {bLoopAnimation=false, sequence=/Game/Animation/Game/MainPlayer/Vehicles/ATV/ATV_Passenger_Lean_End_CMM})]
 }
 Transition {
  path=VehicleSplitBodyOverride.AnimGraphNode_StateMachine_0.ATV_LeansAndBounce.AnimStateTransitionNode_0.Transition
  nodes [Result(bCanEnterTransition=Break Fort Anim Input Quad(FortAnimInput_Quad=Get QuadVehicleInput()) -> (
     bIsUsingQuad=~
     bIsDriver=~
     bIsFrontPassenger=~
     bIsBackPassenger=~
     bIsBackPassengerAndLeaning=~
     bIsDrifting=~
     bIsBoosting=~
     bIsReversing=~
     bIsBraking=~
     bIsMoving=~
     bIsMovingForward=~
     bIsLeaning=~
     bIsLeaningOrBouncing=~
     bIsBounceCrouching=~
     bIsBounceCrouched=~
     bIsBounceJumping=~
     bIsBounceRecoiling=~
     bIsSteeringRight=~
     bIsSteeringLeft=~
     RunForwardAlpha=~
     BounceCompression=~
     LeanPosition=~
     LeanPositionX=~
     LeanPositionY=~
     LeanPositionZ=~
     VerticalVelocity=~
     VerticalAcceleration=~
     bAimFWD=~
     bAimBWD=~
     bAimLFT=~
     bAimRGT=~
     PawnToVehicleDeltaYawAngleDegrees=~
     AimCardDirDeadZoneAngleDegrees=~
     AimCardDirAngleOffsetDegrees=~
     AimFWDDeltaAngleDegrees=~
     AimBWDDeltaAngleDegrees=~
     AimLFTDeltaAngleDegrees=~
     AimRGTDeltaAngleDegrees=~
     SlopePitchDegreeAngle=~
     SlopeRollDegreeAngle=~
     SteerAngle=~
     SteerAlpha=~
     SteerAngleDeadZoneDegrees=~
     SteeringRotation=~
     VehiclePitch=~
     VehicleRoll=~
    ).bIsLeaningOrBouncing)]
 }
 Transition {
  path=VehicleSplitBodyOverride.AnimGraphNode_StateMachine_0.ATV_LeansAndBounce.AnimStateTransitionNode_1.Transition
  nodes [Result(bCanEnterTransition=Get bRecentlyFired() -> (bRecentlyFired:bool))]
 }
 Transition {
  path=VehicleSplitBodyOverride.AnimGraphNode_StateMachine_0.ATV_LeansAndBounce.AnimStateTransitionNode_3.Transition
  nodes [Result(bCanEnterTransition=Get bRecentlyFired())]
 }
 Transition {path=VehicleSplitBodyOverride.AnimGraphNode_StateMachine_0.ATV_LeansAndBounce.AnimStateTransitionNode_5.Transition, nodes [Result()]}
 AnimationTransitionGraph_0 {
  path=VehicleSplitBodyOverride.AnimGraphNode_StateMachine_0.ATV_LeansAndBounce.AnimStateTransitionNode_7.AnimationTransitionGraph_0
  nodes [Result()]
 }
 Transition {
  path=VehicleSplitBodyOverride.AnimGraphNode_StateMachine_0.ATV_LeansAndBounce.AnimStateTransitionNode_8.Transition
  nodes [Result(bCanEnterTransition=Get bRecentlyFired())]
 }
 AnimationTransitionGraph_0 {
  path=VehicleSplitBodyOverride.AnimGraphNode_StateMachine_0.ATV_LeansAndBounce.AnimStateTransitionNode_9.AnimationTransitionGraph_0
  nodes [
   n1 = Get QuadVehicleInput()
   Result(bCanEnterTransition=NOT Boolean(A=Break Fort Anim Input Quad(FortAnimInput_Quad=@n1) -> (
      bIsUsingQuad=~
      bIsDriver=~
      bIsFrontPassenger=~
      bIsBackPassenger=~
      bIsBackPassengerAndLeaning=~
      bIsDrifting=~
      bIsBoosting=~
      bIsReversing=~
      bIsBraking=~
      bIsMoving=~
      bIsMovingForward=~
      bIsLeaning=~
      bIsLeaningOrBouncing=~
      bIsBounceCrouching=~
      bIsBounceCrouched=~
      bIsBounceJumping=~
      bIsBounceRecoiling=~
      bIsSteeringRight=~
      bIsSteeringLeft=~
      RunForwardAlpha=~
      BounceCompression=~
      LeanPosition=~
      LeanPositionX=~
      LeanPositionY=~
      LeanPositionZ=~
      VerticalVelocity=~
      VerticalAcceleration=~
      bAimFWD=~
      bAimBWD=~
      bAimLFT=~
      bAimRGT=~
      PawnToVehicleDeltaYawAngleDegrees=~
      AimCardDirDeadZoneAngleDegrees=~
      AimCardDirAngleOffsetDegrees=~
      AimFWDDeltaAngleDegrees=~
      AimBWDDeltaAngleDegrees=~
      AimLFTDeltaAngleDegrees=~
      AimRGTDeltaAngleDegrees=~
      SlopePitchDegreeAngle=~
      SlopeRollDegreeAngle=~
      SteerAngle=~
      SteerAlpha=~
      SteerAngleDeadZoneDegrees=~
      SteeringRotation=~
      VehiclePitch=~
      VehicleRoll=~
     ).bIsLeaningOrBouncing))
  ]
 }
]
variables []
