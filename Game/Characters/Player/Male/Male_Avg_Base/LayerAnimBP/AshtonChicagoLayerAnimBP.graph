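# Linked-layer Animation Blueprint for the male average player skeleton, supplying AshtonChicago item-specific airborne additive animations through the Item Override layer group.
# Jump up, apex, fall and land layers each play an AshtonChicago additive sequence, with fall rate from FallPlayRate and landing in the JumpLand sync group.
# Jetpack and zipline layers use ARDS additives.
# Other layers have unconnected outputs, and the EventGraph is empty.
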AnimBlueprint /Game/Characters/Player/Male/Male_Avg_Base/LayerAnimBP/AshtonChicagoLayerAnimBP
target_skeleton=&Fortnite_M_Avg_Player_Skeleton
graphs [
 AnimGraph {nodes [Output Pose() {layerGroup=None}]}
 EventGraph {links [], nodes []}
 ItemAimOffset {nodes [
   Input Pose!() -> (Pose, AimOffsetAlpha:double, Yaw:double, Pitch:double) {name=InPose_AimOffset}
   Input Pose!() -> (Pose) {name=InPose_PreAimOffsets}
   Output Pose() {layerGroup=Item Override}
  ]}
 ItemFallAdditive {nodes [
   Input Pose() {name=InPoseItemFallAdditive}
   n1 = "Sequence Player (additive) 'AshtonChicago_JumpFall_Additive_CMM'"!(PlayRate:float=Get FallPlayRate() -> (FallPlayRate:float)) -> (Pose) {
    bLoopAnimation=false
    sequence=/Game/Animation/Game/MainPlayer/Locomotion/Medium/Male/Jump/Ashton/Chicago/AshtonChicago_JumpFall_Additive_CMM
   }
   Output Pose(Result=@n1) {layerGroup=Item Override}
  ]}
 ItemFallLandAdditive {nodes [
   Input Pose() {name=InPoseItemFallLandAdditive}
   n1 = "Sequence Player (additive) 'AshtonChicago_JumpLand_Additive_CMM'"() {
    groupName=JumpLand
    method=SyncGroup
    sequence=/Game/Animation/Game/MainPlayer/Locomotion/Medium/Male/Jump/Ashton/Chicago/AshtonChicago_JumpLand_Additive_CMM
   }
   Output Pose(Result=@n1) {layerGroup=Item Override}
  ]}
 ItemFallLevel2Override {nodes [
   Input Pose() {name=InPose_FallLevel2}
   Output Pose() {layerGroup=Item Override}
  ]}
 ItemFinalPoseOverride {nodes [
   Input Pose() {name=InFinalPose}
   Output Pose() {layerGroup=Item Override}
  ]}
 ItemFullBodyAimOffsets {nodes [
   Input Pose!() -> (Pose, FullBody_AimOffsetAlpha:double, Aim_Yaw:double, Aim_Pitch:double) {name=InPoseFullBodyAimOffsets}
   Input Pose!() -> (Pose) {name=InPosePreFullBodyAimOffsets}
   Output Pose() {layerGroup=Item Override}
  ]}
 ItemFullBodyOverride {nodes [
   Input Pose() {name=InPoseFullBody}
   Output Pose() {layerGroup=Item Override}
  ]}
 ItemIdleOverride {nodes [
   Input Pose() {name=InPose_Idle}
   Output Pose() {layerGroup=Item Override}
  ]}
 ItemJetPackJumpAdditive {nodes [Output Pose(Result="Sequence Player (additive) 'JetpackApex_ARDS_Loop_Additive'"() {sequence=/Game/Animation/Game/MainPlayer/Locomotion/Medium/Male/Jump/AssaultRifle/JetpackApex_ARDS_Loop_Additive}) {layerGroup=Item Override}]}
 ItemJetPackStartAdditive {nodes [Output Pose(Result="Sequence Player (additive) 'JetpackUp_ARDS_Additive'"() {bLoopAnimation=false, sequence=/Game/Animation/Game/MainPlayer/Locomotion/Medium/Male/Jump/AssaultRifle/JetpackUp_ARDS_Additive}) {layerGroup=Item Override}]}
 ItemJumpLoopAdditive {nodes [
   Input Pose() {name=InPoseItemJumpLoopAdditive}
   Output Pose(Result="Sequence Player (additive) 'AshtonChicago_JumpApex_Additive'"() {sequence=/Game/Animation/Game/MainPlayer/Locomotion/Medium/Male/Jump/Ashton/Chicago/AshtonChicago_JumpApex_Additive}) {layerGroup=Item Override}
  ]}
 ItemJumpOverride {nodes [
   Input Pose() {name=InPoseJump}
   Output Pose() {layerGroup=Item Override}
  ]}
 ItemJumpUpAdditive {nodes [
   Input Pose() {name=InPoseItemJumpUpAdditive}
   n1 = "Sequence Player (additive) 'AshtonChicago_JumpUp_Additive_CMM'"() {bLoopAnimation=false, sequence=/Game/Animation/Game/MainPlayer/Locomotion/Medium/Male/Jump/Ashton/Chicago/AshtonChicago_JumpUp_Additive_CMM}
   Output Pose(Result=@n1) {layerGroup=Item Override}
  ]}
 ItemPreAimOffsets {nodes [
   Input Pose() {name=InPosePreAimOffsets}
   Output Pose() {layerGroup=Item Override}
  ]}
 ItemPreIK {nodes [
   Input Pose() {name=InPose_PreIK}
   Output Pose() {layerGroup=Item Override}
  ]}
 ItemRightHandIK {nodes [
   Input Pose() {name=InPosePreRightHandIK}
   Input Pose() {name=InPosePostRightHandIK}
   Output Pose() {layerGroup=Item Override}
  ]}
 ItemSkeletalControl {nodes [
   Input Pose()
   Output Pose() {layerGroup=Item Override}
  ]}
 ItemSwimJumpFallLoop {nodes [Output Pose() {layerGroup=Item Override}]}
 ItemSwimJumpSurfaceEnd {nodes [Output Pose() {layerGroup=Item Override}]}
 ItemSwimJumpSurfaceLoop {nodes [Output Pose() {layerGroup=Item Override}]}
 ItemUpperBody {nodes [
   Output Pose() {layerGroup=Item Override}
   AnimGraphNode_SubInput_0 = Input Pose!() -> (Pose, InputParam:Fort Anim Input Adjusted Aim Structure) {name=InPoseUpperBody}
  ]}
 ItemZipLineStartAdditive {nodes [Output Pose(Result="Sequence Player (additive) 'ZipLine_Entry_ARDS_Additive'"() {bLoopAnimation=false, sequence=/Game/Animation/Game/MainPlayer/Combat/Gadgets/Medium/Male/ZipLine/ZipLine_Entry_ARDS_Additive}) {layerGroup=Item Override}]}
]
variables []
