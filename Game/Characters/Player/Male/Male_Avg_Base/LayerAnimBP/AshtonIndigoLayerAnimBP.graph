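# Item-layer Animation Blueprint for the AshtonIndigo gadget on the average male player skeleton, mainly supplying jump and fall additive animations.
# The event graph sets isOverheated each update from the UserCurve01 curve.
# Jump-up, apex and fall layers use nested bool blends on overheat and relaxed state, with jump-boost variants for apex and fall.
# Jetpack and zipline layers play fixed additives; the other layers contain no animation logic.
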AnimBlueprint /Game/Characters/Player/Male/Male_Avg_Base/LayerAnimBP/AshtonIndigoLayerAnimBP
target_skeleton=&Fortnite_M_Avg_Player_Skeleton
graphs [
 AnimGraph {nodes [Output Pose() {layerGroup=None}]}
 EventGraph {nodes [
   Event BlueprintUpdateAnimation()
   > n1 = Branch(Condition="float > float"(A=GetCurveValue(CurveName=UserCurve01), B=0.1))
   > Set isOverheated(isOverheated:bool=true) -> (Output_Get:bool)
   Set isOverheated(execute=@n1.else, isOverheated=false)
  ]}
 ItemAimOffset {nodes [
   Input Pose!() -> (Pose, AimOffsetAlpha:double, Yaw:double, Pitch:double) {name=InPose_AimOffset}
   Input Pose!() -> (Pose) {name=InPose_PreAimOffsets}
   Output Pose() {layerGroup=Item Override}
  ]}
 ItemFallAdditive {nodes [
   Input Pose() {name=InPoseItemFallAdditive}
   n1 = Get FallPlayRate() -> (FallPlayRate:float)
   n2 = "Sequence Player (additive) 'AshtonIndigo_JumpFall_CMM_Additive'"!(PlayRate:float=@n1) -> (Pose) {bLoopAnimation=false, sequence=/Game/Animation/Game/MainPlayer/Combat/Gadgets/Medium/Male/Ashton/Indigo/AshtonIndigo_JumpFall_CMM_Additive}
   n3 = Blend Poses by bool(
    BlendPose_0=@n2
    BlendPose_1="Sequence Player (additive) 'JumpFall_NoWep_CMM_Additive'"!(PlayRate:float=@n1) -> (Pose) {bLoopAnimation=false, sequence=/Game/Animation/Game/MainPlayer/Locomotion/Medium/Male/Jump/NoWep/JumpFall_NoWep_CMM_Additive}
    BlendTime_0=0.2
    BlendTime_1=0.2
    bActiveValue=Get bIsJumpBoosting() -> (bIsJumpBoosting:bool)
   ) {blendType=Linear}
   n4 = "Sequence Player (additive) 'AshtonIndigo_JumpFall_NonTargeted_Additive'"!(PlayRate:float=@n1) -> (Pose) {
    bLoopAnimation=false
    sequence=/Game/Animation/Game/MainPlayer/Combat/Gadgets/Medium/Male/Ashton/Indigo/AshtonIndigo_JumpFall_NonTargeted_Additive
   }
   n5 = Get bIsRelaxedLevel1() -> (bIsRelaxedLevel1:bool)
   n6 = "Sequence Player (additive) 'AshtonIndigo_JumpFall_NonTargeted_Additive'"(PlayRate=1) {
    bLoopAnimation=false
    sequence=&AshtonIndigo_JumpFall_NonTargeted_Additive
   }
   Output Pose(Result=Blend Poses by bool(
     BlendPose_0=@n6
     BlendPose_1=Blend Poses by bool(BlendPose_0=@n3, BlendPose_1=@n4, BlendTime_0=0.2, BlendTime_1=0.2, bActiveValue=@n5) {blendType=Linear}
     BlendTime_0=0.2
     BlendTime_1=0.2
     bActiveValue=Get isOverheated() -> (isOverheated:bool)
    ) {blendType=Linear}) {layerGroup=Item Override}
  ]}
 ItemFallLandAdditive {nodes [
   Input Pose() {name=InPoseItemFallLandAdditive}
   Output Pose() {layerGroup=Item Override}
   AnimGraphNode_SubInput_0 = Input Pose()
  ]}
 ItemFallLevel2Override {nodes [
   Input Pose() {name=InPose_FallLevel2}
   Output Pose() {layerGroup=Item Override}
  ]}
 ItemFinalPoseOverride {nodes [
   Input Pose() {name=InFinalPose}
   Output Pose() {layerGroup=Item Override}
  ]}
 ItemFullBodyAimOffsets {nodes [
   Input Pose!() -> (Pose, FullBody_AimOffsetAlpha:double, Aim_Yaw:double, Aim_Pitch:double) {name=InPoseFullBodyAimOffsets}
   Input Pose!() -> (Pose) {name=InPosePreFullBodyAimOffsets}
   Output Pose() {layerGroup=Item Override}
  ]}
 ItemFullBodyOverride {nodes [
   Input Pose() {name=InPoseFullBody}
   Output Pose() {layerGroup=Item Override}
  ]}
 ItemIdleOverride {nodes [
   Input Pose() {name=InPose_Idle}
   Output Pose() {layerGroup=Item Override}
  ]}
 ItemJetPackJumpAdditive {nodes [Output Pose(Result="Sequence Player (additive) 'JetpackApex_ARDS_Loop_Additive'"() {sequence=/Game/Animation/Game/MainPlayer/Locomotion/Medium/Male/Jump/AssaultRifle/JetpackApex_ARDS_Loop_Additive}) {layerGroup=Item Override}]}
 ItemJetPackStartAdditive {nodes [Output Pose(Result="Sequence Player (additive) 'JetpackUp_ARDS_Additive'"() {bLoopAnimation=false, sequence=/Game/Animation/Game/MainPlayer/Locomotion/Medium/Male/Jump/AssaultRifle/JetpackUp_ARDS_Additive}) {layerGroup=Item Override}]}
 ItemJumpLoopAdditive {nodes [
   n1 = Blend Poses by bool(
    BlendPose_0="Sequence Player (additive) 'AshtonIndigo_Relaxed_JumpApex_CMM_Additive'"() {sequence=/Game/Animation/Game/MainPlayer/Combat/Gadgets/Medium/Male/Ashton/Indigo/AshtonIndigo_Relaxed_JumpApex_CMM_Additive}
    BlendPose_1="Sequence Player (additive) 'JumpApex_NoWep__Loop_CMM_Additive'"() {sequence=/Game/Animation/Game/MainPlayer/Locomotion/Medium/Male/Jump/NoWep/JumpApex_NoWep__Loop_CMM_Additive}
    BlendTime_0=0.2
    BlendTime_1=0.2
    bActiveValue=Get bIsJumpBoosting()
   ) {blendType=Linear}
   n2 = "Sequence Player (additive) 'AshtonIndigo_JumpApex_NonTargeted_Loop_Additive'"() {sequence=/Game/Animation/Game/MainPlayer/Combat/Gadgets/Medium/Male/Ashton/Indigo/AshtonIndigo_JumpApex_NonTargeted_Loop_Additive}
   n3 = Get bIsRelaxedLevel1()
   Output Pose(Result=Blend Poses by bool(
     BlendPose_0=Input Pose() {name=InPoseItemJumpLoopAdditive}
     BlendPose_1=Blend Poses by bool(BlendPose_0=@n1, BlendPose_1=@n2, BlendTime_0=0.2, BlendTime_1=0.2, bActiveValue=@n3) {blendType=Linear}
     BlendTime_0=0.2
     BlendTime_1=0.2
     bActiveValue=Get isOverheated()
    ) {blendType=Linear}) {layerGroup=Item Override}
  ]}
 ItemJumpOverride {nodes [
   Input Pose() {name=InPoseJump}
   Output Pose() {layerGroup=Item Override}
  ]}
 ItemJumpUpAdditive {nodes [
   Input Pose() {name=InPoseItemJumpUpAdditive}
   n1 = "Sequence Player (additive) 'JumpUp_NoWep_CMM_Additive'"() {bLoopAnimation=false, sequence=/Game/Animation/Game/MainPlayer/Locomotion/Medium/Male/Jump/NoWep/JumpUp_NoWep_CMM_Additive}
   n2 = "Sequence Player (additive) 'M_JumpUp_ARDS_Additive'"() {bLoopAnimation=false, sequence=/Game/Animation/Game/MainPlayer/Locomotion/Medium/Male/Jump/AssaultRifle/M_JumpUp_ARDS_Additive}
   n3 = Get bIsRelaxedLevel1()
   Output Pose(Result=Blend Poses by bool(
     BlendPose_0=Input Pose()
     BlendPose_1=Blend Poses by bool(BlendPose_0=@n1, BlendPose_1=@n2, BlendTime_0=0.2, BlendTime_1=0.2, bActiveValue=@n3) {blendType=Linear}
     BlendTime_0=0.2
     BlendTime_1=0.2
     bActiveValue=Get isOverheated()
    ) {blendType=Linear}) {layerGroup=Item Override}
  ]}
 ItemPreAimOffsets {nodes [
   Input Pose() {name=InPosePreAimOffsets}
   Output Pose() {layerGroup=Item Override}
  ]}
 ItemPreIK {nodes [
   Input Pose() {name=InPose_PreIK}
   Output Pose() {layerGroup=Item Override}
  ]}
 ItemRightHandIK {nodes [
   Input Pose() {name=InPosePreRightHandIK}
   Input Pose() {name=InPosePostRightHandIK}
   Output Pose() {layerGroup=Item Override}
  ]}
 ItemSkeletalControl {nodes [
   Input Pose()
   Output Pose() {layerGroup=Item Override}
  ]}
 ItemSwimJumpFallLoop {nodes [Output Pose() {layerGroup=Item Override}]}
 ItemSwimJumpSurfaceEnd {nodes [Output Pose() {layerGroup=Item Override}]}
 ItemSwimJumpSurfaceLoop {nodes [Output Pose() {layerGroup=Item Override}]}
 ItemUpperBody {nodes [
   Output Pose() {layerGroup=Item Override}
   AnimGraphNode_SubInput_0 = Input Pose!() -> (Pose, InputParam:Fort Anim Input Adjusted Aim Structure) {name=InPoseUpperBody}
  ]}
 ItemZipLineStartAdditive {nodes [Output Pose(Result="Sequence Player (additive) 'ZipLine_Entry_ARDS_Additive'"() {bLoopAnimation=false, sequence=/Game/Animation/Game/MainPlayer/Combat/Gadgets/Medium/Male/ZipLine/ZipLine_Entry_ARDS_Additive}) {layerGroup=Item Override}]}
]
variables [isOverheated {type=boolean, default=true}]
