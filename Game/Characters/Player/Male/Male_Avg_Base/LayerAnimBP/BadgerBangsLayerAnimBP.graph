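# BadgerBangs item layer Anim Blueprint for the Fortnite average male player skeleton, supplying Item Override animation layers.
# Only the three swim-jump layers have content, each playing a DualMelee additive sequence for fall loop, surface end or surface loop.
# The fall loop and surface end players have looping off, while the surface loop starts at an offset.
# All other layer outputs are unwired, and the blueprint has no event logic or variables.
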AnimBlueprint /Game/Characters/Player/Male/Male_Avg_Base/LayerAnimBP/BadgerBangsLayerAnimBP
target_skeleton=&Fortnite_M_Avg_Player_Skeleton
graphs [
 AnimGraph {nodes [Output Pose() {layerGroup=None}]}
 EventGraph {links [], nodes []}
 ItemAimOffset {nodes [
   Input Pose!() -> (Pose, AimOffsetAlpha:double, Yaw:double, Pitch:double) {name=InPose_AimOffset}
   Input Pose!() -> (Pose) {name=InPose_PreAimOffsets}
   Output Pose() {layerGroup=Item Override}
  ]}
 ItemFallAdditive {nodes [
   Input Pose() {name=InPoseItemFallAdditive}
   Output Pose() {layerGroup=Item Override}
  ]}
 ItemFallLandAdditive {nodes [
   Input Pose() {name=InPoseItemFallLandAdditive}
   Output Pose() {layerGroup=Item Override}
  ]}
 ItemFallLevel2Override {nodes [
   Input Pose() {name=InPose_FallLevel2}
   Output Pose() {layerGroup=Item Override}
  ]}
 ItemFinalPoseOverride {nodes [
   Input Pose() {name=InFinalPose}
   Output Pose() {layerGroup=Item Override}
  ]}
 ItemFullBodyAimOffsets {nodes [
   Input Pose!() -> (Pose, FullBody_AimOffsetAlpha:double, Aim_Yaw:double, Aim_Pitch:double) {name=InPoseFullBodyAimOffsets}
   Input Pose!() -> (Pose) {name=InPosePreFullBodyAimOffsets}
   Output Pose() {layerGroup=Item Override}
  ]}
 ItemFullBodyOverride {nodes [
   Input Pose() {name=InPoseFullBody}
   Output Pose() {layerGroup=Item Override}
  ]}
 ItemIdleOverride {nodes [
   Input Pose() {name=InPose_Idle}
   Output Pose() {layerGroup=Item Override}
  ]}
 ItemJetPackJumpAdditive {nodes [Output Pose() {layerGroup=Item Override}]}
 ItemJetPackStartAdditive {nodes [Output Pose() {layerGroup=Item Override}]}
 ItemJumpLoopAdditive {nodes [
   Input Pose() {name=InPoseItemJumpLoopAdditive}
   Output Pose() {layerGroup=Item Override}
  ]}
 ItemJumpOverride {nodes [
   Input Pose() {name=InPoseJump}
   Output Pose() {layerGroup=Item Override}
  ]}
 ItemJumpUpAdditive {nodes [
   Input Pose() {name=InPoseItemJumpUpAdditive}
   Output Pose() {layerGroup=Item Override}
  ]}
 ItemPreAimOffsets {nodes [
   Input Pose() {name=InPosePreAimOffsets}
   Output Pose() {layerGroup=Item Override}
  ]}
 ItemPreIK {nodes [
   Input Pose() {name=InPose_PreIK}
   Output Pose() {layerGroup=Item Override}
  ]}
 ItemRightHandIK {nodes [
   Input Pose() {name=InPosePreRightHandIK}
   Input Pose() {name=InPosePostRightHandIK}
   Output Pose() {layerGroup=Item Override}
  ]}
 ItemSkeletalControl {nodes [
   Input Pose()
   Output Pose() {layerGroup=Item Override}
  ]}
 ItemSwimJumpFallLoop {nodes [Output Pose(Result="Sequence Player (additive) 'DualMelee_Swim_Jump_Fall_Loop'"!(bLoopAnimation:bool=False) -> (Pose) {bLoopAnimation=false, sequence=/Game/Animation/Game/MainPlayer/Swimming/Jump/DualMelee_Swim_Jump_Fall_Loop}) {layerGroup=Item Override}]}
 ItemSwimJumpSurfaceEnd {nodes [Output Pose(Result="Sequence Player (additive) 'DualMelee_Swim_Jump_Surface_End'"!(bLoopAnimation:bool=False) -> (Pose) {bLoopAnimation=false, sequence=/Game/Animation/Game/MainPlayer/Swimming/Jump/DualMelee_Swim_Jump_Surface_End}) {layerGroup=Item Override}]}
 ItemSwimJumpSurfaceLoop {nodes [Output Pose(Result="Sequence Player (additive) 'DualMelee_Swim_Jump_Surface_Loop'"!(StartPosition:float=0.167) -> (Pose) {sequence=/Game/Animation/Game/MainPlayer/Swimming/Jump/DualMelee_Swim_Jump_Surface_Loop}) {layerGroup=Item Override}]}
 ItemUpperBody {nodes [
   Input Pose!() -> (Pose, InputParam:Fort Anim Input Adjusted Aim Structure) {name=InPoseUpperBody}
   Output Pose() {layerGroup=Item Override}
  ]}
 ItemZipLineStartAdditive {nodes [Output Pose() {layerGroup=Item Override}]}
]
variables []
