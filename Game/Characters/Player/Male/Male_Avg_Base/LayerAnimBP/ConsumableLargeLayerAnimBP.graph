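# Item animation layer Blueprint for male average-build players holding large consumables, overriding jump, fall, landing, jetpack, swim-jump and aim layers.
# Inherited bIsAgileChugJugHeldItem selects assault-rifle or held-object additives when true, ConsumableLarge additives when false.
# The same flag enables the HeldObjects full-body aim offset, otherwise passing the input through.
# Zipline start always uses the rifle jetpack-apex additive; other layers are unwired, and the event graph is empty.
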AnimBlueprint /Game/Characters/Player/Male/Male_Avg_Base/LayerAnimBP/ConsumableLargeLayerAnimBP
target_skeleton=&Fortnite_M_Avg_Player_Skeleton
graphs [
 AnimGraph {nodes [Output Pose() {layerGroup=None}]}
 EventGraph {links [], nodes []}
 ItemAimOffset {nodes [
   Input Pose!() -> (Pose, AimOffsetAlpha:double, Yaw:double, Pitch:double) {name=InPose_AimOffset}
   Input Pose!() -> (Pose) {name=InPose_PreAimOffsets}
   Output Pose() {layerGroup=Item Override}
  ]}
 ItemFallAdditive {nodes [
   Input Pose() {name=InPoseItemFallAdditive}
   n1 = Get FallPlayRate() -> (FallPlayRate:float)
   n2 = Get FallPlayRate()
   Output Pose(Result=Blend Poses by bool(
     BlendPose_0="Sequence Player (additive) 'M_JumpFall_ARDS_Additive'"!(PlayRate:float=@n1) -> (Pose) {bLoopAnimation=false, sequence=/Game/Animation/Game/MainPlayer/Locomotion/Medium/Male/Jump/AssaultRifle/M_JumpFall_ARDS_Additive}
     BlendPose_1="Sequence Player (additive) 'ConsumableLarge_JumpFall_Additive_CMM'"!(PlayRate:float=@n2) -> (Pose) {sequence=/Game/Animation/Game/MainPlayer/Locomotion/Medium/Male/Jump/ConsumableLarge/ConsumableLarge_JumpFall_Additive_CMM}
     bActiveValue=Get bIsAgileChugJugHeldItem() -> (bIsAgileChugJugHeldItem:bool)
    ) {blendType=Linear}) {layerGroup=Item Override}
  ]}
 ItemFallLandAdditive {nodes [
   Input Pose() {name=InPoseItemFallLandAdditive}
   n1 = "Sequence Player (additive) 'M_JumpLand_ARDS_Additive'"() {
    bLoopAnimation=false
    groupName=JumpLand
    method=SyncGroup
    sequence=/Game/Animation/Game/MainPlayer/Locomotion/Medium/Male/Jump/AssaultRifle/M_JumpLand_ARDS_Additive
   }
   n2 = "Sequence Player (additive) 'ConsumableLarge_JumpLand_Additive_CMM'"() {
    bLoopAnimation=false
    groupName=JumpLand
    method=SyncGroup
    sequence=/Game/Animation/Game/MainPlayer/Locomotion/Medium/Male/Jump/ConsumableLarge/ConsumableLarge_JumpLand_Additive_CMM
   }
   Output Pose(Result=Blend Poses by bool(BlendPose_0=@n1, BlendPose_1=@n2, bActiveValue=Get bIsAgileChugJugHeldItem()) {blendType=Linear}) {layerGroup=Item Override}
  ]}
 ItemFallLevel2Override {nodes [
   Input Pose() {name=InPose_FallLevel2}
   Output Pose() {layerGroup=Item Override}
  ]}
 ItemFinalPoseOverride {nodes [
   Input Pose() {name=InFinalPose}
   Output Pose() {layerGroup=Item Override}
  ]}
 ItemFullBodyAimOffsets {nodes [
   n1 = Input Pose!() -> (Pose, FullBody_AimOffsetAlpha:double, Aim_Yaw:double, Aim_Pitch:double) {name=InPoseFullBodyAimOffsets}
   n2 = AimOffset Player 'HeldObjects_AO'!(
    X:float=@n1.Aim_Yaw
    Y:float=@n1.Aim_Pitch
    BasePose:Pose Link Structure=Input Pose!() -> (Pose) {name=InPosePreFullBodyAimOffsets}
    Alpha:float=@n1.FullBody_AimOffsetAlpha
   ) -> (Pose:Pose Link Structure) {blendSpace=/Game/Animation/Game/MainPlayer/Combat/Gadgets/Medium/Male/HeldObjects/AO/HeldObjects_AO, alphaBoolBlend {blendOption=Linear}}
   Output Pose(Result=Blend Poses by bool(BlendPose_0=@n2, BlendPose_1=@n1, bActiveValue=Get bIsAgileChugJugHeldItem()) {blendType=Linear}) {layerGroup=Item Override}
  ]}
 ItemFullBodyOverride {nodes [
   Input Pose() {name=InPoseFullBody}
   Output Pose() {layerGroup=Item Override}
  ]}
 ItemIdleOverride {nodes [
   Input Pose() {name=InPose_Idle}
   Output Pose() {layerGroup=Item Override}
  ]}
 ItemJetPackJumpAdditive {nodes [
   n1 = "Sequence Player (additive) 'ConsumableLarge_JetpackApex_Loop_Additive_CMM'"() {
    bLoopAnimation=false
    sequence=/Game/Animation/Game/MainPlayer/Locomotion/Medium/Male/Jump/ConsumableLarge/ConsumableLarge_JetpackApex_Loop_Additive_CMM
   }
   Output Pose(Result=Blend Poses by bool(
     BlendPose_0="Sequence Player (additive) 'JetpackApex_ARDS_Loop_Additive'"() {sequence=/Game/Animation/Game/MainPlayer/Locomotion/Medium/Male/Jump/AssaultRifle/JetpackApex_ARDS_Loop_Additive}
     BlendPose_1=@n1
     bActiveValue=Get bIsAgileChugJugHeldItem()
    ) {blendType=Linear}) {layerGroup=Item Override}
  ]}
 ItemJetPackStartAdditive {nodes [Output Pose() {layerGroup=Item Override}]}
 ItemJumpLoopAdditive {nodes [
   Input Pose() {name=InPoseItemJumpLoopAdditive}
   Output Pose(Result=Blend Poses by bool(
     BlendPose_0="Sequence Player (additive) 'M_JumpApex_ARDS_Loop_Additive'"() {playRate=0.5, sequence=/Game/Animation/Game/MainPlayer/Locomotion/Medium/Male/Jump/AssaultRifle/M_JumpApex_ARDS_Loop_Additive}
     BlendPose_1="Sequence Player (additive) 'ConsumableLarge_JumpApex_Additive_CMM'"() {sequence=/Game/Animation/Game/MainPlayer/Locomotion/Medium/Male/Jump/ConsumableLarge/ConsumableLarge_JumpApex_Additive_CMM}
     bActiveValue=Get bIsAgileChugJugHeldItem()
    ) {blendType=Linear}) {layerGroup=Item Override}
  ]}
 ItemJumpOverride {nodes [
   Input Pose() {name=InPoseJump}
   Output Pose() {layerGroup=Item Override}
  ]}
 ItemJumpUpAdditive {nodes [
   Input Pose() {name=InPoseItemJumpUpAdditive}
   n1 = "Sequence Player (additive) 'ConsumableLarge_JumpUp_Additive_CMM'"() {
    bLoopAnimation=false
    sequence=/Game/Animation/Game/MainPlayer/Locomotion/Medium/Male/Jump/ConsumableLarge/ConsumableLarge_JumpUp_Additive_CMM
   }
   Output Pose(Result=Blend Poses by bool(
     BlendPose_0="Sequence Player (additive) 'M_JumpUp_ARDS_Additive'"() {bLoopAnimation=false, sequence=/Game/Animation/Game/MainPlayer/Locomotion/Medium/Male/Jump/AssaultRifle/M_JumpUp_ARDS_Additive}
     BlendPose_1=@n1
     bActiveValue=Get bIsAgileChugJugHeldItem()
    ) {blendType=Linear}) {layerGroup=Item Override}
  ]}
 ItemPreAimOffsets {nodes [
   Input Pose() {name=InPosePreAimOffsets}
   Output Pose() {layerGroup=Item Override}
  ]}
 ItemPreIK {nodes [
   Input Pose() {name=InPose_PreIK}
   Output Pose() {layerGroup=Item Override}
  ]}
 ItemRightHandIK {nodes [
   Input Pose() {name=InPosePreRightHandIK}
   Input Pose() {name=InPosePostRightHandIK}
   Output Pose() {layerGroup=Item Override}
  ]}
 ItemSkeletalControl {nodes [
   Input Pose()
   Output Pose() {layerGroup=Item Override}
  ]}
 ItemSwimJumpFallLoop {nodes [
   n1 = "Sequence Player (additive) 'ConsumableLarge_Swim_Jump_Fall_Loop'"() {
    bLoopAnimation=false
    playRate=2
    sequence=/Game/Animation/Game/MainPlayer/Swimming/Jump/ConsumableLarge_Swim_Jump_Fall_Loop
    startPosition=0.25
   }
   Output Pose(Result=Blend Poses by bool(
     BlendPose_0="Sequence Player (additive) 'HeldObjects_Swim_Jump_Fall_Loop'"!(PlayRate:float=2, StartPosition:float=0.25) -> (Pose) {bLoopAnimation=false, sequence=/Game/Animation/Game/MainPlayer/Swimming/Jump/HeldObjects_Swim_Jump_Fall_Loop}
     BlendPose_1=@n1
     bActiveValue=Get bIsAgileChugJugHeldItem()
    ) {blendType=Linear}) {layerGroup=Item Override}
  ]}
 ItemSwimJumpSurfaceEnd {nodes [Output Pose(Result=Blend Poses by bool(
     BlendPose_0="Sequence Player (additive) 'HeldObjects_Swim_Jump_Surface_End'"() {bLoopAnimation=false, sequence=/Game/Animation/Game/MainPlayer/Swimming/Jump/HeldObjects_Swim_Jump_Surface_End}
     BlendPose_1="Sequence Player (additive) 'ConsumableLarge_Swim_Jump_Surface_End'"() {bLoopAnimation=false, sequence=/Game/Animation/Game/MainPlayer/Swimming/Jump/ConsumableLarge_Swim_Jump_Surface_End}
     bActiveValue=Get bIsAgileChugJugHeldItem()
    ) {blendType=Linear}) {layerGroup=Item Override}]}
 ItemSwimJumpSurfaceLoop {nodes [
   n1 = "Sequence Player (additive) 'ConsumableLarge_Swim_Jump_Surface_Loop'"() {playRate=0.5, sequence=/Game/Animation/Game/MainPlayer/Swimming/Jump/ConsumableLarge_Swim_Jump_Surface_Loop, startPosition=0.167}
   Output Pose(Result=Blend Poses by bool(
     BlendPose_0="Sequence Player (additive) 'HeldObjects_Swim_Jump_Surface_Loop'"!(PlayRate:float=0.5, StartPosition:float=0.167) -> (Pose) {sequence=/Game/Animation/Game/MainPlayer/Swimming/Jump/HeldObjects_Swim_Jump_Surface_Loop}
     BlendPose_1=@n1
     bActiveValue=Get bIsAgileChugJugHeldItem()
    ) {blendType=Linear}) {layerGroup=Item Override}
  ]}
 ItemUpperBody {nodes [
   Output Pose() {layerGroup=Item Override}
   AnimGraphNode_SubInput_0 = Input Pose!() -> (Pose, InputParam:Fort Anim Input Adjusted Aim Structure) {name=InPoseUpperBody}
  ]}
 ItemZipLineStartAdditive {nodes [Output Pose(Result="Sequence Player (additive) 'JetpackApex_ARDS_Loop_Additive'"() {sequence=&JetpackApex_ARDS_Loop_Additive}) {layerGroup=Item Override}]}
]
variables []
