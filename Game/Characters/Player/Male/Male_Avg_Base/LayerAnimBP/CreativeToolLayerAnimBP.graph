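# Item-override layer AnimBP for the male player that supplies additive jump animations for the Creative Tool.
# Each update reads the owner's FortPlayerAnimInstance, setting IsRelaxedJump when level-1 relaxed with CreativeTool core animation.
# That flag blends jump-up, apex, fall and landing from assault-rifle additives to Creative Tool clips.
# Creative Tool landings vary by crouch and speed and modify the BothHandFK curve; jetpack jump always plays the tool's apex loop.
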AnimBlueprint /Game/Characters/Player/Male/Male_Avg_Base/LayerAnimBP/CreativeToolLayerAnimBP
target_skeleton=&Fortnite_M_Avg_Player_Skeleton
graphs [
 AnimGraph {nodes [Output Pose() {layerGroup=None}]}
 EventGraph {nodes [
   Event BlueprintUpdateAnimation()
   n1 = TryGetPawnOwner()
   > n2 = Cast To FortPlayerAnimInstance(Object=GetAnimInstance(self=GetComponentByClass!(self:Actor Object Reference=@n1, ComponentClass:Actor Component Class Reference=/Script/Engine.SkeletalMeshComponent) -> (ReturnValue:Skeletal Mesh Component Object Reference)))
   n3 = Get WeaponCoreAnim(self=@n2)
   > Set WeaponCoreAnim(WeaponCoreAnim:EFortWeaponCoreAnimation Enum=@n3) -> (Output_Get:EFortWeaponCoreAnimation Enum)
   > Set IsRelaxedJump(IsRelaxedJump:bool=AND Boolean(
     A=Get bIsRelaxedLevel1!(self:Fort Anim Instance Object Reference=@n2) -> (bIsRelaxedLevel1:bool)
     B="Equal (Enum)"(A:EFortWeaponCoreAnimation Enum=@n3, B:EFortWeaponCoreAnimation Enum=CreativeTool)
    )) -> (Output_Get:bool)
  ]}
 ItemAimOffset {nodes [
   Input Pose!() -> (Pose, AimOffsetAlpha:double, Yaw:double, Pitch:double) {name=InPose_AimOffset}
   Input Pose!() -> (Pose) {name=InPose_PreAimOffsets}
   Output Pose() {layerGroup=Item Override}
  ]}
 ItemFallAdditive {nodes [
   Input Pose() {name=InPoseItemFallAdditive}
   n1 = Get FallPlayRate() -> (FallPlayRate:float)
   Output Pose(Result=Blend Poses by bool(
     BlendPose_0="Sequence Player (additive) 'CreativeTool_JumpFall_CMM'"!(PlayRate:float=@n1) -> (Pose) {bLoopAnimation=false, sequence=/Game/Animation/Game/MainPlayer/Locomotion/Medium/Male/Jump/CreativeTool/CreativeTool_JumpFall_CMM}
     BlendPose_1="Sequence Player (additive) 'M_JumpFall_ARDS_Additive'"!(PlayRate:float=@n1) -> (Pose) {bLoopAnimation=false, sequence=/Game/Animation/Game/MainPlayer/Locomotion/Medium/Male/Jump/AssaultRifle/M_JumpFall_ARDS_Additive}
     BlendTime_0=0.2
     BlendTime_1=0.2
     bActiveValue=Get IsRelaxedJump() -> (IsRelaxedJump:bool)
    ) {blendType=Linear}) {layerGroup=Item Override}
  ]}
 ItemFallLandAdditive {nodes [
   Input Pose() {name=InPoseItemFallLandAdditive}
   n1 = "Sequence Player (additive) 'CreativeTool_Land_Moving_CMM'"() {bLoopAnimation=false, sequence=/Game/Animation/Game/MainPlayer/Locomotion/Medium/Male/Jump/CreativeTool/CreativeTool_Land_Moving_CMM}
   n2 = Blend Poses by bool(
    BlendPose_0=Two Way Blend(
     A="Sequence Player (additive) 'CreativeTool_Land_CMM'"() {bLoopAnimation=false, sequence=/Game/Animation/Game/MainPlayer/Locomotion/Medium/Male/Jump/CreativeTool/CreativeTool_Land_CMM}
     B=Additive Identity Pose()
     Alpha=0.6
    ) {blendNode {alphaBoolBlend {blendOption=Linear}}}
    BlendPose_1=Blend Poses by bool(
     BlendPose_0=@n1
     BlendPose_1="Sequence Player (additive) 'CreativeTool_Land_CMM'"() {bLoopAnimation=false, sequence=&CreativeTool_Land_CMM}
     BlendTime_0=.25
     BlendTime_1=.25
     bActiveValue=Get bIsAboveMinimumLocomotionSpeed() -> (bIsAboveMinimumLocomotionSpeed:bool)
    ) {blendType=Linear}
    BlendTime_0=.2
    BlendTime_1=.2
    bActiveValue=Get bIsCrouching()
   ) {blendType=Linear}
   Output Pose(Result=Blend Poses by bool(
     BlendPose_0=Modify Curve(SourcePose=@n2, CurveValues_0=1, Alpha=1) {curveNames [BothHandFK], curveValues [0]}
     BlendPose_1="Sequence Player (additive) 'M_JumpLand_ARDS_Additive'"() {bLoopAnimation=false, sequence=/Game/Animation/Game/MainPlayer/Locomotion/Medium/Male/Jump/AssaultRifle/M_JumpLand_ARDS_Additive}
     BlendTime_0=0.2
     BlendTime_1=0.2
     bActiveValue=Get IsRelaxedJump()
    ) {blendType=Linear}) {layerGroup=Item Override}
  ]}
 ItemFallLevel2Override {nodes [
   Input Pose() {name=InPose_FallLevel2}
   Output Pose() {layerGroup=Item Override}
  ]}
 ItemFinalPoseOverride {nodes [
   Input Pose() {name=InFinalPose}
   Output Pose() {layerGroup=Item Override}
  ]}
 ItemFullBodyAimOffsets {nodes [
   Input Pose!() -> (Pose, FullBody_AimOffsetAlpha:double, Aim_Yaw:double, Aim_Pitch:double) {name=InPoseFullBodyAimOffsets}
   Input Pose!() -> (Pose) {name=InPosePreFullBodyAimOffsets}
   Output Pose() {layerGroup=Item Override}
  ]}
 ItemFullBodyOverride {nodes [
   Input Pose() {name=InPoseFullBody}
   Output Pose() {layerGroup=Item Override}
  ]}
 ItemIdleOverride {nodes [
   Input Pose() {name=InPose_Idle}
   Output Pose() {layerGroup=Item Override}
  ]}
 ItemJetPackJumpAdditive {nodes [Output Pose(Result="Sequence Player (additive) 'CreativeTool_ApexLoop_CMM'"() {sequence=/Game/Animation/Game/MainPlayer/Locomotion/Medium/Male/Jump/CreativeTool/CreativeTool_ApexLoop_CMM}) {layerGroup=Item Override}]}
 ItemJetPackStartAdditive {nodes [Output Pose() {layerGroup=Item Override}]}
 ItemJumpLoopAdditive {nodes [
   Input Pose() {name=InPoseItemJumpLoopAdditive}
   n1 = "Sequence Player (additive) 'M_JumpApex_ARDS_Loop_Additive'"() {bLoopAnimation=false, sequence=/Game/Animation/Game/MainPlayer/Locomotion/Medium/Male/Jump/AssaultRifle/M_JumpApex_ARDS_Loop_Additive}
   Output Pose(Result=Blend Poses by bool(
     BlendPose_0="Sequence Player (additive) 'CreativeTool_ApexLoop_CMM'"() {sequence=&CreativeTool_ApexLoop_CMM}
     BlendPose_1=@n1
     BlendTime_0=0.2
     BlendTime_1=0.2
     bActiveValue=Get IsRelaxedJump()
    ) {blendType=Linear}) {layerGroup=Item Override}
  ]}
 ItemJumpOverride {nodes [
   Input Pose() {name=InPoseJump}
   Output Pose() {layerGroup=Item Override}
  ]}
 ItemJumpUpAdditive {nodes [
   Input Pose() {name=InPoseItemJumpUpAdditive}
   Output Pose(Result=Blend Poses by bool(
     BlendPose_0="Sequence Player (additive) 'CreativeTool_JumpUp_CMM'"() {bLoopAnimation=false, sequence=/Game/Animation/Game/MainPlayer/Locomotion/Medium/Male/Jump/CreativeTool/CreativeTool_JumpUp_CMM}
     BlendPose_1="Sequence Player (additive) 'M_JumpUp_ARDS_Additive'"() {bLoopAnimation=false, sequence=/Game/Animation/Game/MainPlayer/Locomotion/Medium/Male/Jump/AssaultRifle/M_JumpUp_ARDS_Additive}
     BlendTime_0=0.2
     BlendTime_1=0.2
     bActiveValue=Get IsRelaxedJump()
    ) {blendType=Linear}) {layerGroup=Item Override}
  ]}
 ItemPreAimOffsets {nodes [
   Input Pose() {name=InPosePreAimOffsets}
   Output Pose() {layerGroup=Item Override}
  ]}
 ItemPreIK {nodes [
   Input Pose() {name=InPose_PreIK}
   Output Pose() {layerGroup=Item Override}
  ]}
 ItemRightHandIK {nodes [
   Input Pose() {name=InPosePreRightHandIK}
   Input Pose() {name=InPosePostRightHandIK}
   Output Pose() {layerGroup=Item Override}
  ]}
 ItemSkeletalControl {nodes [
   Input Pose()
   Output Pose() {layerGroup=Item Override}
  ]}
 ItemSwimJumpFallLoop {nodes [Output Pose() {layerGroup=Item Override}]}
 ItemSwimJumpSurfaceEnd {nodes [Output Pose() {layerGroup=Item Override}]}
 ItemSwimJumpSurfaceLoop {nodes [Output Pose() {layerGroup=Item Override}]}
 ItemUpperBody {nodes [
   Output Pose() {layerGroup=Item Override}
   AnimGraphNode_SubInput_0 = Input Pose!() -> (Pose, InputParam:Fort Anim Input Adjusted Aim Structure) {name=InPoseUpperBody}
  ]}
 ItemZipLineStartAdditive {nodes [Output Pose() {layerGroup=Item Override}]}
]
variables [
 IsRelaxedJump {type=boolean}
 WeaponCoreAnim {type=EFortWeaponCoreAnimation, default=Melee}
]
