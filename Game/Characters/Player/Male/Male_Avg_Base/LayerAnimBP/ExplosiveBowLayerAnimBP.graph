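# Layer animation Blueprint for the Explosive Bow on the average male player, filling item layers for upper body, airborne additives and right-hand IK.
# While charging, a crouch-speed/pitch blendspace gains a max-charge loop additive and yaw aim offset, blended by inertialization; sliding zeroes crouch speed.
# Jump additives switch between charged, relaxed and normal, while jetpack and zipline use ARDS clips.
# Right-hand IK snaps to offset weapon_l through a Leg IK node.
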AnimBlueprint /Game/Characters/Player/Male/Male_Avg_Base/LayerAnimBP/ExplosiveBowLayerAnimBP
target_skeleton=&Fortnite_M_Avg_Player_Skeleton
graphs [
 AnimGraph {nodes [Output Pose() {layerGroup=None}]}
 EventGraph {nodes [
   Event BlueprintUpdateAnimation()
   n1 = TryGetPawnOwner()
   > Is Valid(InputObject=@n1)
   > n2 = Cast To FortPlayerPawnAthena(Object=@n1)
   > n3 = Sequence()
   > Set IsCrouchShuffling(IsCrouchShuffling:bool="float > float"(A=GetCurveValue(CurveName=CrouchShuffle))) -> (Output_Get:bool)
   n4 = Branch(execute=@n3.then_1, Condition=Get bIsSliding!(self:Fort Pawn Object Reference=@n2) -> (bIsSliding))
   > Set Charge Blend Space Crouch Speed(Charge Blend Space Crouch Speed:float=0) -> (Output_Get:float)
   Set Charge Blend Space Crouch Speed(
    execute=@n4.else
    Charge Blend Space Crouch Speed=Break Fort Anim Input Bow Weapon!(FortAnimInput_BowWeapon:"Fort Anim Input Bow Weapon Structure (by ref)"=Get BowWeaponInput() -> (BowWeaponInput:Fort Anim Input Bow Weapon Structure)) -> (
     RightHandIKOffsetLocation:Vector=~
     RightHandIKOffsetRotation:Rotator=~
     BowAimYaw:float=~
     BowAimPitch:float=~
     ChargeBlendSpaceCrouchSpeed:float=~
     ChargePlayRate:float=~
     ChargeBlendInTime:float=~
     bIsBowEquipped:bool=~
     bIsBowCharging:bool=~
     bIsBowAtMaxCharge:bool=~
     bEnableRightHandIK:bool=~
    ) {is=title K2Node_BreakStruct}.ChargeBlendSpaceCrouchSpeed
   )
  ]}
 ItemAimOffset {nodes [
   Input Pose!() -> (Pose, AimOffsetAlpha:double, Yaw:double, Pitch:double) {name=InPose_AimOffset}
   Input Pose!() -> (Pose) {name=InPose_PreAimOffsets}
   Output Pose() {layerGroup=Item Override}
  ]}
 ItemFallAdditive {nodes [
   Input Pose() {name=InPoseItemFallAdditive}
   n1 = Get FallPlayRate() -> (FallPlayRate:float)
   n2 = "Sequence Player (additive) 'M_JumpFall_ExplosiveBow_Additive'"!(PlayRate:float=@n1) -> (Pose) {bLoopAnimation=false, sequence=/Game/Animation/Game/MainPlayer/Locomotion/Medium/Male/Jump/ExplosvieBow/M_JumpFall_ExplosiveBow_Additive}
   n3 = "Sequence Player (additive) 'M_Relaxed_JumpFall_ExplosiveBow_Additive'"!(PlayRate:float=@n1) -> (Pose) {
    bLoopAnimation=false
    sequence=/Game/Animation/Game/MainPlayer/Locomotion/Medium/Male/Jump/ExplosvieBow/M_Relaxed_JumpFall_ExplosiveBow_Additive
   }
   n4 = Get bIsRelaxedLevel1() -> (bIsRelaxedLevel1:bool)
   n5 = "Sequence Player (additive) 'M_JumpChargedFall_ExplosiveBow_Additive'"!(PlayRate:float=@n1) -> (Pose) {
    bLoopAnimation=false
    sequence=/Game/Animation/Game/MainPlayer/Locomotion/Medium/Male/Jump/ExplosvieBow/M_JumpChargedFall_ExplosiveBow_Additive
   }
   n6 = Get BowWeaponInput()
   Output Pose(Result=Blend Poses by bool(
     BlendPose_0=@n5
     BlendPose_1=Blend Poses by bool(BlendPose_0=@n3, BlendPose_1=@n2, bActiveValue=@n4) {blendType=Linear}
     bActiveValue=Break Fort Anim Input Bow Weapon(FortAnimInput_BowWeapon=@n6) -> (
      RightHandIKOffsetLocation=~
      RightHandIKOffsetRotation=~
      BowAimYaw=~
      BowAimPitch=~
      ChargeBlendSpaceCrouchSpeed=~
      ChargePlayRate=~
      ChargeBlendInTime=~
      bIsBowEquipped=~
      bIsBowCharging=~
      bIsBowAtMaxCharge=~
      bEnableRightHandIK=~
     ).bIsBowCharging
    ) {blendType=Linear}) {layerGroup=Item Override}
  ]}
 ItemFallLandAdditive {nodes [
   Input Pose() {name=InPoseItemFallLandAdditive}
   n1 = "Sequence Player (additive) 'M_JumpLand_ExplosiveBow_Additive'"() {
    bLoopAnimation=false
    groupName=JumpLand
    method=SyncGroup
    sequence=/Game/Animation/Game/MainPlayer/Locomotion/Medium/Male/Jump/ExplosvieBow/M_JumpLand_ExplosiveBow_Additive
   }
   n2 = "Sequence Player (additive) 'M_Relaxed_JumpLand_ExplosiveBow_Additive'"() {
    bLoopAnimation=false
    groupName=JumpLand
    method=SyncGroup
    sequence=/Game/Animation/Game/MainPlayer/Locomotion/Medium/Male/Jump/ExplosvieBow/M_Relaxed_JumpLand_ExplosiveBow_Additive
   }
   n3 = Get bIsRelaxedLevel1()
   n4 = "Sequence Player (additive) 'M_JumpChargedLand_ExplosiveBow_Additive'"() {
    bLoopAnimation=false
    groupName=JumpLand
    method=SyncGroup
    sequence=/Game/Animation/Game/MainPlayer/Locomotion/Medium/Male/Jump/ExplosvieBow/M_JumpChargedLand_ExplosiveBow_Additive
   }
   n5 = Get BowWeaponInput()
   Output Pose(Result=Blend Poses by bool(
     BlendPose_0=@n4
     BlendPose_1=Blend Poses by bool(BlendPose_0=@n2, BlendPose_1=@n1, bActiveValue=@n3) {blendType=Linear}
     bActiveValue=Break Fort Anim Input Bow Weapon(FortAnimInput_BowWeapon=@n5) -> (
      RightHandIKOffsetLocation=~
      RightHandIKOffsetRotation=~
      BowAimYaw=~
      BowAimPitch=~
      ChargeBlendSpaceCrouchSpeed=~
      ChargePlayRate=~
      ChargeBlendInTime=~
      bIsBowEquipped=~
      bIsBowCharging=~
      bIsBowAtMaxCharge=~
      bEnableRightHandIK=~
     ).bIsBowCharging
    ) {blendType=Linear}) {layerGroup=Item Override}
  ]}
 ItemFallLevel2Override {nodes [
   Input Pose() {name=InPose_FallLevel2}
   Output Pose() {layerGroup=Item Override}
  ]}
 ItemFinalPoseOverride {nodes [
   Input Pose() {name=InFinalPose}
   Output Pose() {layerGroup=Item Override}
  ]}
 ItemFullBodyAimOffsets {nodes [
   Input Pose!() -> (Pose, FullBody_AimOffsetAlpha:double, Aim_Yaw:double, Aim_Pitch:double) {name=InPoseFullBodyAimOffsets}
   Input Pose!() -> (Pose) {name=InPosePreFullBodyAimOffsets}
   Output Pose() {layerGroup=Item Override}
  ]}
 ItemFullBodyOverride {nodes [
   Input Pose() {name=InPoseFullBody}
   Output Pose() {layerGroup=Item Override}
  ]}
 ItemIdleOverride {nodes [
   Input Pose() {name=InPose_Idle}
   Output Pose() {layerGroup=Item Override}
  ]}
 ItemJetPackJumpAdditive {nodes [Output Pose(Result="Sequence Player (additive) 'JetpackApex_ARDS_Loop_Additive'"() {sequence=/Game/Animation/Game/MainPlayer/Locomotion/Medium/Male/Jump/AssaultRifle/JetpackApex_ARDS_Loop_Additive}) {layerGroup=Item Override}]}
 ItemJetPackStartAdditive {nodes [Output Pose(Result="Sequence Player (additive) 'JetpackUp_ARDS_Additive'"() {bLoopAnimation=false, sequence=/Game/Animation/Game/MainPlayer/Locomotion/Medium/Male/Jump/AssaultRifle/JetpackUp_ARDS_Additive}) {layerGroup=Item Override}]}
 ItemJumpLoopAdditive {nodes [
   Input Pose() {name=InPoseItemJumpLoopAdditive}
   n1 = "Sequence Player (additive) 'M_Relaxed_JumpApex_ExplosiveBow_Loop_Additive'"() {sequence=/Game/Animation/Game/MainPlayer/Locomotion/Medium/Male/Jump/ExplosvieBow/M_Relaxed_JumpApex_ExplosiveBow_Loop_Additive}
   n2 = "Sequence Player (additive) 'M_JumpApex_ExplosiveBow_Loop_Additive'"() {sequence=/Game/Animation/Game/MainPlayer/Locomotion/Medium/Male/Jump/ExplosvieBow/M_JumpApex_ExplosiveBow_Loop_Additive}
   n3 = Get bIsRelaxedLevel1()
   n4 = Get BowWeaponInput()
   Output Pose(Result=Blend Poses by bool(
     BlendPose_0="Sequence Player (additive) 'M_JumpChargedApex_ExplosiveBow_Loop_Additive'"() {sequence=/Game/Animation/Game/MainPlayer/Locomotion/Medium/Male/Jump/ExplosvieBow/M_JumpChargedApex_ExplosiveBow_Loop_Additive}
     BlendPose_1=Blend Poses by bool(BlendPose_0=@n1, BlendPose_1=@n2, bActiveValue=@n3) {blendType=Linear}
     bActiveValue=Break Fort Anim Input Bow Weapon(FortAnimInput_BowWeapon=@n4) -> (
      RightHandIKOffsetLocation=~
      RightHandIKOffsetRotation=~
      BowAimYaw=~
      BowAimPitch=~
      ChargeBlendSpaceCrouchSpeed=~
      ChargePlayRate=~
      ChargeBlendInTime=~
      bIsBowEquipped=~
      bIsBowCharging=~
      bIsBowAtMaxCharge=~
      bEnableRightHandIK=~
     ).bIsBowCharging
    ) {blendType=Linear}) {layerGroup=Item Override}
  ]}
 ItemJumpOverride {nodes [
   Input Pose() {name=InPoseJump}
   Output Pose() {layerGroup=Item Override}
  ]}
 ItemJumpUpAdditive {nodes [
   Input Pose() {name=InPoseItemJumpUpAdditive}
   n1 = "Sequence Player (additive) 'M_JumpUp_ExplosiveBow_Additive'"() {bLoopAnimation=false, sequence=/Game/Animation/Game/MainPlayer/Locomotion/Medium/Male/Jump/ExplosvieBow/M_JumpUp_ExplosiveBow_Additive}
   n2 = "Sequence Player (additive) 'M_Relaxed_JumpUp_ExplosiveBow_Additive'"() {
    bLoopAnimation=false
    sequence=/Game/Animation/Game/MainPlayer/Locomotion/Medium/Male/Jump/ExplosvieBow/M_Relaxed_JumpUp_ExplosiveBow_Additive
   }
   n3 = Get bIsRelaxedLevel1()
   n4 = "Sequence Player (additive) 'M_JumpChargedUp_ExplosiveBow_Additive'"() {
    bLoopAnimation=false
    sequence=/Game/Animation/Game/MainPlayer/Locomotion/Medium/Male/Jump/ExplosvieBow/M_JumpChargedUp_ExplosiveBow_Additive
   }
   n5 = Get BowWeaponInput()
   Output Pose(Result=Blend Poses by bool(
     BlendPose_0=@n4
     BlendPose_1=Blend Poses by bool(BlendPose_0=@n2, BlendPose_1=@n1, bActiveValue=@n3) {blendType=Linear}
     bActiveValue=Break Fort Anim Input Bow Weapon(FortAnimInput_BowWeapon=@n5) -> (
      RightHandIKOffsetLocation=~
      RightHandIKOffsetRotation=~
      BowAimYaw=~
      BowAimPitch=~
      ChargeBlendSpaceCrouchSpeed=~
      ChargePlayRate=~
      ChargeBlendInTime=~
      bIsBowEquipped=~
      bIsBowCharging=~
      bIsBowAtMaxCharge=~
      bEnableRightHandIK=~
     ).bIsBowCharging
    ) {blendType=Linear}) {layerGroup=Item Override}
  ]}
 ItemPreAimOffsets {nodes [
   Input Pose() {name=InPosePreAimOffsets}
   Output Pose() {layerGroup=Item Override}
  ]}
 ItemPreIK {nodes [
   Input Pose() {name=InPose_PreIK}
   Output Pose() {layerGroup=Item Override}
  ]}
 ItemRightHandIK {nodes [
   n1 = Copy Bone - Source Bone: weapon_l - Target Bone: ik_hand_r!(
    ComponentPose:Component Space Pose Link Structure=Local To Component(LocalPose=Use cached pose 'PreRightHandIK'())
    bAlphaBoolEnabled:bool=True
    bCopyTranslation:bool=true
    bCopyRotation:bool=true
    bCopyScale:bool=False
   ) -> (Pose:Component Space Pose Link Structure) {
    is=title AnimGraphNode_CopyBone
    alphaInputType=Bool
    alphaBoolBlend {blendOption=Linear}
    sourceBone {boneName=weapon_l}
    targetBone {boneName=ik_hand_r}
   }
   n2 = Break Fort Anim Input Bow Weapon(FortAnimInput_BowWeapon=Get BowWeaponInput()) -> (
    RightHandIKOffsetLocation=~
    RightHandIKOffsetRotation=~
    BowAimYaw=~
    BowAimPitch=~
    ChargeBlendSpaceCrouchSpeed=~
    ChargePlayRate=~
    ChargeBlendInTime=~
    bIsBowEquipped=~
    bIsBowCharging=~
    bIsBowAtMaxCharge=~
    bEnableRightHandIK=~
   )
   n3 = "Transform (Modify) Bone - Bone: ik_hand_r"!(ComponentPose=@n1, bAlphaBoolEnabled:bool=True, Translation=@n2, Rotation=@n2.RightHandIKOffsetRotation, Scale="1,1,1") -> (Pose) {
    alphaInputType=Bool
    rotationMode=BMM_Additive
    rotationSpace=BCS_BoneSpace
    translationMode=BMM_Additive
    translationSpace=BCS_BoneSpace
    alphaBoolBlend {blendOption=Linear}
    boneToModify {boneName=ik_hand_r}
   }
   n4 = Leg IK!(ComponentPose:Component Space Pose Link Structure=@n3, Alpha:float=Get RightHandIKAlpha() -> (RightHandIKAlpha:float)) -> (Pose:Component Space Pose Link Structure) {
    is=title AnimGraphNode_LegIK
    alphaBoolBlend {blendOption=Linear}
    legsDefinition [{
      bEnableKneeTwistCorrection=false
      bEnableRotationLimit=false
      footBoneForwardAxis=Y
      hingeRotationAxis=Z
      minRotationAngle=15
      numBonesInLimb=2
      twistOffsetCurveName=None
      fKFootBone {boneName=hand_r}
      iKFootBone {boneName=ik_hand_r}
     }]
   }
   n5 = Component To Local(ComponentPose=@n4)
   n6 = Use cached pose 'PreRightHandIK'()
   Output Pose(Result=Blend Poses by bool(
     BlendPose_0=Blend Poses by bool(BlendPose_0=@n5, BlendPose_1=@n6, BlendTime_0=0.2, BlendTime_1=0, bActiveValue=@n2.bEnableRightHandIK) {blendType=Sinusoidal}
     BlendPose_1=Input Pose() {name=InPosePostRightHandIK}
     bActiveValue=@n2.bIsBowEquipped
    ) {blendType=Linear}) {layerGroup=Item Override}
   Save cached pose 'PreRightHandIK'(Pose=Input Pose() {name=InPosePreRightHandIK}) {cacheName=PreRightHandIK}
  ]}
 ItemSkeletalControl {nodes [
   Input Pose()
   Output Pose() {layerGroup=Item Override}
  ]}
 ItemSwimJumpFallLoop {nodes [Output Pose() {layerGroup=Item Override}]}
 ItemSwimJumpSurfaceEnd {nodes [Output Pose() {layerGroup=Item Override}]}
 ItemSwimJumpSurfaceLoop {nodes [Output Pose() {layerGroup=Item Override}]}
 ItemUpperBody {nodes [
   n1 = Break Fort Anim Input Bow Weapon(FortAnimInput_BowWeapon=Get BowWeaponInput()) -> (
    RightHandIKOffsetLocation=~
    RightHandIKOffsetRotation=~
    BowAimYaw=~
    BowAimPitch=~
    ChargeBlendSpaceCrouchSpeed=~
    ChargePlayRate=~
    ChargeBlendInTime=~
    bIsBowEquipped=~
    bIsBowCharging=~
    bIsBowAtMaxCharge=~
    bEnableRightHandIK=~
   )
   AnimGraphNode_SubInput_0 = Input Pose!() -> (Pose, InputParam:Fort Anim Input Adjusted Aim Structure) {name=InPoseUpperBody}
   n2 = Break Fort Anim Input Adjusted Aim!(FortAnimInput_AdjustedAim:"Fort Anim Input Adjusted Aim Structure (by ref)"=@AnimGraphNode_SubInput_0.InputParam) -> (ResultingYaw:float=~, ResultingPitch:float=~, ZeroOutPitchWeightCurveName:Name) {is=title K2Node_BreakStruct}
   n3 = Blendspace Player 'ExplosiveBow_Combined_Charging_BS_CMM'!(X=Get Charge Blend Space Crouch Speed() -> (Charge Blend Space Crouch Speed:float), Y=@n2.ResultingPitch, PlayRate:float=@n1.ChargePlayRate) -> (Pose) {bLoop=false, blendSpace=/Game/Animation/Game/MainPlayer/Combat/Guns/ExplosiveBow/ExplosiveBow_Combined_Charging_BS_CMM}
   n4 = "Sequence Player (additive) 'ExplosiveBow_ChargingLoop_CMM'"() {sequence=/Game/Animation/Game/MainPlayer/Combat/Guns/ExplosiveBow/ExplosiveBow_ChargingLoop_CMM}
   n5 = Blend Poses by bool(
    BlendPose_0=AimOffset Player 'M_AO_NT_Bow'!(
     X:float=@n2
     Y:float=0
     BasePose:Pose Link Structure=Apply Additive!(Base=@n3, Additive=@n4, bAlphaBoolEnabled:bool=@n1.bIsBowAtMaxCharge) -> (Pose) {alphaInputType=Bool, alphaBoolBlend {blendOption=Linear}}
     Alpha:float=1
    ) -> (Pose:Pose Link Structure) {blendSpace=/Game/Animation/Game/MainPlayer/Poses/Medium/Male/AO/ExplosiveBow/NT/M_AO_NT_Bow, alphaBoolBlend {blendOption=Linear}}
    BlendPose_1=@AnimGraphNode_SubInput_0
    BlendTime_1=0.15
    bActiveValue=@n1.bIsBowCharging
   ) {blendType=Linear, childUpateMode=ResetChildOnActivate, transitionType=Inertialization}
   Output Pose(Result=Modify Curve(SourcePose=Modify Curve(SourcePose=@n5, CurveValues_0=1, Alpha=1) {curveNames [SnapWeapon_LHand], curveValues [0]}, CurveValues_0=0, Alpha=1) {curveNames [LeftHandFK], curveValues [0]}) {layerGroup=Item Override}
  ]}
 ItemZipLineStartAdditive {nodes [Output Pose(Result="Sequence Player (additive) 'ZipLine_Entry_ARDS_Additive'"() {bLoopAnimation=false, sequence=/Game/Animation/Game/MainPlayer/Combat/Gadgets/Medium/Male/ZipLine/ZipLine_Entry_ARDS_Additive}) {layerGroup=Item Override}]}
]
variables [
 Charge Blend Space Crouch Speed {type=number}
 Is Sliding {type=boolean}
 IsCrouchShuffling {type=boolean}
]
