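# Galileo patrol layer AnimBP for the player skeleton, implementing only FullBodyOverride.
# When bIsLandPatrolling is set, a four-state machine replaces the pass-through pose, switching on movement, recent firing and a time-threshold flag.
# Idle randomizes guard loops and fidget breaks; Walk adds spine-masked guard-idle motion, built as a dynamic additive, to a walk blendspace.
# Firing uses the incoming pose; FiringWalk layers it from the spine and IK hand root over the walk.
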AnimBlueprint /Game/Characters/Player/Male/Male_Avg_Base/LayerAnimBP/GalileoPatrolLayerAnimBP
target_skeleton=&Fortnite_M_Avg_Player_Skeleton
graphs [
 AnimGraph {nodes [Output Pose() {layerGroup=None}]}
 CharacterSkeletalControlPostLegIK {nodes [
   Input Pose() {name=InPosePostLegIK}
   Output Pose() {layerGroup=PlayerOverride}
  ]}
 EventGraph {links [], nodes []}
 FullBodyOverride {nodes [
   AnimGraphNode_StateMachine_0 = PatrolLocomotionSM() {is=title AnimGraphNode_StateMachine}
   Output Pose(Result=Blend Poses by bool(
     BlendPose_0=@AnimGraphNode_StateMachine_0
     BlendPose_1=Use cached pose 'SavedInputPose'()
     BlendTime_1=0.0667
     bActiveValue=Get bIsLandPatrolling() -> (bIsLandPatrolling:bool)
    ) {blendType=Linear}) {layerGroup=PlayerOverride}
   Save cached pose 'SavedInputPose'(Pose=Input Pose() {name=PassThroughFullBodyPose}) {cacheName=SavedInputPose}
  ]}
 PatrolLocomotionSM {
  entry=Idle
  path=FullBodyOverride.AnimGraphNode_StateMachine_0.PatrolLocomotionSM
  ids=id
  nodes [
   AnimStateNode_0 = Idle()
   AnimStateNode_1 = FiringWalk() {is=title AnimStateNode}
   AnimStateNode_2 = Firing() {is=title AnimStateNode}
   AnimStateNode_3 = Walk() {is=title AnimStateNode}
   AnimStateTransitionNode_0 = AnimStateTransitionNode() {crossfadeDuration=0.35}
   AnimStateTransitionNode_1 = AnimStateTransitionNode() {crossfadeDuration=0.35}
   AnimStateTransitionNode_2 = AnimStateTransitionNode() {crossfadeDuration=0.033}
   AnimStateTransitionNode_3 = AnimStateTransitionNode() {crossfadeDuration=0.3}
   AnimStateTransitionNode_4 = AnimStateTransitionNode() {crossfadeDuration=0.033}
   AnimStateTransitionNode_5 = AnimStateTransitionNode() {crossfadeDuration=0.3}
   AnimStateTransitionNode_6 = AnimStateTransitionNode()
   AnimStateTransitionNode_7 = AnimStateTransitionNode()
  ]
  links [
   {
    dst=FiringWalk
    rule=FullBodyOverride.AnimGraphNode_StateMachine_0.PatrolLocomotionSM.AnimStateTransitionNode_6.Transition
    src=Firing
    via=AnimStateTransitionNode_6
   }
   {
    dst=Idle
    rule=FullBodyOverride.AnimGraphNode_StateMachine_0.PatrolLocomotionSM.AnimStateTransitionNode_3.Transition
    src=Firing
    via=AnimStateTransitionNode_3
   }
   {
    dst=Firing
    rule=FullBodyOverride.AnimGraphNode_StateMachine_0.PatrolLocomotionSM.AnimStateTransitionNode_7.Transition
    src=FiringWalk
    via=AnimStateTransitionNode_7
   }
   {
    dst=Walk
    rule=FullBodyOverride.AnimGraphNode_StateMachine_0.PatrolLocomotionSM.AnimStateTransitionNode_5.Transition
    src=FiringWalk
    via=AnimStateTransitionNode_5
   }
   {
    dst=Firing
    rule=FullBodyOverride.AnimGraphNode_StateMachine_0.PatrolLocomotionSM.AnimStateTransitionNode_2.Transition
    src=Idle
    via=AnimStateTransitionNode_2
   }
   {
    dst=Walk
    rule=FullBodyOverride.AnimGraphNode_StateMachine_0.PatrolLocomotionSM.AnimStateTransitionNode_0.Transition
    src=Idle
    via=AnimStateTransitionNode_0
   }
   {
    dst=FiringWalk
    rule=FullBodyOverride.AnimGraphNode_StateMachine_0.PatrolLocomotionSM.AnimStateTransitionNode_4.Transition
    src=Walk
    via=AnimStateTransitionNode_4
   }
   {
    dst=Idle
    rule=FullBodyOverride.AnimGraphNode_StateMachine_0.PatrolLocomotionSM.AnimStateTransitionNode_1.Transition
    src=Walk
    via=AnimStateTransitionNode_1
   }
  ]
 }
 Idle {
  path=FullBodyOverride.AnimGraphNode_StateMachine_0.PatrolLocomotionSM.AnimStateNode_0.Idle
  nodes [
   n1 = Random Sequence Player!() -> (Pose:Pose Link Structure) {
    is=title AnimGraphNode_RandomPlayer
    entries [
     {
      chanceToPlay=1
      maxLoopCount=2
      maxPlayRate=1
      minLoopCount=0
      minPlayRate=1
      sequence=/Game/Animation/Game/MainPlayer/Locomotion/Medium/Male/Idle/Galileo/Idle_Guarding_A
      blendIn {blendOption=HermiteCubic, blendTime=1, customCurve=null}
     }
     {
      chanceToPlay=1
      maxLoopCount=2
      maxPlayRate=1
      minLoopCount=0
      minPlayRate=1
      sequence=/Game/Animation/Game/MainPlayer/Locomotion/Medium/Male/Idle/Galileo/Idle_Guarding_B
      blendIn {blendOption=HermiteCubic, blendTime=1, customCurve=null}
     }
     {
      chanceToPlay=0.35
      maxLoopCount=0
      maxPlayRate=1
      minLoopCount=0
      minPlayRate=1
      sequence=/Game/Animation/Game/MainPlayer/Locomotion/Medium/Male/Idle/Galileo/IdleBreak_CollarAdjust
      blendIn {blendOption=HermiteCubic, blendTime=1, customCurve=null}
     }
     {
      chanceToPlay=0.25
      maxLoopCount=0
      maxPlayRate=1
      minLoopCount=0
      minPlayRate=1
      sequence=/Game/Animation/Game/MainPlayer/Locomotion/Medium/Male/Idle/Galileo/IdleBreak_ConfusedLook
      blendIn {blendOption=HermiteCubic, blendTime=1, customCurve=null}
     }
     {
      chanceToPlay=0.25
      maxLoopCount=0
      maxPlayRate=1
      minLoopCount=0
      minPlayRate=1
      sequence=/Game/Animation/Game/MainPlayer/Locomotion/Medium/Male/Idle/Galileo/IdleBreak_DirtyBoot
      blendIn {blendOption=HermiteCubic, blendTime=1, customCurve=null}
     }
     {
      chanceToPlay=0.35
      maxLoopCount=0
      maxPlayRate=1
      minLoopCount=0
      minPlayRate=1
      sequence=/Game/Animation/Game/MainPlayer/Locomotion/Medium/Male/Idle/Galileo/IdleBreak_RadioChatter
      blendIn {blendOption=HermiteCubic, blendTime=1, customCurve=null}
     }
     {
      chanceToPlay=0.25
      maxLoopCount=0
      maxPlayRate=1
      minLoopCount=0
      minPlayRate=1
      sequence=/Game/Animation/Game/MainPlayer/Locomotion/Medium/Male/Idle/Galileo/IdleBreak_Sneeze
      blendIn {blendOption=HermiteCubic, blendTime=1, customCurve=null}
     }
    ]
   }
   Output Animation Pose(Result=@n1)
  ]
 }
 FiringWalk {
  path=FullBodyOverride.AnimGraphNode_StateMachine_0.PatrolLocomotionSM.AnimStateNode_1.FiringWalk
  nodes [
   n1 = Blendspace Player 'Galileo_Walk_BS'!(X=Get LocalVelocityYawAngle() -> (LocalVelocityYawAngle:float), Y=Get Speed2D() -> (Speed2D:float), PlayRate:float=1) -> (Pose) {
    blendSpace=/Game/Animation/Game/MainPlayer/Locomotion/Medium/Male/Walk/Galileo/Galileo_Walk_BS
    groupName=WalkSync
    method=SyncGroup
    playRate=0.75
   }
   n2 = Layered blend per bone(BasePose=@n1, BlendPoses_0=Use cached pose 'SavedInputPose'()) {bMeshSpaceRotationBlend=true, layerSetup [{branchFilters [{blendDepth=3, boneName=Spine_01}, {boneName=ik_hand_root}]}]}
   Output Animation Pose(Result=@n2)
  ]
 }
 Firing {
  path=FullBodyOverride.AnimGraphNode_StateMachine_0.PatrolLocomotionSM.AnimStateNode_2.Firing
  nodes [Output Animation Pose(Result=Use cached pose 'SavedInputPose'())]
 }
 Walk {
  path=FullBodyOverride.AnimGraphNode_StateMachine_0.PatrolLocomotionSM.AnimStateNode_3.Walk
  nodes [
   n1 = Blendspace Player 'Galileo_Walk_BS'(X=Get LocalVelocityYawAngle(), Y=Get Speed2D(), PlayRate=1) {
    blendSpace=&Galileo_Walk_BS
    groupName=WalkSync
    method=SyncGroup
    playRate=0.75
   }
   n2 = Additive Identity Pose()
   n3 = Make Dynamic Additive!(
    Base:Pose Link Structure="Sequence Evaluator (by time) 'Idle_Guarding_A'"!(ExplicitTime:float=0) -> (Pose:Pose Link Structure) {
     is=title AnimGraphNode_SequenceEvaluator
     bIgnoreForRelevancyTest=true
     sequence=&Idle_Guarding_A
    }
    Additive:Pose Link Structure=Sequence Player 'Idle_Guarding_A'!(StartPosition:float=Get WalkAdditiveStartOffset() -> (WalkAdditiveStartOffset:float)) -> (Pose) {bIgnoreForRelevancyTest=true, sequence=&Idle_Guarding_A}
   ) -> (Pose:Pose Link Structure) {is=title AnimGraphNode_MakeDynamicAdditive}
   Output Animation Pose(Result=Apply Additive(Base=@n1, Additive=Layered blend per bone(BasePose=@n2, BlendPoses_0=@n3) {layerSetup [{branchFilters [{blendDepth=3, boneName=Spine_01}]}]}) {alphaBoolBlend {blendOption=Linear}})
  ]
 }
 Transition {
  path=FullBodyOverride.AnimGraphNode_StateMachine_0.PatrolLocomotionSM.AnimStateTransitionNode_0.Transition
  nodes [Result(bCanEnterTransition=Get bIdle_Walk_Transition() -> (bIdle_Walk_Transition:bool))]
 }
 Transition {
  path=FullBodyOverride.AnimGraphNode_StateMachine_0.PatrolLocomotionSM.AnimStateTransitionNode_1.Transition
  nodes [Result(bCanEnterTransition=NOT Boolean(A=Get bIsMoving2D()))]
 }
 Transition {
  path=FullBodyOverride.AnimGraphNode_StateMachine_0.PatrolLocomotionSM.AnimStateTransitionNode_2.Transition
  nodes [Result(bCanEnterTransition=Get bRecentlyFired() -> (bRecentlyFired:bool))]
 }
 Transition {
  path=FullBodyOverride.AnimGraphNode_StateMachine_0.PatrolLocomotionSM.AnimStateTransitionNode_3.Transition
  nodes [Result(bCanEnterTransition=Get bIsPatrolLocomotionCurrentTimeAboveThreshold() -> (bIsPatrolLocomotionCurrentTimeAboveThreshold:bool))]
 }
 Transition {
  path=FullBodyOverride.AnimGraphNode_StateMachine_0.PatrolLocomotionSM.AnimStateTransitionNode_4.Transition
  nodes [Result(bCanEnterTransition=Get bRecentlyFired())]
 }
 Transition {
  path=FullBodyOverride.AnimGraphNode_StateMachine_0.PatrolLocomotionSM.AnimStateTransitionNode_5.Transition
  nodes [Result(bCanEnterTransition=Get bIsPatrolLocomotionCurrentTimeAboveThreshold())]
 }
 Transition {
  path=FullBodyOverride.AnimGraphNode_StateMachine_0.PatrolLocomotionSM.AnimStateTransitionNode_6.Transition
  nodes [Result(bCanEnterTransition=Get bIsMoving2D())]
 }
 Transition {
  path=FullBodyOverride.AnimGraphNode_StateMachine_0.PatrolLocomotionSM.AnimStateTransitionNode_7.Transition
  nodes [Result(bCanEnterTransition=NOT Boolean(A=Get bIsMoving2D()))]
 }
 PreLegIK {nodes [
   Input Pose()
   Output Pose() {layerGroup=PlayerOverride}
  ]}
 UpperBody {nodes [
   Input Pose() {name=InPoseUpperBody}
   Output Pose() {layerGroup=PlayerOverride}
  ]}
]
variables []
