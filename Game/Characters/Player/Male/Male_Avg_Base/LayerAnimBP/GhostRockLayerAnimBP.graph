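# Item animation layer blueprint for the GhostRock gadget on the Fortnite average male player skeleton, supplying its jump and landing additives.
# Jump up, apex and fall layers each play one GhostRock additive sequence, with the fall slowed down.
# Landing nests bool blends on sprinting and 2D movement to pick sprint, in-motion or standing additives, sharing the JumpLand sync group.
# The remaining layers leave their outputs unconnected, and the event graph is empty.
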AnimBlueprint /Game/Characters/Player/Male/Male_Avg_Base/LayerAnimBP/GhostRockLayerAnimBP
target_skeleton=&Fortnite_M_Avg_Player_Skeleton
graphs [
 AnimGraph {nodes [Output Pose() {layerGroup=None}]}
 EventGraph {links [], nodes []}
 ItemAimOffset {nodes [
   Input Pose!() -> (Pose, AimOffsetAlpha:double, Yaw:double, Pitch:double) {name=InPose_AimOffset}
   Input Pose!() -> (Pose) {name=InPose_PreAimOffsets}
   Output Pose() {layerGroup=Item Override}
  ]}
 ItemFallAdditive {nodes [
   Input Pose() {name=InPoseItemFallAdditive}
   Output Pose(Result="Sequence Player (additive) 'GhostRock_JumpFall_Additive'"!(PlayRate:float=0.3) -> (Pose) {bLoopAnimation=false, sequence=/Game/Animation/Game/MainPlayer/Combat/Gadgets/Medium/Male/GhostRock/GhostRock_JumpFall_Additive}) {layerGroup=Item Override}
  ]}
 ItemFallLandAdditive {nodes [
   Input Pose() {name=InPoseItemFallLandAdditive}
   n1 = "Sequence Player (additive) 'GhostRock_JumpLand_InMotion_Additive'"!(PlayRate:float=1.1) -> (Pose) {
    bLoopAnimation=false
    groupName=JumpLand
    method=SyncGroup
    sequence=/Game/Animation/Game/MainPlayer/Combat/Gadgets/Medium/Male/GhostRock/GhostRock_JumpLand_InMotion_Additive
   }
   n2 = "Sequence Player (additive) 'GhostRock_JumpLand_Additive'"!(PlayRate:float=1.1) -> (Pose) {
    bLoopAnimation=false
    groupName=JumpLand
    method=SyncGroup
    sequence=/Game/Animation/Game/MainPlayer/Combat/Gadgets/Medium/Male/GhostRock/GhostRock_JumpLand_Additive
   }
   n3 = Get bIsMoving2D()
   n4 = "Sequence Player (additive) 'GhostRock_JumpLand_Sprint_Additive'"() {
    bLoopAnimation=false
    groupName=JumpLand
    method=SyncGroup
    sequence=/Game/Animation/Game/MainPlayer/Combat/Gadgets/Medium/Male/GhostRock/GhostRock_JumpLand_Sprint_Additive
   }
   Output Pose(Result=Blend Poses by bool(
     BlendPose_0=@n4
     BlendPose_1=Blend Poses by bool(BlendPose_0=@n1, BlendPose_1=@n2, BlendTime_0=0.25, BlendTime_1=0.5, bActiveValue=@n3) {blendType=Linear}
     BlendTime_0=0.25
     BlendTime_1=0.25
     bActiveValue=Get bIsSprinting()
    ) {blendType=Linear}) {layerGroup=Item Override}
  ]}
 ItemFallLevel2Override {nodes [
   Input Pose() {name=InPose_FallLevel2}
   Output Pose() {layerGroup=Item Override}
  ]}
 ItemFinalPoseOverride {nodes [
   Input Pose() {name=InFinalPose}
   Output Pose() {layerGroup=Item Override}
  ]}
 ItemFullBodyAimOffsets {nodes [
   Input Pose!() -> (Pose, FullBody_AimOffsetAlpha:double, Aim_Yaw:double, Aim_Pitch:double) {name=InPoseFullBodyAimOffsets}
   Input Pose!() -> (Pose) {name=InPosePreFullBodyAimOffsets}
   Output Pose() {layerGroup=Item Override}
  ]}
 ItemFullBodyOverride {nodes [
   Input Pose() {name=InPoseFullBody}
   Output Pose() {layerGroup=Item Override}
  ]}
 ItemIdleOverride {nodes [
   Input Pose() {name=InPose_Idle}
   Output Pose() {layerGroup=Item Override}
  ]}
 ItemJetPackJumpAdditive {nodes [Output Pose() {layerGroup=Item Override}]}
 ItemJetPackStartAdditive {nodes [Output Pose() {layerGroup=Item Override}]}
 ItemJumpLoopAdditive {nodes [
   Input Pose() {name=InPoseItemJumpLoopAdditive}
   Output Pose(Result="Sequence Player (additive) 'GhostRock_JumpApex_Additive'"() {sequence=/Game/Animation/Game/MainPlayer/Combat/Gadgets/Medium/Male/GhostRock/GhostRock_JumpApex_Additive}) {layerGroup=Item Override}
  ]}
 ItemJumpOverride {nodes [
   Input Pose() {name=InPoseJump}
   Output Pose() {layerGroup=Item Override}
  ]}
 ItemJumpUpAdditive {nodes [
   Input Pose() {name=InPoseItemJumpUpAdditive}
   Output Pose(Result="Sequence Player (additive) 'GhostRock_JumpUp_Additive'"() {bLoopAnimation=false, sequence=/Game/Animation/Game/MainPlayer/Combat/Gadgets/Medium/Male/GhostRock/GhostRock_JumpUp_Additive}) {layerGroup=Item Override}
  ]}
 ItemPreAimOffsets {nodes [
   Input Pose() {name=InPosePreAimOffsets}
   Output Pose() {layerGroup=Item Override}
  ]}
 ItemPreIK {nodes [
   Input Pose() {name=InPose_PreIK}
   Output Pose() {layerGroup=Item Override}
  ]}
 ItemRightHandIK {nodes [
   Input Pose() {name=InPosePreRightHandIK}
   Input Pose() {name=InPosePostRightHandIK}
   Output Pose() {layerGroup=Item Override}
  ]}
 ItemSkeletalControl {nodes [
   Input Pose()
   Output Pose() {layerGroup=Item Override}
  ]}
 ItemSwimJumpFallLoop {nodes [Output Pose() {layerGroup=Item Override}]}
 ItemSwimJumpSurfaceEnd {nodes [Output Pose() {layerGroup=Item Override}]}
 ItemSwimJumpSurfaceLoop {nodes [Output Pose() {layerGroup=Item Override}]}
 ItemUpperBody {nodes [
   Output Pose() {layerGroup=Item Override}
   AnimGraphNode_SubInput_0 = Input Pose!() -> (Pose, InputParam:Fort Anim Input Adjusted Aim Structure) {name=InPoseUpperBody}
  ]}
 ItemZipLineStartAdditive {nodes [Output Pose() {layerGroup=Item Override}]}
]
variables [
 JogStartPosition {type=number, default=0.625}
 PivotAnimBlendTime {type=number, default=0.1}
 StartBlendTime {type=number, default=0.33}
]
