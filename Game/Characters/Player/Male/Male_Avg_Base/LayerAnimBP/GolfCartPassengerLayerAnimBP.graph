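# Animation layer Blueprint for player passengers in the golf cart, implementing the vehicle full-body override.
# A state machine switches between a seated split-body pose and a nested lean/bounce machine.
# The seated pose counter-rotates turn-in-place yaw around front or back slope aim offsets, with a front-seat downed (DBNO) pose.
# Leans layer additive blendspaces over idles, with gender variants for back seats; recent firing exits to the cached input pose.
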AnimBlueprint /Game/Characters/Player/Male/Male_Avg_Base/LayerAnimBP/GolfCartPassengerLayerAnimBP
target_skeleton=&Fortnite_M_Avg_Player_Skeleton
graphs [
 AnimGraph {nodes [Output Pose() {layerGroup=None}]}
 EventGraph {links [], nodes []}
 VehicleFinalPoseOverride {nodes [
   Input Pose() {name=InFinalPose}
   Output Pose() {layerGroup=VehicleOverride}
  ]}
 VehicleFullBodyOverride {nodes [
   Blend Poses by bool(
    BlendPose_0=Sequence Player 'TrendConcentrate_PassengerDBNO_AnimSET'() {sequence=/Game/Animation/Game/MainPlayer/Locomotion/Medium/Male/DBNO/TrendConcentrate_PassengerDBNO_AnimSET}
    bActiveValue=Get isDBNO() -> (isDBNO:bool)
   ) {blendType=Linear, childUpateMode=ResetChildOnActivate}
   Component To Local(ComponentPose=Reset Root Transform!(ComponentPose:Component Space Pose Link Structure=Local To Component(), Alpha:float=1) -> (Pose:Component Space Pose Link Structure) {is=title AnimGraphNode_ResetRoot, alphaBoolBlend {blendOption=Linear}})
   AnimGraphNode_StateMachine_0 = GolfCart_LeansAndBounce() {is=title AnimGraphNode_StateMachine}
   Output Pose(Result=@AnimGraphNode_StateMachine_0) {layerGroup=VehicleOverride}
   Rotate Root Bone!(
    BasePose:Pose Link Structure
    Yaw:float=Break Fort Anim Input Turn in Place!(FortAnimInput_TurnInPlace:"Fort Anim Input Turn In Place Structure (by ref)"=Get TurnInPlaceInput() -> (TurnInPlaceInput:Fort Anim Input Turn In Place Structure)) -> (
     RootYawOffset:float=~
     TurnAnimPlayRate:float=~
     bWantsToTurn:bool=~
     bExitTurn:bool=~
     bUseTurnA:bool=~
     TurnACurveAlpha:float=~
     TurnBCurveAlpha:float=~
     bWantsToTurnInVehicle:bool=~
     bWantsToTurnAgain:bool=~
     bTurningLeft:bool=~
     ShufflePlayRate:float=~
     CrouchShufflePlayRate:float=~
     CrouchShuffleStartPosition:float=~
     CrouchShuffleSpeedScaling:float=~
     bTransition_LowerBody_Shuffle:bool=~
     bTransition_LowerBody_ShuffleADS:bool=~
     bPlayOutro:bool=~
    ) {is=title K2Node_BreakStruct}
   ) -> (Pose:Pose Link Structure) {is=title AnimGraphNode_RotateRootBone}
   Rotate Root Bone(Yaw=Break Fort Anim Input Turn in Place(FortAnimInput_TurnInPlace=Get TurnInPlaceInput()) -> (
     RootYawOffset=~
     TurnAnimPlayRate=~
     bWantsToTurn=~
     bExitTurn=~
     bUseTurnA=~
     TurnACurveAlpha=~
     TurnBCurveAlpha=~
     bWantsToTurnInVehicle=~
     bWantsToTurnAgain=~
     bTurningLeft=~
     ShufflePlayRate=~
     CrouchShufflePlayRate=~
     CrouchShuffleStartPosition=~
     CrouchShuffleSpeedScaling=~
     bTransition_LowerBody_Shuffle=~
     bTransition_LowerBody_ShuffleADS=~
     bPlayOutro=~
    )) {yawScaleBiasClamp {scale=-1}}
   Save cached pose 'UpperAndLowerBodyCachePose'(Pose=Input Pose() {name=InPoseFullBody}) {cacheName=UpperAndLowerBodyCachePose}
  ]}
 GolfCart_LeansAndBounce {
  entry=SplitBodyCacheState
  path=VehicleFullBodyOverride.AnimGraphNode_StateMachine_0.GolfCart_LeansAndBounce
  ids=id
  nodes [
   AnimStateNode_0 = SplitBodyCacheState() {is=title AnimStateNode}
   AnimStateNode_5 = GolfCartBouncingLeans() {is=title AnimStateNode}
   AnimStateTransitionNode_10 = AnimStateTransitionNode()
   AnimStateTransitionNode_11 = AnimStateTransitionNode()
  ]
  links [
   {
    dst=SplitBodyCacheState
    rule=VehicleFullBodyOverride.AnimGraphNode_StateMachine_0.GolfCart_LeansAndBounce.AnimStateTransitionNode_11.Transition
    src=GolfCartBouncingLeans
    via=AnimStateTransitionNode_11
   }
   {
    dst=GolfCartBouncingLeans
    rule=VehicleFullBodyOverride.AnimGraphNode_StateMachine_0.GolfCart_LeansAndBounce.AnimStateTransitionNode_10.Transition
    src=SplitBodyCacheState
    via=AnimStateTransitionNode_10
   }
  ]
 }
 SplitBodyCacheState {
  path=VehicleFullBodyOverride.AnimGraphNode_StateMachine_0.GolfCart_LeansAndBounce.AnimStateNode_0.SplitBodyCacheState
  nodes [
   n1 = Get TurnInPlaceInput()
   n2 = Local To Component(LocalPose=Rotate Root Bone(
     BasePose=Use cached pose 'UpperAndLowerBodyCachePose'()
     Yaw=Break Fort Anim Input Turn in Place(FortAnimInput_TurnInPlace=@n1) -> (
      RootYawOffset=~
      TurnAnimPlayRate=~
      bWantsToTurn=~
      bExitTurn=~
      bUseTurnA=~
      TurnACurveAlpha=~
      TurnBCurveAlpha=~
      bWantsToTurnInVehicle=~
      bWantsToTurnAgain=~
      bTurningLeft=~
      ShufflePlayRate=~
      CrouchShufflePlayRate=~
      CrouchShuffleStartPosition=~
      CrouchShuffleSpeedScaling=~
      bTransition_LowerBody_Shuffle=~
      bTransition_LowerBody_ShuffleADS=~
      bPlayOutro=~
     )
    ))
   n3 = Break Fort Anim Input Golf Cart!(FortAnimInput_GolfCart:"Fort Anim Input Golf Cart Structure (by ref)"=Get GolfCartVehicleInput() -> (GolfCartVehicleInput:Fort Anim Input Golf Cart Structure)) -> (
    bIsUsingGolfCart:bool=~
    bIsDriver:bool=~
    bIsFrontPassenger:bool=~
    bIsBackLeftPassenger:bool=~
    bIsBackRightPassenger:bool=~
    bIsFrontPassengerAndLeaning:bool=~
    bIsBackPassengerAndLeaning:bool=~
    bIsDrifting:bool=~
    bIsBoosting:bool=~
    bIsEBraking:bool=~
    bIsReversing:bool=~
    bIsBraking:bool=~
    bIsMoving:bool=~
    bIsMovingForward:bool=~
    bIsPowerSliding:bool=~
    bIsLeaning:bool=~
    bIsLeaningOrBouncing:bool=~
    bIsBounceCrouching:bool=~
    bIsBounceCrouched:bool=~
    bIsBounceJumping:bool=~
    bIsBounceRecoiling:bool=~
    bIsSteeringRight:bool=~
    bIsSteeringLeft:bool=~
    RunForwardAlpha:float=~
    BounceCompression:float=~
    LeanPosition:Vector=~
    LeanPositionX:float=~
    LeanPositionY:float=~
    LeanPositionZ:float=~
    bAimFWD:bool=~
    bAimBWD:bool=~
    bAimLFT:bool=~
    bAimRGT:bool=~
    PawnToVehicleDeltaYawAngleDegrees:float=~
    AimCardDirDeadZoneAngleDegrees:float=~
    AimCardDirAngleOffsetDegrees:float=~
    AimFWDDeltaAngleDegrees:float=~
    AimBWDDeltaAngleDegrees:float=~
    AimLFTDeltaAngleDegrees:float=~
    AimRGTDeltaAngleDegrees:float=~
    SlopePitchDegreeAngle:float=~
    SlopeRollDegreeAngle:float=~
    SteerAngle:float=~
   ) {is=title K2Node_BreakStruct}
   n4 = AimOffset Player 'PassengerFront_AO_SlopePitch_GolfCart'!(
    X:float=@n3.SlopePitchDegreeAngle
    BasePose:Pose Link Structure=Component To Local(ComponentPose=Reset Root Transform(ComponentPose=@n2, Alpha=1) {alphaBoolBlend {blendOption=Linear}})
    Alpha:float=1
   ) -> (Pose:Pose Link Structure) {blendSpace=/Game/Animation/Game/MainPlayer/Vehicles/GolfCart/AO/PassengerFront_AO_SlopePitch_GolfCart, alphaBoolBlend {blendOption=Linear}}
   n5 = AimOffset Player 'PassengerFront_AO_SlopeRoll_GolfCart'!(X:float=@n3.SlopeRollDegreeAngle, BasePose:Pose Link Structure=@n4, Alpha:float=1) -> (Pose:Pose Link Structure) {blendSpace=/Game/Animation/Game/MainPlayer/Vehicles/GolfCart/AO/PassengerFront_AO_SlopeRoll_GolfCart, alphaBoolBlend {blendOption=Linear}}
   n6 = Layered blend per bone(BasePose=Use cached pose 'UpperAndLowerBodyCachePose'(), BlendPoses_0=@n5) {bMeshSpaceRotationBlend=true, layerSetup [{branchFilters [{blendDepth=3, boneName=spine_01}, {boneName=ik_hand_root}, {boneName=weapon_r}]}]}
   n7 = Blend Poses by bool(
    BlendPose_0=Sequence Player 'TrendConcentrate_PassengerDBNO_AnimSET'() {sequence=&TrendConcentrate_PassengerDBNO_AnimSET}
    BlendPose_1=@n6
    bActiveValue=Get isDBNO()
   ) {blendType=Linear, childUpateMode=ResetChildOnActivate}
   n8 = Get TurnInPlaceInput()
   n9 = Local To Component(LocalPose=Rotate Root Bone(
     BasePose=Use cached pose 'UpperAndLowerBodyCachePose'()
     Yaw=Break Fort Anim Input Turn in Place(FortAnimInput_TurnInPlace=@n8) -> (
      RootYawOffset=~
      TurnAnimPlayRate=~
      bWantsToTurn=~
      bExitTurn=~
      bUseTurnA=~
      TurnACurveAlpha=~
      TurnBCurveAlpha=~
      bWantsToTurnInVehicle=~
      bWantsToTurnAgain=~
      bTurningLeft=~
      ShufflePlayRate=~
      CrouchShufflePlayRate=~
      CrouchShuffleStartPosition=~
      CrouchShuffleSpeedScaling=~
      bTransition_LowerBody_Shuffle=~
      bTransition_LowerBody_ShuffleADS=~
      bPlayOutro=~
     )
    ))
   n10 = AimOffset Player 'PassengerBack_AO_SlopePitch_GolfCart'!(
    X:float=@n3.SlopePitchDegreeAngle
    BasePose:Pose Link Structure=Component To Local(ComponentPose=Reset Root Transform(ComponentPose=@n9, Alpha=1) {alphaBoolBlend {blendOption=Linear}})
    Alpha:float=1
   ) -> (Pose:Pose Link Structure) {blendSpace=/Game/Animation/Game/MainPlayer/Vehicles/GolfCart/AO/PassengerBack_AO_SlopePitch_GolfCart, alphaBoolBlend {blendOption=Linear}}
   n11 = AimOffset Player 'PassengerBack_AO_SlopeRoll_GolfCart'!(X:float=@n3.SlopeRollDegreeAngle, BasePose:Pose Link Structure=@n10, Alpha:float=1) -> (Pose:Pose Link Structure) {blendSpace=/Game/Animation/Game/MainPlayer/Vehicles/GolfCart/AO/PassengerBack_AO_SlopeRoll_GolfCart, alphaBoolBlend {blendOption=Linear}}
   n12 = Break Fort Anim Input Golf Cart(FortAnimInput_GolfCart=Get GolfCartVehicleInput()) -> (
    bIsUsingGolfCart=~
    bIsDriver=~
    bIsFrontPassenger=~
    bIsBackLeftPassenger=~
    bIsBackRightPassenger=~
    bIsFrontPassengerAndLeaning=~
    bIsBackPassengerAndLeaning=~
    bIsDrifting=~
    bIsBoosting=~
    bIsEBraking=~
    bIsReversing=~
    bIsBraking=~
    bIsMoving=~
    bIsMovingForward=~
    bIsPowerSliding=~
    bIsLeaning=~
    bIsLeaningOrBouncing=~
    bIsBounceCrouching=~
    bIsBounceCrouched=~
    bIsBounceJumping=~
    bIsBounceRecoiling=~
    bIsSteeringRight=~
    bIsSteeringLeft=~
    RunForwardAlpha=~
    BounceCompression=~
    LeanPosition=~
    LeanPositionX=~
    LeanPositionY=~
    LeanPositionZ=~
    bAimFWD=~
    bAimBWD=~
    bAimLFT=~
    bAimRGT=~
    PawnToVehicleDeltaYawAngleDegrees=~
    AimCardDirDeadZoneAngleDegrees=~
    AimCardDirAngleOffsetDegrees=~
    AimFWDDeltaAngleDegrees=~
    AimBWDDeltaAngleDegrees=~
    AimLFTDeltaAngleDegrees=~
    AimRGTDeltaAngleDegrees=~
    SlopePitchDegreeAngle=~
    SlopeRollDegreeAngle=~
    SteerAngle=~
   )
   n13 = Get TurnInPlaceInput()
   Output Animation Pose(Result=Rotate Root Bone(
     BasePose=Blend Poses by bool(BlendPose_0=@n7, BlendPose_1=@n11, BlendTime_0=0, BlendTime_1=0, bActiveValue=@n12.bIsFrontPassenger) {blendType=Linear}
     Yaw=Break Fort Anim Input Turn in Place(FortAnimInput_TurnInPlace=@n13) -> (
      RootYawOffset=~
      TurnAnimPlayRate=~
      bWantsToTurn=~
      bExitTurn=~
      bUseTurnA=~
      TurnACurveAlpha=~
      TurnBCurveAlpha=~
      bWantsToTurnInVehicle=~
      bWantsToTurnAgain=~
      bTurningLeft=~
      ShufflePlayRate=~
      CrouchShufflePlayRate=~
      CrouchShuffleStartPosition=~
      CrouchShuffleSpeedScaling=~
      bTransition_LowerBody_Shuffle=~
      bTransition_LowerBody_ShuffleADS=~
      bPlayOutro=~
     )
    ) {yawScaleBiasClamp {scale=-1}})
  ]
 }
 GolfCartBouncingLeans {
  path=VehicleFullBodyOverride.AnimGraphNode_StateMachine_0.GolfCart_LeansAndBounce.AnimStateNode_5.GolfCartBouncingLeans
  nodes [
   AnimGraphNode_StateMachine_0 = BouncingLeansState() {is=title AnimGraphNode_StateMachine}
   Output Animation Pose(Result=@AnimGraphNode_StateMachine_0)
  ]
 }
 BouncingLeansState {
  entry=Default
  path=VehicleFullBodyOverride.AnimGraphNode_StateMachine_0.GolfCart_LeansAndBounce.AnimStateNode_5.GolfCartBouncingLeans.AnimGraphNode_StateMachine_0.BouncingLeansState
  ids=id
  nodes [
   AnimStateNode_0 = Default() {is=title AnimStateNode, bAlwaysResetOnEntry=true}
   AnimStateNode_1 = PassengerBR_Leans() {is=title AnimStateNode}
   AnimStateNode_2 = PassengerFront_Leans() {is=title AnimStateNode}
   AnimStateNode_3 = PassengerFront_Bounce() {is=title AnimStateNode}
   AnimStateNode_4 = Bounce_Jump() {is=title AnimStateNode, bAlwaysResetOnEntry=true}
   AnimStateNode_5 = Bounce_Hold_Loop() {is=title AnimStateNode}
   AnimStateNode_6 = PassengerFront_Bounce_Hold_Loop() {is=title AnimStateNode}
   AnimStateTransitionNode_0 = AnimStateTransitionNode()
   AnimStateTransitionNode_1 = AnimStateTransitionNode()
   AnimStateTransitionNode_10 = AnimStateTransitionNode() {crossfadeDuration=0}
   AnimStateTransitionNode_11 = AnimStateTransitionNode() {crossfadeDuration=0}
   AnimStateTransitionNode_19 = AnimStateTransitionNode() {crossfadeDuration=0.033, priorityOrder=2}
   AnimStateTransitionNode_2 = AnimStateTransitionNode() {crossfadeDuration=0.033, priorityOrder=2}
   AnimStateTransitionNode_24 = AnimStateTransitionNode() {bAutomaticRuleBasedOnSequencePlayerInState=true, priorityOrder=2}
   AnimStateTransitionNode_25 = AnimStateTransitionNode() {priorityOrder=2}
   AnimStateTransitionNode_3 = AnimStateTransitionNode() {crossfadeDuration=0}
   AnimStateTransitionNode_30 = AnimStateTransitionNode() {crossfadeDuration=0.15}
   AnimStateTransitionNode_4 = AnimStateTransitionNode() {crossfadeDuration=0}
   AnimStateTransitionNode_5 = AnimStateTransitionNode() {bAutomaticRuleBasedOnSequencePlayerInState=true, priorityOrder=2}
   AnimStateTransitionNode_6 = AnimStateTransitionNode() {crossfadeDuration=0}
   AnimStateTransitionNode_7 = AnimStateTransitionNode() {crossfadeDuration=0}
   AnimStateTransitionNode_8 = AnimStateTransitionNode() {crossfadeDuration=0.15}
   AnimStateTransitionNode_9 = AnimStateTransitionNode() {priorityOrder=2}
  ]
  links [
   {
    dst=Bounce_Jump
    rule=VehicleFullBodyOverride.AnimGraphNode_StateMachine_0.GolfCart_LeansAndBounce.AnimStateNode_5.GolfCartBouncingLeans.AnimGraphNode_StateMachine_0.BouncingLeansState.AnimStateTransitionNode_19.AnimationTransitionGraph_0
    src=Bounce_Hold_Loop
    via=AnimStateTransitionNode_19
   }
   {
    dst=Default
    rule=VehicleFullBodyOverride.AnimGraphNode_StateMachine_0.GolfCart_LeansAndBounce.AnimStateNode_5.GolfCartBouncingLeans.AnimGraphNode_StateMachine_0.BouncingLeansState.AnimStateTransitionNode_11.Transition
    src=Bounce_Hold_Loop
    via=AnimStateTransitionNode_11
   }
   {
    dst=Default
    rule=VehicleFullBodyOverride.AnimGraphNode_StateMachine_0.GolfCart_LeansAndBounce.AnimStateNode_5.GolfCartBouncingLeans.AnimGraphNode_StateMachine_0.BouncingLeansState.AnimStateTransitionNode_10.Transition
    src=Bounce_Jump
    via=AnimStateTransitionNode_10
   }
   {
    dst=PassengerBR_Leans
    rule=VehicleFullBodyOverride.AnimGraphNode_StateMachine_0.GolfCart_LeansAndBounce.AnimStateNode_5.GolfCartBouncingLeans.AnimGraphNode_StateMachine_0.BouncingLeansState.AnimStateTransitionNode_24.AnimationTransitionGraph_0
    src=Bounce_Jump
    via=AnimStateTransitionNode_24
   }
   {
    dst=PassengerBR_Leans
    rule=VehicleFullBodyOverride.AnimGraphNode_StateMachine_0.GolfCart_LeansAndBounce.AnimStateNode_5.GolfCartBouncingLeans.AnimGraphNode_StateMachine_0.BouncingLeansState.AnimStateTransitionNode_25.AnimationTransitionGraph_0
    src=Bounce_Jump
    via=AnimStateTransitionNode_25
   }
   {
    dst=PassengerBR_Leans
    rule=VehicleFullBodyOverride.AnimGraphNode_StateMachine_0.GolfCart_LeansAndBounce.AnimStateNode_5.GolfCartBouncingLeans.AnimGraphNode_StateMachine_0.BouncingLeansState.AnimStateTransitionNode_0.AnimationTransitionGraph_0
    src=Default
    via=AnimStateTransitionNode_0
   }
   {
    dst=PassengerFront_Leans
    rule=VehicleFullBodyOverride.AnimGraphNode_StateMachine_0.GolfCart_LeansAndBounce.AnimStateNode_5.GolfCartBouncingLeans.AnimGraphNode_StateMachine_0.BouncingLeansState.AnimStateTransitionNode_1.AnimationTransitionGraph_0
    src=Default
    via=AnimStateTransitionNode_1
   }
   {
    dst=Bounce_Hold_Loop
    rule=VehicleFullBodyOverride.AnimGraphNode_StateMachine_0.GolfCart_LeansAndBounce.AnimStateNode_5.GolfCartBouncingLeans.AnimGraphNode_StateMachine_0.BouncingLeansState.AnimStateTransitionNode_30.AnimationTransitionGraph_0
    src=PassengerBR_Leans
    via=AnimStateTransitionNode_30
   }
   {
    dst=Default
    rule=VehicleFullBodyOverride.AnimGraphNode_StateMachine_0.GolfCart_LeansAndBounce.AnimStateNode_5.GolfCartBouncingLeans.AnimGraphNode_StateMachine_0.BouncingLeansState.AnimStateTransitionNode_7.Transition
    src=PassengerBR_Leans
    via=AnimStateTransitionNode_7
   }
   {
    dst=Default
    rule=VehicleFullBodyOverride.AnimGraphNode_StateMachine_0.GolfCart_LeansAndBounce.AnimStateNode_5.GolfCartBouncingLeans.AnimGraphNode_StateMachine_0.BouncingLeansState.AnimStateTransitionNode_6.Transition
    src=PassengerFront_Bounce
    via=AnimStateTransitionNode_6
   }
   {
    dst=PassengerFront_Leans
    rule=VehicleFullBodyOverride.AnimGraphNode_StateMachine_0.GolfCart_LeansAndBounce.AnimStateNode_5.GolfCartBouncingLeans.AnimGraphNode_StateMachine_0.BouncingLeansState.AnimStateTransitionNode_5.AnimationTransitionGraph_0
    src=PassengerFront_Bounce
    via=AnimStateTransitionNode_5
   }
   {
    dst=PassengerFront_Leans
    rule=VehicleFullBodyOverride.AnimGraphNode_StateMachine_0.GolfCart_LeansAndBounce.AnimStateNode_5.GolfCartBouncingLeans.AnimGraphNode_StateMachine_0.BouncingLeansState.AnimStateTransitionNode_9.AnimationTransitionGraph_0
    src=PassengerFront_Bounce
    via=AnimStateTransitionNode_9
   }
   {
    dst=Default
    rule=VehicleFullBodyOverride.AnimGraphNode_StateMachine_0.GolfCart_LeansAndBounce.AnimStateNode_5.GolfCartBouncingLeans.AnimGraphNode_StateMachine_0.BouncingLeansState.AnimStateTransitionNode_3.Transition
    src=PassengerFront_Bounce_Hold_Loop
    via=AnimStateTransitionNode_3
   }
   {
    dst=PassengerFront_Bounce
    rule=VehicleFullBodyOverride.AnimGraphNode_StateMachine_0.GolfCart_LeansAndBounce.AnimStateNode_5.GolfCartBouncingLeans.AnimGraphNode_StateMachine_0.BouncingLeansState.AnimStateTransitionNode_2.AnimationTransitionGraph_0
    src=PassengerFront_Bounce_Hold_Loop
    via=AnimStateTransitionNode_2
   }
   {
    dst=Default
    rule=VehicleFullBodyOverride.AnimGraphNode_StateMachine_0.GolfCart_LeansAndBounce.AnimStateNode_5.GolfCartBouncingLeans.AnimGraphNode_StateMachine_0.BouncingLeansState.AnimStateTransitionNode_4.Transition
    src=PassengerFront_Leans
    via=AnimStateTransitionNode_4
   }
   {
    dst=PassengerFront_Bounce_Hold_Loop
    rule=VehicleFullBodyOverride.AnimGraphNode_StateMachine_0.GolfCart_LeansAndBounce.AnimStateNode_5.GolfCartBouncingLeans.AnimGraphNode_StateMachine_0.BouncingLeansState.AnimStateTransitionNode_8.AnimationTransitionGraph_0
    src=PassengerFront_Leans
    via=AnimStateTransitionNode_8
   }
  ]
 }
 Default {
  path=VehicleFullBodyOverride.AnimGraphNode_StateMachine_0.GolfCart_LeansAndBounce.AnimStateNode_5.GolfCartBouncingLeans.AnimGraphNode_StateMachine_0.BouncingLeansState.AnimStateNode_0.Default
  nodes [Output Animation Pose(Result=Use cached pose 'UpperAndLowerBodyCachePose'())]
 }
 PassengerBR_Leans {
  path=VehicleFullBodyOverride.AnimGraphNode_StateMachine_0.GolfCart_LeansAndBounce.AnimStateNode_5.GolfCartBouncingLeans.AnimGraphNode_StateMachine_0.BouncingLeansState.AnimStateNode_1.PassengerBR_Leans
  nodes [
   n1 = Break Fort Anim Input Golf Cart(FortAnimInput_GolfCart=Get GolfCartVehicleInput()) -> (
    bIsUsingGolfCart=~
    bIsDriver=~
    bIsFrontPassenger=~
    bIsBackLeftPassenger=~
    bIsBackRightPassenger=~
    bIsFrontPassengerAndLeaning=~
    bIsBackPassengerAndLeaning=~
    bIsDrifting=~
    bIsBoosting=~
    bIsEBraking=~
    bIsReversing=~
    bIsBraking=~
    bIsMoving=~
    bIsMovingForward=~
    bIsPowerSliding=~
    bIsLeaning=~
    bIsLeaningOrBouncing=~
    bIsBounceCrouching=~
    bIsBounceCrouched=~
    bIsBounceJumping=~
    bIsBounceRecoiling=~
    bIsSteeringRight=~
    bIsSteeringLeft=~
    RunForwardAlpha=~
    BounceCompression=~
    LeanPosition=~
    LeanPositionX=~
    LeanPositionY=~
    LeanPositionZ=~
    bAimFWD=~
    bAimBWD=~
    bAimLFT=~
    bAimRGT=~
    PawnToVehicleDeltaYawAngleDegrees=~
    AimCardDirDeadZoneAngleDegrees=~
    AimCardDirAngleOffsetDegrees=~
    AimFWDDeltaAngleDegrees=~
    AimBWDDeltaAngleDegrees=~
    AimLFTDeltaAngleDegrees=~
    AimRGTDeltaAngleDegrees=~
    SlopePitchDegreeAngle=~
    SlopeRollDegreeAngle=~
    SteerAngle=~
   )
   n2 = "Blend Poses (EFortCustomGender)"!(
    BlendPose_0:Pose Link Structure=Blendspace Player 'PassengerBR_Leans_BlendSpace_GolfCart'(X=@n1.LeanPositionY, Y=@n1.LeanPositionZ) {blendSpace=/Game/Animation/Game/MainPlayer/Vehicles/GolfCart/Blendspaces/PassengerBR_Leans_BlendSpace_GolfCart}
    BlendPose_1:Pose Link Structure=Blendspace Player 'PassengerBR_Leans_BlendSpace_CMF_GolfCart'(X=@n1.LeanPositionY, Y=@n1.LeanPositionZ) {blendSpace=/Game/Animation/Game/MainPlayer/Vehicles/GolfCart/Blendspaces/PassengerBR_Leans_BlendSpace_CMF_GolfCart}
    BlendTime_0:float=0
    BlendTime_1:float=0
    ActiveEnumValue:Byte=Get Gender() -> (Gender:EFortCustomGender Enum=Invalid)
   ) -> (Pose:Pose Link Structure) {
    is=title AnimGraphNode_BlendListByEnum
    blendType=Linear
    blendPose [{linkId=-1, sourceLinkId=-1}, {linkId=-1, sourceLinkId=-1}]
    blendTime [0.1,0.1]
   }
   n3 = Sequence Player 'Passenger_BackRight_Lean_Idle_GolfCart'() {groupName=Idle_Lean, method=SyncGroup, sequence=/Game/Animation/Game/MainPlayer/Vehicles/GolfCart/Passenger_BackRight_Lean_Idle_GolfCart}
   n4 = Sequence Player 'Passenger_BackRight_Lean_Idle_CMF_GolfCart'() {groupName=Idle_Lean, method=SyncGroup, sequence=/Game/Animation/Game/MainPlayer/Vehicles/GolfCart/Passenger_BackRight_Lean_Idle_CMF_GolfCart}
   n5 = "Blend Poses (EFortCustomGender)"(BlendPose_0=@n3, BlendPose_1=@n4, BlendTime_0=0, BlendTime_1=0, ActiveEnumValue=Get Gender() -> (Gender=Invalid)) {blendType=Linear, blendPose [{linkId=-1, sourceLinkId=-1}, {linkId=-1, sourceLinkId=-1}], blendTime [0.1,0.1]}
   Output Animation Pose(Result=Apply Additive(Base=@n5, Additive=@n2) {alphaBoolBlend {blendOption=Linear}})
  ]
 }
 PassengerFront_Leans {
  path=VehicleFullBodyOverride.AnimGraphNode_StateMachine_0.GolfCart_LeansAndBounce.AnimStateNode_5.GolfCartBouncingLeans.AnimGraphNode_StateMachine_0.BouncingLeansState.AnimStateNode_2.PassengerFront_Leans
  nodes [
   n1 = Break Fort Anim Input Golf Cart(FortAnimInput_GolfCart=Get GolfCartVehicleInput()) -> (
    bIsUsingGolfCart=~
    bIsDriver=~
    bIsFrontPassenger=~
    bIsBackLeftPassenger=~
    bIsBackRightPassenger=~
    bIsFrontPassengerAndLeaning=~
    bIsBackPassengerAndLeaning=~
    bIsDrifting=~
    bIsBoosting=~
    bIsEBraking=~
    bIsReversing=~
    bIsBraking=~
    bIsMoving=~
    bIsMovingForward=~
    bIsPowerSliding=~
    bIsLeaning=~
    bIsLeaningOrBouncing=~
    bIsBounceCrouching=~
    bIsBounceCrouched=~
    bIsBounceJumping=~
    bIsBounceRecoiling=~
    bIsSteeringRight=~
    bIsSteeringLeft=~
    RunForwardAlpha=~
    BounceCompression=~
    LeanPosition=~
    LeanPositionX=~
    LeanPositionY=~
    LeanPositionZ=~
    bAimFWD=~
    bAimBWD=~
    bAimLFT=~
    bAimRGT=~
    PawnToVehicleDeltaYawAngleDegrees=~
    AimCardDirDeadZoneAngleDegrees=~
    AimCardDirAngleOffsetDegrees=~
    AimFWDDeltaAngleDegrees=~
    AimBWDDeltaAngleDegrees=~
    AimLFTDeltaAngleDegrees=~
    AimRGTDeltaAngleDegrees=~
    SlopePitchDegreeAngle=~
    SlopeRollDegreeAngle=~
    SteerAngle=~
   )
   n2 = Blendspace Player 'PassengerFront_Leans_BlendSpace_GolfCart'(X=@n1.LeanPositionY, Y=@n1.LeanPositionZ) {
    blendSpace=/Game/Animation/Game/MainPlayer/Vehicles/GolfCart/Blendspaces/PassengerFront_Leans_BlendSpace_GolfCart
    groupName=FrontPaseengerLean
    method=SyncGroup
   }
   n3 = Sequence Player 'Passenger_Front_Lean_Idle_GolfCart'() {groupName=FrontLeanIdle, method=SyncGroup, sequence=/Game/Animation/Game/MainPlayer/Vehicles/GolfCart/Passenger_Front_Lean_Idle_GolfCart}
   Output Animation Pose(Result=Apply Additive(Base=@n3, Additive=@n2) {alphaBoolBlend {blendOption=Linear}})
  ]
 }
 PassengerFront_Bounce {
  path=VehicleFullBodyOverride.AnimGraphNode_StateMachine_0.GolfCart_LeansAndBounce.AnimStateNode_5.GolfCartBouncingLeans.AnimGraphNode_StateMachine_0.BouncingLeansState.AnimStateNode_3.PassengerFront_Bounce
  nodes [Output Animation Pose(Result=Sequence Player 'Passenger_Front_Bounce_GolfCart'() {bLoopAnimation=false, sequence=/Game/Animation/Game/MainPlayer/Vehicles/GolfCart/Passenger_Front_Bounce_GolfCart})]
 }
 Bounce_Jump {
  path=VehicleFullBodyOverride.AnimGraphNode_StateMachine_0.GolfCart_LeansAndBounce.AnimStateNode_5.GolfCartBouncingLeans.AnimGraphNode_StateMachine_0.BouncingLeansState.AnimStateNode_4.Bounce_Jump
  nodes [
   n1 = "Blend Poses (EFortCustomGender)"(
    BlendPose_0=Sequence Player 'Passenger_BackRight_Bounce_Jump_GolfCart'() {bLoopAnimation=false, sequence=/Game/Animation/Game/MainPlayer/Vehicles/GolfCart/Passenger_BackRight_Bounce_Jump_GolfCart}
    BlendPose_1=Sequence Player 'Passenger_BackRight_Bounce_Jump_CMF_GolfCart'() {bLoopAnimation=false, sequence=/Game/Animation/Game/MainPlayer/Vehicles/GolfCart/Passenger_BackRight_Bounce_Jump_CMF_GolfCart}
    BlendTime_0=0
    BlendTime_1=0
    ActiveEnumValue=Get Gender() -> (Gender=Invalid)
   ) {blendType=Linear, blendPose [{linkId=-1, sourceLinkId=-1}, {linkId=-1, sourceLinkId=-1}], blendTime [0.1,0.1]}
   Output Animation Pose(Result=@n1)
  ]
 }
 Bounce_Hold_Loop {
  path=VehicleFullBodyOverride.AnimGraphNode_StateMachine_0.GolfCart_LeansAndBounce.AnimStateNode_5.GolfCartBouncingLeans.AnimGraphNode_StateMachine_0.BouncingLeansState.AnimStateNode_5.Bounce_Hold_Loop
  nodes [
   n1 = Sequence Player 'Passenger_BackRight_Lean_Idle_Crouched_GolfCart'() {
    groupName=Idle_Lean
    method=SyncGroup
    sequence=/Game/Animation/Game/MainPlayer/Vehicles/GolfCart/Passenger_BackRight_Lean_Idle_Crouched_GolfCart
   }
   n2 = Sequence Player 'Passenger_BackRight_Lean_Idle_Crouched_CMF_GolfCart'() {
    groupName=Idle_Lean
    method=SyncGroup
    sequence=/Game/Animation/Game/MainPlayer/Vehicles/GolfCart/Passenger_BackRight_Lean_Idle_Crouched_CMF_GolfCart
   }
   n3 = "Blend Poses (EFortCustomGender)"(BlendPose_0=@n1, BlendPose_1=@n2, BlendTime_0=0, BlendTime_1=0, ActiveEnumValue=Get Gender() -> (Gender=Invalid)) {blendType=Linear, blendPose [{linkId=-1, sourceLinkId=-1}, {linkId=-1, sourceLinkId=-1}], blendTime [0.1,0.1]}
   n4 = Break Fort Anim Input Golf Cart(FortAnimInput_GolfCart=Get GolfCartVehicleInput()) -> (
    bIsUsingGolfCart=~
    bIsDriver=~
    bIsFrontPassenger=~
    bIsBackLeftPassenger=~
    bIsBackRightPassenger=~
    bIsFrontPassengerAndLeaning=~
    bIsBackPassengerAndLeaning=~
    bIsDrifting=~
    bIsBoosting=~
    bIsEBraking=~
    bIsReversing=~
    bIsBraking=~
    bIsMoving=~
    bIsMovingForward=~
    bIsPowerSliding=~
    bIsLeaning=~
    bIsLeaningOrBouncing=~
    bIsBounceCrouching=~
    bIsBounceCrouched=~
    bIsBounceJumping=~
    bIsBounceRecoiling=~
    bIsSteeringRight=~
    bIsSteeringLeft=~
    RunForwardAlpha=~
    BounceCompression=~
    LeanPosition=~
    LeanPositionX=~
    LeanPositionY=~
    LeanPositionZ=~
    bAimFWD=~
    bAimBWD=~
    bAimLFT=~
    bAimRGT=~
    PawnToVehicleDeltaYawAngleDegrees=~
    AimCardDirDeadZoneAngleDegrees=~
    AimCardDirAngleOffsetDegrees=~
    AimFWDDeltaAngleDegrees=~
    AimBWDDeltaAngleDegrees=~
    AimLFTDeltaAngleDegrees=~
    AimRGTDeltaAngleDegrees=~
    SlopePitchDegreeAngle=~
    SlopeRollDegreeAngle=~
    SteerAngle=~
   )
   n5 = "Blend Poses (EFortCustomGender)"(
    BlendPose_0=Blendspace Player 'PassengerBR_Leans_Crouched_BlendSpace_GolfCart'(X=@n4.LeanPositionY, Y=@n4.LeanPositionZ) {blendSpace=/Game/Animation/Game/MainPlayer/Vehicles/GolfCart/Blendspaces/PassengerBR_Leans_Crouched_BlendSpace_GolfCart}
    BlendPose_1=Blendspace Player 'PassengerBR_Leans_Crouched_BlendSpace_CMF_GolfCart'!(X=@n4.LeanPositionY) -> (Pose) {blendSpace=/Game/Animation/Game/MainPlayer/Vehicles/GolfCart/Blendspaces/PassengerBR_Leans_Crouched_BlendSpace_CMF_GolfCart}
    BlendTime_0=0
    BlendTime_1=0
    ActiveEnumValue=Get Gender() -> (Gender=Invalid)
   ) {blendType=Linear, blendPose [{linkId=-1, sourceLinkId=-1}, {linkId=-1, sourceLinkId=-1}], blendTime [0.1,0.1]}
   Output Animation Pose(Result=Apply Additive(Base=@n3, Additive=@n5) {alphaBoolBlend {blendOption=Linear}})
  ]
 }
 PassengerFront_Bounce_Hold_Loop {
  path=VehicleFullBodyOverride.AnimGraphNode_StateMachine_0.GolfCart_LeansAndBounce.AnimStateNode_5.GolfCartBouncingLeans.AnimGraphNode_StateMachine_0.BouncingLeansState.AnimStateNode_6.PassengerFront_Bounce_Hold_Loop
  nodes [
   n1 = Blendspace Player 'PassengerFront_Leans_Crouched_BlendSpace_GolfCart'!(X=Break Fort Anim Input Golf Cart(FortAnimInput_GolfCart=Get GolfCartVehicleInput()) -> (
     bIsUsingGolfCart=~
     bIsDriver=~
     bIsFrontPassenger=~
     bIsBackLeftPassenger=~
     bIsBackRightPassenger=~
     bIsFrontPassengerAndLeaning=~
     bIsBackPassengerAndLeaning=~
     bIsDrifting=~
     bIsBoosting=~
     bIsEBraking=~
     bIsReversing=~
     bIsBraking=~
     bIsMoving=~
     bIsMovingForward=~
     bIsPowerSliding=~
     bIsLeaning=~
     bIsLeaningOrBouncing=~
     bIsBounceCrouching=~
     bIsBounceCrouched=~
     bIsBounceJumping=~
     bIsBounceRecoiling=~
     bIsSteeringRight=~
     bIsSteeringLeft=~
     RunForwardAlpha=~
     BounceCompression=~
     LeanPosition=~
     LeanPositionX=~
     LeanPositionY=~
     LeanPositionZ=~
     bAimFWD=~
     bAimBWD=~
     bAimLFT=~
     bAimRGT=~
     PawnToVehicleDeltaYawAngleDegrees=~
     AimCardDirDeadZoneAngleDegrees=~
     AimCardDirAngleOffsetDegrees=~
     AimFWDDeltaAngleDegrees=~
     AimBWDDeltaAngleDegrees=~
     AimLFTDeltaAngleDegrees=~
     AimRGTDeltaAngleDegrees=~
     SlopePitchDegreeAngle=~
     SlopeRollDegreeAngle=~
     SteerAngle=~
    ).LeanPositionY) -> (Pose) {
    blendSpace=/Game/Animation/Game/MainPlayer/Vehicles/GolfCart/Blendspaces/PassengerFront_Leans_Crouched_BlendSpace_GolfCart
    groupName=FrontPaseengerLean
    method=SyncGroup
   }
   n2 = Sequence Player 'Passenger_Front_Lean_Idle_Crouched_GolfCart'() {
    groupName=FrontLeanIdle
    method=SyncGroup
    sequence=/Game/Animation/Game/MainPlayer/Vehicles/GolfCart/Passenger_Front_Lean_Idle_Crouched_GolfCart
   }
   Output Animation Pose(Result=Apply Additive(Base=@n2, Additive=@n1) {alphaBoolBlend {blendOption=Linear}})
  ]
 }
 AnimationTransitionGraph_0 {
  path=VehicleFullBodyOverride.AnimGraphNode_StateMachine_0.GolfCart_LeansAndBounce.AnimStateNode_5.GolfCartBouncingLeans.AnimGraphNode_StateMachine_0.BouncingLeansState.AnimStateTransitionNode_0.AnimationTransitionGraph_0
  nodes [Result(bCanEnterTransition=Break Fort Anim Input Golf Cart(FortAnimInput_GolfCart=Get GolfCartVehicleInput()) -> (
     bIsUsingGolfCart=~
     bIsDriver=~
     bIsFrontPassenger=~
     bIsBackLeftPassenger=~
     bIsBackRightPassenger=~
     bIsFrontPassengerAndLeaning=~
     bIsBackPassengerAndLeaning=~
     bIsDrifting=~
     bIsBoosting=~
     bIsEBraking=~
     bIsReversing=~
     bIsBraking=~
     bIsMoving=~
     bIsMovingForward=~
     bIsPowerSliding=~
     bIsLeaning=~
     bIsLeaningOrBouncing=~
     bIsBounceCrouching=~
     bIsBounceCrouched=~
     bIsBounceJumping=~
     bIsBounceRecoiling=~
     bIsSteeringRight=~
     bIsSteeringLeft=~
     RunForwardAlpha=~
     BounceCompression=~
     LeanPosition=~
     LeanPositionX=~
     LeanPositionY=~
     LeanPositionZ=~
     bAimFWD=~
     bAimBWD=~
     bAimLFT=~
     bAimRGT=~
     PawnToVehicleDeltaYawAngleDegrees=~
     AimCardDirDeadZoneAngleDegrees=~
     AimCardDirAngleOffsetDegrees=~
     AimFWDDeltaAngleDegrees=~
     AimBWDDeltaAngleDegrees=~
     AimLFTDeltaAngleDegrees=~
     AimRGTDeltaAngleDegrees=~
     SlopePitchDegreeAngle=~
     SlopeRollDegreeAngle=~
     SteerAngle=~
    ).bIsBackPassengerAndLeaning)]
 }
 AnimationTransitionGraph_0 {
  path=VehicleFullBodyOverride.AnimGraphNode_StateMachine_0.GolfCart_LeansAndBounce.AnimStateNode_5.GolfCartBouncingLeans.AnimGraphNode_StateMachine_0.BouncingLeansState.AnimStateTransitionNode_1.AnimationTransitionGraph_0
  nodes [Result(bCanEnterTransition=Break Fort Anim Input Golf Cart(FortAnimInput_GolfCart=Get GolfCartVehicleInput()) -> (
     bIsUsingGolfCart=~
     bIsDriver=~
     bIsFrontPassenger=~
     bIsBackLeftPassenger=~
     bIsBackRightPassenger=~
     bIsFrontPassengerAndLeaning=~
     bIsBackPassengerAndLeaning=~
     bIsDrifting=~
     bIsBoosting=~
     bIsEBraking=~
     bIsReversing=~
     bIsBraking=~
     bIsMoving=~
     bIsMovingForward=~
     bIsPowerSliding=~
     bIsLeaning=~
     bIsLeaningOrBouncing=~
     bIsBounceCrouching=~
     bIsBounceCrouched=~
     bIsBounceJumping=~
     bIsBounceRecoiling=~
     bIsSteeringRight=~
     bIsSteeringLeft=~
     RunForwardAlpha=~
     BounceCompression=~
     LeanPosition=~
     LeanPositionX=~
     LeanPositionY=~
     LeanPositionZ=~
     bAimFWD=~
     bAimBWD=~
     bAimLFT=~
     bAimRGT=~
     PawnToVehicleDeltaYawAngleDegrees=~
     AimCardDirDeadZoneAngleDegrees=~
     AimCardDirAngleOffsetDegrees=~
     AimFWDDeltaAngleDegrees=~
     AimBWDDeltaAngleDegrees=~
     AimLFTDeltaAngleDegrees=~
     AimRGTDeltaAngleDegrees=~
     SlopePitchDegreeAngle=~
     SlopeRollDegreeAngle=~
     SteerAngle=~
    ).bIsFrontPassengerAndLeaning)]
 }
 Transition {
  path=VehicleFullBodyOverride.AnimGraphNode_StateMachine_0.GolfCart_LeansAndBounce.AnimStateNode_5.GolfCartBouncingLeans.AnimGraphNode_StateMachine_0.BouncingLeansState.AnimStateTransitionNode_10.Transition
  nodes [Result(bCanEnterTransition=Get bRecentlyFired() -> (bRecentlyFired:bool))]
 }
 Transition {
  path=VehicleFullBodyOverride.AnimGraphNode_StateMachine_0.GolfCart_LeansAndBounce.AnimStateNode_5.GolfCartBouncingLeans.AnimGraphNode_StateMachine_0.BouncingLeansState.AnimStateTransitionNode_11.Transition
  nodes [Result(bCanEnterTransition=Get bRecentlyFired())]
 }
 AnimationTransitionGraph_0 {
  path=VehicleFullBodyOverride.AnimGraphNode_StateMachine_0.GolfCart_LeansAndBounce.AnimStateNode_5.GolfCartBouncingLeans.AnimGraphNode_StateMachine_0.BouncingLeansState.AnimStateTransitionNode_19.AnimationTransitionGraph_0
  nodes [Result(bCanEnterTransition=Break Fort Anim Input Golf Cart(FortAnimInput_GolfCart=Get GolfCartVehicleInput()) -> (
     bIsUsingGolfCart=~
     bIsDriver=~
     bIsFrontPassenger=~
     bIsBackLeftPassenger=~
     bIsBackRightPassenger=~
     bIsFrontPassengerAndLeaning=~
     bIsBackPassengerAndLeaning=~
     bIsDrifting=~
     bIsBoosting=~
     bIsEBraking=~
     bIsReversing=~
     bIsBraking=~
     bIsMoving=~
     bIsMovingForward=~
     bIsPowerSliding=~
     bIsLeaning=~
     bIsLeaningOrBouncing=~
     bIsBounceCrouching=~
     bIsBounceCrouched=~
     bIsBounceJumping=~
     bIsBounceRecoiling=~
     bIsSteeringRight=~
     bIsSteeringLeft=~
     RunForwardAlpha=~
     BounceCompression=~
     LeanPosition=~
     LeanPositionX=~
     LeanPositionY=~
     LeanPositionZ=~
     bAimFWD=~
     bAimBWD=~
     bAimLFT=~
     bAimRGT=~
     PawnToVehicleDeltaYawAngleDegrees=~
     AimCardDirDeadZoneAngleDegrees=~
     AimCardDirAngleOffsetDegrees=~
     AimFWDDeltaAngleDegrees=~
     AimBWDDeltaAngleDegrees=~
     AimLFTDeltaAngleDegrees=~
     AimRGTDeltaAngleDegrees=~
     SlopePitchDegreeAngle=~
     SlopeRollDegreeAngle=~
     SteerAngle=~
    ).bIsBounceJumping)]
 }
 AnimationTransitionGraph_0 {
  path=VehicleFullBodyOverride.AnimGraphNode_StateMachine_0.GolfCart_LeansAndBounce.AnimStateNode_5.GolfCartBouncingLeans.AnimGraphNode_StateMachine_0.BouncingLeansState.AnimStateTransitionNode_2.AnimationTransitionGraph_0
  nodes [Result(bCanEnterTransition=Break Fort Anim Input Golf Cart(FortAnimInput_GolfCart=Get GolfCartVehicleInput()) -> (
     bIsUsingGolfCart=~
     bIsDriver=~
     bIsFrontPassenger=~
     bIsBackLeftPassenger=~
     bIsBackRightPassenger=~
     bIsFrontPassengerAndLeaning=~
     bIsBackPassengerAndLeaning=~
     bIsDrifting=~
     bIsBoosting=~
     bIsEBraking=~
     bIsReversing=~
     bIsBraking=~
     bIsMoving=~
     bIsMovingForward=~
     bIsPowerSliding=~
     bIsLeaning=~
     bIsLeaningOrBouncing=~
     bIsBounceCrouching=~
     bIsBounceCrouched=~
     bIsBounceJumping=~
     bIsBounceRecoiling=~
     bIsSteeringRight=~
     bIsSteeringLeft=~
     RunForwardAlpha=~
     BounceCompression=~
     LeanPosition=~
     LeanPositionX=~
     LeanPositionY=~
     LeanPositionZ=~
     bAimFWD=~
     bAimBWD=~
     bAimLFT=~
     bAimRGT=~
     PawnToVehicleDeltaYawAngleDegrees=~
     AimCardDirDeadZoneAngleDegrees=~
     AimCardDirAngleOffsetDegrees=~
     AimFWDDeltaAngleDegrees=~
     AimBWDDeltaAngleDegrees=~
     AimLFTDeltaAngleDegrees=~
     AimRGTDeltaAngleDegrees=~
     SlopePitchDegreeAngle=~
     SlopeRollDegreeAngle=~
     SteerAngle=~
    ).bIsBounceJumping)]
 }
 AnimationTransitionGraph_0 {
  path=VehicleFullBodyOverride.AnimGraphNode_StateMachine_0.GolfCart_LeansAndBounce.AnimStateNode_5.GolfCartBouncingLeans.AnimGraphNode_StateMachine_0.BouncingLeansState.AnimStateTransitionNode_24.AnimationTransitionGraph_0
  nodes [Result()]
 }
 AnimationTransitionGraph_0 {
  path=VehicleFullBodyOverride.AnimGraphNode_StateMachine_0.GolfCart_LeansAndBounce.AnimStateNode_5.GolfCartBouncingLeans.AnimGraphNode_StateMachine_0.BouncingLeansState.AnimStateTransitionNode_25.AnimationTransitionGraph_0
  nodes [Result(bCanEnterTransition=Break Fort Anim Input Golf Cart(FortAnimInput_GolfCart=Get GolfCartVehicleInput()) -> (
     bIsUsingGolfCart=~
     bIsDriver=~
     bIsFrontPassenger=~
     bIsBackLeftPassenger=~
     bIsBackRightPassenger=~
     bIsFrontPassengerAndLeaning=~
     bIsBackPassengerAndLeaning=~
     bIsDrifting=~
     bIsBoosting=~
     bIsEBraking=~
     bIsReversing=~
     bIsBraking=~
     bIsMoving=~
     bIsMovingForward=~
     bIsPowerSliding=~
     bIsLeaning=~
     bIsLeaningOrBouncing=~
     bIsBounceCrouching=~
     bIsBounceCrouched=~
     bIsBounceJumping=~
     bIsBounceRecoiling=~
     bIsSteeringRight=~
     bIsSteeringLeft=~
     RunForwardAlpha=~
     BounceCompression=~
     LeanPosition=~
     LeanPositionX=~
     LeanPositionY=~
     LeanPositionZ=~
     bAimFWD=~
     bAimBWD=~
     bAimLFT=~
     bAimRGT=~
     PawnToVehicleDeltaYawAngleDegrees=~
     AimCardDirDeadZoneAngleDegrees=~
     AimCardDirAngleOffsetDegrees=~
     AimFWDDeltaAngleDegrees=~
     AimBWDDeltaAngleDegrees=~
     AimLFTDeltaAngleDegrees=~
     AimRGTDeltaAngleDegrees=~
     SlopePitchDegreeAngle=~
     SlopeRollDegreeAngle=~
     SteerAngle=~
    ).bIsBounceCrouching)]
 }
 Transition {
  path=VehicleFullBodyOverride.AnimGraphNode_StateMachine_0.GolfCart_LeansAndBounce.AnimStateNode_5.GolfCartBouncingLeans.AnimGraphNode_StateMachine_0.BouncingLeansState.AnimStateTransitionNode_3.Transition
  nodes [Result(bCanEnterTransition=Get bRecentlyFired())]
 }
 AnimationTransitionGraph_0 {
  path=VehicleFullBodyOverride.AnimGraphNode_StateMachine_0.GolfCart_LeansAndBounce.AnimStateNode_5.GolfCartBouncingLeans.AnimGraphNode_StateMachine_0.BouncingLeansState.AnimStateTransitionNode_30.AnimationTransitionGraph_0
  nodes [Result(bCanEnterTransition=Break Fort Anim Input Golf Cart(FortAnimInput_GolfCart=Get GolfCartVehicleInput()) -> (
     bIsUsingGolfCart=~
     bIsDriver=~
     bIsFrontPassenger=~
     bIsBackLeftPassenger=~
     bIsBackRightPassenger=~
     bIsFrontPassengerAndLeaning=~
     bIsBackPassengerAndLeaning=~
     bIsDrifting=~
     bIsBoosting=~
     bIsEBraking=~
     bIsReversing=~
     bIsBraking=~
     bIsMoving=~
     bIsMovingForward=~
     bIsPowerSliding=~
     bIsLeaning=~
     bIsLeaningOrBouncing=~
     bIsBounceCrouching=~
     bIsBounceCrouched=~
     bIsBounceJumping=~
     bIsBounceRecoiling=~
     bIsSteeringRight=~
     bIsSteeringLeft=~
     RunForwardAlpha=~
     BounceCompression=~
     LeanPosition=~
     LeanPositionX=~
     LeanPositionY=~
     LeanPositionZ=~
     bAimFWD=~
     bAimBWD=~
     bAimLFT=~
     bAimRGT=~
     PawnToVehicleDeltaYawAngleDegrees=~
     AimCardDirDeadZoneAngleDegrees=~
     AimCardDirAngleOffsetDegrees=~
     AimFWDDeltaAngleDegrees=~
     AimBWDDeltaAngleDegrees=~
     AimLFTDeltaAngleDegrees=~
     AimRGTDeltaAngleDegrees=~
     SlopePitchDegreeAngle=~
     SlopeRollDegreeAngle=~
     SteerAngle=~
    ).bIsBounceCrouching)]
 }
 Transition {
  path=VehicleFullBodyOverride.AnimGraphNode_StateMachine_0.GolfCart_LeansAndBounce.AnimStateNode_5.GolfCartBouncingLeans.AnimGraphNode_StateMachine_0.BouncingLeansState.AnimStateTransitionNode_4.Transition
  nodes [Result(bCanEnterTransition=Get bRecentlyFired())]
 }
 AnimationTransitionGraph_0 {
  path=VehicleFullBodyOverride.AnimGraphNode_StateMachine_0.GolfCart_LeansAndBounce.AnimStateNode_5.GolfCartBouncingLeans.AnimGraphNode_StateMachine_0.BouncingLeansState.AnimStateTransitionNode_5.AnimationTransitionGraph_0
  nodes [Result()]
 }
 Transition {
  path=VehicleFullBodyOverride.AnimGraphNode_StateMachine_0.GolfCart_LeansAndBounce.AnimStateNode_5.GolfCartBouncingLeans.AnimGraphNode_StateMachine_0.BouncingLeansState.AnimStateTransitionNode_6.Transition
  nodes [Result(bCanEnterTransition=Get bRecentlyFired())]
 }
 Transition {
  path=VehicleFullBodyOverride.AnimGraphNode_StateMachine_0.GolfCart_LeansAndBounce.AnimStateNode_5.GolfCartBouncingLeans.AnimGraphNode_StateMachine_0.BouncingLeansState.AnimStateTransitionNode_7.Transition
  nodes [Result(bCanEnterTransition=Get bRecentlyFired())]
 }
 AnimationTransitionGraph_0 {
  path=VehicleFullBodyOverride.AnimGraphNode_StateMachine_0.GolfCart_LeansAndBounce.AnimStateNode_5.GolfCartBouncingLeans.AnimGraphNode_StateMachine_0.BouncingLeansState.AnimStateTransitionNode_8.AnimationTransitionGraph_0
  nodes [Result(bCanEnterTransition=Break Fort Anim Input Golf Cart(FortAnimInput_GolfCart=Get GolfCartVehicleInput()) -> (
     bIsUsingGolfCart=~
     bIsDriver=~
     bIsFrontPassenger=~
     bIsBackLeftPassenger=~
     bIsBackRightPassenger=~
     bIsFrontPassengerAndLeaning=~
     bIsBackPassengerAndLeaning=~
     bIsDrifting=~
     bIsBoosting=~
     bIsEBraking=~
     bIsReversing=~
     bIsBraking=~
     bIsMoving=~
     bIsMovingForward=~
     bIsPowerSliding=~
     bIsLeaning=~
     bIsLeaningOrBouncing=~
     bIsBounceCrouching=~
     bIsBounceCrouched=~
     bIsBounceJumping=~
     bIsBounceRecoiling=~
     bIsSteeringRight=~
     bIsSteeringLeft=~
     RunForwardAlpha=~
     BounceCompression=~
     LeanPosition=~
     LeanPositionX=~
     LeanPositionY=~
     LeanPositionZ=~
     bAimFWD=~
     bAimBWD=~
     bAimLFT=~
     bAimRGT=~
     PawnToVehicleDeltaYawAngleDegrees=~
     AimCardDirDeadZoneAngleDegrees=~
     AimCardDirAngleOffsetDegrees=~
     AimFWDDeltaAngleDegrees=~
     AimBWDDeltaAngleDegrees=~
     AimLFTDeltaAngleDegrees=~
     AimRGTDeltaAngleDegrees=~
     SlopePitchDegreeAngle=~
     SlopeRollDegreeAngle=~
     SteerAngle=~
    ).bIsBounceCrouching)]
 }
 AnimationTransitionGraph_0 {
  path=VehicleFullBodyOverride.AnimGraphNode_StateMachine_0.GolfCart_LeansAndBounce.AnimStateNode_5.GolfCartBouncingLeans.AnimGraphNode_StateMachine_0.BouncingLeansState.AnimStateTransitionNode_9.AnimationTransitionGraph_0
  nodes [Result(bCanEnterTransition=Break Fort Anim Input Golf Cart(FortAnimInput_GolfCart=Get GolfCartVehicleInput()) -> (
     bIsUsingGolfCart=~
     bIsDriver=~
     bIsFrontPassenger=~
     bIsBackLeftPassenger=~
     bIsBackRightPassenger=~
     bIsFrontPassengerAndLeaning=~
     bIsBackPassengerAndLeaning=~
     bIsDrifting=~
     bIsBoosting=~
     bIsEBraking=~
     bIsReversing=~
     bIsBraking=~
     bIsMoving=~
     bIsMovingForward=~
     bIsPowerSliding=~
     bIsLeaning=~
     bIsLeaningOrBouncing=~
     bIsBounceCrouching=~
     bIsBounceCrouched=~
     bIsBounceJumping=~
     bIsBounceRecoiling=~
     bIsSteeringRight=~
     bIsSteeringLeft=~
     RunForwardAlpha=~
     BounceCompression=~
     LeanPosition=~
     LeanPositionX=~
     LeanPositionY=~
     LeanPositionZ=~
     bAimFWD=~
     bAimBWD=~
     bAimLFT=~
     bAimRGT=~
     PawnToVehicleDeltaYawAngleDegrees=~
     AimCardDirDeadZoneAngleDegrees=~
     AimCardDirAngleOffsetDegrees=~
     AimFWDDeltaAngleDegrees=~
     AimBWDDeltaAngleDegrees=~
     AimLFTDeltaAngleDegrees=~
     AimRGTDeltaAngleDegrees=~
     SlopePitchDegreeAngle=~
     SlopeRollDegreeAngle=~
     SteerAngle=~
    ).bIsBounceCrouching)]
 }
 Transition {
  path=VehicleFullBodyOverride.AnimGraphNode_StateMachine_0.GolfCart_LeansAndBounce.AnimStateTransitionNode_10.Transition
  nodes [Result(bCanEnterTransition=Break Fort Anim Input Golf Cart(FortAnimInput_GolfCart=Get GolfCartVehicleInput()) -> (
     bIsUsingGolfCart=~
     bIsDriver=~
     bIsFrontPassenger=~
     bIsBackLeftPassenger=~
     bIsBackRightPassenger=~
     bIsFrontPassengerAndLeaning=~
     bIsBackPassengerAndLeaning=~
     bIsDrifting=~
     bIsBoosting=~
     bIsEBraking=~
     bIsReversing=~
     bIsBraking=~
     bIsMoving=~
     bIsMovingForward=~
     bIsPowerSliding=~
     bIsLeaning=~
     bIsLeaningOrBouncing=~
     bIsBounceCrouching=~
     bIsBounceCrouched=~
     bIsBounceJumping=~
     bIsBounceRecoiling=~
     bIsSteeringRight=~
     bIsSteeringLeft=~
     RunForwardAlpha=~
     BounceCompression=~
     LeanPosition=~
     LeanPositionX=~
     LeanPositionY=~
     LeanPositionZ=~
     bAimFWD=~
     bAimBWD=~
     bAimLFT=~
     bAimRGT=~
     PawnToVehicleDeltaYawAngleDegrees=~
     AimCardDirDeadZoneAngleDegrees=~
     AimCardDirAngleOffsetDegrees=~
     AimFWDDeltaAngleDegrees=~
     AimBWDDeltaAngleDegrees=~
     AimLFTDeltaAngleDegrees=~
     AimRGTDeltaAngleDegrees=~
     SlopePitchDegreeAngle=~
     SlopeRollDegreeAngle=~
     SteerAngle=~
    ).bIsLeaningOrBouncing)]
 }
 Transition {
  path=VehicleFullBodyOverride.AnimGraphNode_StateMachine_0.GolfCart_LeansAndBounce.AnimStateTransitionNode_11.Transition
  nodes [
   n1 = Get GolfCartVehicleInput()
   Result(bCanEnterTransition=NOT Boolean(A=Break Fort Anim Input Golf Cart(FortAnimInput_GolfCart=@n1) -> (
      bIsUsingGolfCart=~
      bIsDriver=~
      bIsFrontPassenger=~
      bIsBackLeftPassenger=~
      bIsBackRightPassenger=~
      bIsFrontPassengerAndLeaning=~
      bIsBackPassengerAndLeaning=~
      bIsDrifting=~
      bIsBoosting=~
      bIsEBraking=~
      bIsReversing=~
      bIsBraking=~
      bIsMoving=~
      bIsMovingForward=~
      bIsPowerSliding=~
      bIsLeaning=~
      bIsLeaningOrBouncing=~
      bIsBounceCrouching=~
      bIsBounceCrouched=~
      bIsBounceJumping=~
      bIsBounceRecoiling=~
      bIsSteeringRight=~
      bIsSteeringLeft=~
      RunForwardAlpha=~
      BounceCompression=~
      LeanPosition=~
      LeanPositionX=~
      LeanPositionY=~
      LeanPositionZ=~
      bAimFWD=~
      bAimBWD=~
      bAimLFT=~
      bAimRGT=~
      PawnToVehicleDeltaYawAngleDegrees=~
      AimCardDirDeadZoneAngleDegrees=~
      AimCardDirAngleOffsetDegrees=~
      AimFWDDeltaAngleDegrees=~
      AimBWDDeltaAngleDegrees=~
      AimLFTDeltaAngleDegrees=~
      AimRGTDeltaAngleDegrees=~
      SlopePitchDegreeAngle=~
      SlopeRollDegreeAngle=~
      SteerAngle=~
     ).bIsLeaningOrBouncing))
  ]
 }
 VehicleLowerBodyOverride {nodes [
   AnimGraphNode_StateMachine_0 = GolfCartFrontPassenger() {is=title AnimGraphNode_StateMachine}
   n1 = Get GolfCartVehicleInput()
   Output Pose(Result=Blend Poses by bool(
     BlendPose_0=@AnimGraphNode_StateMachine_0
     BlendPose_1=Input Pose() {name=InPoseLowerBodyDefault}
     bActiveValue=Break Fort Anim Input Golf Cart(FortAnimInput_GolfCart=@n1) -> (
      bIsUsingGolfCart=~
      bIsDriver=~
      bIsFrontPassenger=~
      bIsBackLeftPassenger=~
      bIsBackRightPassenger=~
      bIsFrontPassengerAndLeaning=~
      bIsBackPassengerAndLeaning=~
      bIsDrifting=~
      bIsBoosting=~
      bIsEBraking=~
      bIsReversing=~
      bIsBraking=~
      bIsMoving=~
      bIsMovingForward=~
      bIsPowerSliding=~
      bIsLeaning=~
      bIsLeaningOrBouncing=~
      bIsBounceCrouching=~
      bIsBounceCrouched=~
      bIsBounceJumping=~
      bIsBounceRecoiling=~
      bIsSteeringRight=~
      bIsSteeringLeft=~
      RunForwardAlpha=~
      BounceCompression=~
      LeanPosition=~
      LeanPositionX=~
      LeanPositionY=~
      LeanPositionZ=~
      bAimFWD=~
      bAimBWD=~
      bAimLFT=~
      bAimRGT=~
      PawnToVehicleDeltaYawAngleDegrees=~
      AimCardDirDeadZoneAngleDegrees=~
      AimCardDirAngleOffsetDegrees=~
      AimFWDDeltaAngleDegrees=~
      AimBWDDeltaAngleDegrees=~
      AimLFTDeltaAngleDegrees=~
      AimRGTDeltaAngleDegrees=~
      SlopePitchDegreeAngle=~
      SlopeRollDegreeAngle=~
      SteerAngle=~
     ).bIsFrontPassenger
    ) {blendType=Linear}) {layerGroup=VehicleOverride}
   Save cached pose 'UpperBodyCachePose'(Pose=Input Pose() {name=InPoseUpperBody}) {cacheName=UpperBodyCachePose}
  ]}
 GolfCartFrontPassenger {
  entry=GolfCartPassenger_Front
  path=VehicleLowerBodyOverride.AnimGraphNode_StateMachine_0.GolfCartFrontPassenger
  ids=id
  nodes [AnimStateNode_16 = GolfCartPassenger_Front() {is=title AnimStateNode}]
 }
 GolfCartPassenger_Front {
  path=VehicleLowerBodyOverride.AnimGraphNode_StateMachine_0.GolfCartFrontPassenger.AnimStateNode_16.GolfCartPassenger_Front
  nodes [
   AnimGraphNode_StateMachine_1 = GolfCartPassenger_Front() {is=title AnimGraphNode_StateMachine}
   n1 = Get GolfCartVehicleInput()
   Output Animation Pose(Result=Rotate Root Bone!(
     BasePose=@AnimGraphNode_StateMachine_1
     Pitch:float=0
     Yaw=Break Fort Anim Input Golf Cart(FortAnimInput_GolfCart=@n1) -> (
      bIsUsingGolfCart=~
      bIsDriver=~
      bIsFrontPassenger=~
      bIsBackLeftPassenger=~
      bIsBackRightPassenger=~
      bIsFrontPassengerAndLeaning=~
      bIsBackPassengerAndLeaning=~
      bIsDrifting=~
      bIsBoosting=~
      bIsEBraking=~
      bIsReversing=~
      bIsBraking=~
      bIsMoving=~
      bIsMovingForward=~
      bIsPowerSliding=~
      bIsLeaning=~
      bIsLeaningOrBouncing=~
      bIsBounceCrouching=~
      bIsBounceCrouched=~
      bIsBounceJumping=~
      bIsBounceRecoiling=~
      bIsSteeringRight=~
      bIsSteeringLeft=~
      RunForwardAlpha=~
      BounceCompression=~
      LeanPosition=~
      LeanPositionX=~
      LeanPositionY=~
      LeanPositionZ=~
      bAimFWD=~
      bAimBWD=~
      bAimLFT=~
      bAimRGT=~
      PawnToVehicleDeltaYawAngleDegrees=~
      AimCardDirDeadZoneAngleDegrees=~
      AimCardDirAngleOffsetDegrees=~
      AimFWDDeltaAngleDegrees=~
      AimBWDDeltaAngleDegrees=~
      AimLFTDeltaAngleDegrees=~
      AimRGTDeltaAngleDegrees=~
      SlopePitchDegreeAngle=~
      SlopeRollDegreeAngle=~
      SteerAngle=~
     ).PawnToVehicleDeltaYawAngleDegrees
     MeshToComponent:Rotator="0,0,0"
    ) -> (Pose))
  ]
 }
 GolfCartPassenger_Front {
  entry=PassengerFrontPose_N
  path=VehicleLowerBodyOverride.AnimGraphNode_StateMachine_0.GolfCartFrontPassenger.AnimStateNode_16.GolfCartPassenger_Front.AnimGraphNode_StateMachine_1.GolfCartPassenger_Front
  ids=id
  nodes [
   AnimStateNode_10 = Transition_EtoS() {is=title AnimStateNode}
   AnimStateNode_11 = PassengerPose_S() {is=title AnimStateNode}
   AnimStateNode_12 = Transition_StoW() {is=title AnimStateNode}
   AnimStateNode_13 = PassengerFrontPose_W() {is=title AnimStateNode}
   AnimStateNode_15 = Transition_WtoN() {is=title AnimStateNode}
   AnimStateNode_16 = PassengerFrontPose_N() {is=title AnimStateNode}
   AnimStateNode_17 = Transition_EtoN() {is=title AnimStateNode}
   AnimStateNode_18 = Transition_StoE() {is=title AnimStateNode}
   AnimStateNode_19 = Transition_WtoS() {is=title AnimStateNode}
   AnimStateNode_20 = Transition_NtoW() {is=title AnimStateNode}
   AnimStateNode_8 = Transition_NtoE() {is=title AnimStateNode}
   AnimStateNode_9 = PassengerFrontPose_E() {is=title AnimStateNode}
   AnimStateTransitionNode_0 = AnimStateTransitionNode()
   AnimStateTransitionNode_10 = AnimStateTransitionNode() {priorityOrder=2}
   AnimStateTransitionNode_11 = AnimStateTransitionNode() {priorityOrder=2}
   AnimStateTransitionNode_15 = AnimStateTransitionNode()
   AnimStateTransitionNode_16 = AnimStateTransitionNode()
   AnimStateTransitionNode_17 = AnimStateTransitionNode()
   AnimStateTransitionNode_18 = AnimStateTransitionNode() {bAutomaticRuleBasedOnSequencePlayerInState=true}
   AnimStateTransitionNode_19 = AnimStateTransitionNode()
   AnimStateTransitionNode_2 = AnimStateTransitionNode() {bAutomaticRuleBasedOnSequencePlayerInState=true}
   AnimStateTransitionNode_20 = AnimStateTransitionNode()
   AnimStateTransitionNode_21 = AnimStateTransitionNode()
   AnimStateTransitionNode_22 = AnimStateTransitionNode() {bAutomaticRuleBasedOnSequencePlayerInState=true}
   AnimStateTransitionNode_25 = AnimStateTransitionNode()
   AnimStateTransitionNode_26 = AnimStateTransitionNode() {bAutomaticRuleBasedOnSequencePlayerInState=true}
   AnimStateTransitionNode_28 = AnimStateTransitionNode()
   AnimStateTransitionNode_29 = AnimStateTransitionNode()
   AnimStateTransitionNode_3 = AnimStateTransitionNode() {bAutomaticRuleBasedOnSequencePlayerInState=true}
   AnimStateTransitionNode_31 = AnimStateTransitionNode()
   AnimStateTransitionNode_32 = AnimStateTransitionNode()
   AnimStateTransitionNode_35 = AnimStateTransitionNode()
   AnimStateTransitionNode_36 = AnimStateTransitionNode()
   AnimStateTransitionNode_37 = AnimStateTransitionNode()
   AnimStateTransitionNode_4 = AnimStateTransitionNode()
   AnimStateTransitionNode_40 = AnimStateTransitionNode() {bAutomaticRuleBasedOnSequencePlayerInState=true}
   AnimStateTransitionNode_41 = AnimStateTransitionNode()
   AnimStateTransitionNode_42 = AnimStateTransitionNode() {bAutomaticRuleBasedOnSequencePlayerInState=true}
   AnimStateTransitionNode_5 = AnimStateTransitionNode()
   AnimStateTransitionNode_7 = AnimStateTransitionNode() {bAutomaticRuleBasedOnSequencePlayerInState=true}
  ]
  links [
   {
    dst=PassengerFrontPose_W
    rule=VehicleLowerBodyOverride.AnimGraphNode_StateMachine_0.GolfCartFrontPassenger.AnimStateNode_16.GolfCartPassenger_Front.AnimGraphNode_StateMachine_1.GolfCartPassenger_Front.AnimStateTransitionNode_11.Transition
    src=PassengerFrontPose_E
    via=AnimStateTransitionNode_11
   }
   {
    dst=Transition_EtoN
    rule=VehicleLowerBodyOverride.AnimGraphNode_StateMachine_0.GolfCartFrontPassenger.AnimStateNode_16.GolfCartPassenger_Front.AnimGraphNode_StateMachine_1.GolfCartPassenger_Front.AnimStateTransitionNode_35.AnimationTransitionGraph_1
    src=PassengerFrontPose_E
    via=AnimStateTransitionNode_35
   }
   {
    dst=Transition_EtoS
    rule=VehicleLowerBodyOverride.AnimGraphNode_StateMachine_0.GolfCartFrontPassenger.AnimStateNode_16.GolfCartPassenger_Front.AnimGraphNode_StateMachine_1.GolfCartPassenger_Front.AnimStateTransitionNode_17.Transition
    src=PassengerFrontPose_E
    via=AnimStateTransitionNode_17
   }
   {
    dst=PassengerPose_S
    rule=VehicleLowerBodyOverride.AnimGraphNode_StateMachine_0.GolfCartFrontPassenger.AnimStateNode_16.GolfCartPassenger_Front.AnimGraphNode_StateMachine_1.GolfCartPassenger_Front.AnimStateTransitionNode_15.Transition
    src=PassengerFrontPose_N
    via=AnimStateTransitionNode_15
   }
   {
    dst=Transition_NtoE
    rule=VehicleLowerBodyOverride.AnimGraphNode_StateMachine_0.GolfCartFrontPassenger.AnimStateNode_16.GolfCartPassenger_Front.AnimGraphNode_StateMachine_1.GolfCartPassenger_Front.AnimStateTransitionNode_25.AnimationTransitionGraph_1
    src=PassengerFrontPose_N
    via=AnimStateTransitionNode_25
   }
   {
    dst=Transition_NtoW
    rule=VehicleLowerBodyOverride.AnimGraphNode_StateMachine_0.GolfCartFrontPassenger.AnimStateNode_16.GolfCartPassenger_Front.AnimGraphNode_StateMachine_1.GolfCartPassenger_Front.AnimStateTransitionNode_41.AnimationTransitionGraph_1
    src=PassengerFrontPose_N
    via=AnimStateTransitionNode_41
   }
   {
    dst=PassengerFrontPose_E
    rule=VehicleLowerBodyOverride.AnimGraphNode_StateMachine_0.GolfCartFrontPassenger.AnimStateNode_16.GolfCartPassenger_Front.AnimGraphNode_StateMachine_1.GolfCartPassenger_Front.AnimStateTransitionNode_10.Transition
    src=PassengerFrontPose_W
    via=AnimStateTransitionNode_10
   }
   {
    dst=Transition_WtoN
    rule=VehicleLowerBodyOverride.AnimGraphNode_StateMachine_0.GolfCartFrontPassenger.AnimStateNode_16.GolfCartPassenger_Front.AnimGraphNode_StateMachine_1.GolfCartPassenger_Front.AnimStateTransitionNode_31.AnimationTransitionGraph_1
    src=PassengerFrontPose_W
    via=AnimStateTransitionNode_31
   }
   {
    dst=Transition_WtoS
    rule=VehicleLowerBodyOverride.AnimGraphNode_StateMachine_0.GolfCartFrontPassenger.AnimStateNode_16.GolfCartPassenger_Front.AnimGraphNode_StateMachine_1.GolfCartPassenger_Front.AnimStateTransitionNode_20.Transition
    src=PassengerFrontPose_W
    via=AnimStateTransitionNode_20
   }
   {
    dst=PassengerFrontPose_N
    rule=VehicleLowerBodyOverride.AnimGraphNode_StateMachine_0.GolfCartFrontPassenger.AnimStateNode_16.GolfCartPassenger_Front.AnimGraphNode_StateMachine_1.GolfCartPassenger_Front.AnimStateTransitionNode_16.Transition
    src=PassengerPose_S
    via=AnimStateTransitionNode_16
   }
   {
    dst=Transition_StoE
    rule=VehicleLowerBodyOverride.AnimGraphNode_StateMachine_0.GolfCartFrontPassenger.AnimStateNode_16.GolfCartPassenger_Front.AnimGraphNode_StateMachine_1.GolfCartPassenger_Front.AnimStateTransitionNode_37.AnimationTransitionGraph_1
    src=PassengerPose_S
    via=AnimStateTransitionNode_37
   }
   {
    dst=Transition_StoW
    rule=VehicleLowerBodyOverride.AnimGraphNode_StateMachine_0.GolfCartFrontPassenger.AnimStateNode_16.GolfCartPassenger_Front.AnimGraphNode_StateMachine_1.GolfCartPassenger_Front.AnimStateTransitionNode_29.AnimationTransitionGraph_0
    src=PassengerPose_S
    via=AnimStateTransitionNode_29
   }
   {
    dst=PassengerFrontPose_N
    rule=VehicleLowerBodyOverride.AnimGraphNode_StateMachine_0.GolfCartFrontPassenger.AnimStateNode_16.GolfCartPassenger_Front.AnimGraphNode_StateMachine_1.GolfCartPassenger_Front.AnimStateTransitionNode_36.AnimationTransitionGraph_1
    src=Transition_EtoN
    via=AnimStateTransitionNode_36
   }
   {
    dst=PassengerFrontPose_N
    rule=VehicleLowerBodyOverride.AnimGraphNode_StateMachine_0.GolfCartFrontPassenger.AnimStateNode_16.GolfCartPassenger_Front.AnimGraphNode_StateMachine_1.GolfCartPassenger_Front.AnimStateTransitionNode_7.Transition
    src=Transition_EtoN
    via=AnimStateTransitionNode_7
   }
   {
    dst=PassengerPose_S
    rule=VehicleLowerBodyOverride.AnimGraphNode_StateMachine_0.GolfCartFrontPassenger.AnimStateNode_16.GolfCartPassenger_Front.AnimGraphNode_StateMachine_1.GolfCartPassenger_Front.AnimStateTransitionNode_2.AnimationTransitionGraph_0
    src=Transition_EtoS
    via=AnimStateTransitionNode_2
   }
   {
    dst=PassengerPose_S
    rule=VehicleLowerBodyOverride.AnimGraphNode_StateMachine_0.GolfCartFrontPassenger.AnimStateNode_16.GolfCartPassenger_Front.AnimGraphNode_StateMachine_1.GolfCartPassenger_Front.AnimStateTransitionNode_28.AnimationTransitionGraph_0
    src=Transition_EtoS
    via=AnimStateTransitionNode_28
   }
   {
    dst=PassengerFrontPose_E
    rule=VehicleLowerBodyOverride.AnimGraphNode_StateMachine_0.GolfCartFrontPassenger.AnimStateNode_16.GolfCartPassenger_Front.AnimGraphNode_StateMachine_1.GolfCartPassenger_Front.AnimStateTransitionNode_0.Transition
    src=Transition_NtoE
    via=AnimStateTransitionNode_0
   }
   {
    dst=PassengerFrontPose_E
    rule=VehicleLowerBodyOverride.AnimGraphNode_StateMachine_0.GolfCartFrontPassenger.AnimStateNode_16.GolfCartPassenger_Front.AnimGraphNode_StateMachine_1.GolfCartPassenger_Front.AnimStateTransitionNode_26.AnimationTransitionGraph_1
    src=Transition_NtoE
    via=AnimStateTransitionNode_26
   }
   {
    dst=PassengerFrontPose_W
    rule=VehicleLowerBodyOverride.AnimGraphNode_StateMachine_0.GolfCartFrontPassenger.AnimStateNode_16.GolfCartPassenger_Front.AnimGraphNode_StateMachine_1.GolfCartPassenger_Front.AnimStateTransitionNode_4.Transition
    src=Transition_NtoW
    via=AnimStateTransitionNode_4
   }
   {
    dst=PassengerFrontPose_W
    rule=VehicleLowerBodyOverride.AnimGraphNode_StateMachine_0.GolfCartFrontPassenger.AnimStateNode_16.GolfCartPassenger_Front.AnimGraphNode_StateMachine_1.GolfCartPassenger_Front.AnimStateTransitionNode_42.AnimationTransitionGraph_1
    src=Transition_NtoW
    via=AnimStateTransitionNode_42
   }
   {
    dst=PassengerFrontPose_E
    rule=VehicleLowerBodyOverride.AnimGraphNode_StateMachine_0.GolfCartFrontPassenger.AnimStateNode_16.GolfCartPassenger_Front.AnimGraphNode_StateMachine_1.GolfCartPassenger_Front.AnimStateTransitionNode_18.Transition
    src=Transition_StoE
    via=AnimStateTransitionNode_18
   }
   {
    dst=PassengerFrontPose_E
    rule=VehicleLowerBodyOverride.AnimGraphNode_StateMachine_0.GolfCartFrontPassenger.AnimStateNode_16.GolfCartPassenger_Front.AnimGraphNode_StateMachine_1.GolfCartPassenger_Front.AnimStateTransitionNode_19.Transition
    src=Transition_StoE
    via=AnimStateTransitionNode_19
   }
   {
    dst=PassengerFrontPose_W
    rule=VehicleLowerBodyOverride.AnimGraphNode_StateMachine_0.GolfCartFrontPassenger.AnimStateNode_16.GolfCartPassenger_Front.AnimGraphNode_StateMachine_1.GolfCartPassenger_Front.AnimStateTransitionNode_21.Transition
    src=Transition_StoW
    via=AnimStateTransitionNode_21
   }
   {
    dst=PassengerFrontPose_W
    rule=VehicleLowerBodyOverride.AnimGraphNode_StateMachine_0.GolfCartFrontPassenger.AnimStateNode_16.GolfCartPassenger_Front.AnimGraphNode_StateMachine_1.GolfCartPassenger_Front.AnimStateTransitionNode_22.Transition
    src=Transition_StoW
    via=AnimStateTransitionNode_22
   }
   {
    dst=PassengerFrontPose_N
    rule=VehicleLowerBodyOverride.AnimGraphNode_StateMachine_0.GolfCartFrontPassenger.AnimStateNode_16.GolfCartPassenger_Front.AnimGraphNode_StateMachine_1.GolfCartPassenger_Front.AnimStateTransitionNode_3.Transition
    src=Transition_WtoN
    via=AnimStateTransitionNode_3
   }
   {
    dst=PassengerFrontPose_N
    rule=VehicleLowerBodyOverride.AnimGraphNode_StateMachine_0.GolfCartFrontPassenger.AnimStateNode_16.GolfCartPassenger_Front.AnimGraphNode_StateMachine_1.GolfCartPassenger_Front.AnimStateTransitionNode_32.AnimationTransitionGraph_1
    src=Transition_WtoN
    via=AnimStateTransitionNode_32
   }
   {
    dst=PassengerPose_S
    rule=VehicleLowerBodyOverride.AnimGraphNode_StateMachine_0.GolfCartFrontPassenger.AnimStateNode_16.GolfCartPassenger_Front.AnimGraphNode_StateMachine_1.GolfCartPassenger_Front.AnimStateTransitionNode_40.AnimationTransitionGraph_1
    src=Transition_WtoS
    via=AnimStateTransitionNode_40
   }
   {
    dst=PassengerPose_S
    rule=VehicleLowerBodyOverride.AnimGraphNode_StateMachine_0.GolfCartFrontPassenger.AnimStateNode_16.GolfCartPassenger_Front.AnimGraphNode_StateMachine_1.GolfCartPassenger_Front.AnimStateTransitionNode_5.AnimationTransitionGraph_1
    src=Transition_WtoS
    via=AnimStateTransitionNode_5
   }
  ]
 }
 Transition_EtoS {
  path=VehicleLowerBodyOverride.AnimGraphNode_StateMachine_0.GolfCartFrontPassenger.AnimStateNode_16.GolfCartPassenger_Front.AnimGraphNode_StateMachine_1.GolfCartPassenger_Front.AnimStateNode_10.Transition_EtoS
  nodes [Output Animation Pose(Result=Sequence Player 'Passenger_Front_Turn_EtoS_GolfCart'() {bLoopAnimation=false, sequence=/Game/Animation/Game/MainPlayer/Vehicles/GolfCart/Passenger_Front_Turn_EtoS_GolfCart})]
 }
 PassengerPose_S {
  path=VehicleLowerBodyOverride.AnimGraphNode_StateMachine_0.GolfCartFrontPassenger.AnimStateNode_16.GolfCartPassenger_Front.AnimGraphNode_StateMachine_1.GolfCartPassenger_Front.AnimStateNode_11.PassengerPose_S
  nodes [
   n1 = AimOffset Player 'ShoppingCart_Passenger_LowerBodyAO'!(
    X:float=Break Fort Anim Input Golf Cart(FortAnimInput_GolfCart=Get GolfCartVehicleInput()) -> (
     bIsUsingGolfCart=~
     bIsDriver=~
     bIsFrontPassenger=~
     bIsBackLeftPassenger=~
     bIsBackRightPassenger=~
     bIsFrontPassengerAndLeaning=~
     bIsBackPassengerAndLeaning=~
     bIsDrifting=~
     bIsBoosting=~
     bIsEBraking=~
     bIsReversing=~
     bIsBraking=~
     bIsMoving=~
     bIsMovingForward=~
     bIsPowerSliding=~
     bIsLeaning=~
     bIsLeaningOrBouncing=~
     bIsBounceCrouching=~
     bIsBounceCrouched=~
     bIsBounceJumping=~
     bIsBounceRecoiling=~
     bIsSteeringRight=~
     bIsSteeringLeft=~
     RunForwardAlpha=~
     BounceCompression=~
     LeanPosition=~
     LeanPositionX=~
     LeanPositionY=~
     LeanPositionZ=~
     bAimFWD=~
     bAimBWD=~
     bAimLFT=~
     bAimRGT=~
     PawnToVehicleDeltaYawAngleDegrees=~
     AimCardDirDeadZoneAngleDegrees=~
     AimCardDirAngleOffsetDegrees=~
     AimFWDDeltaAngleDegrees=~
     AimBWDDeltaAngleDegrees=~
     AimLFTDeltaAngleDegrees=~
     AimRGTDeltaAngleDegrees=~
     SlopePitchDegreeAngle=~
     SlopeRollDegreeAngle=~
     SteerAngle=~
    ).AimBWDDeltaAngleDegrees
    BasePose:Pose Link Structure=Rotate Root Bone!(
     BasePose=Sequence Player 'Passenger_Front_Pose_S_GolfCart'() {sequence=/Game/Animation/Game/MainPlayer/Vehicles/GolfCart/Passenger_Front_Pose_S_GolfCart}
     Yaw=180
     MeshToComponent="0,0,0"
    ) -> (Pose)
    Alpha:float=1
   ) -> (Pose:Pose Link Structure) {blendSpace=/Game/Animation/Game/MainPlayer/Vehicles/ShoppingCart/AO/ShoppingCart_Passenger_LowerBodyAO, alphaBoolBlend {blendOption=Linear}}
   Output Animation Pose(Result=Rotate Root Bone(BasePose=@n1, Yaw=-180, MeshToComponent="0,0,0"))
  ]
 }
 Transition_StoW {
  path=VehicleLowerBodyOverride.AnimGraphNode_StateMachine_0.GolfCartFrontPassenger.AnimStateNode_16.GolfCartPassenger_Front.AnimGraphNode_StateMachine_1.GolfCartPassenger_Front.AnimStateNode_12.Transition_StoW
  nodes [Output Animation Pose(Result=Sequence Player 'Passenger_Front_Turn_StoW_GolfCart'() {sequence=/Game/Animation/Game/MainPlayer/Vehicles/GolfCart/Passenger_Front_Turn_StoW_GolfCart})]
 }
 PassengerFrontPose_W {
  path=VehicleLowerBodyOverride.AnimGraphNode_StateMachine_0.GolfCartFrontPassenger.AnimStateNode_16.GolfCartPassenger_Front.AnimGraphNode_StateMachine_1.GolfCartPassenger_Front.AnimStateNode_13.PassengerFrontPose_W
  nodes [
   n1 = AimOffset Player 'ShoppingCart_Passenger_LowerBodyAO'(
    X=Break Fort Anim Input Golf Cart(FortAnimInput_GolfCart=Get GolfCartVehicleInput()) -> (
     bIsUsingGolfCart=~
     bIsDriver=~
     bIsFrontPassenger=~
     bIsBackLeftPassenger=~
     bIsBackRightPassenger=~
     bIsFrontPassengerAndLeaning=~
     bIsBackPassengerAndLeaning=~
     bIsDrifting=~
     bIsBoosting=~
     bIsEBraking=~
     bIsReversing=~
     bIsBraking=~
     bIsMoving=~
     bIsMovingForward=~
     bIsPowerSliding=~
     bIsLeaning=~
     bIsLeaningOrBouncing=~
     bIsBounceCrouching=~
     bIsBounceCrouched=~
     bIsBounceJumping=~
     bIsBounceRecoiling=~
     bIsSteeringRight=~
     bIsSteeringLeft=~
     RunForwardAlpha=~
     BounceCompression=~
     LeanPosition=~
     LeanPositionX=~
     LeanPositionY=~
     LeanPositionZ=~
     bAimFWD=~
     bAimBWD=~
     bAimLFT=~
     bAimRGT=~
     PawnToVehicleDeltaYawAngleDegrees=~
     AimCardDirDeadZoneAngleDegrees=~
     AimCardDirAngleOffsetDegrees=~
     AimFWDDeltaAngleDegrees=~
     AimBWDDeltaAngleDegrees=~
     AimLFTDeltaAngleDegrees=~
     AimRGTDeltaAngleDegrees=~
     SlopePitchDegreeAngle=~
     SlopeRollDegreeAngle=~
     SteerAngle=~
    ).AimLFTDeltaAngleDegrees
    BasePose=Rotate Root Bone(
     BasePose=Sequence Player 'Passenger_Front_Pose_W_GolfCart'() {sequence=/Game/Animation/Game/MainPlayer/Vehicles/GolfCart/Passenger_Front_Pose_W_GolfCart}
     Yaw=90
     MeshToComponent="0,0,0"
    )
    Alpha=1
   ) {blendSpace=&ShoppingCart_Passenger_LowerBodyAO, alphaBoolBlend {blendOption=Linear}}
   Output Animation Pose(Result=Rotate Root Bone(BasePose=@n1, Yaw=-90, MeshToComponent="0,0,0"))
  ]
 }
 Transition_WtoN {
  path=VehicleLowerBodyOverride.AnimGraphNode_StateMachine_0.GolfCartFrontPassenger.AnimStateNode_16.GolfCartPassenger_Front.AnimGraphNode_StateMachine_1.GolfCartPassenger_Front.AnimStateNode_15.Transition_WtoN
  nodes [Output Animation Pose(Result=Sequence Player 'Passenger_Front_Turn_WtoN_GolfCart'() {bLoopAnimation=false, sequence=/Game/Animation/Game/MainPlayer/Vehicles/GolfCart/Passenger_Front_Turn_WtoN_GolfCart})]
 }
 PassengerFrontPose_N {
  path=VehicleLowerBodyOverride.AnimGraphNode_StateMachine_0.GolfCartFrontPassenger.AnimStateNode_16.GolfCartPassenger_Front.AnimGraphNode_StateMachine_1.GolfCartPassenger_Front.AnimStateNode_16.PassengerFrontPose_N
  nodes [
   n1 = AimOffset Player 'ShoppingCart_Passenger_LowerBodyAO'(
    X=Break Fort Anim Input Golf Cart(FortAnimInput_GolfCart=Get GolfCartVehicleInput()) -> (
     bIsUsingGolfCart=~
     bIsDriver=~
     bIsFrontPassenger=~
     bIsBackLeftPassenger=~
     bIsBackRightPassenger=~
     bIsFrontPassengerAndLeaning=~
     bIsBackPassengerAndLeaning=~
     bIsDrifting=~
     bIsBoosting=~
     bIsEBraking=~
     bIsReversing=~
     bIsBraking=~
     bIsMoving=~
     bIsMovingForward=~
     bIsPowerSliding=~
     bIsLeaning=~
     bIsLeaningOrBouncing=~
     bIsBounceCrouching=~
     bIsBounceCrouched=~
     bIsBounceJumping=~
     bIsBounceRecoiling=~
     bIsSteeringRight=~
     bIsSteeringLeft=~
     RunForwardAlpha=~
     BounceCompression=~
     LeanPosition=~
     LeanPositionX=~
     LeanPositionY=~
     LeanPositionZ=~
     bAimFWD=~
     bAimBWD=~
     bAimLFT=~
     bAimRGT=~
     PawnToVehicleDeltaYawAngleDegrees=~
     AimCardDirDeadZoneAngleDegrees=~
     AimCardDirAngleOffsetDegrees=~
     AimFWDDeltaAngleDegrees=~
     AimBWDDeltaAngleDegrees=~
     AimLFTDeltaAngleDegrees=~
     AimRGTDeltaAngleDegrees=~
     SlopePitchDegreeAngle=~
     SlopeRollDegreeAngle=~
     SteerAngle=~
    ).AimFWDDeltaAngleDegrees
    BasePose=Sequence Player 'Passenger_Front_Pose_N_GolfCart'() {sequence=/Game/Animation/Game/MainPlayer/Vehicles/GolfCart/Passenger_Front_Pose_N_GolfCart}
    Alpha=1
   ) {blendSpace=&ShoppingCart_Passenger_LowerBodyAO, alphaBoolBlend {blendOption=Linear}}
   Output Animation Pose(Result=@n1)
  ]
 }
 Transition_EtoN {
  path=VehicleLowerBodyOverride.AnimGraphNode_StateMachine_0.GolfCartFrontPassenger.AnimStateNode_16.GolfCartPassenger_Front.AnimGraphNode_StateMachine_1.GolfCartPassenger_Front.AnimStateNode_17.Transition_EtoN
  nodes [Output Animation Pose(Result=Sequence Player 'Passenger_Front_Turn_EtoN_GolfCart'() {bLoopAnimation=false, sequence=/Game/Animation/Game/MainPlayer/Vehicles/GolfCart/Passenger_Front_Turn_EtoN_GolfCart})]
 }
 Transition_StoE {
  path=VehicleLowerBodyOverride.AnimGraphNode_StateMachine_0.GolfCartFrontPassenger.AnimStateNode_16.GolfCartPassenger_Front.AnimGraphNode_StateMachine_1.GolfCartPassenger_Front.AnimStateNode_18.Transition_StoE
  nodes [Output Animation Pose(Result=Sequence Player 'Passenger_Front_Turn_StoE_GolfCart'() {bLoopAnimation=false, sequence=/Game/Animation/Game/MainPlayer/Vehicles/GolfCart/Passenger_Front_Turn_StoE_GolfCart})]
 }
 Transition_WtoS {
  path=VehicleLowerBodyOverride.AnimGraphNode_StateMachine_0.GolfCartFrontPassenger.AnimStateNode_16.GolfCartPassenger_Front.AnimGraphNode_StateMachine_1.GolfCartPassenger_Front.AnimStateNode_19.Transition_WtoS
  nodes [Output Animation Pose(Result=Sequence Player 'Passenger_Front_Turn_WtoS_GolfCart'() {bLoopAnimation=false, sequence=/Game/Animation/Game/MainPlayer/Vehicles/GolfCart/Passenger_Front_Turn_WtoS_GolfCart})]
 }
 Transition_NtoW {
  path=VehicleLowerBodyOverride.AnimGraphNode_StateMachine_0.GolfCartFrontPassenger.AnimStateNode_16.GolfCartPassenger_Front.AnimGraphNode_StateMachine_1.GolfCartPassenger_Front.AnimStateNode_20.Transition_NtoW
  nodes [Output Animation Pose(Result=Sequence Player 'Passenger_Front_Turn_NtoW_GolfCart'() {bLoopAnimation=false, sequence=/Game/Animation/Game/MainPlayer/Vehicles/GolfCart/Passenger_Front_Turn_NtoW_GolfCart})]
 }
 Transition_NtoE {
  path=VehicleLowerBodyOverride.AnimGraphNode_StateMachine_0.GolfCartFrontPassenger.AnimStateNode_16.GolfCartPassenger_Front.AnimGraphNode_StateMachine_1.GolfCartPassenger_Front.AnimStateNode_8.Transition_NtoE
  nodes [Output Animation Pose(Result=Sequence Player 'Passenger_Front_Turn_NtoE_GolfCart'() {bLoopAnimation=false, sequence=/Game/Animation/Game/MainPlayer/Vehicles/GolfCart/Passenger_Front_Turn_NtoE_GolfCart})]
 }
 PassengerFrontPose_E {
  path=VehicleLowerBodyOverride.AnimGraphNode_StateMachine_0.GolfCartFrontPassenger.AnimStateNode_16.GolfCartPassenger_Front.AnimGraphNode_StateMachine_1.GolfCartPassenger_Front.AnimStateNode_9.PassengerFrontPose_E
  nodes [
   n1 = AimOffset Player 'ShoppingCart_Passenger_LowerBodyAO'(
    X=Break Fort Anim Input Golf Cart(FortAnimInput_GolfCart=Get GolfCartVehicleInput()) -> (
     bIsUsingGolfCart=~
     bIsDriver=~
     bIsFrontPassenger=~
     bIsBackLeftPassenger=~
     bIsBackRightPassenger=~
     bIsFrontPassengerAndLeaning=~
     bIsBackPassengerAndLeaning=~
     bIsDrifting=~
     bIsBoosting=~
     bIsEBraking=~
     bIsReversing=~
     bIsBraking=~
     bIsMoving=~
     bIsMovingForward=~
     bIsPowerSliding=~
     bIsLeaning=~
     bIsLeaningOrBouncing=~
     bIsBounceCrouching=~
     bIsBounceCrouched=~
     bIsBounceJumping=~
     bIsBounceRecoiling=~
     bIsSteeringRight=~
     bIsSteeringLeft=~
     RunForwardAlpha=~
     BounceCompression=~
     LeanPosition=~
     LeanPositionX=~
     LeanPositionY=~
     LeanPositionZ=~
     bAimFWD=~
     bAimBWD=~
     bAimLFT=~
     bAimRGT=~
     PawnToVehicleDeltaYawAngleDegrees=~
     AimCardDirDeadZoneAngleDegrees=~
     AimCardDirAngleOffsetDegrees=~
     AimFWDDeltaAngleDegrees=~
     AimBWDDeltaAngleDegrees=~
     AimLFTDeltaAngleDegrees=~
     AimRGTDeltaAngleDegrees=~
     SlopePitchDegreeAngle=~
     SlopeRollDegreeAngle=~
     SteerAngle=~
    ).AimRGTDeltaAngleDegrees
    BasePose=Rotate Root Bone(
     BasePose=Sequence Player 'Passenger_Front_Pose_E_GolfCart'() {sequence=/Game/Animation/Game/MainPlayer/Vehicles/GolfCart/Passenger_Front_Pose_E_GolfCart}
     Yaw=-90
     MeshToComponent="0,0,0"
    )
    Alpha=1
   ) {blendSpace=&ShoppingCart_Passenger_LowerBodyAO, alphaBoolBlend {blendOption=Linear}}
   Output Animation Pose(Result=Rotate Root Bone(BasePose=@n1, Yaw=90, MeshToComponent="0,0,0"))
  ]
 }
 Transition {
  path=VehicleLowerBodyOverride.AnimGraphNode_StateMachine_0.GolfCartFrontPassenger.AnimStateNode_16.GolfCartPassenger_Front.AnimGraphNode_StateMachine_1.GolfCartPassenger_Front.AnimStateTransitionNode_0.Transition
  nodes [
   n1 = Get GolfCartVehicleInput()
   Result(bCanEnterTransition=NOT Boolean(A=Break Fort Anim Input Golf Cart(FortAnimInput_GolfCart=@n1) -> (
      bIsUsingGolfCart=~
      bIsDriver=~
      bIsFrontPassenger=~
      bIsBackLeftPassenger=~
      bIsBackRightPassenger=~
      bIsFrontPassengerAndLeaning=~
      bIsBackPassengerAndLeaning=~
      bIsDrifting=~
      bIsBoosting=~
      bIsEBraking=~
      bIsReversing=~
      bIsBraking=~
      bIsMoving=~
      bIsMovingForward=~
      bIsPowerSliding=~
      bIsLeaning=~
      bIsLeaningOrBouncing=~
      bIsBounceCrouching=~
      bIsBounceCrouched=~
      bIsBounceJumping=~
      bIsBounceRecoiling=~
      bIsSteeringRight=~
      bIsSteeringLeft=~
      RunForwardAlpha=~
      BounceCompression=~
      LeanPosition=~
      LeanPositionX=~
      LeanPositionY=~
      LeanPositionZ=~
      bAimFWD=~
      bAimBWD=~
      bAimLFT=~
      bAimRGT=~
      PawnToVehicleDeltaYawAngleDegrees=~
      AimCardDirDeadZoneAngleDegrees=~
      AimCardDirAngleOffsetDegrees=~
      AimFWDDeltaAngleDegrees=~
      AimBWDDeltaAngleDegrees=~
      AimLFTDeltaAngleDegrees=~
      AimRGTDeltaAngleDegrees=~
      SlopePitchDegreeAngle=~
      SlopeRollDegreeAngle=~
      SteerAngle=~
     ).bAimRGT))
  ]
 }
 Transition {
  path=VehicleLowerBodyOverride.AnimGraphNode_StateMachine_0.GolfCartFrontPassenger.AnimStateNode_16.GolfCartPassenger_Front.AnimGraphNode_StateMachine_1.GolfCartPassenger_Front.AnimStateTransitionNode_10.Transition
  nodes [Result(bCanEnterTransition=Break Fort Anim Input Golf Cart(FortAnimInput_GolfCart=Get GolfCartVehicleInput()) -> (
     bIsUsingGolfCart=~
     bIsDriver=~
     bIsFrontPassenger=~
     bIsBackLeftPassenger=~
     bIsBackRightPassenger=~
     bIsFrontPassengerAndLeaning=~
     bIsBackPassengerAndLeaning=~
     bIsDrifting=~
     bIsBoosting=~
     bIsEBraking=~
     bIsReversing=~
     bIsBraking=~
     bIsMoving=~
     bIsMovingForward=~
     bIsPowerSliding=~
     bIsLeaning=~
     bIsLeaningOrBouncing=~
     bIsBounceCrouching=~
     bIsBounceCrouched=~
     bIsBounceJumping=~
     bIsBounceRecoiling=~
     bIsSteeringRight=~
     bIsSteeringLeft=~
     RunForwardAlpha=~
     BounceCompression=~
     LeanPosition=~
     LeanPositionX=~
     LeanPositionY=~
     LeanPositionZ=~
     bAimFWD=~
     bAimBWD=~
     bAimLFT=~
     bAimRGT=~
     PawnToVehicleDeltaYawAngleDegrees=~
     AimCardDirDeadZoneAngleDegrees=~
     AimCardDirAngleOffsetDegrees=~
     AimFWDDeltaAngleDegrees=~
     AimBWDDeltaAngleDegrees=~
     AimLFTDeltaAngleDegrees=~
     AimRGTDeltaAngleDegrees=~
     SlopePitchDegreeAngle=~
     SlopeRollDegreeAngle=~
     SteerAngle=~
    ).bAimRGT)]
 }
 Transition {
  path=VehicleLowerBodyOverride.AnimGraphNode_StateMachine_0.GolfCartFrontPassenger.AnimStateNode_16.GolfCartPassenger_Front.AnimGraphNode_StateMachine_1.GolfCartPassenger_Front.AnimStateTransitionNode_11.Transition
  nodes [Result(bCanEnterTransition=Break Fort Anim Input Golf Cart(FortAnimInput_GolfCart=Get GolfCartVehicleInput()) -> (
     bIsUsingGolfCart=~
     bIsDriver=~
     bIsFrontPassenger=~
     bIsBackLeftPassenger=~
     bIsBackRightPassenger=~
     bIsFrontPassengerAndLeaning=~
     bIsBackPassengerAndLeaning=~
     bIsDrifting=~
     bIsBoosting=~
     bIsEBraking=~
     bIsReversing=~
     bIsBraking=~
     bIsMoving=~
     bIsMovingForward=~
     bIsPowerSliding=~
     bIsLeaning=~
     bIsLeaningOrBouncing=~
     bIsBounceCrouching=~
     bIsBounceCrouched=~
     bIsBounceJumping=~
     bIsBounceRecoiling=~
     bIsSteeringRight=~
     bIsSteeringLeft=~
     RunForwardAlpha=~
     BounceCompression=~
     LeanPosition=~
     LeanPositionX=~
     LeanPositionY=~
     LeanPositionZ=~
     bAimFWD=~
     bAimBWD=~
     bAimLFT=~
     bAimRGT=~
     PawnToVehicleDeltaYawAngleDegrees=~
     AimCardDirDeadZoneAngleDegrees=~
     AimCardDirAngleOffsetDegrees=~
     AimFWDDeltaAngleDegrees=~
     AimBWDDeltaAngleDegrees=~
     AimLFTDeltaAngleDegrees=~
     AimRGTDeltaAngleDegrees=~
     SlopePitchDegreeAngle=~
     SlopeRollDegreeAngle=~
     SteerAngle=~
    ).bAimLFT)]
 }
 Transition {
  path=VehicleLowerBodyOverride.AnimGraphNode_StateMachine_0.GolfCartFrontPassenger.AnimStateNode_16.GolfCartPassenger_Front.AnimGraphNode_StateMachine_1.GolfCartPassenger_Front.AnimStateTransitionNode_15.Transition
  nodes [Result(bCanEnterTransition=Break Fort Anim Input Golf Cart(FortAnimInput_GolfCart=Get GolfCartVehicleInput()) -> (
     bIsUsingGolfCart=~
     bIsDriver=~
     bIsFrontPassenger=~
     bIsBackLeftPassenger=~
     bIsBackRightPassenger=~
     bIsFrontPassengerAndLeaning=~
     bIsBackPassengerAndLeaning=~
     bIsDrifting=~
     bIsBoosting=~
     bIsEBraking=~
     bIsReversing=~
     bIsBraking=~
     bIsMoving=~
     bIsMovingForward=~
     bIsPowerSliding=~
     bIsLeaning=~
     bIsLeaningOrBouncing=~
     bIsBounceCrouching=~
     bIsBounceCrouched=~
     bIsBounceJumping=~
     bIsBounceRecoiling=~
     bIsSteeringRight=~
     bIsSteeringLeft=~
     RunForwardAlpha=~
     BounceCompression=~
     LeanPosition=~
     LeanPositionX=~
     LeanPositionY=~
     LeanPositionZ=~
     bAimFWD=~
     bAimBWD=~
     bAimLFT=~
     bAimRGT=~
     PawnToVehicleDeltaYawAngleDegrees=~
     AimCardDirDeadZoneAngleDegrees=~
     AimCardDirAngleOffsetDegrees=~
     AimFWDDeltaAngleDegrees=~
     AimBWDDeltaAngleDegrees=~
     AimLFTDeltaAngleDegrees=~
     AimRGTDeltaAngleDegrees=~
     SlopePitchDegreeAngle=~
     SlopeRollDegreeAngle=~
     SteerAngle=~
    ).bAimBWD)]
 }
 Transition {
  path=VehicleLowerBodyOverride.AnimGraphNode_StateMachine_0.GolfCartFrontPassenger.AnimStateNode_16.GolfCartPassenger_Front.AnimGraphNode_StateMachine_1.GolfCartPassenger_Front.AnimStateTransitionNode_16.Transition
  nodes [Result(bCanEnterTransition=Break Fort Anim Input Golf Cart(FortAnimInput_GolfCart=Get GolfCartVehicleInput()) -> (
     bIsUsingGolfCart=~
     bIsDriver=~
     bIsFrontPassenger=~
     bIsBackLeftPassenger=~
     bIsBackRightPassenger=~
     bIsFrontPassengerAndLeaning=~
     bIsBackPassengerAndLeaning=~
     bIsDrifting=~
     bIsBoosting=~
     bIsEBraking=~
     bIsReversing=~
     bIsBraking=~
     bIsMoving=~
     bIsMovingForward=~
     bIsPowerSliding=~
     bIsLeaning=~
     bIsLeaningOrBouncing=~
     bIsBounceCrouching=~
     bIsBounceCrouched=~
     bIsBounceJumping=~
     bIsBounceRecoiling=~
     bIsSteeringRight=~
     bIsSteeringLeft=~
     RunForwardAlpha=~
     BounceCompression=~
     LeanPosition=~
     LeanPositionX=~
     LeanPositionY=~
     LeanPositionZ=~
     bAimFWD=~
     bAimBWD=~
     bAimLFT=~
     bAimRGT=~
     PawnToVehicleDeltaYawAngleDegrees=~
     AimCardDirDeadZoneAngleDegrees=~
     AimCardDirAngleOffsetDegrees=~
     AimFWDDeltaAngleDegrees=~
     AimBWDDeltaAngleDegrees=~
     AimLFTDeltaAngleDegrees=~
     AimRGTDeltaAngleDegrees=~
     SlopePitchDegreeAngle=~
     SlopeRollDegreeAngle=~
     SteerAngle=~
    ).bAimFWD)]
 }
 Transition {
  path=VehicleLowerBodyOverride.AnimGraphNode_StateMachine_0.GolfCartFrontPassenger.AnimStateNode_16.GolfCartPassenger_Front.AnimGraphNode_StateMachine_1.GolfCartPassenger_Front.AnimStateTransitionNode_17.Transition
  nodes [Result(bCanEnterTransition=Break Fort Anim Input Golf Cart(FortAnimInput_GolfCart=Get GolfCartVehicleInput()) -> (
     bIsUsingGolfCart=~
     bIsDriver=~
     bIsFrontPassenger=~
     bIsBackLeftPassenger=~
     bIsBackRightPassenger=~
     bIsFrontPassengerAndLeaning=~
     bIsBackPassengerAndLeaning=~
     bIsDrifting=~
     bIsBoosting=~
     bIsEBraking=~
     bIsReversing=~
     bIsBraking=~
     bIsMoving=~
     bIsMovingForward=~
     bIsPowerSliding=~
     bIsLeaning=~
     bIsLeaningOrBouncing=~
     bIsBounceCrouching=~
     bIsBounceCrouched=~
     bIsBounceJumping=~
     bIsBounceRecoiling=~
     bIsSteeringRight=~
     bIsSteeringLeft=~
     RunForwardAlpha=~
     BounceCompression=~
     LeanPosition=~
     LeanPositionX=~
     LeanPositionY=~
     LeanPositionZ=~
     bAimFWD=~
     bAimBWD=~
     bAimLFT=~
     bAimRGT=~
     PawnToVehicleDeltaYawAngleDegrees=~
     AimCardDirDeadZoneAngleDegrees=~
     AimCardDirAngleOffsetDegrees=~
     AimFWDDeltaAngleDegrees=~
     AimBWDDeltaAngleDegrees=~
     AimLFTDeltaAngleDegrees=~
     AimRGTDeltaAngleDegrees=~
     SlopePitchDegreeAngle=~
     SlopeRollDegreeAngle=~
     SteerAngle=~
    ).bAimBWD)]
 }
 Transition {
  path=VehicleLowerBodyOverride.AnimGraphNode_StateMachine_0.GolfCartFrontPassenger.AnimStateNode_16.GolfCartPassenger_Front.AnimGraphNode_StateMachine_1.GolfCartPassenger_Front.AnimStateTransitionNode_18.Transition
  nodes [Result()]
 }
 Transition {
  path=VehicleLowerBodyOverride.AnimGraphNode_StateMachine_0.GolfCartFrontPassenger.AnimStateNode_16.GolfCartPassenger_Front.AnimGraphNode_StateMachine_1.GolfCartPassenger_Front.AnimStateTransitionNode_19.Transition
  nodes [
   n1 = Get GolfCartVehicleInput()
   Result(bCanEnterTransition=NOT Boolean(A=Break Fort Anim Input Golf Cart(FortAnimInput_GolfCart=@n1) -> (
      bIsUsingGolfCart=~
      bIsDriver=~
      bIsFrontPassenger=~
      bIsBackLeftPassenger=~
      bIsBackRightPassenger=~
      bIsFrontPassengerAndLeaning=~
      bIsBackPassengerAndLeaning=~
      bIsDrifting=~
      bIsBoosting=~
      bIsEBraking=~
      bIsReversing=~
      bIsBraking=~
      bIsMoving=~
      bIsMovingForward=~
      bIsPowerSliding=~
      bIsLeaning=~
      bIsLeaningOrBouncing=~
      bIsBounceCrouching=~
      bIsBounceCrouched=~
      bIsBounceJumping=~
      bIsBounceRecoiling=~
      bIsSteeringRight=~
      bIsSteeringLeft=~
      RunForwardAlpha=~
      BounceCompression=~
      LeanPosition=~
      LeanPositionX=~
      LeanPositionY=~
      LeanPositionZ=~
      bAimFWD=~
      bAimBWD=~
      bAimLFT=~
      bAimRGT=~
      PawnToVehicleDeltaYawAngleDegrees=~
      AimCardDirDeadZoneAngleDegrees=~
      AimCardDirAngleOffsetDegrees=~
      AimFWDDeltaAngleDegrees=~
      AimBWDDeltaAngleDegrees=~
      AimLFTDeltaAngleDegrees=~
      AimRGTDeltaAngleDegrees=~
      SlopePitchDegreeAngle=~
      SlopeRollDegreeAngle=~
      SteerAngle=~
     ).bAimRGT))
  ]
 }
 AnimationTransitionGraph_0 {
  path=VehicleLowerBodyOverride.AnimGraphNode_StateMachine_0.GolfCartFrontPassenger.AnimStateNode_16.GolfCartPassenger_Front.AnimGraphNode_StateMachine_1.GolfCartPassenger_Front.AnimStateTransitionNode_2.AnimationTransitionGraph_0
  nodes [Result()]
 }
 Transition {
  path=VehicleLowerBodyOverride.AnimGraphNode_StateMachine_0.GolfCartFrontPassenger.AnimStateNode_16.GolfCartPassenger_Front.AnimGraphNode_StateMachine_1.GolfCartPassenger_Front.AnimStateTransitionNode_20.Transition
  nodes [Result(bCanEnterTransition=Break Fort Anim Input Golf Cart(FortAnimInput_GolfCart=Get GolfCartVehicleInput()) -> (
     bIsUsingGolfCart=~
     bIsDriver=~
     bIsFrontPassenger=~
     bIsBackLeftPassenger=~
     bIsBackRightPassenger=~
     bIsFrontPassengerAndLeaning=~
     bIsBackPassengerAndLeaning=~
     bIsDrifting=~
     bIsBoosting=~
     bIsEBraking=~
     bIsReversing=~
     bIsBraking=~
     bIsMoving=~
     bIsMovingForward=~
     bIsPowerSliding=~
     bIsLeaning=~
     bIsLeaningOrBouncing=~
     bIsBounceCrouching=~
     bIsBounceCrouched=~
     bIsBounceJumping=~
     bIsBounceRecoiling=~
     bIsSteeringRight=~
     bIsSteeringLeft=~
     RunForwardAlpha=~
     BounceCompression=~
     LeanPosition=~
     LeanPositionX=~
     LeanPositionY=~
     LeanPositionZ=~
     bAimFWD=~
     bAimBWD=~
     bAimLFT=~
     bAimRGT=~
     PawnToVehicleDeltaYawAngleDegrees=~
     AimCardDirDeadZoneAngleDegrees=~
     AimCardDirAngleOffsetDegrees=~
     AimFWDDeltaAngleDegrees=~
     AimBWDDeltaAngleDegrees=~
     AimLFTDeltaAngleDegrees=~
     AimRGTDeltaAngleDegrees=~
     SlopePitchDegreeAngle=~
     SlopeRollDegreeAngle=~
     SteerAngle=~
    ).bAimBWD)]
 }
 Transition {
  path=VehicleLowerBodyOverride.AnimGraphNode_StateMachine_0.GolfCartFrontPassenger.AnimStateNode_16.GolfCartPassenger_Front.AnimGraphNode_StateMachine_1.GolfCartPassenger_Front.AnimStateTransitionNode_21.Transition
  nodes [
   n1 = Get GolfCartVehicleInput()
   Result(bCanEnterTransition=NOT Boolean(A=Break Fort Anim Input Golf Cart(FortAnimInput_GolfCart=@n1) -> (
      bIsUsingGolfCart=~
      bIsDriver=~
      bIsFrontPassenger=~
      bIsBackLeftPassenger=~
      bIsBackRightPassenger=~
      bIsFrontPassengerAndLeaning=~
      bIsBackPassengerAndLeaning=~
      bIsDrifting=~
      bIsBoosting=~
      bIsEBraking=~
      bIsReversing=~
      bIsBraking=~
      bIsMoving=~
      bIsMovingForward=~
      bIsPowerSliding=~
      bIsLeaning=~
      bIsLeaningOrBouncing=~
      bIsBounceCrouching=~
      bIsBounceCrouched=~
      bIsBounceJumping=~
      bIsBounceRecoiling=~
      bIsSteeringRight=~
      bIsSteeringLeft=~
      RunForwardAlpha=~
      BounceCompression=~
      LeanPosition=~
      LeanPositionX=~
      LeanPositionY=~
      LeanPositionZ=~
      bAimFWD=~
      bAimBWD=~
      bAimLFT=~
      bAimRGT=~
      PawnToVehicleDeltaYawAngleDegrees=~
      AimCardDirDeadZoneAngleDegrees=~
      AimCardDirAngleOffsetDegrees=~
      AimFWDDeltaAngleDegrees=~
      AimBWDDeltaAngleDegrees=~
      AimLFTDeltaAngleDegrees=~
      AimRGTDeltaAngleDegrees=~
      SlopePitchDegreeAngle=~
      SlopeRollDegreeAngle=~
      SteerAngle=~
     ).bAimLFT))
  ]
 }
 Transition {
  path=VehicleLowerBodyOverride.AnimGraphNode_StateMachine_0.GolfCartFrontPassenger.AnimStateNode_16.GolfCartPassenger_Front.AnimGraphNode_StateMachine_1.GolfCartPassenger_Front.AnimStateTransitionNode_22.Transition
  nodes [Result()]
 }
 AnimationTransitionGraph_1 {
  path=VehicleLowerBodyOverride.AnimGraphNode_StateMachine_0.GolfCartFrontPassenger.AnimStateNode_16.GolfCartPassenger_Front.AnimGraphNode_StateMachine_1.GolfCartPassenger_Front.AnimStateTransitionNode_25.AnimationTransitionGraph_1
  nodes [Result(bCanEnterTransition=Break Fort Anim Input Golf Cart(FortAnimInput_GolfCart=Get GolfCartVehicleInput()) -> (
     bIsUsingGolfCart=~
     bIsDriver=~
     bIsFrontPassenger=~
     bIsBackLeftPassenger=~
     bIsBackRightPassenger=~
     bIsFrontPassengerAndLeaning=~
     bIsBackPassengerAndLeaning=~
     bIsDrifting=~
     bIsBoosting=~
     bIsEBraking=~
     bIsReversing=~
     bIsBraking=~
     bIsMoving=~
     bIsMovingForward=~
     bIsPowerSliding=~
     bIsLeaning=~
     bIsLeaningOrBouncing=~
     bIsBounceCrouching=~
     bIsBounceCrouched=~
     bIsBounceJumping=~
     bIsBounceRecoiling=~
     bIsSteeringRight=~
     bIsSteeringLeft=~
     RunForwardAlpha=~
     BounceCompression=~
     LeanPosition=~
     LeanPositionX=~
     LeanPositionY=~
     LeanPositionZ=~
     bAimFWD=~
     bAimBWD=~
     bAimLFT=~
     bAimRGT=~
     PawnToVehicleDeltaYawAngleDegrees=~
     AimCardDirDeadZoneAngleDegrees=~
     AimCardDirAngleOffsetDegrees=~
     AimFWDDeltaAngleDegrees=~
     AimBWDDeltaAngleDegrees=~
     AimLFTDeltaAngleDegrees=~
     AimRGTDeltaAngleDegrees=~
     SlopePitchDegreeAngle=~
     SlopeRollDegreeAngle=~
     SteerAngle=~
    ).bAimRGT)]
 }
 AnimationTransitionGraph_1 {
  path=VehicleLowerBodyOverride.AnimGraphNode_StateMachine_0.GolfCartFrontPassenger.AnimStateNode_16.GolfCartPassenger_Front.AnimGraphNode_StateMachine_1.GolfCartPassenger_Front.AnimStateTransitionNode_26.AnimationTransitionGraph_1
  nodes [Result()]
 }
 AnimationTransitionGraph_0 {
  path=VehicleLowerBodyOverride.AnimGraphNode_StateMachine_0.GolfCartFrontPassenger.AnimStateNode_16.GolfCartPassenger_Front.AnimGraphNode_StateMachine_1.GolfCartPassenger_Front.AnimStateTransitionNode_28.AnimationTransitionGraph_0
  nodes [
   n1 = Get GolfCartVehicleInput()
   Result(bCanEnterTransition=NOT Boolean(A=Break Fort Anim Input Golf Cart(FortAnimInput_GolfCart=@n1) -> (
      bIsUsingGolfCart=~
      bIsDriver=~
      bIsFrontPassenger=~
      bIsBackLeftPassenger=~
      bIsBackRightPassenger=~
      bIsFrontPassengerAndLeaning=~
      bIsBackPassengerAndLeaning=~
      bIsDrifting=~
      bIsBoosting=~
      bIsEBraking=~
      bIsReversing=~
      bIsBraking=~
      bIsMoving=~
      bIsMovingForward=~
      bIsPowerSliding=~
      bIsLeaning=~
      bIsLeaningOrBouncing=~
      bIsBounceCrouching=~
      bIsBounceCrouched=~
      bIsBounceJumping=~
      bIsBounceRecoiling=~
      bIsSteeringRight=~
      bIsSteeringLeft=~
      RunForwardAlpha=~
      BounceCompression=~
      LeanPosition=~
      LeanPositionX=~
      LeanPositionY=~
      LeanPositionZ=~
      bAimFWD=~
      bAimBWD=~
      bAimLFT=~
      bAimRGT=~
      PawnToVehicleDeltaYawAngleDegrees=~
      AimCardDirDeadZoneAngleDegrees=~
      AimCardDirAngleOffsetDegrees=~
      AimFWDDeltaAngleDegrees=~
      AimBWDDeltaAngleDegrees=~
      AimLFTDeltaAngleDegrees=~
      AimRGTDeltaAngleDegrees=~
      SlopePitchDegreeAngle=~
      SlopeRollDegreeAngle=~
      SteerAngle=~
     ).bAimBWD))
  ]
 }
 AnimationTransitionGraph_0 {
  path=VehicleLowerBodyOverride.AnimGraphNode_StateMachine_0.GolfCartFrontPassenger.AnimStateNode_16.GolfCartPassenger_Front.AnimGraphNode_StateMachine_1.GolfCartPassenger_Front.AnimStateTransitionNode_29.AnimationTransitionGraph_0
  nodes [Result(bCanEnterTransition=Break Fort Anim Input Golf Cart(FortAnimInput_GolfCart=Get GolfCartVehicleInput()) -> (
     bIsUsingGolfCart=~
     bIsDriver=~
     bIsFrontPassenger=~
     bIsBackLeftPassenger=~
     bIsBackRightPassenger=~
     bIsFrontPassengerAndLeaning=~
     bIsBackPassengerAndLeaning=~
     bIsDrifting=~
     bIsBoosting=~
     bIsEBraking=~
     bIsReversing=~
     bIsBraking=~
     bIsMoving=~
     bIsMovingForward=~
     bIsPowerSliding=~
     bIsLeaning=~
     bIsLeaningOrBouncing=~
     bIsBounceCrouching=~
     bIsBounceCrouched=~
     bIsBounceJumping=~
     bIsBounceRecoiling=~
     bIsSteeringRight=~
     bIsSteeringLeft=~
     RunForwardAlpha=~
     BounceCompression=~
     LeanPosition=~
     LeanPositionX=~
     LeanPositionY=~
     LeanPositionZ=~
     bAimFWD=~
     bAimBWD=~
     bAimLFT=~
     bAimRGT=~
     PawnToVehicleDeltaYawAngleDegrees=~
     AimCardDirDeadZoneAngleDegrees=~
     AimCardDirAngleOffsetDegrees=~
     AimFWDDeltaAngleDegrees=~
     AimBWDDeltaAngleDegrees=~
     AimLFTDeltaAngleDegrees=~
     AimRGTDeltaAngleDegrees=~
     SlopePitchDegreeAngle=~
     SlopeRollDegreeAngle=~
     SteerAngle=~
    ).bAimLFT)]
 }
 Transition {
  path=VehicleLowerBodyOverride.AnimGraphNode_StateMachine_0.GolfCartFrontPassenger.AnimStateNode_16.GolfCartPassenger_Front.AnimGraphNode_StateMachine_1.GolfCartPassenger_Front.AnimStateTransitionNode_3.Transition
  nodes [Result()]
 }
 AnimationTransitionGraph_1 {
  path=VehicleLowerBodyOverride.AnimGraphNode_StateMachine_0.GolfCartFrontPassenger.AnimStateNode_16.GolfCartPassenger_Front.AnimGraphNode_StateMachine_1.GolfCartPassenger_Front.AnimStateTransitionNode_31.AnimationTransitionGraph_1
  nodes [Result(bCanEnterTransition=Break Fort Anim Input Golf Cart(FortAnimInput_GolfCart=Get GolfCartVehicleInput()) -> (
     bIsUsingGolfCart=~
     bIsDriver=~
     bIsFrontPassenger=~
     bIsBackLeftPassenger=~
     bIsBackRightPassenger=~
     bIsFrontPassengerAndLeaning=~
     bIsBackPassengerAndLeaning=~
     bIsDrifting=~
     bIsBoosting=~
     bIsEBraking=~
     bIsReversing=~
     bIsBraking=~
     bIsMoving=~
     bIsMovingForward=~
     bIsPowerSliding=~
     bIsLeaning=~
     bIsLeaningOrBouncing=~
     bIsBounceCrouching=~
     bIsBounceCrouched=~
     bIsBounceJumping=~
     bIsBounceRecoiling=~
     bIsSteeringRight=~
     bIsSteeringLeft=~
     RunForwardAlpha=~
     BounceCompression=~
     LeanPosition=~
     LeanPositionX=~
     LeanPositionY=~
     LeanPositionZ=~
     bAimFWD=~
     bAimBWD=~
     bAimLFT=~
     bAimRGT=~
     PawnToVehicleDeltaYawAngleDegrees=~
     AimCardDirDeadZoneAngleDegrees=~
     AimCardDirAngleOffsetDegrees=~
     AimFWDDeltaAngleDegrees=~
     AimBWDDeltaAngleDegrees=~
     AimLFTDeltaAngleDegrees=~
     AimRGTDeltaAngleDegrees=~
     SlopePitchDegreeAngle=~
     SlopeRollDegreeAngle=~
     SteerAngle=~
    ).bAimFWD)]
 }
 AnimationTransitionGraph_1 {
  path=VehicleLowerBodyOverride.AnimGraphNode_StateMachine_0.GolfCartFrontPassenger.AnimStateNode_16.GolfCartPassenger_Front.AnimGraphNode_StateMachine_1.GolfCartPassenger_Front.AnimStateTransitionNode_32.AnimationTransitionGraph_1
  nodes [
   n1 = Get GolfCartVehicleInput()
   Result(bCanEnterTransition=NOT Boolean(A=Break Fort Anim Input Golf Cart(FortAnimInput_GolfCart=@n1) -> (
      bIsUsingGolfCart=~
      bIsDriver=~
      bIsFrontPassenger=~
      bIsBackLeftPassenger=~
      bIsBackRightPassenger=~
      bIsFrontPassengerAndLeaning=~
      bIsBackPassengerAndLeaning=~
      bIsDrifting=~
      bIsBoosting=~
      bIsEBraking=~
      bIsReversing=~
      bIsBraking=~
      bIsMoving=~
      bIsMovingForward=~
      bIsPowerSliding=~
      bIsLeaning=~
      bIsLeaningOrBouncing=~
      bIsBounceCrouching=~
      bIsBounceCrouched=~
      bIsBounceJumping=~
      bIsBounceRecoiling=~
      bIsSteeringRight=~
      bIsSteeringLeft=~
      RunForwardAlpha=~
      BounceCompression=~
      LeanPosition=~
      LeanPositionX=~
      LeanPositionY=~
      LeanPositionZ=~
      bAimFWD=~
      bAimBWD=~
      bAimLFT=~
      bAimRGT=~
      PawnToVehicleDeltaYawAngleDegrees=~
      AimCardDirDeadZoneAngleDegrees=~
      AimCardDirAngleOffsetDegrees=~
      AimFWDDeltaAngleDegrees=~
      AimBWDDeltaAngleDegrees=~
      AimLFTDeltaAngleDegrees=~
      AimRGTDeltaAngleDegrees=~
      SlopePitchDegreeAngle=~
      SlopeRollDegreeAngle=~
      SteerAngle=~
     ).bAimFWD))
  ]
 }
 AnimationTransitionGraph_1 {
  path=VehicleLowerBodyOverride.AnimGraphNode_StateMachine_0.GolfCartFrontPassenger.AnimStateNode_16.GolfCartPassenger_Front.AnimGraphNode_StateMachine_1.GolfCartPassenger_Front.AnimStateTransitionNode_35.AnimationTransitionGraph_1
  nodes [Result(bCanEnterTransition=Break Fort Anim Input Golf Cart(FortAnimInput_GolfCart=Get GolfCartVehicleInput()) -> (
     bIsUsingGolfCart=~
     bIsDriver=~
     bIsFrontPassenger=~
     bIsBackLeftPassenger=~
     bIsBackRightPassenger=~
     bIsFrontPassengerAndLeaning=~
     bIsBackPassengerAndLeaning=~
     bIsDrifting=~
     bIsBoosting=~
     bIsEBraking=~
     bIsReversing=~
     bIsBraking=~
     bIsMoving=~
     bIsMovingForward=~
     bIsPowerSliding=~
     bIsLeaning=~
     bIsLeaningOrBouncing=~
     bIsBounceCrouching=~
     bIsBounceCrouched=~
     bIsBounceJumping=~
     bIsBounceRecoiling=~
     bIsSteeringRight=~
     bIsSteeringLeft=~
     RunForwardAlpha=~
     BounceCompression=~
     LeanPosition=~
     LeanPositionX=~
     LeanPositionY=~
     LeanPositionZ=~
     bAimFWD=~
     bAimBWD=~
     bAimLFT=~
     bAimRGT=~
     PawnToVehicleDeltaYawAngleDegrees=~
     AimCardDirDeadZoneAngleDegrees=~
     AimCardDirAngleOffsetDegrees=~
     AimFWDDeltaAngleDegrees=~
     AimBWDDeltaAngleDegrees=~
     AimLFTDeltaAngleDegrees=~
     AimRGTDeltaAngleDegrees=~
     SlopePitchDegreeAngle=~
     SlopeRollDegreeAngle=~
     SteerAngle=~
    ).bAimFWD)]
 }
 AnimationTransitionGraph_1 {
  path=VehicleLowerBodyOverride.AnimGraphNode_StateMachine_0.GolfCartFrontPassenger.AnimStateNode_16.GolfCartPassenger_Front.AnimGraphNode_StateMachine_1.GolfCartPassenger_Front.AnimStateTransitionNode_36.AnimationTransitionGraph_1
  nodes [
   n1 = Get GolfCartVehicleInput()
   Result(bCanEnterTransition=NOT Boolean(A=Break Fort Anim Input Golf Cart(FortAnimInput_GolfCart=@n1) -> (
      bIsUsingGolfCart=~
      bIsDriver=~
      bIsFrontPassenger=~
      bIsBackLeftPassenger=~
      bIsBackRightPassenger=~
      bIsFrontPassengerAndLeaning=~
      bIsBackPassengerAndLeaning=~
      bIsDrifting=~
      bIsBoosting=~
      bIsEBraking=~
      bIsReversing=~
      bIsBraking=~
      bIsMoving=~
      bIsMovingForward=~
      bIsPowerSliding=~
      bIsLeaning=~
      bIsLeaningOrBouncing=~
      bIsBounceCrouching=~
      bIsBounceCrouched=~
      bIsBounceJumping=~
      bIsBounceRecoiling=~
      bIsSteeringRight=~
      bIsSteeringLeft=~
      RunForwardAlpha=~
      BounceCompression=~
      LeanPosition=~
      LeanPositionX=~
      LeanPositionY=~
      LeanPositionZ=~
      bAimFWD=~
      bAimBWD=~
      bAimLFT=~
      bAimRGT=~
      PawnToVehicleDeltaYawAngleDegrees=~
      AimCardDirDeadZoneAngleDegrees=~
      AimCardDirAngleOffsetDegrees=~
      AimFWDDeltaAngleDegrees=~
      AimBWDDeltaAngleDegrees=~
      AimLFTDeltaAngleDegrees=~
      AimRGTDeltaAngleDegrees=~
      SlopePitchDegreeAngle=~
      SlopeRollDegreeAngle=~
      SteerAngle=~
     ).bAimFWD))
  ]
 }
 AnimationTransitionGraph_1 {
  path=VehicleLowerBodyOverride.AnimGraphNode_StateMachine_0.GolfCartFrontPassenger.AnimStateNode_16.GolfCartPassenger_Front.AnimGraphNode_StateMachine_1.GolfCartPassenger_Front.AnimStateTransitionNode_37.AnimationTransitionGraph_1
  nodes [Result(bCanEnterTransition=Break Fort Anim Input Golf Cart(FortAnimInput_GolfCart=Get GolfCartVehicleInput()) -> (
     bIsUsingGolfCart=~
     bIsDriver=~
     bIsFrontPassenger=~
     bIsBackLeftPassenger=~
     bIsBackRightPassenger=~
     bIsFrontPassengerAndLeaning=~
     bIsBackPassengerAndLeaning=~
     bIsDrifting=~
     bIsBoosting=~
     bIsEBraking=~
     bIsReversing=~
     bIsBraking=~
     bIsMoving=~
     bIsMovingForward=~
     bIsPowerSliding=~
     bIsLeaning=~
     bIsLeaningOrBouncing=~
     bIsBounceCrouching=~
     bIsBounceCrouched=~
     bIsBounceJumping=~
     bIsBounceRecoiling=~
     bIsSteeringRight=~
     bIsSteeringLeft=~
     RunForwardAlpha=~
     BounceCompression=~
     LeanPosition=~
     LeanPositionX=~
     LeanPositionY=~
     LeanPositionZ=~
     bAimFWD=~
     bAimBWD=~
     bAimLFT=~
     bAimRGT=~
     PawnToVehicleDeltaYawAngleDegrees=~
     AimCardDirDeadZoneAngleDegrees=~
     AimCardDirAngleOffsetDegrees=~
     AimFWDDeltaAngleDegrees=~
     AimBWDDeltaAngleDegrees=~
     AimLFTDeltaAngleDegrees=~
     AimRGTDeltaAngleDegrees=~
     SlopePitchDegreeAngle=~
     SlopeRollDegreeAngle=~
     SteerAngle=~
    ).bAimRGT)]
 }
 Transition {
  path=VehicleLowerBodyOverride.AnimGraphNode_StateMachine_0.GolfCartFrontPassenger.AnimStateNode_16.GolfCartPassenger_Front.AnimGraphNode_StateMachine_1.GolfCartPassenger_Front.AnimStateTransitionNode_4.Transition
  nodes [
   n1 = Get GolfCartVehicleInput()
   Result(bCanEnterTransition=NOT Boolean(A=Break Fort Anim Input Golf Cart(FortAnimInput_GolfCart=@n1) -> (
      bIsUsingGolfCart=~
      bIsDriver=~
      bIsFrontPassenger=~
      bIsBackLeftPassenger=~
      bIsBackRightPassenger=~
      bIsFrontPassengerAndLeaning=~
      bIsBackPassengerAndLeaning=~
      bIsDrifting=~
      bIsBoosting=~
      bIsEBraking=~
      bIsReversing=~
      bIsBraking=~
      bIsMoving=~
      bIsMovingForward=~
      bIsPowerSliding=~
      bIsLeaning=~
      bIsLeaningOrBouncing=~
      bIsBounceCrouching=~
      bIsBounceCrouched=~
      bIsBounceJumping=~
      bIsBounceRecoiling=~
      bIsSteeringRight=~
      bIsSteeringLeft=~
      RunForwardAlpha=~
      BounceCompression=~
      LeanPosition=~
      LeanPositionX=~
      LeanPositionY=~
      LeanPositionZ=~
      bAimFWD=~
      bAimBWD=~
      bAimLFT=~
      bAimRGT=~
      PawnToVehicleDeltaYawAngleDegrees=~
      AimCardDirDeadZoneAngleDegrees=~
      AimCardDirAngleOffsetDegrees=~
      AimFWDDeltaAngleDegrees=~
      AimBWDDeltaAngleDegrees=~
      AimLFTDeltaAngleDegrees=~
      AimRGTDeltaAngleDegrees=~
      SlopePitchDegreeAngle=~
      SlopeRollDegreeAngle=~
      SteerAngle=~
     ).bAimLFT))
  ]
 }
 AnimationTransitionGraph_1 {
  path=VehicleLowerBodyOverride.AnimGraphNode_StateMachine_0.GolfCartFrontPassenger.AnimStateNode_16.GolfCartPassenger_Front.AnimGraphNode_StateMachine_1.GolfCartPassenger_Front.AnimStateTransitionNode_40.AnimationTransitionGraph_1
  nodes [Result()]
 }
 AnimationTransitionGraph_1 {
  path=VehicleLowerBodyOverride.AnimGraphNode_StateMachine_0.GolfCartFrontPassenger.AnimStateNode_16.GolfCartPassenger_Front.AnimGraphNode_StateMachine_1.GolfCartPassenger_Front.AnimStateTransitionNode_41.AnimationTransitionGraph_1
  nodes [Result(bCanEnterTransition=Break Fort Anim Input Golf Cart(FortAnimInput_GolfCart=Get GolfCartVehicleInput()) -> (
     bIsUsingGolfCart=~
     bIsDriver=~
     bIsFrontPassenger=~
     bIsBackLeftPassenger=~
     bIsBackRightPassenger=~
     bIsFrontPassengerAndLeaning=~
     bIsBackPassengerAndLeaning=~
     bIsDrifting=~
     bIsBoosting=~
     bIsEBraking=~
     bIsReversing=~
     bIsBraking=~
     bIsMoving=~
     bIsMovingForward=~
     bIsPowerSliding=~
     bIsLeaning=~
     bIsLeaningOrBouncing=~
     bIsBounceCrouching=~
     bIsBounceCrouched=~
     bIsBounceJumping=~
     bIsBounceRecoiling=~
     bIsSteeringRight=~
     bIsSteeringLeft=~
     RunForwardAlpha=~
     BounceCompression=~
     LeanPosition=~
     LeanPositionX=~
     LeanPositionY=~
     LeanPositionZ=~
     bAimFWD=~
     bAimBWD=~
     bAimLFT=~
     bAimRGT=~
     PawnToVehicleDeltaYawAngleDegrees=~
     AimCardDirDeadZoneAngleDegrees=~
     AimCardDirAngleOffsetDegrees=~
     AimFWDDeltaAngleDegrees=~
     AimBWDDeltaAngleDegrees=~
     AimLFTDeltaAngleDegrees=~
     AimRGTDeltaAngleDegrees=~
     SlopePitchDegreeAngle=~
     SlopeRollDegreeAngle=~
     SteerAngle=~
    ).bAimLFT)]
 }
 AnimationTransitionGraph_1 {
  path=VehicleLowerBodyOverride.AnimGraphNode_StateMachine_0.GolfCartFrontPassenger.AnimStateNode_16.GolfCartPassenger_Front.AnimGraphNode_StateMachine_1.GolfCartPassenger_Front.AnimStateTransitionNode_42.AnimationTransitionGraph_1
  nodes [Result()]
 }
 AnimationTransitionGraph_1 {
  path=VehicleLowerBodyOverride.AnimGraphNode_StateMachine_0.GolfCartFrontPassenger.AnimStateNode_16.GolfCartPassenger_Front.AnimGraphNode_StateMachine_1.GolfCartPassenger_Front.AnimStateTransitionNode_5.AnimationTransitionGraph_1
  nodes [
   n1 = Get GolfCartVehicleInput()
   Result(bCanEnterTransition=NOT Boolean(A=Break Fort Anim Input Golf Cart(FortAnimInput_GolfCart=@n1) -> (
      bIsUsingGolfCart=~
      bIsDriver=~
      bIsFrontPassenger=~
      bIsBackLeftPassenger=~
      bIsBackRightPassenger=~
      bIsFrontPassengerAndLeaning=~
      bIsBackPassengerAndLeaning=~
      bIsDrifting=~
      bIsBoosting=~
      bIsEBraking=~
      bIsReversing=~
      bIsBraking=~
      bIsMoving=~
      bIsMovingForward=~
      bIsPowerSliding=~
      bIsLeaning=~
      bIsLeaningOrBouncing=~
      bIsBounceCrouching=~
      bIsBounceCrouched=~
      bIsBounceJumping=~
      bIsBounceRecoiling=~
      bIsSteeringRight=~
      bIsSteeringLeft=~
      RunForwardAlpha=~
      BounceCompression=~
      LeanPosition=~
      LeanPositionX=~
      LeanPositionY=~
      LeanPositionZ=~
      bAimFWD=~
      bAimBWD=~
      bAimLFT=~
      bAimRGT=~
      PawnToVehicleDeltaYawAngleDegrees=~
      AimCardDirDeadZoneAngleDegrees=~
      AimCardDirAngleOffsetDegrees=~
      AimFWDDeltaAngleDegrees=~
      AimBWDDeltaAngleDegrees=~
      AimLFTDeltaAngleDegrees=~
      AimRGTDeltaAngleDegrees=~
      SlopePitchDegreeAngle=~
      SlopeRollDegreeAngle=~
      SteerAngle=~
     ).bAimBWD))
  ]
 }
 Transition {
  path=VehicleLowerBodyOverride.AnimGraphNode_StateMachine_0.GolfCartFrontPassenger.AnimStateNode_16.GolfCartPassenger_Front.AnimGraphNode_StateMachine_1.GolfCartPassenger_Front.AnimStateTransitionNode_7.Transition
  nodes [Result()]
 }
 VehicleSplitBodyOverride {nodes [
   Input Pose() {name=InPoseSplitBody}
   Input Pose() {name=InPoseUpperAndLowerBody}
   Input Pose() {name=InPoseUpperAndLowerBodyRemoveRoot}
   Output Pose() {layerGroup=Vehicle Override}
  ]}
]
variables [isDBNO {type=boolean}]
