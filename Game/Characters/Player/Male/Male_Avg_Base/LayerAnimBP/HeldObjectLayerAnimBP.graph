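# Animation layer Blueprint for the Fortnite medium male player skeleton that supplies Item Override layers while carrying a held object.
# Jump, fall, land, jetpack, swim-jump and zipline layers play additive sequences, many borrowed from assault rifle animations.
# Full-body aiming uses the HeldObjects aim offset, while several other layers have unwired outputs.
# Pickup and place notifies toggle IsHeldObjectDropped, which switches the upper body to a relaxed no-weapon pose once placed.
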AnimBlueprint /Game/Characters/Player/Male/Male_Avg_Base/LayerAnimBP/HeldObjectLayerAnimBP
target_skeleton=&Fortnite_M_Avg_Player_Skeleton
graphs [
 AnimGraph {nodes [Output Pose() {layerGroup=None}]}
 EventGraph {nodes [
   n1 = Event BlueprintInitializeAnimation()
   Set IsHeldObjectDropped(execute=[@n1.then, @n2.then], IsHeldObjectDropped:bool=false) -> (Output_Get:bool)
   n2 = AnimNotify_HeldObjectPickUp() {is=title K2Node_Event}
   AnimNotify_HeldObjectPlace() {is=title K2Node_Event}
   > Set IsHeldObjectDropped(IsHeldObjectDropped=true)
  ]}
 ItemAimOffset {nodes [
   Input Pose!() -> (Pose, AimOffsetAlpha:double, Yaw:double, Pitch:double) {name=InPose_AimOffset}
   Input Pose!() -> (Pose) {name=InPose_PreAimOffsets}
   Output Pose() {layerGroup=Item Override}
  ]}
 ItemFallAdditive {nodes [
   Input Pose() {name=InPoseItemFallAdditive}
   Output Pose(Result="Sequence Player (additive) 'M_JumpFall_ARDS_Additive'"!(PlayRate:float=Get FallPlayRate() -> (FallPlayRate:float)) -> (Pose) {bLoopAnimation=false, sequence=/Game/Animation/Game/MainPlayer/Locomotion/Medium/Male/Jump/AssaultRifle/M_JumpFall_ARDS_Additive}) {layerGroup=Item Override}
  ]}
 ItemFallLandAdditive {nodes [
   Input Pose() {name=InPoseItemFallLandAdditive}
   n1 = "Sequence Player (additive) 'M_JumpLand_ARDS_Additive'"() {
    bLoopAnimation=false
    groupName=JumpLand
    method=SyncGroup
    sequence=/Game/Animation/Game/MainPlayer/Locomotion/Medium/Male/Jump/AssaultRifle/M_JumpLand_ARDS_Additive
   }
   Output Pose(Result=@n1) {layerGroup=Item Override}
  ]}
 ItemFallLevel2Override {nodes [
   Input Pose() {name=InPose_FallLevel2}
   Output Pose() {layerGroup=Item Override}
  ]}
 ItemFinalPoseOverride {nodes [
   Input Pose() {name=InFinalPose}
   Output Pose() {layerGroup=Item Override}
  ]}
 ItemFullBodyAimOffsets {nodes [
   n1 = Input Pose!() -> (Pose, FullBody_AimOffsetAlpha:double, Aim_Yaw:double, Aim_Pitch:double) {name=InPoseFullBodyAimOffsets}
   n2 = AimOffset Player 'HeldObjects_AO'!(
    X:float=@n1.Aim_Yaw
    Y:float=@n1.Aim_Pitch
    BasePose:Pose Link Structure=Input Pose!() -> (Pose) {name=InPosePreFullBodyAimOffsets}
    Alpha:float=@n1.FullBody_AimOffsetAlpha
   ) -> (Pose:Pose Link Structure) {blendSpace=/Game/Animation/Game/MainPlayer/Combat/Gadgets/Medium/Male/HeldObjects/AO/HeldObjects_AO, alphaBoolBlend {blendOption=Linear}}
   Output Pose(Result=@n2) {layerGroup=Item Override}
  ]}
 ItemFullBodyOverride {nodes [
   Input Pose() {name=InPoseFullBody}
   Output Pose() {layerGroup=Item Override}
  ]}
 ItemIdleOverride {nodes [
   Input Pose() {name=InPose_Idle}
   Output Pose() {layerGroup=Item Override}
  ]}
 ItemJetPackJumpAdditive {nodes [Output Pose(Result="Sequence Player (additive) 'JetpackApex_ARDS_Loop_Additive'"() {sequence=/Game/Animation/Game/MainPlayer/Locomotion/Medium/Male/Jump/AssaultRifle/JetpackApex_ARDS_Loop_Additive}) {layerGroup=Item Override}]}
 ItemJetPackStartAdditive {nodes [Output Pose(Result="Sequence Player (additive) 'JetpackUp_ARDS_Additive'"() {bLoopAnimation=false, sequence=/Game/Animation/Game/MainPlayer/Locomotion/Medium/Male/Jump/AssaultRifle/JetpackUp_ARDS_Additive}) {layerGroup=Item Override}]}
 ItemJumpLoopAdditive {nodes [
   Input Pose() {name=InPoseItemJumpLoopAdditive}
   Output Pose(Result="Sequence Player (additive) 'M_JumpApex_ARDS_Loop_Additive'"() {playRate=0.5, sequence=/Game/Animation/Game/MainPlayer/Locomotion/Medium/Male/Jump/AssaultRifle/M_JumpApex_ARDS_Loop_Additive}) {layerGroup=Item Override}
  ]}
 ItemJumpOverride {nodes [
   Input Pose() {name=InPoseJump}
   Output Pose() {layerGroup=Item Override}
  ]}
 ItemJumpUpAdditive {nodes [
   Input Pose() {name=InPoseItemJumpUpAdditive}
   Output Pose(Result="Sequence Player (additive) 'M_JumpUp_ARDS_Additive'"() {bLoopAnimation=false, sequence=/Game/Animation/Game/MainPlayer/Locomotion/Medium/Male/Jump/AssaultRifle/M_JumpUp_ARDS_Additive}) {layerGroup=Item Override}
  ]}
 ItemPreAimOffsets {nodes [
   Input Pose() {name=InPosePreAimOffsets}
   Output Pose() {layerGroup=Item Override}
  ]}
 ItemPreIK {nodes [
   Input Pose() {name=InPose_PreIK}
   Output Pose() {layerGroup=Item Override}
  ]}
 ItemRightHandIK {nodes [
   Input Pose() {name=InPosePreRightHandIK}
   Input Pose() {name=InPosePostRightHandIK}
   Output Pose() {layerGroup=Item Override}
  ]}
 ItemSkeletalControl {nodes [
   Input Pose()
   Output Pose() {layerGroup=Item Override}
  ]}
 ItemSwimJumpFallLoop {nodes [Output Pose(Result="Sequence Player (additive) 'HeldObjects_Swim_Jump_Fall_Loop'"!(PlayRate:float=2, StartPosition:float=0.25) -> (Pose) {bLoopAnimation=false, sequence=/Game/Animation/Game/MainPlayer/Swimming/Jump/HeldObjects_Swim_Jump_Fall_Loop}) {layerGroup=Item Override}]}
 ItemSwimJumpSurfaceEnd {nodes [Output Pose(Result="Sequence Player (additive) 'HeldObjects_Swim_Jump_Surface_End'"() {bLoopAnimation=false, sequence=/Game/Animation/Game/MainPlayer/Swimming/Jump/HeldObjects_Swim_Jump_Surface_End}) {layerGroup=Item Override}]}
 ItemSwimJumpSurfaceLoop {nodes [Output Pose(Result="Sequence Player (additive) 'HeldObjects_Swim_Jump_Surface_Loop'"!(PlayRate:float=0.5, StartPosition:float=0.167) -> (Pose) {sequence=/Game/Animation/Game/MainPlayer/Swimming/Jump/HeldObjects_Swim_Jump_Surface_Loop}) {layerGroup=Item Override}]}
 ItemUpperBody {nodes [Output Pose(Result=Blend Poses by bool(
     BlendPose_0=Sequence Player 'M_NoWep_Relaxed'() {sequence=/Game/Animation/Game/MainPlayer/Poses/Medium/Male/M_NoWep_Relaxed}
     BlendPose_1=Input Pose!() -> (Pose, InputParam:Fort Anim Input Adjusted Aim Structure) {name=InPoseUpperBody}
     BlendTime_0=0
     bActiveValue=Get IsHeldObjectDropped() -> (IsHeldObjectDropped:bool)
    ) {blendType=Linear}) {layerGroup=Item Override}]}
 ItemZipLineStartAdditive {nodes [
   n1 = "Sequence Player (additive) 'ZipLine_Entry_SneakySnowman_Additive'"() {bLoopAnimation=false, sequence=/Game/Animation/Game/MainPlayer/Combat/Gadgets/Medium/Male/ZipLine/ZipLine_Entry_SneakySnowman_Additive}
   Output Pose(Result=@n1) {layerGroup=Item Override}
  ]}
]
variables [IsHeldObjectDropped {type=boolean}]
