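# Layer Anim Blueprint that poses a player riding as a shopping cart passenger, driven by ShoppingCartInput.
# The lower-body layer switches between the incoming upper-body pose and a four-direction state machine with turn clips, yawed by the pawn-to-vehicle delta.
# Each directional idle feeds a shared lower-body aim offset, with root rotations around it for non-forward facings.
# The full-body layer adds slope pitch and roll aim offsets, or swaps to a DBNO animation.
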AnimBlueprint /Game/Characters/Player/Male/Male_Avg_Base/LayerAnimBP/ShoppingCartPassengerLayerAnimBP
target_skeleton=&Fortnite_M_Avg_Player_Skeleton
graphs [
 AnimGraph {nodes [Output Pose() {layerGroup=None}]}
 EventGraph {links [], nodes []}
 VehicleFinalPoseOverride {nodes [
   Input Pose() {name=InFinalPose}
   Output Pose() {layerGroup=VehicleOverride}
  ]}
 VehicleFullBodyOverride {nodes [
   n1 = Local To Component(LocalPose=Use cached pose 'UpperAndLowerBodyCachePose'())
   n2 = Break Fort Anim Input Shopping Cart!(FortAnimInput_ShoppingCart:"Fort Anim Input Shopping Cart Structure (by ref)"=Get ShoppingCartInput() -> (ShoppingCartInput:Fort Anim Input Shopping Cart Structure)) -> (
    bIsUsingShoppingCart:bool=~
    bIsUsingVehicle:bool=~
    CoastState:ECoastState Enum
    bIsCoastStatePedaling:bool=~
    bIsCoastStateCoasting:bool=~
    bIsCoastStateDismount:bool=~
    bIsCoastStateIdle:bool=~
    bIsInAir:bool=~
    bIsCoasting:bool=~
    bIsPedaling:bool=~
    bIsReadyToPedal:bool=~
    IsReadyToPedal:float=~
    bWantsToCoast:bool=~
    bIsCoastIdling:bool=~
    bIsStartCoasting:bool=~
    bIsEndCoasting:bool=~
    bIsDismountingFromCoast:bool=~
    bIsCoastingOrDismountingFromCoast:bool=~
    bIsStandingInPlace:bool=~
    bIsSprinting:bool=~
    bIsSprintingAndMovingForward:bool=~
    bIsMovingForwardNotSprinting:bool=~
    bIsBraking:bool=~
    bIsReversing:bool=~
    bIsMoving:bool=~
    bIsMovingForward:bool=~
    bIsMovingBackwards:bool=~
    bIsMovingOrTurningInPlace:bool=~
    bIsInAirSteady:bool=~
    bIsOnSlope:bool=~
    bAimFWD:bool=~
    bAimBWD:bool=~
    bAimLFT:bool=~
    bAimRGT:bool=~
    ForwardVelocity:float=~
    ForwardSpeedKmH:float=~
    CurrentBrakeForce:float=~
    RunForwardAlpha:float=~
    bIsAcceleratingForward:bool=~
    bIsAccelBreakingOrReversing:bool=~
    SteerAngle:float=~
    SteerAngleInterpSpeed:float=~
    CoastSteerAngleInterpSpeed:float=~
    IsReadyToPedalInterpSpeed:float=~
    StandingInPlaceSteerAngle:float=~
    SlopePitchDegreeAngle:float=~
    SlopeRollDegreeAngle:float=~
    PawnToVehicleDeltaYawAngleDegrees:float=~
    AimCardDirDeadZoneAngleDegrees:float=~
    AimCardDirAngleOffsetDegrees:float=~
    AimFWDDeltaAngleDegrees:float=~
    AimBWDDeltaAngleDegrees:float=~
    AimLFTDeltaAngleDegrees:float=~
    AimRGTDeltaAngleDegrees:float=~
   ) {is=title K2Node_BreakStruct}
   n3 = AimOffset Player 'ShoppingCart_Passenger_Slope_PitchAO'!(
    X:float=@n2.SlopePitchDegreeAngle
    BasePose:Pose Link Structure=Component To Local(ComponentPose=Reset Root Transform!(ComponentPose:Component Space Pose Link Structure=@n1, Alpha:float=1) -> (Pose:Component Space Pose Link Structure) {is=title AnimGraphNode_ResetRoot, alphaBoolBlend {blendOption=Linear}})
    Alpha:float=1
   ) -> (Pose:Pose Link Structure) {
    blendSpace=/Game/Animation/Game/MainPlayer/Vehicles/ShoppingCart/AO/ShoppingCart_Passenger_Slope_PitchAO
    alphaBoolBlend {blendOption=Linear}
   }
   n4 = AimOffset Player 'ShoppingCart_Passenger_Slope_RollAO'!(X:float=@n2.SlopeRollDegreeAngle, BasePose:Pose Link Structure=@n3, Alpha:float=1) -> (Pose:Pose Link Structure) {blendSpace=/Game/Animation/Game/MainPlayer/Vehicles/ShoppingCart/AO/ShoppingCart_Passenger_Slope_RollAO, alphaBoolBlend {blendOption=Linear}}
   Output Pose(Result=Blend Poses by bool(
     BlendPose_0=Sequence Player 'TrendConcentrate_PassengerDBNO_AnimSET'() {sequence=/Game/Animation/Game/MainPlayer/Locomotion/Medium/Male/DBNO/TrendConcentrate_PassengerDBNO_AnimSET}
     BlendPose_1=@n4
     bActiveValue=Get isDBNO() -> (isDBNO:bool)
    ) {blendType=Linear, childUpateMode=ResetChildOnActivate}) {layerGroup=VehicleOverride}
   Save cached pose 'UpperAndLowerBodyCachePose'(Pose=Input Pose() {name=InPoseFullBody}) {cacheName=UpperAndLowerBodyCachePose}
  ]}
 VehicleLowerBodyOverride {nodes [
   Input Pose() {name=InPoseLowerBodyDefault}
   AnimGraphNode_StateMachine_0 = ShoppingCartPassenger() {is=title AnimGraphNode_StateMachine}
   Output Pose(Result=@AnimGraphNode_StateMachine_0) {layerGroup=VehicleOverride}
   Save cached pose 'UpperBodyCachePose'(Pose=Input Pose() {name=InPoseUpperBody}) {cacheName=UpperBodyCachePose}
  ]}
 ShoppingCartPassenger {
  entry=IdlesCacheState
  path=VehicleLowerBodyOverride.AnimGraphNode_StateMachine_0.ShoppingCartPassenger
  ids=id
  nodes [
   AnimStateNode_0 = IdlesCacheState() {is=title AnimStateNode}
   AnimStateNode_7 = ShoppingCartPassenger_SM() {is=title AnimStateNode}
   AnimStateTransitionNode_0 = AnimStateTransitionNode()
   AnimStateTransitionNode_1 = AnimStateTransitionNode()
  ]
  links [
   {
    dst=ShoppingCartPassenger_SM
    rule=VehicleLowerBodyOverride.AnimGraphNode_StateMachine_0.ShoppingCartPassenger.AnimStateTransitionNode_0.Transition
    src=IdlesCacheState
    via=AnimStateTransitionNode_0
   }
   {
    dst=IdlesCacheState
    rule=VehicleLowerBodyOverride.AnimGraphNode_StateMachine_0.ShoppingCartPassenger.AnimStateTransitionNode_1.Transition
    src=ShoppingCartPassenger_SM
    via=AnimStateTransitionNode_1
   }
  ]
 }
 IdlesCacheState {
  path=VehicleLowerBodyOverride.AnimGraphNode_StateMachine_0.ShoppingCartPassenger.AnimStateNode_0.IdlesCacheState
  nodes [Output Animation Pose(Result=Use cached pose 'UpperBodyCachePose'())]
 }
 ShoppingCartPassenger_SM {
  path=VehicleLowerBodyOverride.AnimGraphNode_StateMachine_0.ShoppingCartPassenger.AnimStateNode_7.ShoppingCartPassenger_SM
  nodes [
   AnimGraphNode_StateMachine_1 = ShoppingCartPassenger()
   n1 = Get ShoppingCartInput()
   Output Animation Pose(Result=Rotate Root Bone!(
     BasePose:Pose Link Structure=@AnimGraphNode_StateMachine_1
     Pitch:float=0
     Yaw:float=Break Fort Anim Input Shopping Cart(FortAnimInput_ShoppingCart=@n1) -> (
      bIsUsingShoppingCart=~
      bIsUsingVehicle=~
      bIsCoastStatePedaling=~
      bIsCoastStateCoasting=~
      bIsCoastStateDismount=~
      bIsCoastStateIdle=~
      bIsInAir=~
      bIsCoasting=~
      bIsPedaling=~
      bIsReadyToPedal=~
      IsReadyToPedal=~
      bWantsToCoast=~
      bIsCoastIdling=~
      bIsStartCoasting=~
      bIsEndCoasting=~
      bIsDismountingFromCoast=~
      bIsCoastingOrDismountingFromCoast=~
      bIsStandingInPlace=~
      bIsSprinting=~
      bIsSprintingAndMovingForward=~
      bIsMovingForwardNotSprinting=~
      bIsBraking=~
      bIsReversing=~
      bIsMoving=~
      bIsMovingForward=~
      bIsMovingBackwards=~
      bIsMovingOrTurningInPlace=~
      bIsInAirSteady=~
      bIsOnSlope=~
      bAimFWD=~
      bAimBWD=~
      bAimLFT=~
      bAimRGT=~
      ForwardVelocity=~
      ForwardSpeedKmH=~
      CurrentBrakeForce=~
      RunForwardAlpha=~
      bIsAcceleratingForward=~
      bIsAccelBreakingOrReversing=~
      SteerAngle=~
      SteerAngleInterpSpeed=~
      CoastSteerAngleInterpSpeed=~
      IsReadyToPedalInterpSpeed=~
      StandingInPlaceSteerAngle=~
      SlopePitchDegreeAngle=~
      SlopeRollDegreeAngle=~
      PawnToVehicleDeltaYawAngleDegrees=~
      AimCardDirDeadZoneAngleDegrees=~
      AimCardDirAngleOffsetDegrees=~
      AimFWDDeltaAngleDegrees=~
      AimBWDDeltaAngleDegrees=~
      AimLFTDeltaAngleDegrees=~
      AimRGTDeltaAngleDegrees=~
     ).PawnToVehicleDeltaYawAngleDegrees
     MeshToComponent:Rotator="0,0,0"
    ) -> (Pose:Pose Link Structure) {is=title AnimGraphNode_RotateRootBone})
  ]
 }
 ShoppingCartPassenger {
  entry=PassengerPose_N
  path=VehicleLowerBodyOverride.AnimGraphNode_StateMachine_0.ShoppingCartPassenger.AnimStateNode_7.ShoppingCartPassenger_SM.AnimGraphNode_StateMachine_1.ShoppingCartPassenger
  ids=id
  nodes [
   AnimStateNode_10 = Transition_EtoS() {is=title AnimStateNode}
   AnimStateNode_11 = PassengerPose_S() {is=title AnimStateNode}
   AnimStateNode_12 = Transition_StoW() {is=title AnimStateNode}
   AnimStateNode_13 = PassengerPose_W() {is=title AnimStateNode}
   AnimStateNode_15 = Transition_WtoN() {is=title AnimStateNode}
   AnimStateNode_16 = PassengerPose_N() {is=title AnimStateNode}
   AnimStateNode_17 = Transition_EtoN() {is=title AnimStateNode}
   AnimStateNode_18 = Transition_StoE() {is=title AnimStateNode}
   AnimStateNode_19 = Transition_WtoS() {is=title AnimStateNode}
   AnimStateNode_20 = Transition_NtoW() {is=title AnimStateNode}
   AnimStateNode_8 = Transition_NtoE() {is=title AnimStateNode}
   AnimStateNode_9 = PassengerPose_E() {is=title AnimStateNode}
   AnimStateTransitionNode_0 = AnimStateTransitionNode()
   AnimStateTransitionNode_1 = AnimStateTransitionNode() {bAutomaticRuleBasedOnSequencePlayerInState=true}
   AnimStateTransitionNode_10 = AnimStateTransitionNode() {priorityOrder=2}
   AnimStateTransitionNode_11 = AnimStateTransitionNode() {priorityOrder=2}
   AnimStateTransitionNode_2 = AnimStateTransitionNode() {bAutomaticRuleBasedOnSequencePlayerInState=true}
   AnimStateTransitionNode_25 = AnimStateTransitionNode()
   AnimStateTransitionNode_26 = AnimStateTransitionNode() {bAutomaticRuleBasedOnSequencePlayerInState=true}
   AnimStateTransitionNode_27 = AnimStateTransitionNode()
   AnimStateTransitionNode_28 = AnimStateTransitionNode()
   AnimStateTransitionNode_29 = AnimStateTransitionNode()
   AnimStateTransitionNode_3 = AnimStateTransitionNode() {bAutomaticRuleBasedOnSequencePlayerInState=true}
   AnimStateTransitionNode_30 = AnimStateTransitionNode()
   AnimStateTransitionNode_31 = AnimStateTransitionNode()
   AnimStateTransitionNode_32 = AnimStateTransitionNode()
   AnimStateTransitionNode_35 = AnimStateTransitionNode()
   AnimStateTransitionNode_36 = AnimStateTransitionNode()
   AnimStateTransitionNode_37 = AnimStateTransitionNode()
   AnimStateTransitionNode_38 = AnimStateTransitionNode() {bAutomaticRuleBasedOnSequencePlayerInState=true}
   AnimStateTransitionNode_39 = AnimStateTransitionNode()
   AnimStateTransitionNode_4 = AnimStateTransitionNode()
   AnimStateTransitionNode_40 = AnimStateTransitionNode() {bAutomaticRuleBasedOnSequencePlayerInState=true}
   AnimStateTransitionNode_41 = AnimStateTransitionNode()
   AnimStateTransitionNode_42 = AnimStateTransitionNode() {bAutomaticRuleBasedOnSequencePlayerInState=true}
   AnimStateTransitionNode_5 = AnimStateTransitionNode()
   AnimStateTransitionNode_6 = AnimStateTransitionNode()
   AnimStateTransitionNode_7 = AnimStateTransitionNode() {bAutomaticRuleBasedOnSequencePlayerInState=true}
   AnimStateTransitionNode_8 = AnimStateTransitionNode() {priorityOrder=2}
   AnimStateTransitionNode_9 = AnimStateTransitionNode() {priorityOrder=2}
  ]
  links [
   {
    dst=PassengerPose_W
    rule=VehicleLowerBodyOverride.AnimGraphNode_StateMachine_0.ShoppingCartPassenger.AnimStateNode_7.ShoppingCartPassenger_SM.AnimGraphNode_StateMachine_1.ShoppingCartPassenger.AnimStateTransitionNode_11.Transition
    src=PassengerPose_E
    via=AnimStateTransitionNode_11
   }
   {
    dst=Transition_EtoN
    rule=VehicleLowerBodyOverride.AnimGraphNode_StateMachine_0.ShoppingCartPassenger.AnimStateNode_7.ShoppingCartPassenger_SM.AnimGraphNode_StateMachine_1.ShoppingCartPassenger.AnimStateTransitionNode_35.AnimationTransitionGraph_1
    src=PassengerPose_E
    via=AnimStateTransitionNode_35
   }
   {
    dst=Transition_EtoS
    rule=VehicleLowerBodyOverride.AnimGraphNode_StateMachine_0.ShoppingCartPassenger.AnimStateNode_7.ShoppingCartPassenger_SM.AnimGraphNode_StateMachine_1.ShoppingCartPassenger.AnimStateTransitionNode_27.AnimationTransitionGraph_0
    src=PassengerPose_E
    via=AnimStateTransitionNode_27
   }
   {
    dst=PassengerPose_S
    rule=VehicleLowerBodyOverride.AnimGraphNode_StateMachine_0.ShoppingCartPassenger.AnimStateNode_7.ShoppingCartPassenger_SM.AnimGraphNode_StateMachine_1.ShoppingCartPassenger.AnimStateTransitionNode_8.AnimationTransitionGraph_0
    src=PassengerPose_N
    via=AnimStateTransitionNode_8
   }
   {
    dst=Transition_NtoE
    rule=VehicleLowerBodyOverride.AnimGraphNode_StateMachine_0.ShoppingCartPassenger.AnimStateNode_7.ShoppingCartPassenger_SM.AnimGraphNode_StateMachine_1.ShoppingCartPassenger.AnimStateTransitionNode_25.AnimationTransitionGraph_1
    src=PassengerPose_N
    via=AnimStateTransitionNode_25
   }
   {
    dst=Transition_NtoW
    rule=VehicleLowerBodyOverride.AnimGraphNode_StateMachine_0.ShoppingCartPassenger.AnimStateNode_7.ShoppingCartPassenger_SM.AnimGraphNode_StateMachine_1.ShoppingCartPassenger.AnimStateTransitionNode_41.AnimationTransitionGraph_1
    src=PassengerPose_N
    via=AnimStateTransitionNode_41
   }
   {
    dst=PassengerPose_N
    rule=VehicleLowerBodyOverride.AnimGraphNode_StateMachine_0.ShoppingCartPassenger.AnimStateNode_7.ShoppingCartPassenger_SM.AnimGraphNode_StateMachine_1.ShoppingCartPassenger.AnimStateTransitionNode_9.AnimationTransitionGraph_0
    src=PassengerPose_S
    via=AnimStateTransitionNode_9
   }
   {
    dst=Transition_StoE
    rule=VehicleLowerBodyOverride.AnimGraphNode_StateMachine_0.ShoppingCartPassenger.AnimStateNode_7.ShoppingCartPassenger_SM.AnimGraphNode_StateMachine_1.ShoppingCartPassenger.AnimStateTransitionNode_37.AnimationTransitionGraph_1
    src=PassengerPose_S
    via=AnimStateTransitionNode_37
   }
   {
    dst=Transition_StoW
    rule=VehicleLowerBodyOverride.AnimGraphNode_StateMachine_0.ShoppingCartPassenger.AnimStateNode_7.ShoppingCartPassenger_SM.AnimGraphNode_StateMachine_1.ShoppingCartPassenger.AnimStateTransitionNode_29.AnimationTransitionGraph_0
    src=PassengerPose_S
    via=AnimStateTransitionNode_29
   }
   {
    dst=PassengerPose_E
    rule=VehicleLowerBodyOverride.AnimGraphNode_StateMachine_0.ShoppingCartPassenger.AnimStateNode_7.ShoppingCartPassenger_SM.AnimGraphNode_StateMachine_1.ShoppingCartPassenger.AnimStateTransitionNode_10.Transition
    src=PassengerPose_W
    via=AnimStateTransitionNode_10
   }
   {
    dst=Transition_WtoN
    rule=VehicleLowerBodyOverride.AnimGraphNode_StateMachine_0.ShoppingCartPassenger.AnimStateNode_7.ShoppingCartPassenger_SM.AnimGraphNode_StateMachine_1.ShoppingCartPassenger.AnimStateTransitionNode_31.AnimationTransitionGraph_1
    src=PassengerPose_W
    via=AnimStateTransitionNode_31
   }
   {
    dst=Transition_WtoS
    rule=VehicleLowerBodyOverride.AnimGraphNode_StateMachine_0.ShoppingCartPassenger.AnimStateNode_7.ShoppingCartPassenger_SM.AnimGraphNode_StateMachine_1.ShoppingCartPassenger.AnimStateTransitionNode_39.AnimationTransitionGraph_0
    src=PassengerPose_W
    via=AnimStateTransitionNode_39
   }
   {
    dst=PassengerPose_N
    rule=VehicleLowerBodyOverride.AnimGraphNode_StateMachine_0.ShoppingCartPassenger.AnimStateNode_7.ShoppingCartPassenger_SM.AnimGraphNode_StateMachine_1.ShoppingCartPassenger.AnimStateTransitionNode_36.AnimationTransitionGraph_1
    src=Transition_EtoN
    via=AnimStateTransitionNode_36
   }
   {
    dst=PassengerPose_N
    rule=VehicleLowerBodyOverride.AnimGraphNode_StateMachine_0.ShoppingCartPassenger.AnimStateNode_7.ShoppingCartPassenger_SM.AnimGraphNode_StateMachine_1.ShoppingCartPassenger.AnimStateTransitionNode_7.Transition
    src=Transition_EtoN
    via=AnimStateTransitionNode_7
   }
   {
    dst=PassengerPose_S
    rule=VehicleLowerBodyOverride.AnimGraphNode_StateMachine_0.ShoppingCartPassenger.AnimStateNode_7.ShoppingCartPassenger_SM.AnimGraphNode_StateMachine_1.ShoppingCartPassenger.AnimStateTransitionNode_2.AnimationTransitionGraph_0
    src=Transition_EtoS
    via=AnimStateTransitionNode_2
   }
   {
    dst=PassengerPose_S
    rule=VehicleLowerBodyOverride.AnimGraphNode_StateMachine_0.ShoppingCartPassenger.AnimStateNode_7.ShoppingCartPassenger_SM.AnimGraphNode_StateMachine_1.ShoppingCartPassenger.AnimStateTransitionNode_28.AnimationTransitionGraph_0
    src=Transition_EtoS
    via=AnimStateTransitionNode_28
   }
   {
    dst=PassengerPose_E
    rule=VehicleLowerBodyOverride.AnimGraphNode_StateMachine_0.ShoppingCartPassenger.AnimStateNode_7.ShoppingCartPassenger_SM.AnimGraphNode_StateMachine_1.ShoppingCartPassenger.AnimStateTransitionNode_0.Transition
    src=Transition_NtoE
    via=AnimStateTransitionNode_0
   }
   {
    dst=PassengerPose_E
    rule=VehicleLowerBodyOverride.AnimGraphNode_StateMachine_0.ShoppingCartPassenger.AnimStateNode_7.ShoppingCartPassenger_SM.AnimGraphNode_StateMachine_1.ShoppingCartPassenger.AnimStateTransitionNode_26.AnimationTransitionGraph_1
    src=Transition_NtoE
    via=AnimStateTransitionNode_26
   }
   {
    dst=PassengerPose_W
    rule=VehicleLowerBodyOverride.AnimGraphNode_StateMachine_0.ShoppingCartPassenger.AnimStateNode_7.ShoppingCartPassenger_SM.AnimGraphNode_StateMachine_1.ShoppingCartPassenger.AnimStateTransitionNode_4.Transition
    src=Transition_NtoW
    via=AnimStateTransitionNode_4
   }
   {
    dst=PassengerPose_W
    rule=VehicleLowerBodyOverride.AnimGraphNode_StateMachine_0.ShoppingCartPassenger.AnimStateNode_7.ShoppingCartPassenger_SM.AnimGraphNode_StateMachine_1.ShoppingCartPassenger.AnimStateTransitionNode_42.AnimationTransitionGraph_1
    src=Transition_NtoW
    via=AnimStateTransitionNode_42
   }
   {
    dst=PassengerPose_E
    rule=VehicleLowerBodyOverride.AnimGraphNode_StateMachine_0.ShoppingCartPassenger.AnimStateNode_7.ShoppingCartPassenger_SM.AnimGraphNode_StateMachine_1.ShoppingCartPassenger.AnimStateTransitionNode_38.AnimationTransitionGraph_0
    src=Transition_StoE
    via=AnimStateTransitionNode_38
   }
   {
    dst=PassengerPose_E
    rule=VehicleLowerBodyOverride.AnimGraphNode_StateMachine_0.ShoppingCartPassenger.AnimStateNode_7.ShoppingCartPassenger_SM.AnimGraphNode_StateMachine_1.ShoppingCartPassenger.AnimStateTransitionNode_6.AnimationTransitionGraph_0
    src=Transition_StoE
    via=AnimStateTransitionNode_6
   }
   {
    dst=PassengerPose_W
    rule=VehicleLowerBodyOverride.AnimGraphNode_StateMachine_0.ShoppingCartPassenger.AnimStateNode_7.ShoppingCartPassenger_SM.AnimGraphNode_StateMachine_1.ShoppingCartPassenger.AnimStateTransitionNode_1.AnimationTransitionGraph_0
    src=Transition_StoW
    via=AnimStateTransitionNode_1
   }
   {
    dst=PassengerPose_W
    rule=VehicleLowerBodyOverride.AnimGraphNode_StateMachine_0.ShoppingCartPassenger.AnimStateNode_7.ShoppingCartPassenger_SM.AnimGraphNode_StateMachine_1.ShoppingCartPassenger.AnimStateTransitionNode_30.AnimationTransitionGraph_0
    src=Transition_StoW
    via=AnimStateTransitionNode_30
   }
   {
    dst=PassengerPose_N
    rule=VehicleLowerBodyOverride.AnimGraphNode_StateMachine_0.ShoppingCartPassenger.AnimStateNode_7.ShoppingCartPassenger_SM.AnimGraphNode_StateMachine_1.ShoppingCartPassenger.AnimStateTransitionNode_3.Transition
    src=Transition_WtoN
    via=AnimStateTransitionNode_3
   }
   {
    dst=PassengerPose_N
    rule=VehicleLowerBodyOverride.AnimGraphNode_StateMachine_0.ShoppingCartPassenger.AnimStateNode_7.ShoppingCartPassenger_SM.AnimGraphNode_StateMachine_1.ShoppingCartPassenger.AnimStateTransitionNode_32.AnimationTransitionGraph_1
    src=Transition_WtoN
    via=AnimStateTransitionNode_32
   }
   {
    dst=PassengerPose_S
    rule=VehicleLowerBodyOverride.AnimGraphNode_StateMachine_0.ShoppingCartPassenger.AnimStateNode_7.ShoppingCartPassenger_SM.AnimGraphNode_StateMachine_1.ShoppingCartPassenger.AnimStateTransitionNode_40.AnimationTransitionGraph_1
    src=Transition_WtoS
    via=AnimStateTransitionNode_40
   }
   {
    dst=PassengerPose_S
    rule=VehicleLowerBodyOverride.AnimGraphNode_StateMachine_0.ShoppingCartPassenger.AnimStateNode_7.ShoppingCartPassenger_SM.AnimGraphNode_StateMachine_1.ShoppingCartPassenger.AnimStateTransitionNode_5.AnimationTransitionGraph_1
    src=Transition_WtoS
    via=AnimStateTransitionNode_5
   }
  ]
 }
 Transition_EtoS {
  path=VehicleLowerBodyOverride.AnimGraphNode_StateMachine_0.ShoppingCartPassenger.AnimStateNode_7.ShoppingCartPassenger_SM.AnimGraphNode_StateMachine_1.ShoppingCartPassenger.AnimStateNode_10.Transition_EtoS
  nodes [Output Animation Pose(Result=Sequence Player 'ShoppingCart_Passenger_Turn_EtoS'() {bLoopAnimation=false, sequence=/Game/Animation/Game/MainPlayer/Vehicles/ShoppingCart/ShoppingCart_Passenger_Turn_EtoS})]
 }
 PassengerPose_S {
  path=VehicleLowerBodyOverride.AnimGraphNode_StateMachine_0.ShoppingCartPassenger.AnimStateNode_7.ShoppingCartPassenger_SM.AnimGraphNode_StateMachine_1.ShoppingCartPassenger.AnimStateNode_11.PassengerPose_S
  nodes [
   n1 = AimOffset Player 'ShoppingCart_Passenger_LowerBodyAO'!(
    X:float=Break Fort Anim Input Shopping Cart(FortAnimInput_ShoppingCart=Get ShoppingCartInput()) -> (
     bIsUsingShoppingCart=~
     bIsUsingVehicle=~
     bIsCoastStatePedaling=~
     bIsCoastStateCoasting=~
     bIsCoastStateDismount=~
     bIsCoastStateIdle=~
     bIsInAir=~
     bIsCoasting=~
     bIsPedaling=~
     bIsReadyToPedal=~
     IsReadyToPedal=~
     bWantsToCoast=~
     bIsCoastIdling=~
     bIsStartCoasting=~
     bIsEndCoasting=~
     bIsDismountingFromCoast=~
     bIsCoastingOrDismountingFromCoast=~
     bIsStandingInPlace=~
     bIsSprinting=~
     bIsSprintingAndMovingForward=~
     bIsMovingForwardNotSprinting=~
     bIsBraking=~
     bIsReversing=~
     bIsMoving=~
     bIsMovingForward=~
     bIsMovingBackwards=~
     bIsMovingOrTurningInPlace=~
     bIsInAirSteady=~
     bIsOnSlope=~
     bAimFWD=~
     bAimBWD=~
     bAimLFT=~
     bAimRGT=~
     ForwardVelocity=~
     ForwardSpeedKmH=~
     CurrentBrakeForce=~
     RunForwardAlpha=~
     bIsAcceleratingForward=~
     bIsAccelBreakingOrReversing=~
     SteerAngle=~
     SteerAngleInterpSpeed=~
     CoastSteerAngleInterpSpeed=~
     IsReadyToPedalInterpSpeed=~
     StandingInPlaceSteerAngle=~
     SlopePitchDegreeAngle=~
     SlopeRollDegreeAngle=~
     PawnToVehicleDeltaYawAngleDegrees=~
     AimCardDirDeadZoneAngleDegrees=~
     AimCardDirAngleOffsetDegrees=~
     AimFWDDeltaAngleDegrees=~
     AimBWDDeltaAngleDegrees=~
     AimLFTDeltaAngleDegrees=~
     AimRGTDeltaAngleDegrees=~
    ).AimBWDDeltaAngleDegrees
    BasePose:Pose Link Structure=Rotate Root Bone!(
     BasePose=Sequence Player 'ShoppingCart_Passenger_Idle_S'() {sequence=/Game/Animation/Game/MainPlayer/Vehicles/ShoppingCart/ShoppingCart_Passenger_Idle_S}
     Yaw=180
     MeshToComponent="0,0,0"
    ) -> (Pose)
    Alpha:float=1
   ) -> (Pose:Pose Link Structure) {blendSpace=/Game/Animation/Game/MainPlayer/Vehicles/ShoppingCart/AO/ShoppingCart_Passenger_LowerBodyAO, alphaBoolBlend {blendOption=Linear}}
   Output Animation Pose(Result=Rotate Root Bone(BasePose=@n1, Yaw=-180, MeshToComponent="0,0,0"))
  ]
 }
 Transition_StoW {
  path=VehicleLowerBodyOverride.AnimGraphNode_StateMachine_0.ShoppingCartPassenger.AnimStateNode_7.ShoppingCartPassenger_SM.AnimGraphNode_StateMachine_1.ShoppingCartPassenger.AnimStateNode_12.Transition_StoW
  nodes [Output Animation Pose(Result=Sequence Player 'ShoppingCart_Passenger_Turn_StoW'() {bLoopAnimation=false, sequence=/Game/Animation/Game/MainPlayer/Vehicles/ShoppingCart/ShoppingCart_Passenger_Turn_StoW})]
 }
 PassengerPose_W {
  path=VehicleLowerBodyOverride.AnimGraphNode_StateMachine_0.ShoppingCartPassenger.AnimStateNode_7.ShoppingCartPassenger_SM.AnimGraphNode_StateMachine_1.ShoppingCartPassenger.AnimStateNode_13.PassengerPose_W
  nodes [
   n1 = AimOffset Player 'ShoppingCart_Passenger_LowerBodyAO'(
    X=Break Fort Anim Input Shopping Cart(FortAnimInput_ShoppingCart=Get ShoppingCartInput()) -> (
     bIsUsingShoppingCart=~
     bIsUsingVehicle=~
     bIsCoastStatePedaling=~
     bIsCoastStateCoasting=~
     bIsCoastStateDismount=~
     bIsCoastStateIdle=~
     bIsInAir=~
     bIsCoasting=~
     bIsPedaling=~
     bIsReadyToPedal=~
     IsReadyToPedal=~
     bWantsToCoast=~
     bIsCoastIdling=~
     bIsStartCoasting=~
     bIsEndCoasting=~
     bIsDismountingFromCoast=~
     bIsCoastingOrDismountingFromCoast=~
     bIsStandingInPlace=~
     bIsSprinting=~
     bIsSprintingAndMovingForward=~
     bIsMovingForwardNotSprinting=~
     bIsBraking=~
     bIsReversing=~
     bIsMoving=~
     bIsMovingForward=~
     bIsMovingBackwards=~
     bIsMovingOrTurningInPlace=~
     bIsInAirSteady=~
     bIsOnSlope=~
     bAimFWD=~
     bAimBWD=~
     bAimLFT=~
     bAimRGT=~
     ForwardVelocity=~
     ForwardSpeedKmH=~
     CurrentBrakeForce=~
     RunForwardAlpha=~
     bIsAcceleratingForward=~
     bIsAccelBreakingOrReversing=~
     SteerAngle=~
     SteerAngleInterpSpeed=~
     CoastSteerAngleInterpSpeed=~
     IsReadyToPedalInterpSpeed=~
     StandingInPlaceSteerAngle=~
     SlopePitchDegreeAngle=~
     SlopeRollDegreeAngle=~
     PawnToVehicleDeltaYawAngleDegrees=~
     AimCardDirDeadZoneAngleDegrees=~
     AimCardDirAngleOffsetDegrees=~
     AimFWDDeltaAngleDegrees=~
     AimBWDDeltaAngleDegrees=~
     AimLFTDeltaAngleDegrees=~
     AimRGTDeltaAngleDegrees=~
    ).AimLFTDeltaAngleDegrees
    BasePose=Rotate Root Bone(
     BasePose=Sequence Player 'ShoppingCart_Passenger_Idle_W'() {sequence=/Game/Animation/Game/MainPlayer/Vehicles/ShoppingCart/ShoppingCart_Passenger_Idle_W}
     Yaw=90
     MeshToComponent="0,0,0"
    )
    Alpha=1
   ) {blendSpace=&ShoppingCart_Passenger_LowerBodyAO, alphaBoolBlend {blendOption=Linear}}
   Output Animation Pose(Result=Rotate Root Bone(BasePose=@n1, Yaw=-90, MeshToComponent="0,0,0"))
  ]
 }
 Transition_WtoN {
  path=VehicleLowerBodyOverride.AnimGraphNode_StateMachine_0.ShoppingCartPassenger.AnimStateNode_7.ShoppingCartPassenger_SM.AnimGraphNode_StateMachine_1.ShoppingCartPassenger.AnimStateNode_15.Transition_WtoN
  nodes [Output Animation Pose(Result=Sequence Player 'ShoppingCart_Passenger_Turn_WtoN'() {bLoopAnimation=false, sequence=/Game/Animation/Game/MainPlayer/Vehicles/ShoppingCart/ShoppingCart_Passenger_Turn_WtoN})]
 }
 PassengerPose_N {
  path=VehicleLowerBodyOverride.AnimGraphNode_StateMachine_0.ShoppingCartPassenger.AnimStateNode_7.ShoppingCartPassenger_SM.AnimGraphNode_StateMachine_1.ShoppingCartPassenger.AnimStateNode_16.PassengerPose_N
  nodes [
   n1 = AimOffset Player 'ShoppingCart_Passenger_LowerBodyAO'(
    X=Break Fort Anim Input Shopping Cart(FortAnimInput_ShoppingCart=Get ShoppingCartInput()) -> (
     bIsUsingShoppingCart=~
     bIsUsingVehicle=~
     bIsCoastStatePedaling=~
     bIsCoastStateCoasting=~
     bIsCoastStateDismount=~
     bIsCoastStateIdle=~
     bIsInAir=~
     bIsCoasting=~
     bIsPedaling=~
     bIsReadyToPedal=~
     IsReadyToPedal=~
     bWantsToCoast=~
     bIsCoastIdling=~
     bIsStartCoasting=~
     bIsEndCoasting=~
     bIsDismountingFromCoast=~
     bIsCoastingOrDismountingFromCoast=~
     bIsStandingInPlace=~
     bIsSprinting=~
     bIsSprintingAndMovingForward=~
     bIsMovingForwardNotSprinting=~
     bIsBraking=~
     bIsReversing=~
     bIsMoving=~
     bIsMovingForward=~
     bIsMovingBackwards=~
     bIsMovingOrTurningInPlace=~
     bIsInAirSteady=~
     bIsOnSlope=~
     bAimFWD=~
     bAimBWD=~
     bAimLFT=~
     bAimRGT=~
     ForwardVelocity=~
     ForwardSpeedKmH=~
     CurrentBrakeForce=~
     RunForwardAlpha=~
     bIsAcceleratingForward=~
     bIsAccelBreakingOrReversing=~
     SteerAngle=~
     SteerAngleInterpSpeed=~
     CoastSteerAngleInterpSpeed=~
     IsReadyToPedalInterpSpeed=~
     StandingInPlaceSteerAngle=~
     SlopePitchDegreeAngle=~
     SlopeRollDegreeAngle=~
     PawnToVehicleDeltaYawAngleDegrees=~
     AimCardDirDeadZoneAngleDegrees=~
     AimCardDirAngleOffsetDegrees=~
     AimFWDDeltaAngleDegrees=~
     AimBWDDeltaAngleDegrees=~
     AimLFTDeltaAngleDegrees=~
     AimRGTDeltaAngleDegrees=~
    ).AimFWDDeltaAngleDegrees
    BasePose=Sequence Player 'ShoppingCart_Passenger_Idle_N'() {sequence=/Game/Animation/Game/MainPlayer/Vehicles/ShoppingCart/ShoppingCart_Passenger_Idle_N}
    Alpha=1
   ) {blendSpace=&ShoppingCart_Passenger_LowerBodyAO, alphaBoolBlend {blendOption=Linear}}
   Output Animation Pose(Result=@n1)
  ]
 }
 Transition_EtoN {
  path=VehicleLowerBodyOverride.AnimGraphNode_StateMachine_0.ShoppingCartPassenger.AnimStateNode_7.ShoppingCartPassenger_SM.AnimGraphNode_StateMachine_1.ShoppingCartPassenger.AnimStateNode_17.Transition_EtoN
  nodes [Output Animation Pose(Result=Sequence Player 'ShoppingCart_Passenger_Turn_EtoN'() {bLoopAnimation=false, sequence=/Game/Animation/Game/MainPlayer/Vehicles/ShoppingCart/ShoppingCart_Passenger_Turn_EtoN})]
 }
 Transition_StoE {
  path=VehicleLowerBodyOverride.AnimGraphNode_StateMachine_0.ShoppingCartPassenger.AnimStateNode_7.ShoppingCartPassenger_SM.AnimGraphNode_StateMachine_1.ShoppingCartPassenger.AnimStateNode_18.Transition_StoE
  nodes [Output Animation Pose(Result=Sequence Player 'ShoppingCart_Passenger_Turn_StoE'() {bLoopAnimation=false, sequence=/Game/Animation/Game/MainPlayer/Vehicles/ShoppingCart/ShoppingCart_Passenger_Turn_StoE})]
 }
 Transition_WtoS {
  path=VehicleLowerBodyOverride.AnimGraphNode_StateMachine_0.ShoppingCartPassenger.AnimStateNode_7.ShoppingCartPassenger_SM.AnimGraphNode_StateMachine_1.ShoppingCartPassenger.AnimStateNode_19.Transition_WtoS
  nodes [Output Animation Pose(Result=Sequence Player 'ShoppingCart_Passenger_Turn_WtoS'() {bLoopAnimation=false, sequence=/Game/Animation/Game/MainPlayer/Vehicles/ShoppingCart/ShoppingCart_Passenger_Turn_WtoS})]
 }
 Transition_NtoW {
  path=VehicleLowerBodyOverride.AnimGraphNode_StateMachine_0.ShoppingCartPassenger.AnimStateNode_7.ShoppingCartPassenger_SM.AnimGraphNode_StateMachine_1.ShoppingCartPassenger.AnimStateNode_20.Transition_NtoW
  nodes [Output Animation Pose(Result=Sequence Player 'ShoppingCart_Passenger_Turn_NtoW'() {bLoopAnimation=false, sequence=/Game/Animation/Game/MainPlayer/Vehicles/ShoppingCart/ShoppingCart_Passenger_Turn_NtoW})]
 }
 Transition_NtoE {
  path=VehicleLowerBodyOverride.AnimGraphNode_StateMachine_0.ShoppingCartPassenger.AnimStateNode_7.ShoppingCartPassenger_SM.AnimGraphNode_StateMachine_1.ShoppingCartPassenger.AnimStateNode_8.Transition_NtoE
  nodes [Output Animation Pose(Result=Sequence Player 'ShoppingCart_Passenger_Turn_NtoE'() {bLoopAnimation=false, sequence=/Game/Animation/Game/MainPlayer/Vehicles/ShoppingCart/ShoppingCart_Passenger_Turn_NtoE})]
 }
 PassengerPose_E {
  path=VehicleLowerBodyOverride.AnimGraphNode_StateMachine_0.ShoppingCartPassenger.AnimStateNode_7.ShoppingCartPassenger_SM.AnimGraphNode_StateMachine_1.ShoppingCartPassenger.AnimStateNode_9.PassengerPose_E
  nodes [
   n1 = AimOffset Player 'ShoppingCart_Passenger_LowerBodyAO'(
    X=Break Fort Anim Input Shopping Cart(FortAnimInput_ShoppingCart=Get ShoppingCartInput()) -> (
     bIsUsingShoppingCart=~
     bIsUsingVehicle=~
     bIsCoastStatePedaling=~
     bIsCoastStateCoasting=~
     bIsCoastStateDismount=~
     bIsCoastStateIdle=~
     bIsInAir=~
     bIsCoasting=~
     bIsPedaling=~
     bIsReadyToPedal=~
     IsReadyToPedal=~
     bWantsToCoast=~
     bIsCoastIdling=~
     bIsStartCoasting=~
     bIsEndCoasting=~
     bIsDismountingFromCoast=~
     bIsCoastingOrDismountingFromCoast=~
     bIsStandingInPlace=~
     bIsSprinting=~
     bIsSprintingAndMovingForward=~
     bIsMovingForwardNotSprinting=~
     bIsBraking=~
     bIsReversing=~
     bIsMoving=~
     bIsMovingForward=~
     bIsMovingBackwards=~
     bIsMovingOrTurningInPlace=~
     bIsInAirSteady=~
     bIsOnSlope=~
     bAimFWD=~
     bAimBWD=~
     bAimLFT=~
     bAimRGT=~
     ForwardVelocity=~
     ForwardSpeedKmH=~
     CurrentBrakeForce=~
     RunForwardAlpha=~
     bIsAcceleratingForward=~
     bIsAccelBreakingOrReversing=~
     SteerAngle=~
     SteerAngleInterpSpeed=~
     CoastSteerAngleInterpSpeed=~
     IsReadyToPedalInterpSpeed=~
     StandingInPlaceSteerAngle=~
     SlopePitchDegreeAngle=~
     SlopeRollDegreeAngle=~
     PawnToVehicleDeltaYawAngleDegrees=~
     AimCardDirDeadZoneAngleDegrees=~
     AimCardDirAngleOffsetDegrees=~
     AimFWDDeltaAngleDegrees=~
     AimBWDDeltaAngleDegrees=~
     AimLFTDeltaAngleDegrees=~
     AimRGTDeltaAngleDegrees=~
    ).AimRGTDeltaAngleDegrees
    BasePose=Rotate Root Bone(
     BasePose=Sequence Player 'ShoppingCart_Passenger_Idle_E'() {sequence=/Game/Animation/Game/MainPlayer/Vehicles/ShoppingCart/ShoppingCart_Passenger_Idle_E}
     Yaw=-90
     MeshToComponent="0,0,0"
    )
    Alpha=1
   ) {blendSpace=&ShoppingCart_Passenger_LowerBodyAO, alphaBoolBlend {blendOption=Linear}}
   Output Animation Pose(Result=Rotate Root Bone(BasePose=@n1, Yaw=90, MeshToComponent="0,0,0"))
  ]
 }
 Transition {
  path=VehicleLowerBodyOverride.AnimGraphNode_StateMachine_0.ShoppingCartPassenger.AnimStateNode_7.ShoppingCartPassenger_SM.AnimGraphNode_StateMachine_1.ShoppingCartPassenger.AnimStateTransitionNode_0.Transition
  nodes [
   n1 = Get ShoppingCartInput()
   Result(bCanEnterTransition=NOT Boolean(A=Break Fort Anim Input Shopping Cart(FortAnimInput_ShoppingCart=@n1) -> (
      bIsUsingShoppingCart=~
      bIsUsingVehicle=~
      bIsCoastStatePedaling=~
      bIsCoastStateCoasting=~
      bIsCoastStateDismount=~
      bIsCoastStateIdle=~
      bIsInAir=~
      bIsCoasting=~
      bIsPedaling=~
      bIsReadyToPedal=~
      IsReadyToPedal=~
      bWantsToCoast=~
      bIsCoastIdling=~
      bIsStartCoasting=~
      bIsEndCoasting=~
      bIsDismountingFromCoast=~
      bIsCoastingOrDismountingFromCoast=~
      bIsStandingInPlace=~
      bIsSprinting=~
      bIsSprintingAndMovingForward=~
      bIsMovingForwardNotSprinting=~
      bIsBraking=~
      bIsReversing=~
      bIsMoving=~
      bIsMovingForward=~
      bIsMovingBackwards=~
      bIsMovingOrTurningInPlace=~
      bIsInAirSteady=~
      bIsOnSlope=~
      bAimFWD=~
      bAimBWD=~
      bAimLFT=~
      bAimRGT=~
      ForwardVelocity=~
      ForwardSpeedKmH=~
      CurrentBrakeForce=~
      RunForwardAlpha=~
      bIsAcceleratingForward=~
      bIsAccelBreakingOrReversing=~
      SteerAngle=~
      SteerAngleInterpSpeed=~
      CoastSteerAngleInterpSpeed=~
      IsReadyToPedalInterpSpeed=~
      StandingInPlaceSteerAngle=~
      SlopePitchDegreeAngle=~
      SlopeRollDegreeAngle=~
      PawnToVehicleDeltaYawAngleDegrees=~
      AimCardDirDeadZoneAngleDegrees=~
      AimCardDirAngleOffsetDegrees=~
      AimFWDDeltaAngleDegrees=~
      AimBWDDeltaAngleDegrees=~
      AimLFTDeltaAngleDegrees=~
      AimRGTDeltaAngleDegrees=~
     ).bAimRGT))
  ]
 }
 AnimationTransitionGraph_0 {
  path=VehicleLowerBodyOverride.AnimGraphNode_StateMachine_0.ShoppingCartPassenger.AnimStateNode_7.ShoppingCartPassenger_SM.AnimGraphNode_StateMachine_1.ShoppingCartPassenger.AnimStateTransitionNode_1.AnimationTransitionGraph_0
  nodes [Result()]
 }
 Transition {
  path=VehicleLowerBodyOverride.AnimGraphNode_StateMachine_0.ShoppingCartPassenger.AnimStateNode_7.ShoppingCartPassenger_SM.AnimGraphNode_StateMachine_1.ShoppingCartPassenger.AnimStateTransitionNode_10.Transition
  nodes [Result(bCanEnterTransition=Break Fort Anim Input Shopping Cart(FortAnimInput_ShoppingCart=Get ShoppingCartInput()) -> (
     bIsUsingShoppingCart=~
     bIsUsingVehicle=~
     bIsCoastStatePedaling=~
     bIsCoastStateCoasting=~
     bIsCoastStateDismount=~
     bIsCoastStateIdle=~
     bIsInAir=~
     bIsCoasting=~
     bIsPedaling=~
     bIsReadyToPedal=~
     IsReadyToPedal=~
     bWantsToCoast=~
     bIsCoastIdling=~
     bIsStartCoasting=~
     bIsEndCoasting=~
     bIsDismountingFromCoast=~
     bIsCoastingOrDismountingFromCoast=~
     bIsStandingInPlace=~
     bIsSprinting=~
     bIsSprintingAndMovingForward=~
     bIsMovingForwardNotSprinting=~
     bIsBraking=~
     bIsReversing=~
     bIsMoving=~
     bIsMovingForward=~
     bIsMovingBackwards=~
     bIsMovingOrTurningInPlace=~
     bIsInAirSteady=~
     bIsOnSlope=~
     bAimFWD=~
     bAimBWD=~
     bAimLFT=~
     bAimRGT=~
     ForwardVelocity=~
     ForwardSpeedKmH=~
     CurrentBrakeForce=~
     RunForwardAlpha=~
     bIsAcceleratingForward=~
     bIsAccelBreakingOrReversing=~
     SteerAngle=~
     SteerAngleInterpSpeed=~
     CoastSteerAngleInterpSpeed=~
     IsReadyToPedalInterpSpeed=~
     StandingInPlaceSteerAngle=~
     SlopePitchDegreeAngle=~
     SlopeRollDegreeAngle=~
     PawnToVehicleDeltaYawAngleDegrees=~
     AimCardDirDeadZoneAngleDegrees=~
     AimCardDirAngleOffsetDegrees=~
     AimFWDDeltaAngleDegrees=~
     AimBWDDeltaAngleDegrees=~
     AimLFTDeltaAngleDegrees=~
     AimRGTDeltaAngleDegrees=~
    ).bAimRGT)]
 }
 Transition {
  path=VehicleLowerBodyOverride.AnimGraphNode_StateMachine_0.ShoppingCartPassenger.AnimStateNode_7.ShoppingCartPassenger_SM.AnimGraphNode_StateMachine_1.ShoppingCartPassenger.AnimStateTransitionNode_11.Transition
  nodes [Result(bCanEnterTransition=Break Fort Anim Input Shopping Cart(FortAnimInput_ShoppingCart=Get ShoppingCartInput()) -> (
     bIsUsingShoppingCart=~
     bIsUsingVehicle=~
     bIsCoastStatePedaling=~
     bIsCoastStateCoasting=~
     bIsCoastStateDismount=~
     bIsCoastStateIdle=~
     bIsInAir=~
     bIsCoasting=~
     bIsPedaling=~
     bIsReadyToPedal=~
     IsReadyToPedal=~
     bWantsToCoast=~
     bIsCoastIdling=~
     bIsStartCoasting=~
     bIsEndCoasting=~
     bIsDismountingFromCoast=~
     bIsCoastingOrDismountingFromCoast=~
     bIsStandingInPlace=~
     bIsSprinting=~
     bIsSprintingAndMovingForward=~
     bIsMovingForwardNotSprinting=~
     bIsBraking=~
     bIsReversing=~
     bIsMoving=~
     bIsMovingForward=~
     bIsMovingBackwards=~
     bIsMovingOrTurningInPlace=~
     bIsInAirSteady=~
     bIsOnSlope=~
     bAimFWD=~
     bAimBWD=~
     bAimLFT=~
     bAimRGT=~
     ForwardVelocity=~
     ForwardSpeedKmH=~
     CurrentBrakeForce=~
     RunForwardAlpha=~
     bIsAcceleratingForward=~
     bIsAccelBreakingOrReversing=~
     SteerAngle=~
     SteerAngleInterpSpeed=~
     CoastSteerAngleInterpSpeed=~
     IsReadyToPedalInterpSpeed=~
     StandingInPlaceSteerAngle=~
     SlopePitchDegreeAngle=~
     SlopeRollDegreeAngle=~
     PawnToVehicleDeltaYawAngleDegrees=~
     AimCardDirDeadZoneAngleDegrees=~
     AimCardDirAngleOffsetDegrees=~
     AimFWDDeltaAngleDegrees=~
     AimBWDDeltaAngleDegrees=~
     AimLFTDeltaAngleDegrees=~
     AimRGTDeltaAngleDegrees=~
    ).bAimLFT)]
 }
 AnimationTransitionGraph_0 {
  path=VehicleLowerBodyOverride.AnimGraphNode_StateMachine_0.ShoppingCartPassenger.AnimStateNode_7.ShoppingCartPassenger_SM.AnimGraphNode_StateMachine_1.ShoppingCartPassenger.AnimStateTransitionNode_2.AnimationTransitionGraph_0
  nodes [Result()]
 }
 AnimationTransitionGraph_1 {
  path=VehicleLowerBodyOverride.AnimGraphNode_StateMachine_0.ShoppingCartPassenger.AnimStateNode_7.ShoppingCartPassenger_SM.AnimGraphNode_StateMachine_1.ShoppingCartPassenger.AnimStateTransitionNode_25.AnimationTransitionGraph_1
  nodes [Result(bCanEnterTransition=Break Fort Anim Input Shopping Cart(FortAnimInput_ShoppingCart=Get ShoppingCartInput()) -> (
     bIsUsingShoppingCart=~
     bIsUsingVehicle=~
     bIsCoastStatePedaling=~
     bIsCoastStateCoasting=~
     bIsCoastStateDismount=~
     bIsCoastStateIdle=~
     bIsInAir=~
     bIsCoasting=~
     bIsPedaling=~
     bIsReadyToPedal=~
     IsReadyToPedal=~
     bWantsToCoast=~
     bIsCoastIdling=~
     bIsStartCoasting=~
     bIsEndCoasting=~
     bIsDismountingFromCoast=~
     bIsCoastingOrDismountingFromCoast=~
     bIsStandingInPlace=~
     bIsSprinting=~
     bIsSprintingAndMovingForward=~
     bIsMovingForwardNotSprinting=~
     bIsBraking=~
     bIsReversing=~
     bIsMoving=~
     bIsMovingForward=~
     bIsMovingBackwards=~
     bIsMovingOrTurningInPlace=~
     bIsInAirSteady=~
     bIsOnSlope=~
     bAimFWD=~
     bAimBWD=~
     bAimLFT=~
     bAimRGT=~
     ForwardVelocity=~
     ForwardSpeedKmH=~
     CurrentBrakeForce=~
     RunForwardAlpha=~
     bIsAcceleratingForward=~
     bIsAccelBreakingOrReversing=~
     SteerAngle=~
     SteerAngleInterpSpeed=~
     CoastSteerAngleInterpSpeed=~
     IsReadyToPedalInterpSpeed=~
     StandingInPlaceSteerAngle=~
     SlopePitchDegreeAngle=~
     SlopeRollDegreeAngle=~
     PawnToVehicleDeltaYawAngleDegrees=~
     AimCardDirDeadZoneAngleDegrees=~
     AimCardDirAngleOffsetDegrees=~
     AimFWDDeltaAngleDegrees=~
     AimBWDDeltaAngleDegrees=~
     AimLFTDeltaAngleDegrees=~
     AimRGTDeltaAngleDegrees=~
    ).bAimRGT)]
 }
 AnimationTransitionGraph_1 {
  path=VehicleLowerBodyOverride.AnimGraphNode_StateMachine_0.ShoppingCartPassenger.AnimStateNode_7.ShoppingCartPassenger_SM.AnimGraphNode_StateMachine_1.ShoppingCartPassenger.AnimStateTransitionNode_26.AnimationTransitionGraph_1
  nodes [Result()]
 }
 AnimationTransitionGraph_0 {
  path=VehicleLowerBodyOverride.AnimGraphNode_StateMachine_0.ShoppingCartPassenger.AnimStateNode_7.ShoppingCartPassenger_SM.AnimGraphNode_StateMachine_1.ShoppingCartPassenger.AnimStateTransitionNode_27.AnimationTransitionGraph_0
  nodes [Result(bCanEnterTransition=Break Fort Anim Input Shopping Cart(FortAnimInput_ShoppingCart=Get ShoppingCartInput()) -> (
     bIsUsingShoppingCart=~
     bIsUsingVehicle=~
     bIsCoastStatePedaling=~
     bIsCoastStateCoasting=~
     bIsCoastStateDismount=~
     bIsCoastStateIdle=~
     bIsInAir=~
     bIsCoasting=~
     bIsPedaling=~
     bIsReadyToPedal=~
     IsReadyToPedal=~
     bWantsToCoast=~
     bIsCoastIdling=~
     bIsStartCoasting=~
     bIsEndCoasting=~
     bIsDismountingFromCoast=~
     bIsCoastingOrDismountingFromCoast=~
     bIsStandingInPlace=~
     bIsSprinting=~
     bIsSprintingAndMovingForward=~
     bIsMovingForwardNotSprinting=~
     bIsBraking=~
     bIsReversing=~
     bIsMoving=~
     bIsMovingForward=~
     bIsMovingBackwards=~
     bIsMovingOrTurningInPlace=~
     bIsInAirSteady=~
     bIsOnSlope=~
     bAimFWD=~
     bAimBWD=~
     bAimLFT=~
     bAimRGT=~
     ForwardVelocity=~
     ForwardSpeedKmH=~
     CurrentBrakeForce=~
     RunForwardAlpha=~
     bIsAcceleratingForward=~
     bIsAccelBreakingOrReversing=~
     SteerAngle=~
     SteerAngleInterpSpeed=~
     CoastSteerAngleInterpSpeed=~
     IsReadyToPedalInterpSpeed=~
     StandingInPlaceSteerAngle=~
     SlopePitchDegreeAngle=~
     SlopeRollDegreeAngle=~
     PawnToVehicleDeltaYawAngleDegrees=~
     AimCardDirDeadZoneAngleDegrees=~
     AimCardDirAngleOffsetDegrees=~
     AimFWDDeltaAngleDegrees=~
     AimBWDDeltaAngleDegrees=~
     AimLFTDeltaAngleDegrees=~
     AimRGTDeltaAngleDegrees=~
    ).bAimBWD)]
 }
 AnimationTransitionGraph_0 {
  path=VehicleLowerBodyOverride.AnimGraphNode_StateMachine_0.ShoppingCartPassenger.AnimStateNode_7.ShoppingCartPassenger_SM.AnimGraphNode_StateMachine_1.ShoppingCartPassenger.AnimStateTransitionNode_28.AnimationTransitionGraph_0
  nodes [
   n1 = Get ShoppingCartInput()
   Result(bCanEnterTransition=NOT Boolean(A=Break Fort Anim Input Shopping Cart(FortAnimInput_ShoppingCart=@n1) -> (
      bIsUsingShoppingCart=~
      bIsUsingVehicle=~
      bIsCoastStatePedaling=~
      bIsCoastStateCoasting=~
      bIsCoastStateDismount=~
      bIsCoastStateIdle=~
      bIsInAir=~
      bIsCoasting=~
      bIsPedaling=~
      bIsReadyToPedal=~
      IsReadyToPedal=~
      bWantsToCoast=~
      bIsCoastIdling=~
      bIsStartCoasting=~
      bIsEndCoasting=~
      bIsDismountingFromCoast=~
      bIsCoastingOrDismountingFromCoast=~
      bIsStandingInPlace=~
      bIsSprinting=~
      bIsSprintingAndMovingForward=~
      bIsMovingForwardNotSprinting=~
      bIsBraking=~
      bIsReversing=~
      bIsMoving=~
      bIsMovingForward=~
      bIsMovingBackwards=~
      bIsMovingOrTurningInPlace=~
      bIsInAirSteady=~
      bIsOnSlope=~
      bAimFWD=~
      bAimBWD=~
      bAimLFT=~
      bAimRGT=~
      ForwardVelocity=~
      ForwardSpeedKmH=~
      CurrentBrakeForce=~
      RunForwardAlpha=~
      bIsAcceleratingForward=~
      bIsAccelBreakingOrReversing=~
      SteerAngle=~
      SteerAngleInterpSpeed=~
      CoastSteerAngleInterpSpeed=~
      IsReadyToPedalInterpSpeed=~
      StandingInPlaceSteerAngle=~
      SlopePitchDegreeAngle=~
      SlopeRollDegreeAngle=~
      PawnToVehicleDeltaYawAngleDegrees=~
      AimCardDirDeadZoneAngleDegrees=~
      AimCardDirAngleOffsetDegrees=~
      AimFWDDeltaAngleDegrees=~
      AimBWDDeltaAngleDegrees=~
      AimLFTDeltaAngleDegrees=~
      AimRGTDeltaAngleDegrees=~
     ).bAimBWD))
  ]
 }
 AnimationTransitionGraph_0 {
  path=VehicleLowerBodyOverride.AnimGraphNode_StateMachine_0.ShoppingCartPassenger.AnimStateNode_7.ShoppingCartPassenger_SM.AnimGraphNode_StateMachine_1.ShoppingCartPassenger.AnimStateTransitionNode_29.AnimationTransitionGraph_0
  nodes [Result(bCanEnterTransition=Break Fort Anim Input Shopping Cart(FortAnimInput_ShoppingCart=Get ShoppingCartInput()) -> (
     bIsUsingShoppingCart=~
     bIsUsingVehicle=~
     bIsCoastStatePedaling=~
     bIsCoastStateCoasting=~
     bIsCoastStateDismount=~
     bIsCoastStateIdle=~
     bIsInAir=~
     bIsCoasting=~
     bIsPedaling=~
     bIsReadyToPedal=~
     IsReadyToPedal=~
     bWantsToCoast=~
     bIsCoastIdling=~
     bIsStartCoasting=~
     bIsEndCoasting=~
     bIsDismountingFromCoast=~
     bIsCoastingOrDismountingFromCoast=~
     bIsStandingInPlace=~
     bIsSprinting=~
     bIsSprintingAndMovingForward=~
     bIsMovingForwardNotSprinting=~
     bIsBraking=~
     bIsReversing=~
     bIsMoving=~
     bIsMovingForward=~
     bIsMovingBackwards=~
     bIsMovingOrTurningInPlace=~
     bIsInAirSteady=~
     bIsOnSlope=~
     bAimFWD=~
     bAimBWD=~
     bAimLFT=~
     bAimRGT=~
     ForwardVelocity=~
     ForwardSpeedKmH=~
     CurrentBrakeForce=~
     RunForwardAlpha=~
     bIsAcceleratingForward=~
     bIsAccelBreakingOrReversing=~
     SteerAngle=~
     SteerAngleInterpSpeed=~
     CoastSteerAngleInterpSpeed=~
     IsReadyToPedalInterpSpeed=~
     StandingInPlaceSteerAngle=~
     SlopePitchDegreeAngle=~
     SlopeRollDegreeAngle=~
     PawnToVehicleDeltaYawAngleDegrees=~
     AimCardDirDeadZoneAngleDegrees=~
     AimCardDirAngleOffsetDegrees=~
     AimFWDDeltaAngleDegrees=~
     AimBWDDeltaAngleDegrees=~
     AimLFTDeltaAngleDegrees=~
     AimRGTDeltaAngleDegrees=~
    ).bAimLFT)]
 }
 Transition {
  path=VehicleLowerBodyOverride.AnimGraphNode_StateMachine_0.ShoppingCartPassenger.AnimStateNode_7.ShoppingCartPassenger_SM.AnimGraphNode_StateMachine_1.ShoppingCartPassenger.AnimStateTransitionNode_3.Transition
  nodes [Result()]
 }
 AnimationTransitionGraph_0 {
  path=VehicleLowerBodyOverride.AnimGraphNode_StateMachine_0.ShoppingCartPassenger.AnimStateNode_7.ShoppingCartPassenger_SM.AnimGraphNode_StateMachine_1.ShoppingCartPassenger.AnimStateTransitionNode_30.AnimationTransitionGraph_0
  nodes [
   n1 = Get ShoppingCartInput()
   Result(bCanEnterTransition=NOT Boolean(A=Break Fort Anim Input Shopping Cart(FortAnimInput_ShoppingCart=@n1) -> (
      bIsUsingShoppingCart=~
      bIsUsingVehicle=~
      bIsCoastStatePedaling=~
      bIsCoastStateCoasting=~
      bIsCoastStateDismount=~
      bIsCoastStateIdle=~
      bIsInAir=~
      bIsCoasting=~
      bIsPedaling=~
      bIsReadyToPedal=~
      IsReadyToPedal=~
      bWantsToCoast=~
      bIsCoastIdling=~
      bIsStartCoasting=~
      bIsEndCoasting=~
      bIsDismountingFromCoast=~
      bIsCoastingOrDismountingFromCoast=~
      bIsStandingInPlace=~
      bIsSprinting=~
      bIsSprintingAndMovingForward=~
      bIsMovingForwardNotSprinting=~
      bIsBraking=~
      bIsReversing=~
      bIsMoving=~
      bIsMovingForward=~
      bIsMovingBackwards=~
      bIsMovingOrTurningInPlace=~
      bIsInAirSteady=~
      bIsOnSlope=~
      bAimFWD=~
      bAimBWD=~
      bAimLFT=~
      bAimRGT=~
      ForwardVelocity=~
      ForwardSpeedKmH=~
      CurrentBrakeForce=~
      RunForwardAlpha=~
      bIsAcceleratingForward=~
      bIsAccelBreakingOrReversing=~
      SteerAngle=~
      SteerAngleInterpSpeed=~
      CoastSteerAngleInterpSpeed=~
      IsReadyToPedalInterpSpeed=~
      StandingInPlaceSteerAngle=~
      SlopePitchDegreeAngle=~
      SlopeRollDegreeAngle=~
      PawnToVehicleDeltaYawAngleDegrees=~
      AimCardDirDeadZoneAngleDegrees=~
      AimCardDirAngleOffsetDegrees=~
      AimFWDDeltaAngleDegrees=~
      AimBWDDeltaAngleDegrees=~
      AimLFTDeltaAngleDegrees=~
      AimRGTDeltaAngleDegrees=~
     ).bAimLFT))
  ]
 }
 AnimationTransitionGraph_1 {
  path=VehicleLowerBodyOverride.AnimGraphNode_StateMachine_0.ShoppingCartPassenger.AnimStateNode_7.ShoppingCartPassenger_SM.AnimGraphNode_StateMachine_1.ShoppingCartPassenger.AnimStateTransitionNode_31.AnimationTransitionGraph_1
  nodes [Result(bCanEnterTransition=Break Fort Anim Input Shopping Cart(FortAnimInput_ShoppingCart=Get ShoppingCartInput()) -> (
     bIsUsingShoppingCart=~
     bIsUsingVehicle=~
     bIsCoastStatePedaling=~
     bIsCoastStateCoasting=~
     bIsCoastStateDismount=~
     bIsCoastStateIdle=~
     bIsInAir=~
     bIsCoasting=~
     bIsPedaling=~
     bIsReadyToPedal=~
     IsReadyToPedal=~
     bWantsToCoast=~
     bIsCoastIdling=~
     bIsStartCoasting=~
     bIsEndCoasting=~
     bIsDismountingFromCoast=~
     bIsCoastingOrDismountingFromCoast=~
     bIsStandingInPlace=~
     bIsSprinting=~
     bIsSprintingAndMovingForward=~
     bIsMovingForwardNotSprinting=~
     bIsBraking=~
     bIsReversing=~
     bIsMoving=~
     bIsMovingForward=~
     bIsMovingBackwards=~
     bIsMovingOrTurningInPlace=~
     bIsInAirSteady=~
     bIsOnSlope=~
     bAimFWD=~
     bAimBWD=~
     bAimLFT=~
     bAimRGT=~
     ForwardVelocity=~
     ForwardSpeedKmH=~
     CurrentBrakeForce=~
     RunForwardAlpha=~
     bIsAcceleratingForward=~
     bIsAccelBreakingOrReversing=~
     SteerAngle=~
     SteerAngleInterpSpeed=~
     CoastSteerAngleInterpSpeed=~
     IsReadyToPedalInterpSpeed=~
     StandingInPlaceSteerAngle=~
     SlopePitchDegreeAngle=~
     SlopeRollDegreeAngle=~
     PawnToVehicleDeltaYawAngleDegrees=~
     AimCardDirDeadZoneAngleDegrees=~
     AimCardDirAngleOffsetDegrees=~
     AimFWDDeltaAngleDegrees=~
     AimBWDDeltaAngleDegrees=~
     AimLFTDeltaAngleDegrees=~
     AimRGTDeltaAngleDegrees=~
    ).bAimFWD)]
 }
 AnimationTransitionGraph_1 {
  path=VehicleLowerBodyOverride.AnimGraphNode_StateMachine_0.ShoppingCartPassenger.AnimStateNode_7.ShoppingCartPassenger_SM.AnimGraphNode_StateMachine_1.ShoppingCartPassenger.AnimStateTransitionNode_32.AnimationTransitionGraph_1
  nodes [
   n1 = Get ShoppingCartInput()
   Result(bCanEnterTransition=NOT Boolean(A=Break Fort Anim Input Shopping Cart(FortAnimInput_ShoppingCart=@n1) -> (
      bIsUsingShoppingCart=~
      bIsUsingVehicle=~
      bIsCoastStatePedaling=~
      bIsCoastStateCoasting=~
      bIsCoastStateDismount=~
      bIsCoastStateIdle=~
      bIsInAir=~
      bIsCoasting=~
      bIsPedaling=~
      bIsReadyToPedal=~
      IsReadyToPedal=~
      bWantsToCoast=~
      bIsCoastIdling=~
      bIsStartCoasting=~
      bIsEndCoasting=~
      bIsDismountingFromCoast=~
      bIsCoastingOrDismountingFromCoast=~
      bIsStandingInPlace=~
      bIsSprinting=~
      bIsSprintingAndMovingForward=~
      bIsMovingForwardNotSprinting=~
      bIsBraking=~
      bIsReversing=~
      bIsMoving=~
      bIsMovingForward=~
      bIsMovingBackwards=~
      bIsMovingOrTurningInPlace=~
      bIsInAirSteady=~
      bIsOnSlope=~
      bAimFWD=~
      bAimBWD=~
      bAimLFT=~
      bAimRGT=~
      ForwardVelocity=~
      ForwardSpeedKmH=~
      CurrentBrakeForce=~
      RunForwardAlpha=~
      bIsAcceleratingForward=~
      bIsAccelBreakingOrReversing=~
      SteerAngle=~
      SteerAngleInterpSpeed=~
      CoastSteerAngleInterpSpeed=~
      IsReadyToPedalInterpSpeed=~
      StandingInPlaceSteerAngle=~
      SlopePitchDegreeAngle=~
      SlopeRollDegreeAngle=~
      PawnToVehicleDeltaYawAngleDegrees=~
      AimCardDirDeadZoneAngleDegrees=~
      AimCardDirAngleOffsetDegrees=~
      AimFWDDeltaAngleDegrees=~
      AimBWDDeltaAngleDegrees=~
      AimLFTDeltaAngleDegrees=~
      AimRGTDeltaAngleDegrees=~
     ).bAimFWD))
  ]
 }
 AnimationTransitionGraph_1 {
  path=VehicleLowerBodyOverride.AnimGraphNode_StateMachine_0.ShoppingCartPassenger.AnimStateNode_7.ShoppingCartPassenger_SM.AnimGraphNode_StateMachine_1.ShoppingCartPassenger.AnimStateTransitionNode_35.AnimationTransitionGraph_1
  nodes [Result(bCanEnterTransition=Break Fort Anim Input Shopping Cart(FortAnimInput_ShoppingCart=Get ShoppingCartInput()) -> (
     bIsUsingShoppingCart=~
     bIsUsingVehicle=~
     bIsCoastStatePedaling=~
     bIsCoastStateCoasting=~
     bIsCoastStateDismount=~
     bIsCoastStateIdle=~
     bIsInAir=~
     bIsCoasting=~
     bIsPedaling=~
     bIsReadyToPedal=~
     IsReadyToPedal=~
     bWantsToCoast=~
     bIsCoastIdling=~
     bIsStartCoasting=~
     bIsEndCoasting=~
     bIsDismountingFromCoast=~
     bIsCoastingOrDismountingFromCoast=~
     bIsStandingInPlace=~
     bIsSprinting=~
     bIsSprintingAndMovingForward=~
     bIsMovingForwardNotSprinting=~
     bIsBraking=~
     bIsReversing=~
     bIsMoving=~
     bIsMovingForward=~
     bIsMovingBackwards=~
     bIsMovingOrTurningInPlace=~
     bIsInAirSteady=~
     bIsOnSlope=~
     bAimFWD=~
     bAimBWD=~
     bAimLFT=~
     bAimRGT=~
     ForwardVelocity=~
     ForwardSpeedKmH=~
     CurrentBrakeForce=~
     RunForwardAlpha=~
     bIsAcceleratingForward=~
     bIsAccelBreakingOrReversing=~
     SteerAngle=~
     SteerAngleInterpSpeed=~
     CoastSteerAngleInterpSpeed=~
     IsReadyToPedalInterpSpeed=~
     StandingInPlaceSteerAngle=~
     SlopePitchDegreeAngle=~
     SlopeRollDegreeAngle=~
     PawnToVehicleDeltaYawAngleDegrees=~
     AimCardDirDeadZoneAngleDegrees=~
     AimCardDirAngleOffsetDegrees=~
     AimFWDDeltaAngleDegrees=~
     AimBWDDeltaAngleDegrees=~
     AimLFTDeltaAngleDegrees=~
     AimRGTDeltaAngleDegrees=~
    ).bAimFWD)]
 }
 AnimationTransitionGraph_1 {
  path=VehicleLowerBodyOverride.AnimGraphNode_StateMachine_0.ShoppingCartPassenger.AnimStateNode_7.ShoppingCartPassenger_SM.AnimGraphNode_StateMachine_1.ShoppingCartPassenger.AnimStateTransitionNode_36.AnimationTransitionGraph_1
  nodes [
   n1 = Get ShoppingCartInput()
   Result(bCanEnterTransition=NOT Boolean(A=Break Fort Anim Input Shopping Cart(FortAnimInput_ShoppingCart=@n1) -> (
      bIsUsingShoppingCart=~
      bIsUsingVehicle=~
      bIsCoastStatePedaling=~
      bIsCoastStateCoasting=~
      bIsCoastStateDismount=~
      bIsCoastStateIdle=~
      bIsInAir=~
      bIsCoasting=~
      bIsPedaling=~
      bIsReadyToPedal=~
      IsReadyToPedal=~
      bWantsToCoast=~
      bIsCoastIdling=~
      bIsStartCoasting=~
      bIsEndCoasting=~
      bIsDismountingFromCoast=~
      bIsCoastingOrDismountingFromCoast=~
      bIsStandingInPlace=~
      bIsSprinting=~
      bIsSprintingAndMovingForward=~
      bIsMovingForwardNotSprinting=~
      bIsBraking=~
      bIsReversing=~
      bIsMoving=~
      bIsMovingForward=~
      bIsMovingBackwards=~
      bIsMovingOrTurningInPlace=~
      bIsInAirSteady=~
      bIsOnSlope=~
      bAimFWD=~
      bAimBWD=~
      bAimLFT=~
      bAimRGT=~
      ForwardVelocity=~
      ForwardSpeedKmH=~
      CurrentBrakeForce=~
      RunForwardAlpha=~
      bIsAcceleratingForward=~
      bIsAccelBreakingOrReversing=~
      SteerAngle=~
      SteerAngleInterpSpeed=~
      CoastSteerAngleInterpSpeed=~
      IsReadyToPedalInterpSpeed=~
      StandingInPlaceSteerAngle=~
      SlopePitchDegreeAngle=~
      SlopeRollDegreeAngle=~
      PawnToVehicleDeltaYawAngleDegrees=~
      AimCardDirDeadZoneAngleDegrees=~
      AimCardDirAngleOffsetDegrees=~
      AimFWDDeltaAngleDegrees=~
      AimBWDDeltaAngleDegrees=~
      AimLFTDeltaAngleDegrees=~
      AimRGTDeltaAngleDegrees=~
     ).bAimFWD))
  ]
 }
 AnimationTransitionGraph_1 {
  path=VehicleLowerBodyOverride.AnimGraphNode_StateMachine_0.ShoppingCartPassenger.AnimStateNode_7.ShoppingCartPassenger_SM.AnimGraphNode_StateMachine_1.ShoppingCartPassenger.AnimStateTransitionNode_37.AnimationTransitionGraph_1
  nodes [Result(bCanEnterTransition=Break Fort Anim Input Shopping Cart(FortAnimInput_ShoppingCart=Get ShoppingCartInput()) -> (
     bIsUsingShoppingCart=~
     bIsUsingVehicle=~
     bIsCoastStatePedaling=~
     bIsCoastStateCoasting=~
     bIsCoastStateDismount=~
     bIsCoastStateIdle=~
     bIsInAir=~
     bIsCoasting=~
     bIsPedaling=~
     bIsReadyToPedal=~
     IsReadyToPedal=~
     bWantsToCoast=~
     bIsCoastIdling=~
     bIsStartCoasting=~
     bIsEndCoasting=~
     bIsDismountingFromCoast=~
     bIsCoastingOrDismountingFromCoast=~
     bIsStandingInPlace=~
     bIsSprinting=~
     bIsSprintingAndMovingForward=~
     bIsMovingForwardNotSprinting=~
     bIsBraking=~
     bIsReversing=~
     bIsMoving=~
     bIsMovingForward=~
     bIsMovingBackwards=~
     bIsMovingOrTurningInPlace=~
     bIsInAirSteady=~
     bIsOnSlope=~
     bAimFWD=~
     bAimBWD=~
     bAimLFT=~
     bAimRGT=~
     ForwardVelocity=~
     ForwardSpeedKmH=~
     CurrentBrakeForce=~
     RunForwardAlpha=~
     bIsAcceleratingForward=~
     bIsAccelBreakingOrReversing=~
     SteerAngle=~
     SteerAngleInterpSpeed=~
     CoastSteerAngleInterpSpeed=~
     IsReadyToPedalInterpSpeed=~
     StandingInPlaceSteerAngle=~
     SlopePitchDegreeAngle=~
     SlopeRollDegreeAngle=~
     PawnToVehicleDeltaYawAngleDegrees=~
     AimCardDirDeadZoneAngleDegrees=~
     AimCardDirAngleOffsetDegrees=~
     AimFWDDeltaAngleDegrees=~
     AimBWDDeltaAngleDegrees=~
     AimLFTDeltaAngleDegrees=~
     AimRGTDeltaAngleDegrees=~
    ).bAimRGT)]
 }
 AnimationTransitionGraph_0 {
  path=VehicleLowerBodyOverride.AnimGraphNode_StateMachine_0.ShoppingCartPassenger.AnimStateNode_7.ShoppingCartPassenger_SM.AnimGraphNode_StateMachine_1.ShoppingCartPassenger.AnimStateTransitionNode_38.AnimationTransitionGraph_0
  nodes [Result()]
 }
 AnimationTransitionGraph_0 {
  path=VehicleLowerBodyOverride.AnimGraphNode_StateMachine_0.ShoppingCartPassenger.AnimStateNode_7.ShoppingCartPassenger_SM.AnimGraphNode_StateMachine_1.ShoppingCartPassenger.AnimStateTransitionNode_39.AnimationTransitionGraph_0
  nodes [Result(bCanEnterTransition=Break Fort Anim Input Shopping Cart(FortAnimInput_ShoppingCart=Get ShoppingCartInput()) -> (
     bIsUsingShoppingCart=~
     bIsUsingVehicle=~
     bIsCoastStatePedaling=~
     bIsCoastStateCoasting=~
     bIsCoastStateDismount=~
     bIsCoastStateIdle=~
     bIsInAir=~
     bIsCoasting=~
     bIsPedaling=~
     bIsReadyToPedal=~
     IsReadyToPedal=~
     bWantsToCoast=~
     bIsCoastIdling=~
     bIsStartCoasting=~
     bIsEndCoasting=~
     bIsDismountingFromCoast=~
     bIsCoastingOrDismountingFromCoast=~
     bIsStandingInPlace=~
     bIsSprinting=~
     bIsSprintingAndMovingForward=~
     bIsMovingForwardNotSprinting=~
     bIsBraking=~
     bIsReversing=~
     bIsMoving=~
     bIsMovingForward=~
     bIsMovingBackwards=~
     bIsMovingOrTurningInPlace=~
     bIsInAirSteady=~
     bIsOnSlope=~
     bAimFWD=~
     bAimBWD=~
     bAimLFT=~
     bAimRGT=~
     ForwardVelocity=~
     ForwardSpeedKmH=~
     CurrentBrakeForce=~
     RunForwardAlpha=~
     bIsAcceleratingForward=~
     bIsAccelBreakingOrReversing=~
     SteerAngle=~
     SteerAngleInterpSpeed=~
     CoastSteerAngleInterpSpeed=~
     IsReadyToPedalInterpSpeed=~
     StandingInPlaceSteerAngle=~
     SlopePitchDegreeAngle=~
     SlopeRollDegreeAngle=~
     PawnToVehicleDeltaYawAngleDegrees=~
     AimCardDirDeadZoneAngleDegrees=~
     AimCardDirAngleOffsetDegrees=~
     AimFWDDeltaAngleDegrees=~
     AimBWDDeltaAngleDegrees=~
     AimLFTDeltaAngleDegrees=~
     AimRGTDeltaAngleDegrees=~
    ).bAimBWD)]
 }
 Transition {
  path=VehicleLowerBodyOverride.AnimGraphNode_StateMachine_0.ShoppingCartPassenger.AnimStateNode_7.ShoppingCartPassenger_SM.AnimGraphNode_StateMachine_1.ShoppingCartPassenger.AnimStateTransitionNode_4.Transition
  nodes [
   n1 = Get ShoppingCartInput()
   Result(bCanEnterTransition=NOT Boolean(A=Break Fort Anim Input Shopping Cart(FortAnimInput_ShoppingCart=@n1) -> (
      bIsUsingShoppingCart=~
      bIsUsingVehicle=~
      bIsCoastStatePedaling=~
      bIsCoastStateCoasting=~
      bIsCoastStateDismount=~
      bIsCoastStateIdle=~
      bIsInAir=~
      bIsCoasting=~
      bIsPedaling=~
      bIsReadyToPedal=~
      IsReadyToPedal=~
      bWantsToCoast=~
      bIsCoastIdling=~
      bIsStartCoasting=~
      bIsEndCoasting=~
      bIsDismountingFromCoast=~
      bIsCoastingOrDismountingFromCoast=~
      bIsStandingInPlace=~
      bIsSprinting=~
      bIsSprintingAndMovingForward=~
      bIsMovingForwardNotSprinting=~
      bIsBraking=~
      bIsReversing=~
      bIsMoving=~
      bIsMovingForward=~
      bIsMovingBackwards=~
      bIsMovingOrTurningInPlace=~
      bIsInAirSteady=~
      bIsOnSlope=~
      bAimFWD=~
      bAimBWD=~
      bAimLFT=~
      bAimRGT=~
      ForwardVelocity=~
      ForwardSpeedKmH=~
      CurrentBrakeForce=~
      RunForwardAlpha=~
      bIsAcceleratingForward=~
      bIsAccelBreakingOrReversing=~
      SteerAngle=~
      SteerAngleInterpSpeed=~
      CoastSteerAngleInterpSpeed=~
      IsReadyToPedalInterpSpeed=~
      StandingInPlaceSteerAngle=~
      SlopePitchDegreeAngle=~
      SlopeRollDegreeAngle=~
      PawnToVehicleDeltaYawAngleDegrees=~
      AimCardDirDeadZoneAngleDegrees=~
      AimCardDirAngleOffsetDegrees=~
      AimFWDDeltaAngleDegrees=~
      AimBWDDeltaAngleDegrees=~
      AimLFTDeltaAngleDegrees=~
      AimRGTDeltaAngleDegrees=~
     ).bAimLFT))
  ]
 }
 AnimationTransitionGraph_1 {
  path=VehicleLowerBodyOverride.AnimGraphNode_StateMachine_0.ShoppingCartPassenger.AnimStateNode_7.ShoppingCartPassenger_SM.AnimGraphNode_StateMachine_1.ShoppingCartPassenger.AnimStateTransitionNode_40.AnimationTransitionGraph_1
  nodes [Result()]
 }
 AnimationTransitionGraph_1 {
  path=VehicleLowerBodyOverride.AnimGraphNode_StateMachine_0.ShoppingCartPassenger.AnimStateNode_7.ShoppingCartPassenger_SM.AnimGraphNode_StateMachine_1.ShoppingCartPassenger.AnimStateTransitionNode_41.AnimationTransitionGraph_1
  nodes [Result(bCanEnterTransition=Break Fort Anim Input Shopping Cart(FortAnimInput_ShoppingCart=Get ShoppingCartInput()) -> (
     bIsUsingShoppingCart=~
     bIsUsingVehicle=~
     bIsCoastStatePedaling=~
     bIsCoastStateCoasting=~
     bIsCoastStateDismount=~
     bIsCoastStateIdle=~
     bIsInAir=~
     bIsCoasting=~
     bIsPedaling=~
     bIsReadyToPedal=~
     IsReadyToPedal=~
     bWantsToCoast=~
     bIsCoastIdling=~
     bIsStartCoasting=~
     bIsEndCoasting=~
     bIsDismountingFromCoast=~
     bIsCoastingOrDismountingFromCoast=~
     bIsStandingInPlace=~
     bIsSprinting=~
     bIsSprintingAndMovingForward=~
     bIsMovingForwardNotSprinting=~
     bIsBraking=~
     bIsReversing=~
     bIsMoving=~
     bIsMovingForward=~
     bIsMovingBackwards=~
     bIsMovingOrTurningInPlace=~
     bIsInAirSteady=~
     bIsOnSlope=~
     bAimFWD=~
     bAimBWD=~
     bAimLFT=~
     bAimRGT=~
     ForwardVelocity=~
     ForwardSpeedKmH=~
     CurrentBrakeForce=~
     RunForwardAlpha=~
     bIsAcceleratingForward=~
     bIsAccelBreakingOrReversing=~
     SteerAngle=~
     SteerAngleInterpSpeed=~
     CoastSteerAngleInterpSpeed=~
     IsReadyToPedalInterpSpeed=~
     StandingInPlaceSteerAngle=~
     SlopePitchDegreeAngle=~
     SlopeRollDegreeAngle=~
     PawnToVehicleDeltaYawAngleDegrees=~
     AimCardDirDeadZoneAngleDegrees=~
     AimCardDirAngleOffsetDegrees=~
     AimFWDDeltaAngleDegrees=~
     AimBWDDeltaAngleDegrees=~
     AimLFTDeltaAngleDegrees=~
     AimRGTDeltaAngleDegrees=~
    ).bAimLFT)]
 }
 AnimationTransitionGraph_1 {
  path=VehicleLowerBodyOverride.AnimGraphNode_StateMachine_0.ShoppingCartPassenger.AnimStateNode_7.ShoppingCartPassenger_SM.AnimGraphNode_StateMachine_1.ShoppingCartPassenger.AnimStateTransitionNode_42.AnimationTransitionGraph_1
  nodes [Result()]
 }
 AnimationTransitionGraph_1 {
  path=VehicleLowerBodyOverride.AnimGraphNode_StateMachine_0.ShoppingCartPassenger.AnimStateNode_7.ShoppingCartPassenger_SM.AnimGraphNode_StateMachine_1.ShoppingCartPassenger.AnimStateTransitionNode_5.AnimationTransitionGraph_1
  nodes [
   n1 = Get ShoppingCartInput()
   Result(bCanEnterTransition=NOT Boolean(A=Break Fort Anim Input Shopping Cart(FortAnimInput_ShoppingCart=@n1) -> (
      bIsUsingShoppingCart=~
      bIsUsingVehicle=~
      bIsCoastStatePedaling=~
      bIsCoastStateCoasting=~
      bIsCoastStateDismount=~
      bIsCoastStateIdle=~
      bIsInAir=~
      bIsCoasting=~
      bIsPedaling=~
      bIsReadyToPedal=~
      IsReadyToPedal=~
      bWantsToCoast=~
      bIsCoastIdling=~
      bIsStartCoasting=~
      bIsEndCoasting=~
      bIsDismountingFromCoast=~
      bIsCoastingOrDismountingFromCoast=~
      bIsStandingInPlace=~
      bIsSprinting=~
      bIsSprintingAndMovingForward=~
      bIsMovingForwardNotSprinting=~
      bIsBraking=~
      bIsReversing=~
      bIsMoving=~
      bIsMovingForward=~
      bIsMovingBackwards=~
      bIsMovingOrTurningInPlace=~
      bIsInAirSteady=~
      bIsOnSlope=~
      bAimFWD=~
      bAimBWD=~
      bAimLFT=~
      bAimRGT=~
      ForwardVelocity=~
      ForwardSpeedKmH=~
      CurrentBrakeForce=~
      RunForwardAlpha=~
      bIsAcceleratingForward=~
      bIsAccelBreakingOrReversing=~
      SteerAngle=~
      SteerAngleInterpSpeed=~
      CoastSteerAngleInterpSpeed=~
      IsReadyToPedalInterpSpeed=~
      StandingInPlaceSteerAngle=~
      SlopePitchDegreeAngle=~
      SlopeRollDegreeAngle=~
      PawnToVehicleDeltaYawAngleDegrees=~
      AimCardDirDeadZoneAngleDegrees=~
      AimCardDirAngleOffsetDegrees=~
      AimFWDDeltaAngleDegrees=~
      AimBWDDeltaAngleDegrees=~
      AimLFTDeltaAngleDegrees=~
      AimRGTDeltaAngleDegrees=~
     ).bAimBWD))
  ]
 }
 AnimationTransitionGraph_0 {
  path=VehicleLowerBodyOverride.AnimGraphNode_StateMachine_0.ShoppingCartPassenger.AnimStateNode_7.ShoppingCartPassenger_SM.AnimGraphNode_StateMachine_1.ShoppingCartPassenger.AnimStateTransitionNode_6.AnimationTransitionGraph_0
  nodes [
   n1 = Get ShoppingCartInput()
   Result(bCanEnterTransition=NOT Boolean(A=Break Fort Anim Input Shopping Cart(FortAnimInput_ShoppingCart=@n1) -> (
      bIsUsingShoppingCart=~
      bIsUsingVehicle=~
      bIsCoastStatePedaling=~
      bIsCoastStateCoasting=~
      bIsCoastStateDismount=~
      bIsCoastStateIdle=~
      bIsInAir=~
      bIsCoasting=~
      bIsPedaling=~
      bIsReadyToPedal=~
      IsReadyToPedal=~
      bWantsToCoast=~
      bIsCoastIdling=~
      bIsStartCoasting=~
      bIsEndCoasting=~
      bIsDismountingFromCoast=~
      bIsCoastingOrDismountingFromCoast=~
      bIsStandingInPlace=~
      bIsSprinting=~
      bIsSprintingAndMovingForward=~
      bIsMovingForwardNotSprinting=~
      bIsBraking=~
      bIsReversing=~
      bIsMoving=~
      bIsMovingForward=~
      bIsMovingBackwards=~
      bIsMovingOrTurningInPlace=~
      bIsInAirSteady=~
      bIsOnSlope=~
      bAimFWD=~
      bAimBWD=~
      bAimLFT=~
      bAimRGT=~
      ForwardVelocity=~
      ForwardSpeedKmH=~
      CurrentBrakeForce=~
      RunForwardAlpha=~
      bIsAcceleratingForward=~
      bIsAccelBreakingOrReversing=~
      SteerAngle=~
      SteerAngleInterpSpeed=~
      CoastSteerAngleInterpSpeed=~
      IsReadyToPedalInterpSpeed=~
      StandingInPlaceSteerAngle=~
      SlopePitchDegreeAngle=~
      SlopeRollDegreeAngle=~
      PawnToVehicleDeltaYawAngleDegrees=~
      AimCardDirDeadZoneAngleDegrees=~
      AimCardDirAngleOffsetDegrees=~
      AimFWDDeltaAngleDegrees=~
      AimBWDDeltaAngleDegrees=~
      AimLFTDeltaAngleDegrees=~
      AimRGTDeltaAngleDegrees=~
     ).bAimRGT))
  ]
 }
 Transition {
  path=VehicleLowerBodyOverride.AnimGraphNode_StateMachine_0.ShoppingCartPassenger.AnimStateNode_7.ShoppingCartPassenger_SM.AnimGraphNode_StateMachine_1.ShoppingCartPassenger.AnimStateTransitionNode_7.Transition
  nodes [Result()]
 }
 AnimationTransitionGraph_0 {
  path=VehicleLowerBodyOverride.AnimGraphNode_StateMachine_0.ShoppingCartPassenger.AnimStateNode_7.ShoppingCartPassenger_SM.AnimGraphNode_StateMachine_1.ShoppingCartPassenger.AnimStateTransitionNode_8.AnimationTransitionGraph_0
  nodes [Result(bCanEnterTransition=Break Fort Anim Input Shopping Cart(FortAnimInput_ShoppingCart=Get ShoppingCartInput()) -> (
     bIsUsingShoppingCart=~
     bIsUsingVehicle=~
     bIsCoastStatePedaling=~
     bIsCoastStateCoasting=~
     bIsCoastStateDismount=~
     bIsCoastStateIdle=~
     bIsInAir=~
     bIsCoasting=~
     bIsPedaling=~
     bIsReadyToPedal=~
     IsReadyToPedal=~
     bWantsToCoast=~
     bIsCoastIdling=~
     bIsStartCoasting=~
     bIsEndCoasting=~
     bIsDismountingFromCoast=~
     bIsCoastingOrDismountingFromCoast=~
     bIsStandingInPlace=~
     bIsSprinting=~
     bIsSprintingAndMovingForward=~
     bIsMovingForwardNotSprinting=~
     bIsBraking=~
     bIsReversing=~
     bIsMoving=~
     bIsMovingForward=~
     bIsMovingBackwards=~
     bIsMovingOrTurningInPlace=~
     bIsInAirSteady=~
     bIsOnSlope=~
     bAimFWD=~
     bAimBWD=~
     bAimLFT=~
     bAimRGT=~
     ForwardVelocity=~
     ForwardSpeedKmH=~
     CurrentBrakeForce=~
     RunForwardAlpha=~
     bIsAcceleratingForward=~
     bIsAccelBreakingOrReversing=~
     SteerAngle=~
     SteerAngleInterpSpeed=~
     CoastSteerAngleInterpSpeed=~
     IsReadyToPedalInterpSpeed=~
     StandingInPlaceSteerAngle=~
     SlopePitchDegreeAngle=~
     SlopeRollDegreeAngle=~
     PawnToVehicleDeltaYawAngleDegrees=~
     AimCardDirDeadZoneAngleDegrees=~
     AimCardDirAngleOffsetDegrees=~
     AimFWDDeltaAngleDegrees=~
     AimBWDDeltaAngleDegrees=~
     AimLFTDeltaAngleDegrees=~
     AimRGTDeltaAngleDegrees=~
    ).bAimBWD)]
 }
 AnimationTransitionGraph_0 {
  path=VehicleLowerBodyOverride.AnimGraphNode_StateMachine_0.ShoppingCartPassenger.AnimStateNode_7.ShoppingCartPassenger_SM.AnimGraphNode_StateMachine_1.ShoppingCartPassenger.AnimStateTransitionNode_9.AnimationTransitionGraph_0
  nodes [Result(bCanEnterTransition=Break Fort Anim Input Shopping Cart(FortAnimInput_ShoppingCart=Get ShoppingCartInput()) -> (
     bIsUsingShoppingCart=~
     bIsUsingVehicle=~
     bIsCoastStatePedaling=~
     bIsCoastStateCoasting=~
     bIsCoastStateDismount=~
     bIsCoastStateIdle=~
     bIsInAir=~
     bIsCoasting=~
     bIsPedaling=~
     bIsReadyToPedal=~
     IsReadyToPedal=~
     bWantsToCoast=~
     bIsCoastIdling=~
     bIsStartCoasting=~
     bIsEndCoasting=~
     bIsDismountingFromCoast=~
     bIsCoastingOrDismountingFromCoast=~
     bIsStandingInPlace=~
     bIsSprinting=~
     bIsSprintingAndMovingForward=~
     bIsMovingForwardNotSprinting=~
     bIsBraking=~
     bIsReversing=~
     bIsMoving=~
     bIsMovingForward=~
     bIsMovingBackwards=~
     bIsMovingOrTurningInPlace=~
     bIsInAirSteady=~
     bIsOnSlope=~
     bAimFWD=~
     bAimBWD=~
     bAimLFT=~
     bAimRGT=~
     ForwardVelocity=~
     ForwardSpeedKmH=~
     CurrentBrakeForce=~
     RunForwardAlpha=~
     bIsAcceleratingForward=~
     bIsAccelBreakingOrReversing=~
     SteerAngle=~
     SteerAngleInterpSpeed=~
     CoastSteerAngleInterpSpeed=~
     IsReadyToPedalInterpSpeed=~
     StandingInPlaceSteerAngle=~
     SlopePitchDegreeAngle=~
     SlopeRollDegreeAngle=~
     PawnToVehicleDeltaYawAngleDegrees=~
     AimCardDirDeadZoneAngleDegrees=~
     AimCardDirAngleOffsetDegrees=~
     AimFWDDeltaAngleDegrees=~
     AimBWDDeltaAngleDegrees=~
     AimLFTDeltaAngleDegrees=~
     AimRGTDeltaAngleDegrees=~
    ).bAimFWD)]
 }
 Transition {
  path=VehicleLowerBodyOverride.AnimGraphNode_StateMachine_0.ShoppingCartPassenger.AnimStateTransitionNode_0.Transition
  nodes [Result(bCanEnterTransition=Break Fort Anim Input Shopping Cart(FortAnimInput_ShoppingCart=Get ShoppingCartInput()) -> (
     bIsUsingShoppingCart=~
     bIsUsingVehicle=~
     bIsCoastStatePedaling=~
     bIsCoastStateCoasting=~
     bIsCoastStateDismount=~
     bIsCoastStateIdle=~
     bIsInAir=~
     bIsCoasting=~
     bIsPedaling=~
     bIsReadyToPedal=~
     IsReadyToPedal=~
     bWantsToCoast=~
     bIsCoastIdling=~
     bIsStartCoasting=~
     bIsEndCoasting=~
     bIsDismountingFromCoast=~
     bIsCoastingOrDismountingFromCoast=~
     bIsStandingInPlace=~
     bIsSprinting=~
     bIsSprintingAndMovingForward=~
     bIsMovingForwardNotSprinting=~
     bIsBraking=~
     bIsReversing=~
     bIsMoving=~
     bIsMovingForward=~
     bIsMovingBackwards=~
     bIsMovingOrTurningInPlace=~
     bIsInAirSteady=~
     bIsOnSlope=~
     bAimFWD=~
     bAimBWD=~
     bAimLFT=~
     bAimRGT=~
     ForwardVelocity=~
     ForwardSpeedKmH=~
     CurrentBrakeForce=~
     RunForwardAlpha=~
     bIsAcceleratingForward=~
     bIsAccelBreakingOrReversing=~
     SteerAngle=~
     SteerAngleInterpSpeed=~
     CoastSteerAngleInterpSpeed=~
     IsReadyToPedalInterpSpeed=~
     StandingInPlaceSteerAngle=~
     SlopePitchDegreeAngle=~
     SlopeRollDegreeAngle=~
     PawnToVehicleDeltaYawAngleDegrees=~
     AimCardDirDeadZoneAngleDegrees=~
     AimCardDirAngleOffsetDegrees=~
     AimFWDDeltaAngleDegrees=~
     AimBWDDeltaAngleDegrees=~
     AimLFTDeltaAngleDegrees=~
     AimRGTDeltaAngleDegrees=~
    ))]
 }
 Transition {
  path=VehicleLowerBodyOverride.AnimGraphNode_StateMachine_0.ShoppingCartPassenger.AnimStateTransitionNode_1.Transition
  nodes [
   n1 = Get ShoppingCartInput()
   Result(bCanEnterTransition=NOT Boolean(A=Break Fort Anim Input Shopping Cart(FortAnimInput_ShoppingCart=@n1) -> (
      bIsUsingShoppingCart=~
      bIsUsingVehicle=~
      bIsCoastStatePedaling=~
      bIsCoastStateCoasting=~
      bIsCoastStateDismount=~
      bIsCoastStateIdle=~
      bIsInAir=~
      bIsCoasting=~
      bIsPedaling=~
      bIsReadyToPedal=~
      IsReadyToPedal=~
      bWantsToCoast=~
      bIsCoastIdling=~
      bIsStartCoasting=~
      bIsEndCoasting=~
      bIsDismountingFromCoast=~
      bIsCoastingOrDismountingFromCoast=~
      bIsStandingInPlace=~
      bIsSprinting=~
      bIsSprintingAndMovingForward=~
      bIsMovingForwardNotSprinting=~
      bIsBraking=~
      bIsReversing=~
      bIsMoving=~
      bIsMovingForward=~
      bIsMovingBackwards=~
      bIsMovingOrTurningInPlace=~
      bIsInAirSteady=~
      bIsOnSlope=~
      bAimFWD=~
      bAimBWD=~
      bAimLFT=~
      bAimRGT=~
      ForwardVelocity=~
      ForwardSpeedKmH=~
      CurrentBrakeForce=~
      RunForwardAlpha=~
      bIsAcceleratingForward=~
      bIsAccelBreakingOrReversing=~
      SteerAngle=~
      SteerAngleInterpSpeed=~
      CoastSteerAngleInterpSpeed=~
      IsReadyToPedalInterpSpeed=~
      StandingInPlaceSteerAngle=~
      SlopePitchDegreeAngle=~
      SlopeRollDegreeAngle=~
      PawnToVehicleDeltaYawAngleDegrees=~
      AimCardDirDeadZoneAngleDegrees=~
      AimCardDirAngleOffsetDegrees=~
      AimFWDDeltaAngleDegrees=~
      AimBWDDeltaAngleDegrees=~
      AimLFTDeltaAngleDegrees=~
      AimRGTDeltaAngleDegrees=~
     )))
  ]
 }
 VehicleSplitBodyOverride {nodes [
   Input Pose() {name=InPoseSplitBody}
   Input Pose() {name=InPoseUpperAndLowerBody}
   Input Pose() {name=InPoseUpperAndLowerBodyRemoveRoot}
   Output Pose() {layerGroup=Vehicle Override}
  ]}
]
variables [isDBNO {type=boolean}]
