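# Animation layer Blueprint for vehicle passengers on the Fortnite average male player skeleton, adjusting the body pose to vehicle slope.
# Only the split-body layer does work, running the cached upper-and-lower-body pose through slope aim offsets.
# bIsFrontPassenger selects chained GolfCart back-passenger pitch and roll offsets when true, otherwise a combined ATV offset.
# Each update sets DisableNoise from bIsTargeting, though no graph here reads it.
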AnimBlueprint /Game/Characters/Player/Male/Male_Avg_Base/LayerAnimBP/VehiclePassengerLayerAnimBP
target_skeleton=&Fortnite_M_Avg_Player_Skeleton
graphs [
 AnimGraph {nodes [Output Pose() {layerGroup=None}]}
 EventGraph {nodes [
   Event BlueprintUpdateAnimation()
   > n1 = Branch(Condition=Get bIsTargeting())
   > Set DisableNoise(DisableNoise:double=0) -> (Output_Get:double)
   Set DisableNoise(execute=@n1.else, DisableNoise=1)
  ]}
 VehicleFinalPoseOverride {nodes [
   Input Pose() {name=InFinalPose}
   Output Pose() {layerGroup=VehicleOverride}
  ]}
 VehicleFullBodyOverride {nodes [
   Input Pose() {name=InPoseFullBody}
   Output Pose() {layerGroup=VehicleOverride}
  ]}
 VehicleLowerBodyOverride {nodes [
   Input Pose() {name=InPoseLowerBodyDefault}
   Input Pose() {name=InPoseUpperBody}
   Output Pose() {layerGroup=VehicleOverride}
  ]}
 VehicleSplitBodyOverride {nodes [
   Input Pose() {name=InPoseSplitBody}
   Input Pose() {name=InPoseUpperAndLowerBodyRemoveRoot}
   n1 = AimOffset Player 'PassengerBack_AO_SlopePitch_GolfCart'!(
    X:float=Get SlopePitchDegreeAngle() -> (SlopePitchDegreeAngle:float)
    BasePose:Pose Link Structure=Use cached pose 'UpperAndLowerBodyCachePose'()
    Alpha:float=1
   ) -> (Pose:Pose Link Structure) {blendSpace=/Game/Animation/Game/MainPlayer/Vehicles/GolfCart/AO/PassengerBack_AO_SlopePitch_GolfCart, alphaBoolBlend {blendOption=Linear}}
   n2 = AimOffset Player 'PassengerBack_AO_SlopeRoll_GolfCart'!(X:float=Get SlopeRollDegreeAngle() -> (SlopeRollDegreeAngle:float), BasePose:Pose Link Structure=@n1, Alpha:float=1) -> (Pose:Pose Link Structure) {blendSpace=/Game/Animation/Game/MainPlayer/Vehicles/GolfCart/AO/PassengerBack_AO_SlopeRoll_GolfCart, alphaBoolBlend {blendOption=Linear}}
   n3 = Use cached pose 'UpperAndLowerBodyCachePose'()
   n4 = Get SlopePitchDegreeAngle()
   n5 = Get SlopeRollDegreeAngle()
   Output Pose(Result=Blend Poses by bool(
     BlendPose_0=@n2
     BlendPose_1=AimOffset Player 'ATV_PassengerSlope_AO'!(X:float=@n5, Y:float=@n4, BasePose:Pose Link Structure=@n3, Alpha:float=1) -> (Pose:Pose Link Structure) {blendSpace=/Game/Animation/Game/MainPlayer/Vehicles/ATV/AO/ATV_PassengerSlope_AO, alphaBoolBlend {blendOption=Linear}}
     bActiveValue=Get bIsFrontPassenger() -> (bIsFrontPassenger:bool)
    ) {blendType=Linear}) {layerGroup=Vehicle Override}
   Save cached pose 'UpperAndLowerBodyCachePose'(Pose=Input Pose() {name=InPoseUpperAndLowerBody}) {cacheName=UpperAndLowerBodyCachePose}
  ]}
]
variables [DisableNoise {type=number, default=1}]
