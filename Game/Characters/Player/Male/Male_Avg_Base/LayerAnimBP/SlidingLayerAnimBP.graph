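# Player sliding animation layer for the MovementMode layer group.
# The lower-body state machine handles baseball and knee slides with slope blendspaces, then yaws the root outside first person.
# Source-pose and additive layers supply spine tweak and lean.
# Bounce, noise and RigidBodyWithControl physics chains exist but don't reach layer outputs, and update and initialize events are unwired from the velocity, floor-material and debug-print logic.
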AnimBlueprint /Game/Characters/Player/Male/Male_Avg_Base/LayerAnimBP/SlidingLayerAnimBP
target_skeleton=&Fortnite_M_Avg_Player_Skeleton
graphs [
 AnimGraph {nodes [Output Pose() {layerGroup=None}]}
 EventGraph {nodes [
   n1 = Cast To PlayerPawn_Athena(Object=GetOwningActor())
   > Set As Player Pawn Athena(As Player Pawn Athena:Player Pawn Athena Object Reference=@n1) -> (Output_Get:Player Pawn Athena Object Reference)
   AnimNotify_bounceEnter() {is=title K2Node_Event}
   > Delay(Duration=RandomFloatInRange(Min=0.5, Max=3))
   n2 = Get outputVelocity() -> (outputVelocity:double)
   > Set bounceStrength(bounceStrength:double=MapRangeClamped(Value=@n2, InRangeA=600, InRangeB=1200, OutRangeA=0.5)) -> (Output_Get:double)
   > Set bounceSpeed(bounceSpeed:double=MapRangeClamped(Value=@n2, InRangeA=600, InRangeB=1200, OutRangeA=0.7, OutRangeB=1.3)) -> (Output_Get:double)
   > Set outputShouldBounce(outputShouldBounce:bool=true) -> (Output_Get:bool)
   > DelayUntilNextTick!(execute) -> (then)
   > Set outputShouldBounce(outputShouldBounce=false)
   n3 = Sequence!(execute) -> (then_0, then_1, then_2:Exec, then_3:Exec, then_4:Exec)
   n4 = Get Actor Location!(self=Get As Player Pawn Athena() -> (As Player Pawn Athena:Player Pawn Athena Object Reference)) -> (ReturnValue_X:double, ReturnValue_Y:double, ReturnValue_Z:double)
   n5 = Get zPositionTick() -> (zPositionTick:double)
   > n6 = Set zPositionVelocity(zPositionVelocity:double=float / float(A=float - float(A=@n5, B=@n4.ReturnValue_Z), B=GetWorldDeltaSeconds!() -> (ReturnValue:double))) -> (Output_Get:double)
   > Set zPositionTick(zPositionTick:double=@n4.ReturnValue_Z) -> (Output_Get:double)
   > n7 = Set zVelocityChange(zVelocityChange:double=FInterpTo(Current=Get zVelocityChange() -> (zVelocityChange:double), Target=@n6, DeltaTime=GetWorldDeltaSeconds(), InterpSpeed=3)) -> (Output_Get:double)
   n8 = float - float(A=@n7, B=@n6)
   > Set outputZVelocityChange(outputZVelocityChange:double="To Float (Integer)"(InInt=Round(A=@n8) {is=title K2Node_CallFunction})) -> (Output_Get:double)
   n9 = Get Velocity!(self:Movement Component Object Reference=Get CharacterMovement!(self:Character Object Reference=Get As Player Pawn Athena()) -> (CharacterMovement:Character Movement Component Object Reference)) -> (Velocity:Vector)
   n10 = Set outputVelocity(execute=@n3.then_2, outputVelocity:double=Vector Length(A=@n9)) -> (Output_Get:double)
   n11 = vector * vector!(A=@n9, B:float=-1.3) -> (ReturnValue_X:double, ReturnValue_Y:double, ReturnValue_Z:double)
   > Set!(execute, External Force_X:double=@n11, External Force_Y:double=@n11.ReturnValue_Y, External Force_Z:double=BreakVector(InVec=@n9).Z) -> (then, Output_Get:Vector)
   > Set noiseStrength(noiseStrength:double=MapRangeClamped(Value=@n10, InRangeA=600, InRangeB=1200, OutRangeA=0.3, OutRangeB=0.8)) -> (Output_Get:double)
   n12 = Get outputZVelocityChange() -> (outputZVelocityChange:double)
   n13 = Branch(Condition="float > float"(A="Absolute (Float)"(A=@n12), B=250))
   > Do Once(Reset=@n14.then)
   > Set outputShouldBounce(outputShouldBounce=true)
   > DelayUntilNextTick()
   > Set outputShouldBounce(outputShouldBounce=false)
   n14 = Delay(execute=@n13.else, Duration=3)
   n15 = Get outputShouldBounce() -> (outputShouldBounce:bool)
   n16 = Branch(Condition=@n15)
   n17 = Get outputZVelocityChange()
   > n18 = PrintString!(
    execute
    InString=Append!(A="Bounce Strength = ", B="To String (Float)"(InDouble=@n17), C:String="  ", D:String="To String (Boolean)"(InBool=@n15)) -> (ReturnValue)
    bPrintToScreen
    bPrintToLog
    TextColor
    Duration=1
    Key
   ) -> (then)
   n19 = GetWorldDeltaSeconds()
   n20 = Get outputZVelocityChange()
   PrintString(
    execute=[@n18.then, @n16.else]
    InString=Append!(A="Z Velocity Change = ", B="To String (Float)"(InDouble=@n20)) -> (ReturnValue)
    Duration=@n19
   )
   n21 = Get outputVelocity()
   > PrintString(InString=Append(A="Speed = ", B="To String (Float)"(InDouble=@n21)), Duration=@n19)
   n22 = Get outputFloorMaterial() -> (outputFloorMaterial:Physical Material Object Reference)
   > PrintString(InString=Append(A="Surface Type = ", B=GetDisplayName(Object=@n22)), Duration=@n19)
   n23 = Get VelocityAimYawDelta() -> (VelocityAimYawDelta:float)
   > PrintString(InString=Append(A="Aim Delta = ", B="To String (Float)"(InDouble=@n23)), Duration=GetWorldDeltaSeconds())
   Event BlueprintUpdateAnimation()
   Event BlueprintInitializeAnimation()
   n24 = Get As Player Pawn Athena()
   Set outputFloorMaterial(outputFloorMaterial:Physical Material Object Reference=Get!(self:Character Movement Component Object Reference=Get CharacterMovement(self=@n24)) -> (
     CurrentFloor_bBlockingHit:bool=~
     CurrentFloor_bWalkableFloor:bool=~
     CurrentFloor_bLineTrace:bool=~
     CurrentFloor_FloorDist:float=~
     CurrentFloor_LineDist:float=~
     CurrentFloor_HitResult_bBlockingHit:bool
     CurrentFloor_HitResult_bInitialOverlap:bool
     CurrentFloor_HitResult_Time:float
     CurrentFloor_HitResult_Distance:float
     CurrentFloor_HitResult_Location:Vector
     CurrentFloor_HitResult_ImpactPoint:Vector
     CurrentFloor_HitResult_Normal:Vector
     CurrentFloor_HitResult_ImpactNormal:Vector
     CurrentFloor_HitResult_PhysMat:Physical Material Object Reference
     CurrentFloor_HitResult_HitActor:Actor Object Reference
     CurrentFloor_HitResult_HitComponent:Primitive Component Object Reference
     CurrentFloor_HitResult_HitBoneName:Name=None
     CurrentFloor_HitResult_BoneName:Name=None
     CurrentFloor_HitResult_HitItem:int=0
     CurrentFloor_HitResult_ElementIndex:int=0
     CurrentFloor_HitResult_FaceIndex:int=0
     CurrentFloor_HitResult_TraceStart:Vector
     CurrentFloor_HitResult_TraceEnd:Vector
    ).CurrentFloor_HitResult_PhysMat) -> (Output_Get:Physical Material Object Reference)
  ]}
 MovementMode_AimOffset {nodes [
   Input Pose() {name=InAimOffsetsPose}
   Input Pose() {name=InSourcePosePreAimOffsets}
   Output Pose() {layerGroup=MovementMode}
  ]}
 MovementMode_BaseLocomotion {nodes [
   Input Pose() {name=InPoseBaseLocomotion}
   Output Pose() {layerGroup=MovementMode}
  ]}
 MovementMode_FinalPoseOverride {nodes [
   Input Pose() {name=InFinalPose}
   Output Pose() {layerGroup=MovementMode}
  ]}
 MovementMode_FullBodyOverride {nodes [
   n1 = Get VelocityAimYawDelta()
   Apply Mesh Space Additive!(
    Base:Pose Link Structure=Apply Mesh Space Additive(
     Additive=Blendspace Player 'slopeSlide_YawVeloctiy_BS'(X=@n1, Y=0) {blendSpace=/Game/Animation/Game/MainPlayer/Locomotion/Medium/Male/Sliding/KneeSliding/layers/slopeSlide_YawVeloctiy_BS}
     Alpha=1
    )
    Additive:Pose Link Structure=Blendspace Player 'slopeSlide_ZVelocity_BS'(X=0, Y=Get outputZVelocityChange()) {blendSpace=/Game/Animation/Game/MainPlayer/Locomotion/Medium/Male/Sliding/KneeSliding/layers/slopeSlide_ZVelocity_BS}
    Alpha:float=0.5
   ) -> (Pose:Pose Link Structure) {is=title AnimGraphNode_ApplyMeshSpaceAdditive}
   n2 = RigidBodyWithControl!(
    ComponentPose:Component Space Pose Link Structure=Local To Component()
    Alpha:float=1
    ExternalForce:Vector=Get External Force() -> (External Force:Vector)
    ControlAndModifierParameters:Physics Control Control And Modifier Parameters Structure=(ControlParameters=,ControlMultiplierParameters=,ModifierParameters=)
    ControlAndModifierUpdates:Physics Control Control And Modifier Updates Structure=(ControlUpdates=,ControlMultiplierUpdates=,ModifierUpdates=)
    ControlTargets:Rigid Body Control Targets Structure=(Targets=())
    KinematicTargets:Rigid Body Kinematic Targets Structure=(Targets=())
    ForcesAndTorques:Rigid Body Forces And Torques Structure=(ForcesAndTorques=())
   ) -> (Pose:Component Space Pose Link Structure) {
    is=title AnimGraphNode_RigidBodyWithControl
    bEnableControls=true
    bOverrideWorldGravity=true
    overridePhysicsAsset=/Game/sliding_physics
    additionalControlsAndBodyModifiers {controls {none {
       control {
        childBoneName=None
        parentBoneName=None
        controlData {
         angularDampingRatio=1
         angularExtraDamping=0
         angularStrength=0
         angularTargetVelocityMultiplier=1
         bDisableCollision=false
         bEnabled=true
         bOnlyControlChildObject=false
         bUseAccelerationDriveMode=true
         bUseCustomControlPoint=false
         bUseSkeletalAnimation=true
         linearDampingRatio=1
         linearExtraDamping=0
         linearStrength=0
         linearTargetVelocityMultiplier=1
         maxForce=0
         maxTorque=0
         customControlPoint=<0 0 0>
        }
        controlMultiplier {
         angularDampingRatioMultiplier=1
         angularExtraDampingMultiplier=1
         angularStrengthMultiplier=1
         maxTorqueMultiplier=1
         linearDampingRatioMultiplier=<1 1 1>
         linearExtraDampingMultiplier=<1 1 1>
         linearStrengthMultiplier=<1 1 1>
         maxForceMultiplier=<1 1 1>
        }
       }
       sets []
      }}}
    additionalSets {controlSetUpdates [
      {setName=WorldSpace_UpperBody, names [WorldSpace_LeftArm, WorldSpace_RightArm, WorldSpace_Head, WorldSpace_Spine, thigh_l, thigh_r]}
      {
       setName=ParentSpace_UpperBody
       names [ParentSpace_LeftArm, ParentSpace_RightArm, ParentSpace_Head, ParentSpace_Spine, pelvis_thigh_l, pelvis_thigh_r]
      }
      {setName=Hands, names [hand_l,hand_r]}
      {setName=ParentSpace_pelvis, names [root_pelvis]}
     ]}
    characterSetupData {limbSetupData [
      {
       bCreateBodyModifiers=true
       bCreateParentSpaceControls=true
       bCreateWorldSpaceControls=true
       bIncludeParentBone=false
       limbName=LeftArm
       startBone=clavicle_l
      }
      {
       bCreateBodyModifiers=true
       bCreateParentSpaceControls=true
       bCreateWorldSpaceControls=true
       bIncludeParentBone=false
       limbName=RightArm
       startBone=clavicle_r
      }
      {
       bCreateBodyModifiers=true
       bCreateParentSpaceControls=true
       bCreateWorldSpaceControls=true
       bIncludeParentBone=false
       limbName=Head
       startBone=neck_01
      }
      {
       bCreateBodyModifiers=true
       bCreateParentSpaceControls=true
       bCreateWorldSpaceControls=true
       bIncludeParentBone=false
       limbName=LeftLeg
       startBone=thigh_l
      }
      {
       bCreateBodyModifiers=true
       bCreateParentSpaceControls=true
       bCreateWorldSpaceControls=true
       bIncludeParentBone=false
       limbName=RightLeg
       startBone=thigh_r
      }
      {
       bCreateBodyModifiers=true
       bCreateParentSpaceControls=true
       bCreateWorldSpaceControls=true
       bIncludeParentBone=false
       limbName=Spine
       startBone=pelvis
      }
     ]}
    componentAppliedLinearAccClamp=<200 200 200>
    initialControlAndBodyModifierUpdates {
     controlParameters [
      WorldSpace_UpperBody {data {
        angularDampingRatio=1
        angularExtraDamping=0
        angularStrength=3
        angularTargetVelocityMultiplier=1
        bDisableCollision=false
        bEnableAngularDampingRatio=true
        bEnableAngularExtraDamping=true
        bEnableAngularStrength=true
        bEnableAngularTargetVelocityMultiplier=true
        bEnableCustomControlPoint=true
        bEnableLinearDampingRatio=true
        bEnableLinearExtraDamping=true
        bEnableLinearStrength=true
        bEnableLinearTargetVelocityMultiplier=true
        bEnableMaxForce=true
        bEnableMaxTorque=true
        bEnablebDisableCollision=true
        bEnablebEnabled=true
        bEnablebOnlyControlChildObject=true
        bEnablebUseAccelerationDriveMode=true
        bEnablebUseCustomControlPoint=true
        bEnablebUseSkeletalAnimation=true
        bEnabled=true
        bOnlyControlChildObject=false
        bUseAccelerationDriveMode=true
        bUseCustomControlPoint=false
        bUseSkeletalAnimation=true
        linearDampingRatio=1
        linearExtraDamping=0
        linearStrength=1
        linearTargetVelocityMultiplier=1
        maxForce=0
        maxTorque=0
        customControlPoint=<0 0 0>
       }}
      pelvis {data {
        angularDampingRatio=1
        angularExtraDamping=0
        angularStrength=2
        angularTargetVelocityMultiplier=1
        bDisableCollision=false
        bEnableAngularDampingRatio=true
        bEnableAngularExtraDamping=true
        bEnableAngularStrength=true
        bEnableAngularTargetVelocityMultiplier=true
        bEnableCustomControlPoint=true
        bEnableLinearDampingRatio=true
        bEnableLinearExtraDamping=true
        bEnableLinearStrength=true
        bEnableLinearTargetVelocityMultiplier=true
        bEnableMaxForce=true
        bEnableMaxTorque=true
        bEnablebDisableCollision=true
        bEnablebEnabled=true
        bEnablebOnlyControlChildObject=true
        bEnablebUseAccelerationDriveMode=true
        bEnablebUseCustomControlPoint=true
        bEnablebUseSkeletalAnimation=true
        bEnabled=true
        bOnlyControlChildObject=false
        bUseAccelerationDriveMode=true
        bUseCustomControlPoint=false
        bUseSkeletalAnimation=true
        linearDampingRatio=1
        linearExtraDamping=0
        linearStrength=0
        linearTargetVelocityMultiplier=1
        maxForce=0
        maxTorque=0
        customControlPoint=<0 0 0>
       }}
      Hands {data {
        angularDampingRatio=1
        angularExtraDamping=0
        angularStrength=4
        angularTargetVelocityMultiplier=1
        bDisableCollision=false
        bEnableAngularDampingRatio=true
        bEnableAngularExtraDamping=true
        bEnableAngularStrength=true
        bEnableAngularTargetVelocityMultiplier=true
        bEnableCustomControlPoint=true
        bEnableLinearDampingRatio=true
        bEnableLinearExtraDamping=true
        bEnableLinearStrength=true
        bEnableLinearTargetVelocityMultiplier=true
        bEnableMaxForce=true
        bEnableMaxTorque=true
        bEnablebDisableCollision=true
        bEnablebEnabled=true
        bEnablebOnlyControlChildObject=true
        bEnablebUseAccelerationDriveMode=true
        bEnablebUseCustomControlPoint=true
        bEnablebUseSkeletalAnimation=true
        bEnabled=true
        bOnlyControlChildObject=false
        bUseAccelerationDriveMode=true
        bUseCustomControlPoint=false
        bUseSkeletalAnimation=true
        linearDampingRatio=1
        linearExtraDamping=0
        linearStrength=0
        linearTargetVelocityMultiplier=1
        maxForce=0
        maxTorque=0
        customControlPoint=<0 0 0>
       }}
      Head {data {
        angularDampingRatio=1
        angularExtraDamping=0
        angularStrength=4
        angularTargetVelocityMultiplier=1
        bDisableCollision=false
        bEnableAngularDampingRatio=true
        bEnableAngularExtraDamping=true
        bEnableAngularStrength=true
        bEnableAngularTargetVelocityMultiplier=true
        bEnableCustomControlPoint=true
        bEnableLinearDampingRatio=true
        bEnableLinearExtraDamping=true
        bEnableLinearStrength=true
        bEnableLinearTargetVelocityMultiplier=true
        bEnableMaxForce=true
        bEnableMaxTorque=true
        bEnablebDisableCollision=true
        bEnablebEnabled=true
        bEnablebOnlyControlChildObject=true
        bEnablebUseAccelerationDriveMode=true
        bEnablebUseCustomControlPoint=true
        bEnablebUseSkeletalAnimation=true
        bEnabled=true
        bOnlyControlChildObject=false
        bUseAccelerationDriveMode=true
        bUseCustomControlPoint=false
        bUseSkeletalAnimation=true
        linearDampingRatio=1
        linearExtraDamping=0
        linearStrength=0
        linearTargetVelocityMultiplier=1
        maxForce=0
        maxTorque=0
        customControlPoint=<0 0 0>
       }}
     ]
     modifierParameters [
      LeftLeg {data {
        bEnableCCD=false
        bEnableCollisionType=true
        bEnableGravityMultiplier=true
        bEnableKinematicTargetSpace=true
        bEnableMovementType=true
        bEnablePhysicsBlendWeight=true
        bEnablebEnableCCD=true
        bEnablebUpdateKinematicFromSimulation=true
        bUpdateKinematicFromSimulation=true
        collisionType=QueryAndPhysics
        gravityMultiplier=1
        kinematicTargetSpace=OffsetInBoneSpace
        movementType=Kinematic
        physicsBlendWeight=1
       }}
      RightLeg {data {
        bEnableCCD=false
        bEnableCollisionType=true
        bEnableGravityMultiplier=true
        bEnableKinematicTargetSpace=true
        bEnableMovementType=true
        bEnablePhysicsBlendWeight=true
        bEnablebEnableCCD=true
        bEnablebUpdateKinematicFromSimulation=true
        bUpdateKinematicFromSimulation=true
        collisionType=QueryAndPhysics
        gravityMultiplier=1
        kinematicTargetSpace=OffsetInBoneSpace
        movementType=Kinematic
        physicsBlendWeight=1
       }}
      Pelvis {data {
        bEnableCCD=false
        bEnableCollisionType=true
        bEnableGravityMultiplier=true
        bEnableKinematicTargetSpace=true
        bEnableMovementType=true
        bEnablePhysicsBlendWeight=true
        bEnablebEnableCCD=true
        bEnablebUpdateKinematicFromSimulation=true
        bUpdateKinematicFromSimulation=true
        collisionType=QueryAndPhysics
        gravityMultiplier=1
        kinematicTargetSpace=OffsetInBoneSpace
        movementType=Simulated
        physicsBlendWeight=1
       }}
      thigh_l {data {
        bEnableCCD=false
        bEnableCollisionType=true
        bEnableGravityMultiplier=true
        bEnableKinematicTargetSpace=true
        bEnableMovementType=true
        bEnablePhysicsBlendWeight=true
        bEnablebEnableCCD=true
        bEnablebUpdateKinematicFromSimulation=true
        bUpdateKinematicFromSimulation=true
        collisionType=QueryAndPhysics
        gravityMultiplier=1
        kinematicTargetSpace=OffsetInBoneSpace
        movementType=Simulated
        physicsBlendWeight=1
       }}
      thigh_r {data {
        bEnableCCD=false
        bEnableCollisionType=true
        bEnableGravityMultiplier=true
        bEnableKinematicTargetSpace=true
        bEnableMovementType=true
        bEnablePhysicsBlendWeight=true
        bEnablebEnableCCD=true
        bEnablebUpdateKinematicFromSimulation=true
        bUpdateKinematicFromSimulation=true
        collisionType=QueryAndPhysics
        gravityMultiplier=1
        kinematicTargetSpace=OffsetInBoneSpace
        movementType=Simulated
        physicsBlendWeight=1
       }}
     ]
    }
    simSpaceSettings {maxAngularAcceleration=200, maxAngularVelocity=200, maxLinearAcceleration=200, maxLinearVelocity=200}
   }
   Component To Local(ComponentPose=@n2)
   Output Pose(Result=Input Pose() {name=InputPoseFullBody}) {layerGroup=MovementMode}
  ]}
 MovementMode_InAirOverride {nodes [
   Input Pose() {name=InSource_Pose}
   Input Pose() {name=InAirPose}
   Output Pose() {layerGroup=MovementMode}
  ]}
 MovementMode_LocomotionAdditives {nodes [
   AnimGraphNode_StateMachine_0 = bounce_SM() {is=title AnimGraphNode_StateMachine}
   n1 = "Sequence Player (additive) 'Locomotion_KneeSlidingGranularNoise'"() {playRate=0.8, sequence=/Game/Animation/Game/MainPlayer/Locomotion/Medium/Male/Sliding/KneeSliding/layers/Locomotion_KneeSlidingGranularNoise}
   Apply Mesh Space Additive(
    Base=Apply Mesh Space Additive(Additive=@AnimGraphNode_StateMachine_0, Alpha=Get bounceStrength() -> (bounceStrength:double))
    Additive=@n1
    Alpha=Get noiseStrength() -> (noiseStrength:double)
   )
   Input Pose() {name=InLocomotionAdditivesPose}
   Random Sequence Player!() -> (Pose:Pose Link Structure) {
    is=title AnimGraphNode_RandomPlayer
    entries [
     {
      chanceToPlay=1
      maxLoopCount=0
      maxPlayRate=1
      minLoopCount=0
      minPlayRate=1
      sequence=/Game/Animation/Game/MainPlayer/Locomotion/Medium/Male/Sliding/KneeSliding/layers/Locomotion_KneeSlidingBouncyNoiseTrimmed
      blendIn {blendOption=HermiteCubic, blendTime=0.2, customCurve=null}
     }
     {
      chanceToPlay=1
      maxLoopCount=0
      maxPlayRate=1
      minLoopCount=0
      minPlayRate=1
      sequence=null
      blendIn {blendOption=HermiteCubic, blendTime=0.2, customCurve=null}
     }
     {
      chanceToPlay=1
      maxLoopCount=0
      maxPlayRate=1
      minLoopCount=0
      minPlayRate=1
      sequence=null
      blendIn {blendOption=HermiteCubic, blendTime=0.2, customCurve=null}
     }
     {
      chanceToPlay=1
      maxLoopCount=0
      maxPlayRate=1
      minLoopCount=0
      minPlayRate=1
      sequence=null
      blendIn {blendOption=HermiteCubic, blendTime=0.2, customCurve=null}
     }
    ]
   }
   n2 = Get SpineTweakAnimSequence() -> (SpineTweakAnimSequence:Anim Sequence Object Reference)
   Output Pose(Result=Blend Poses by bool(
     BlendPose_0=Sequence Player!(Sequence:Anim Sequence Base Object Reference=@n2) -> (Pose) {bLoopAnimation=false}
     BlendPose_1=Additive Identity Pose()
     bActiveValue=Get bPlaySpineTweak() -> (bPlaySpineTweak:bool)
    ) {blendType=Linear}) {layerGroup=MovementMode}
  ]}
 bounce_SM {
  entry=null
  path=MovementMode_LocomotionAdditives.AnimGraphNode_StateMachine_0.bounce_SM
  ids=id
  nodes [
   AnimStateNode_0 = null() {is=title AnimStateNode, stateEntered {guid=#1, notifyName=bounceEnter}}
   AnimStateNode_1 = bounce() {is=title AnimStateNode}
   AnimStateTransitionNode_0 = AnimStateTransitionNode() {bAllowInertializationForSelfTransitions=true}
   AnimStateTransitionNode_1 = AnimStateTransitionNode() {bAllowInertializationForSelfTransitions=true, bAutomaticRuleBasedOnSequencePlayerInState=true}
  ]
  links [
   {
    dst=null
    rule=MovementMode_LocomotionAdditives.AnimGraphNode_StateMachine_0.bounce_SM.AnimStateTransitionNode_1.Transition
    src=bounce
    via=AnimStateTransitionNode_1
   }
   {
    dst=bounce
    rule=MovementMode_LocomotionAdditives.AnimGraphNode_StateMachine_0.bounce_SM.AnimStateTransitionNode_0.Transition
    src=null
    via=AnimStateTransitionNode_0
   }
  ]
 }
 null {
  path=MovementMode_LocomotionAdditives.AnimGraphNode_StateMachine_0.bounce_SM.AnimStateNode_0.null
  nodes [Output Animation Pose(Result=Additive Identity Pose())]
 }
 bounce {
  path=MovementMode_LocomotionAdditives.AnimGraphNode_StateMachine_0.bounce_SM.AnimStateNode_1.bounce
  nodes [
   n1 = "Sequence Player (additive) 'Locomotion_KneeSlidingBouncyNoiseTrimmed'"!(PlayRate:float=Get bounceSpeed() -> (bounceSpeed:double)) -> (Pose) {
    bLoopAnimation=false
    playRate=1.3
    sequence=&Locomotion_KneeSlidingBouncyNoiseTrimmed
   }
   Output Animation Pose(Result=@n1)
  ]
 }
 Transition {
  path=MovementMode_LocomotionAdditives.AnimGraphNode_StateMachine_0.bounce_SM.AnimStateTransitionNode_0.Transition
  nodes [Result(bCanEnterTransition=Get outputShouldBounce())]
 }
 Transition {path=MovementMode_LocomotionAdditives.AnimGraphNode_StateMachine_0.bounce_SM.AnimStateTransitionNode_1.Transition, nodes [Result()]}
 MovementMode_LowerBodyOverride {nodes [
   Input Pose() {name=InLowerBodyPose}
   Input Pose() {name=InSourcePose}
   n1 = Rotate Root Bone!(
    BasePose:Pose Link Structure=Use cached pose 'LowerBodySlide'()
    Pitch:float=0
    Yaw:float=Get KneeSlideRootYawOffset() -> (KneeSlideRootYawOffset:float)
    MeshToComponent:Rotator="0,0,0"
   ) -> (Pose:Pose Link Structure) {is=title AnimGraphNode_RotateRootBone}
   n2 = Rotate Root Bone(BasePose=Use cached pose 'LowerBodySlide'(), Pitch=0, Yaw=Get VelocityAimYawDelta(), MeshToComponent="0,0,0")
   n3 = Get bTransitionToKneeSlide() -> (bTransitionToKneeSlide:bool)
   Output Pose(Result=Blend Poses by bool(
     BlendPose_0=Use cached pose 'LowerBodySlide'()
     BlendPose_1=Blend Poses by bool(BlendPose_0=@n1, BlendPose_1=@n2, BlendTime_0=0.2, BlendTime_1=0.2, bActiveValue=@n3) {blendType=Linear}
     BlendTime_0=0.2
     BlendTime_1=0.2
     bActiveValue=Get bIsInFirstPerson() -> (bIsInFirstPerson:bool)
    ) {blendType=Linear}) {layerGroup=MovementMode}
   AnimGraphNode_StateMachine_0 = LowerBodySlide() {is=title AnimGraphNode_StateMachine}
   Save cached pose 'LowerBodySlide'(Pose=@AnimGraphNode_StateMachine_0) {cacheName=LowerBodySlide}
  ]}
 LowerBodySlide {
  entry=Initiate
  path=MovementMode_LowerBodyOverride.AnimGraphNode_StateMachine_0.LowerBodySlide
  ids=id
  nodes [
   AnimStateConduitNode_1 = Conduit()
   AnimStateNode_0 = BaseballSliding() {is=title AnimStateNode}
   AnimStateNode_1 = KneeSliding() {is=title AnimStateNode}
   AnimStateNode_2 = BaseballToKneeTransition() {is=title AnimStateNode}
   AnimStateNode_3 = Initiate() {is=title AnimStateNode}
   AnimStateTransitionNode_0 = AnimStateTransitionNode()
   AnimStateTransitionNode_1 = AnimStateTransitionNode()
   AnimStateTransitionNode_10 = AnimStateTransitionNode()
   AnimStateTransitionNode_11 = AnimStateTransitionNode()
   AnimStateTransitionNode_2 = AnimStateTransitionNode() {bAutomaticRuleBasedOnSequencePlayerInState=true}
   AnimStateTransitionNode_6 = AnimStateTransitionNode() {bAutomaticRuleBasedOnSequencePlayerInState=true}
  ]
  links [
   {
    dst=BaseballToKneeTransition
    rule=MovementMode_LowerBodyOverride.AnimGraphNode_StateMachine_0.LowerBodySlide.AnimStateTransitionNode_1.Transition
    src=BaseballSliding
    via=AnimStateTransitionNode_1
   }
   {
    dst=KneeSliding
    rule=MovementMode_LowerBodyOverride.AnimGraphNode_StateMachine_0.LowerBodySlide.AnimStateTransitionNode_2.Transition
    src=BaseballToKneeTransition
    via=AnimStateTransitionNode_2
   }
   {
    dst=BaseballSliding
    rule=MovementMode_LowerBodyOverride.AnimGraphNode_StateMachine_0.LowerBodySlide.AnimStateTransitionNode_10.Transition
    src=Conduit
    via=AnimStateTransitionNode_10
   }
   {
    dst=KneeSliding
    rule=MovementMode_LowerBodyOverride.AnimGraphNode_StateMachine_0.LowerBodySlide.AnimStateTransitionNode_11.Transition
    src=Conduit
    via=AnimStateTransitionNode_11
   }
   {
    dst=Conduit
    rule=MovementMode_LowerBodyOverride.AnimGraphNode_StateMachine_0.LowerBodySlide.AnimStateTransitionNode_0.Transition
    src=Initiate
    via=AnimStateTransitionNode_0
   }
   {
    dst=Conduit
    rule=MovementMode_LowerBodyOverride.AnimGraphNode_StateMachine_0.LowerBodySlide.AnimStateTransitionNode_6.Transition
    src=Initiate
    via=AnimStateTransitionNode_6
   }
  ]
 }
 Conduit {
  path=MovementMode_LowerBodyOverride.AnimGraphNode_StateMachine_0.LowerBodySlide.AnimStateConduitNode_1.Conduit
  nodes [Result(bCanEnterTransition=true)]
 }
 BaseballSliding {
  path=MovementMode_LowerBodyOverride.AnimGraphNode_StateMachine_0.LowerBodySlide.AnimStateNode_0.BaseballSliding
  nodes [
   n1 = Get VelocityAimYawDelta()
   n2 = Get KneeSlideSlopeAngle() -> (KneeSlideSlopeAngle:float)
   Output Animation Pose(Result=Apply Mesh Space Additive(
     Base=Sequence Player 'BaseballSliding'() {sequence=/Game/Animation/Game/MainPlayer/Locomotion/Medium/Male/Sliding/KneeSliding/BaseballSliding}
     Additive=Blendspace Player 'BaseballSlidingSlopeAngleBS'(X=@n1, Y=@n2) {blendSpace=/Game/Animation/Game/MainPlayer/Locomotion/Medium/Male/Sliding/KneeSliding/BaseballSlidingSlopeAngleBS}
     Alpha=1
    ) {alphaBoolBlend {blendOption=Linear}})
  ]
 }
 KneeSliding {
  path=MovementMode_LowerBodyOverride.AnimGraphNode_StateMachine_0.LowerBodySlide.AnimStateNode_1.KneeSliding
  nodes [
   n1 = Get KneeSlideSlopeYawMinusRootRot() -> (KneeSlideSlopeYawMinusRootRot:float)
   n2 = Get KneeSlideSlopeAngle()
   Output Animation Pose(Result=Apply Mesh Space Additive(
     Base=Sequence Player 'KneeSliding'() {sequence=/Game/Animation/Game/MainPlayer/Locomotion/Medium/Male/Sliding/KneeSliding/KneeSliding}
     Additive=Blendspace Player 'KneeSlidingSlopeAngleBS'(X=@n1, Y=@n2) {blendSpace=/Game/Animation/Game/MainPlayer/Locomotion/Medium/Male/Sliding/KneeSliding/KneeSlidingSlopeAngleBS}
     Alpha=1
    ) {alphaBoolBlend {blendOption=Linear}})
  ]
 }
 BaseballToKneeTransition {
  path=MovementMode_LowerBodyOverride.AnimGraphNode_StateMachine_0.LowerBodySlide.AnimStateNode_2.BaseballToKneeTransition
  nodes [
   n1 = Blendspace Player 'BaseballToKneeSlidingTransitionBS'(X=Get KneeSlideSlopeYaw() -> (KneeSlideSlopeYaw:float), Y=Get KneeSlideSlopeAngle()) {bLoop=false, blendSpace=/Game/Animation/Game/MainPlayer/Locomotion/Medium/Male/Sliding/KneeSliding/BaseballToKneeSlidingTransitionBS}
   Output Animation Pose(Result=@n1)
  ]
 }
 Initiate {
  path=MovementMode_LowerBodyOverride.AnimGraphNode_StateMachine_0.LowerBodySlide.AnimStateNode_3.Initiate
  nodes [
   n1 = Blendspace Player 'KneeSlidingSlopeAngleBS'(X=Get KneeSlideSlopeYaw(), Y=Get KneeSlideSlopeAngle()) {
    bIgnoreForRelevancyTest=true
    bLoop=false
    bResetPlayTimeWhenBlendSpaceChanges=false
    blendSpace=&KneeSlidingSlopeAngleBS
   }
   n2 = Sequence Player 'KneeSlideInitiate'() {bLoopAnimation=false, sequence=/Game/Animation/Game/MainPlayer/Locomotion/Medium/Male/Sliding/KneeSliding/KneeSlideInitiate}
   n3 = Blendspace Player 'BaseballSlidingSlopeAngleBS'(X=Get VelocityAimYawDelta(), Y=Get KneeSlideSlopeAngle()) {
    bIgnoreForRelevancyTest=true
    bLoop=false
    bResetPlayTimeWhenBlendSpaceChanges=false
    blendSpace=&BaseballSlidingSlopeAngleBS
   }
   n4 = Sequence Player 'BaseballSlideInitiate'() {bLoopAnimation=false, sequence=/Game/Animation/Game/MainPlayer/Locomotion/Medium/Male/Sliding/KneeSliding/BaseballSlideInitiate}
   Output Animation Pose(Result=Blend Poses by bool(
     BlendPose_0=Apply Mesh Space Additive(Base=@n2, Additive=@n1, Alpha=1) {alphaBoolBlend {blendOption=Linear}}
     BlendPose_1=Apply Mesh Space Additive(Base=@n4, Additive=@n3, Alpha=1) {alphaBoolBlend {blendOption=Linear}}
     bActiveValue=Get bInitiatedSlideUpHill() -> (bInitiatedSlideUpHill:bool)
    ) {blendType=Linear})
  ]
 }
 Transition {
  path=MovementMode_LowerBodyOverride.AnimGraphNode_StateMachine_0.LowerBodySlide.AnimStateTransitionNode_0.Transition
  nodes [Result(bCanEnterTransition=Get bBypassSlideInitiate() -> (bBypassSlideInitiate:bool))]
 }
 Transition {
  path=MovementMode_LowerBodyOverride.AnimGraphNode_StateMachine_0.LowerBodySlide.AnimStateTransitionNode_1.Transition
  nodes [Result(bCanEnterTransition=Get bTransitionToKneeSlide())]
 }
 Transition {
  path=MovementMode_LowerBodyOverride.AnimGraphNode_StateMachine_0.LowerBodySlide.AnimStateTransitionNode_10.Transition
  nodes [Result(bCanEnterTransition=NOT Boolean(A=Get bInitiatedSlideUpHill()))]
 }
 Transition {
  path=MovementMode_LowerBodyOverride.AnimGraphNode_StateMachine_0.LowerBodySlide.AnimStateTransitionNode_11.Transition
  nodes [Result(bCanEnterTransition=Get bInitiatedSlideUpHill())]
 }
 Transition {path=MovementMode_LowerBodyOverride.AnimGraphNode_StateMachine_0.LowerBodySlide.AnimStateTransitionNode_2.Transition, nodes [Result()]}
 Transition {path=MovementMode_LowerBodyOverride.AnimGraphNode_StateMachine_0.LowerBodySlide.AnimStateTransitionNode_6.Transition, nodes [Result()]}
 MovementMode_OffScreen {nodes [
   Input Pose() {name=InPoseOffscreen}
   Output Pose() {layerGroup=MovementMode}
  ]}
 MovementMode_SourcePose {nodes [
   n1 = Blendspace Player 'SlidingSpineTweak_BS'(X=Get SpineTweakBlendSpaceYaw() -> (SpineTweakBlendSpaceYaw:float), Y=0) {bIgnoreForRelevancyTest=true, blendSpace=/Game/Animation/Game/MainPlayer/Locomotion/Medium/Male/Sliding/KneeSliding/SlidingSpineTweak_BS}
   n2 = Blend Poses by bool(BlendPose_0=@n1, BlendPose_1=Additive Identity Pose(), bActiveValue=Get bShouldApplySpineTweak() -> (bShouldApplySpineTweak:bool)) {blendType=Linear}
   n3 = Input Pose() {name=InSourcePoseBase}
   n4 = Get SpineTweakLean() -> (SpineTweakLean:float)
   Output Pose(Result=Apply Mesh Space Additive(
     Base=Apply Mesh Space Additive(Base=@n3, Additive=@n2, Alpha=1) {alphaBoolBlend {blendOption=Linear}}
     Additive=Blendspace Player 'SlidingSpineTweakLean_BS'(X=@n4, Y=0) {blendSpace=/Game/Animation/Game/MainPlayer/Locomotion/Medium/Male/Sliding/KneeSliding/SlidingSpineTweakLean_BS}
     Alpha=Get SpineTweakAlpha() -> (SpineTweakAlpha:float)
    ) {alphaBoolBlend {blendOption=Linear}}) {layerGroup=MovementMode}
  ]}
]
variables [
 As Player Pawn Athena {type=PlayerPawn_Athena_C}
 bounceSpeed {type=number, default=1}
 bounceStrength {type=number, default=0.5}
 External Force {type=Vector}
 noiseStrength {type=number, default=1}
 outputFloorMaterial {type=PhysicalMaterial, default=null}
 outputShouldBounce {type=boolean}
 outputVelocity {type=number}
 outputZVelocityChange {type=number}
 zPositionTick {type=number}
 zPositionVelocity {type=number}
 zVelocityChange {type=number}
]
